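# Chevron arrow button widget for Discovery activity browser rows, built on CommonButtonLegacy and shown only for mouse and keyboard input.
# Flip mirrors the arrow, background fade and slide direction.
# Row active events play RowActive or RowInactive animations, fading the wrapper and animating ActiveRatio, whose setter slides it horizontally; reset to inactive hides it instantly.
# Hover animates the arrow material forward or backward and toggles the focus outline.
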WidgetBlueprint /DiscoveryBrowser/ActivityBrowserArrowButton
animations [
 Hover {
  binding_count=1
  display_label=""
  length_seconds=0.12
  object=&self:Hover
  sequence {
   bindings [Arrow {
     id=#1
     possessed_class=Image
     tracks [{
       type=MovieSceneWidgetMaterialTrack
       sections [{type=MovieSceneParameterSection, channels [Hover_Animate {type=MovieSceneScriptingFloatChannel, keys [[0,0],[12,1]]}]}]
      }]
    }]
   display_rate {numerator=100}
  }
 }
 RowActive {
  binding_count=2
  length_seconds=0.48
  object=&self:RowActive
  sequence {
   bindings [
    ActivityBrowserArrowButton_C_133 {
     id=#2
     possessed_class=ActivityBrowserArrowButton_C
     tracks [{
       type=MovieSceneDoubleTrack
       display_name=Active Ratio
       sections [{type=MovieSceneDoubleSection, end_frame=48, start_frame=0, channels [None {type=MovieSceneScriptingDoubleChannel, keys [[0,0],[48,1]]}]}]
      }]
    }
    GP_ButtonWrapper {
     id=#3
     possessed_class=SizeBox
     tracks [{
       type=MovieSceneFloatTrack
       display_name=Render Opacity
       sections [{type=MovieSceneFloatSection, channels [None {type=MovieSceneScriptingFloatChannel, keys [[0,0],[48,1]]}]}]
      }]
    }
   ]
   display_rate {numerator=100}
  }
 }
 RowInactive {
  binding_count=2
  length_seconds=0.24
  object=&self:RowInactive
  sequence {
   bindings [
    GP_ButtonWrapper {
     id=#4
     possessed_class=SizeBox
     tracks [{
       type=MovieSceneFloatTrack
       display_name=Render Opacity
       sections [{type=MovieSceneFloatSection, channels [None {type=MovieSceneScriptingFloatChannel, keys [[0,1],[24,0]]}]}]
      }]
    }
    ActivityBrowserArrowButton_C_134 {
     id=#5
     possessed_class=ActivityBrowserArrowButton_C
     tracks [{
       type=MovieSceneDoubleTrack
       display_name=Active Ratio
       sections [{type=MovieSceneDoubleSection, end_frame=24, start_frame=0, channels [None {type=MovieSceneScriptingDoubleChannel, keys [[0,1],[24,0]]}]}]
      }]
    }
   ]
   display_rate {numerator=100}
  }
 }
]
graphs [
 EventGraph {nodes [
   n1 = OnRowActiveChanged!() -> (OutputDelegate, then, IsActive:bool=~) {is=title K2Node_CustomEvent}
   > Branch(Condition="Not Equal (Boolean)"(A=@n1.IsActive, B=Get IsActive() -> (IsActive:bool)) -> (ReturnValue=~))
   > n2 = Set IsActive(IsActive:bool=@n1.IsActive) -> (Output_Get:bool)
   > n3 = Branch(Condition=@n2)
   > PlayAnimation(InAnimation=GetActiveAnimation!() -> (ActiveAnimation:Widget Animation Object Reference))
   Check and Stop Animation!(execute:Exec=@n3.else, Animation:Widget Animation Object Reference=GetActiveAnimation()) -> (then:Exec) {is=title K2Node_MacroInstance}
   > PlayAnimation(InAnimation=Get RowInactive() -> (RowInactive:Widget Animation Object Reference))
   n4 = OnRowReset!() -> (OutputDelegate, then, IsActive:bool=~) {is=title K2Node_CustomEvent}
   > n5 = Set IsActive(IsActive=@n4.IsActive)
   > n6 = Branch(Condition=@n5)
   > Check and Stop Animation(Animation=Get RowInactive())
   > n7 = PlayAnimation(InAnimation=GetActiveAnimation())
   Check and Stop Animation(execute=[@n8.then, @n7.then], Animation=Get Hover() -> (Hover:Widget Animation Object Reference))
   Check and Stop Animation(execute=@n6.else, Animation=GetActiveAnimation())
   > n8 = SetRenderOpacity(self=Get GP_ButtonWrapper() -> (GP_ButtonWrapper:Grid Panel Object Reference), InOpacity=0)
   Event PreConstruct()
   n9 = Get Flip() -> (Flip:bool)
   > SetRenderScale!(
    execute
    self:Widget Object Reference=Get Arrow() -> (Arrow:Image Object Reference)
    Scale:Vector 2D Structure=Select(Option 0:Vector 2D Structure=(X=1,Y=1), Option 1:Vector 2D Structure=(X=-1,Y=1), Index=@n9) -> (ReturnValue:Vector 2D Structure)
   ) -> (then)
   > n10 = GetDynamicMaterial(self=Get BgFade() -> (BgFade:Image Object Reference))
   > SetScalarParameterValue(self=@n10, ParameterName=Rotation, Value=Select(Option 0:double=0, Option 1:double=0.5, Index=@n9) -> (ReturnValue:double))
   > Set Padding!(
    execute
    Padding:Margin Structure=Select!(
     Option 0_Left:float=-96
     Option 0_Top:float=-80
     Option 0_Right:float=-60
     Option 0_Bottom:float=-80
     Option 1_Left:float=-60
     Option 1_Top:float=-80
     Option 1_Right:float=-96
     Option 1_Bottom:float=-80
     Index=Get Flip()
    ) -> (ReturnValue:Margin Structure)
    self:Grid Slot Object Reference=SlotAsGridSlot(Widget=Get BgFade())
   ) -> (then, Output_Get:Margin Structure)
   > SetUseArrowBackgroundFade!(execute, self, IsVisible:bool=Get ShowArrowBackgroundFade() -> (ShowArrowBackgroundFade:bool)) -> (then)
   Event OnInitialized()
   n11 = Get CommonInputSubsystem()
   > Bind Event to On Input Method Changed(self=@n11, Delegate=Create Event())
   > HandleInputMethodChanged!(execute, self, bNewInputType:ECommonInputType Enum=GetCurrentInputType(self=@n11)) -> (then)
   Event On Hovered()
   n12 = Get Hover()
   > PlayAnimationForward(InAnimation=@n12)
   n13 = Get WBP_UIKit_Block_Outline() -> (WBP_UIKit_Block_Outline:WBP UIKit Block Outline Object Reference)
   > RequestTransition!(execute, self:UIKit Block Object Reference=@n13, State:EUIKitBlockVisualState Enum=Focused) -> (then)
   Event On Unhovered()
   > PlayAnimationReverse(InAnimation=@n12)
   > RequestTransition(self=@n13, State=Unfocused)
  ]}
 GetActiveAnimation {nodes [
   GetActiveAnimation() {is=title K2Node_FunctionEntry}
   > Return Node!(
    execute
    ActiveAnimation:Widget Animation Object Reference=Select!(
     Option 0:Widget Animation Object Reference=Get RowInactive()
     Option 1:Widget Animation Object Reference=Get RowActive() -> (RowActive:Widget Animation Object Reference)
     Index=Get IsActive()
    ) -> (ReturnValue:Widget Animation Object Reference)
   )
  ]}
 HandleInputMethodChanged {nodes [
   n1 = HandleInputMethodChanged!() -> (then, bNewInputType:ECommonInputType Enum) {is=title K2Node_FunctionEntry}
   > Set Visibility(
    Visibility=Select(
     Option 0:ESlateVisibility Enum=Collapsed
     Option 1:ESlateVisibility Enum=SelfHitTestInvisible
     Index="Equal (Enum)"(A=@n1, B=MouseAndKeyboard)
    ) -> (ReturnValue:ESlateVisibility Enum)
    self=Get GP_ButtonWrapper()
   )
   > Set Size!(execute, Size:Vector 2D Structure=(X=48,Y=48), self:Spacer Object Reference=Get S_ButtonSize() -> (S_ButtonSize:Spacer Object Reference)) -> (then, Output_Get:Vector 2D Structure)
  ]}
 SetActiveRatio {nodes [
   n1 = SetActiveRatio!() -> (then, NewParam:double=~) {is=title K2Node_FunctionEntry}
   > SetRenderTranslation(
    self=Get GP_ButtonWrapper()
    Translation=Select(
     Option 0:Vector 2D Structure=Lerp Vector 2D!(A:Vector 2D Structure=(X=-72,Y=0), B:Vector 2D Structure, Alpha:double=@n1) -> (ReturnValue:Vector 2D Structure)
     Option 1:Vector 2D Structure=Lerp Vector 2D(A=(X=72,Y=0), Alpha=@n1)
     Index=Get Flip()
    ) -> (ReturnValue:Vector 2D Structure)
   )
  ]}
 SetUseArrowBackgroundFade {nodes [
   n1 = SetUseArrowBackgroundFade!() -> (then, IsVisible:bool=~) {is=title K2Node_FunctionEntry}
   > n2 = Set ShowArrowBackgroundFade(ShowArrowBackgroundFade:bool=@n1) -> (Output_Get:bool)
   > SetVisibility(
    self=Get BgFade()
    InVisibility=Select(Option 0:ESlateVisibility Enum=Collapsed, Option 1:ESlateVisibility Enum=HitTestInvisible, Index=@n2) -> (ReturnValue:ESlateVisibility Enum)
   )
  ]}
 UpdateArrowColor {nodes [
   n1 = GetDynamicMaterial(self=Get Arrow())
   > SetVectorParameterValue(self=@n1, ParameterName=TailColor)
   > SetVectorParameterValue(self=@n1, ParameterName=HoverColor)
   UpdateArrowColor!() -> (then, Color:Linear Color Structure, HoverColor:Linear Color Structure) {is=title K2Node_FunctionEntry}
  ]}
]
variables [
 ActiveRatio {type=number}
 Flip {type=boolean}
 InputActionOnSide {type=boolean}
 IsActive {type=boolean}
 ShowArrowBackgroundFade {type=boolean, default=true}
]
widgets {
 info {inheritedWidgetCount=1, parentClass=&CommonButtonLegacy, rootWidgetClass=&GridPanel, widgetCount=8}
 widgets [
  GP_ButtonWrapper {widgetClassPath=&GridPanel, bIsVariable=true, widget=GP_ButtonWrapper, properties {columnFill [1], rowFill [1]}}
  WBP_CaptureForPostBufferUpdate {
   widgetClassPath=&WBP_CaptureForPostBufferUpdate
   parent=GP_ButtonWrapper
   slot {class=GridSlot, properties {layer=-10}}
  }
  BgFade {
   widgetClassPath=&Image
   bIsVariable=true
   parent=GP_ButtonWrapper
   widget=BgFade
   properties {
    visibility=HitTestInvisible
    brush {
     bIsDynamicallyLoaded=false
     drawAs=Image
     imageType=FullColor
     mirroring=NoMirror
     resourceName=None
     resourceObject=/DiscoveryBrowser/Art/Materials/Buttons/MI_UI_Button_Arrow_BgFade
     tiling=NoTile
     imageSize=<32 32>
     margin=<ltrb 0 0 0 0>
     outlineSettings {
      bUseBrushTransparency=false
      roundingType=HalfHeightRadius
      width=0
      color {colorUseRule=UseColor_Specified, specifiedColor=<rgba 0 0 0 0>}
      cornerRadii=<0 0 0 0>
     }
     tintColor {colorUseRule=UseColor_Specified, specifiedColor=<rgba 1 1 1 1>}
     uVRegion {bIsValid=false, max=<0 0>, min=<0 0>}
    }
   }
   slot {class=GridSlot, properties {layer=-1, padding=<ltrb -44 -80 -96 -80>}}
  }
  WBP_UIKit_Backplate {
   widgetClassPath=&WBP_UIKit_Backplate
   bIsVariable=true
   parent=GP_ButtonWrapper
   widget=WBP_UIKit_Backplate
   properties {cornerRadius=Button}
   slot {class=GridSlot}
  }
  S_ButtonSize {
   widgetClassPath=&Spacer
   bIsVariable=true
   parent=GP_ButtonWrapper
   widget=S_ButtonSize
   properties {size=<48 48>}
   slot {class=GridSlot}
  }
  Arrow {
   widgetClassPath=&Image
   bIsVariable=true
   parent=GP_ButtonWrapper
   widget=Arrow
   properties {
    visibility=HitTestInvisible
    brush {
     bIsDynamicallyLoaded=false
     drawAs=Image
     imageType=FullColor
     mirroring=NoMirror
     resourceName=None
     resourceObject=/FortUILibrary/Textures/MaterialInstances/MI_UIKit_Icon_Chevron
     tiling=NoTile
     imageSize=<32 32>
     margin=<ltrb 0 0 0 0>
     outlineSettings {
      bUseBrushTransparency=false
      roundingType=HalfHeightRadius
      width=0
      color {colorUseRule=UseColor_Specified, specifiedColor=<rgba 0 0 0 0>}
      cornerRadii=<0 0 0 0>
     }
     tintColor {colorUseRule=UseColor_Specified, specifiedColor=<rgba 1 1 1 1>}
     uVRegion {bIsValid=false, max=<0 0>, min=<0 0>}
    }
    colorAndOpacity=<rgba 0 0 0 1>
   }
   slot {class=GridSlot, properties {layer=1, padding=<ltrb 8 8 8 8>}}
  }
  WBP_UIKit_Block_Outline {
   widgetClassPath=/FortUILibrary/UIKit/Buttons/Blocks/WBP_UIKit_Block_Outline
   bIsVariable=true
   parent=GP_ButtonWrapper
   widget=WBP_UIKit_Block_Outline
   properties {bottomLeftCornerRadius=0.325, bottomRightCornerRadius=0.325, topLeftCornerRadius=0.325, topRightCornerRadius=0.325}
   slot {class=GridSlot, properties {layer=10}}
  }
  InputActionWidget {widgetClassPath=&CommonActionWidget, bInherited=true, widget=null}
 ]
}
