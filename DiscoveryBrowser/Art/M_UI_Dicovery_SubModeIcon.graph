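# Translucent UI material for the Discovery Browser sub-mode icon, procedurally drawing three squares in a grid with a checkmark in the fourth slot.
# Custom-node box signed distance functions on offset UVs, each reduced by its own scale parameter and smoothstepped, form the square masks.
# These masks are summed with the alpha of a check texture, scaled about its center by its own parameter, alongside an RGB color parameter.
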Material /DiscoveryBrowser/Art/M_UI_Dicovery_SubModeIcon
graph {
 properties {
  bUsedWithStaticMesh=false
  blendMode=BLEND_Translucent
  materialDomain=MD_UI
  pixelDepthOffsetMode=PDOM_Legacy
  usedShadingModels=Default Lit
 }
 nodes [
  MaterialExpressionAdd_2 = Add(A=ConstantBiasScale(None=TextureCoordinate()) {bias=-0.25, scale=2}, B=Constant2Vector() {g=-0.125}) {constB=0}
  MaterialExpressionConstant2Vector_2 = Constant2Vector() {g=0.3, r=0.3}
  n1 = Custom!(p=@MaterialExpressionAdd_2, b=@MaterialExpressionConstant2Vector_2) {
   code="{\r\n    float2 d = abs(p)-b;\r\n    return length(max(d,0.0)) + min(max(d.x,d.y),0.0);\r\n}"
   outputType=CMOT_Float2
   inputs [
    {inputName=p, input {expression=&self:MaterialExpressionAdd_2}}
    {inputName=b, input {expression=&self:MaterialExpressionConstant2Vector_2}}
   ]
  }
  n2 = Subtract(A=@n1, B=ScalarParameter() {parameterName=S01-Scale}) {constB=0}
  MaterialExpressionAdd_3 = Add(A=ConstantBiasScale(None=TextureCoordinate()) {bias=-0.25, scale=2}, B=Constant2Vector() {g=-0.125, r=-0.75})
  MaterialExpressionConstant2Vector_1 = Constant2Vector() {g=0.3, r=0.3}
  n3 = Custom(p=@MaterialExpressionAdd_3, b=@MaterialExpressionConstant2Vector_1) {
   code="{\r\n    float2 d = abs(p)-b;\r\n    return length(max(d,0.0)) + min(max(d.x,d.y),0.0);\r\n}"
   outputType=CMOT_Float2
   inputs [
    {inputName=p, input {expression=&self:MaterialExpressionAdd_3}}
    {inputName=b, input {expression=&self:MaterialExpressionConstant2Vector_1}}
   ]
  }
  n4 = Subtract(A=@n3, B=ScalarParameter() {parameterName=S02-Scale}) {constB=0}
  n5 = Add(A=OneMinus(None=SmoothStep(Value=@n2) {constMax=0.01}), B=OneMinus(None=SmoothStep(Value=@n4) {constMax=0.01}))
  MaterialExpressionAdd_4 = Add(A=ConstantBiasScale(None=TextureCoordinate()) {bias=-0.25, scale=2}, B=Constant2Vector() {g=-0.875})
  MaterialExpressionConstant2Vector_0 = Constant2Vector() {g=0.3, r=0.3}
  n6 = Custom(p=@MaterialExpressionAdd_4, b=@MaterialExpressionConstant2Vector_0) {
   code="{\r\n    float2 d = abs(p)-b;\r\n    return length(max(d,0.0)) + min(max(d.x,d.y),0.0);\r\n}"
   outputType=CMOT_Float2
   inputs [
    {inputName=p, input {expression=&self:MaterialExpressionAdd_4}}
    {inputName=b, input {expression=&self:MaterialExpressionConstant2Vector_0}}
   ]
  }
  n7 = ScalarParameter() {parameterName=S03-Scale}
  n8 = OneMinus(None=SmoothStep(Value=Subtract(A=@n6, B=@n7) {constB=0}) {constMax=0.01})
  n9 = TextureCoordinate()
  n10 = MaterialFunctionCall(
   UVs=Add(A=Multiply(A=@n9) {constB=1.75}, B=Constant2Vector() {g=-0.6, r=-0.725})
   Texture Scale=ScalarParameter() {defaultValue=0.9, parameterName=Check-Scale}
  ) {materialFunction=&ScaleUVsByCenter}
  ComponentMask(None=Add(A=Add(A=@n5, B=@n8), B=TextureSample(UVs=@n10.UVs) {texture=/Game/UI/Foundation/Textures/Icons/Checks/T-Icon-Check-128}.A)) {r=true}
  VectorParameter() {parameterName=RGB, defaultValue=<rgba 1 1 1 1>}
 ]
}
