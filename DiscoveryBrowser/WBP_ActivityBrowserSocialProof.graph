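# Discovery browser widget showing an activity's social proof: player count (CCU), friends playing and an optional likes row.
# Set CCU Count Data and Set Friends Count Data events update their rows, collapsing when hidden, CCU is negative or friends are zero or fewer.
# CCU shortens above 999; friend text pluralizes.
# PreConstruct tints CCU/friend elements, builds a dynamic icon material tuned by a flag, and toggles likes, whose text no graph sets.
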WidgetBlueprint /DiscoveryBrowser/WBP_ActivityBrowserSocialProof
graphs [
 EventGraph {nodes [
   n1 = Set CCU Count Data!() -> (OutputDelegate, then, CCU Count:int, Hide:bool=~) {is=title K2Node_CustomEvent}
   > SetCCUCount!(execute, self, CCUCount:int=@n1.CCU Count, Hide:bool=@n1.Hide) -> (then)
   n2 = Set Friends Count Data!() -> (OutputDelegate, then, Friends Count:int, Hide:bool=~) {is=title K2Node_CustomEvent}
   n3 = "integer <= integer"!(A:int=@n2.Friends Count, B:int=0) -> (ReturnValue:bool) {is=title K2Node_CallFunction}
   > SetVisibility(
    self=Get HB_FriendsPlaying() -> (HB_FriendsPlaying:Horizontal Box Object Reference)
    InVisibility=Select(Option 0=HitTestInvisible, Option 1=Collapsed, Index=OR Boolean(A=@n3, B=@n2.Hide))
   )
   > "SetText (Text)"(
    self=Get Text_FriendCount() -> (Text_FriendCount:Common Text Object Reference)
    InText=Format Text!(Format="LOCTEXT(\"{Number} {Number}|plural(one=friend,other=friends) playing\")", Number:int=@n2.Friends Count) -> (Result)
   )
   Event PreConstruct()
   > Set Color() {is=title K2Node_CallFunction}
   > Set CCU Icon Vertical Scale() {is=title K2Node_CallFunction}
   > SetLikesVisible()
  ]}
 Set CCU Icon Vertical Scale {nodes [
   Set CCU Icon Vertical Scale() {is=title K2Node_FunctionEntry}
   > n1 = CreateDynamicMaterialInstance!(execute, Parent=/FortniteCoreUI/UI/DiscoveryShared/MI_UI_CCU_UserCount, OptionalName=None, CreationFlags=None) -> (then, ReturnValue)
   > Set CCU Icon(CCU Icon:Material Instance Dynamic Object Reference=@n1) -> (Output_Get:Material Instance Dynamic Object Reference)
   > n2 = Branch(Condition=Get CCU Icon Vertical Scale() -> (CCU Icon Vertical Scale:bool))
   > SetScalarParameterValue(self=Get CCU Icon() -> (CCU Icon:Material Instance Dynamic Object Reference), ParameterName=Scale, Value=1.37)
   > SetScalarParameterValue(self=Get CCU Icon(), ParameterName=OffsetY, Value=0.01)
   > n3 = SetScalarParameterValue(self=Get CCU Icon(), ParameterName=Red Channel Opacity, Value=0.6)
   SetBrushFromMaterial(execute=[@n4.then, @n3.then], self=Get Image_CCU() -> (Image_CCU:Fort Mobile Image Object Reference), Material=Get CCU Icon())
   SetScalarParameterValue(execute=@n2.else, self=Get CCU Icon(), ParameterName=Scale, Value=1.2)
   > SetScalarParameterValue(self=Get CCU Icon(), ParameterName=OffsetY, Value=-0.04)
   > n4 = SetScalarParameterValue(self=Get CCU Icon(), ParameterName=Red Channel Opacity, Value=0.7)
  ]}
 Set Color {nodes [
   Set Color() {is=title K2Node_FunctionEntry}
   n1 = Make Slate Color(SpecifiedColor=Get CCU and Friends Text Color() -> (CCU and Friends Text Color:Linear Color Structure))
   > Set ColorAndOpacity!(
    execute
    ColorAndOpacity:Slate Color Structure=@n1
    self:Text Object Reference=Get Text_CCUCount() -> (Text_CCUCount:Common Text Object Reference)
   ) -> (then, Output_Get:Slate Color Structure)
   > Set ColorAndOpacity(ColorAndOpacity=@n1, self=Get Text_FriendCount())
   > SetColorAndOpacity(self:Image Object Reference=Get Image_CCU(), InColorAndOpacity:Linear Color Structure=Get CCU and Friends Text Color())
  ]}
 SetCCUCount {nodes [
   n1 = SetCCUCount!() -> (then, CCUCount:int, Hide:bool) {is=title K2Node_FunctionEntry}
   n2 = "integer < integer"(A=@n1)
   > SetVisibility(
    self=Get HB_CCUCount() -> (HB_CCUCount:Horizontal Box Object Reference)
    InVisibility=Select(Option 0=HitTestInvisible, Option 1=Collapsed, Index=OR Boolean(A=@n2, B=@n1.Hide))
   )
   > n3 = Branch(Condition="integer > integer"(A=@n1, B=999))
   n4 = Get Text_CCUCount()
   > "SetText (Text)"(
    self=@n4
    InText=GetShorthandQuantity!(Quantity:int=@n1, MinimumFractionalDigits:int=0, MaximumFractionalDigits:int=1, bShouldRoundDown:bool=false) -> (ReturnValue:Text)
   )
   "SetText (Text)"(execute=@n3.else, self=@n4, InText="To Text (Integer)"(Value=@n1))
  ]}
 SetLikesVisible {nodes [
   SetLikesVisible() {is=title K2Node_FunctionEntry}
   > SetVisibility(
    self=Get HB_LikesPercent() -> (HB_LikesPercent:Horizontal Box Object Reference)
    InVisibility=Select(Option 0=Collapsed, Option 1=HitTestInvisible, Index=Get IsLikesVisible() -> (IsLikesVisible:bool))
   )
  ]}
]
variables [
 CCU And Friends Text Color {type=LinearColor, default=<rgba 0.5776 0.5395 1 1>}
 CCU Icon {type=MaterialInstanceDynamic}
 CCU Icon Vertical Scale {type=boolean}
 IsLikesVisible {type=boolean}
]
widgets {
 info {inheritedWidgetCount=0, parentClass=&UserWidget, rootWidgetClass=&Overlay, widgetCount=10}
 widgets [
  Overlay {widgetClassPath=&Overlay, widget=Overlay}
  HB_CCUAndFriendsCount {
   widgetClassPath=&HorizontalBox
   bIsVariable=true
   parent=Overlay
   widget=HB_CCUAndFriendsCount
   slot {class=OverlaySlot, properties {horizontalAlignment=HAlign_Fill, verticalAlignment=VAlign_Fill}}
  }
  HB_CCUCount {
   widgetClassPath=&HorizontalBox
   bIsVariable=true
   parent=HB_CCUAndFriendsCount
   widget=HB_CCUCount
   properties {visibility=Collapsed}
   slot {class=HorizontalBoxSlot, properties {padding=<ltrb 0 0 10 0>}}
  }
  Image_CCU {
   widgetClassPath=/Script/FortniteUI.FortMobileImage
   bIsVariable=true
   parent=HB_CCUCount
   widget=Image_CCU
   properties {
    mobileSizeMultiplier=1.2
    visibility=HitTestInvisible
    brush {
     bIsDynamicallyLoaded=false
     drawAs=Image
     imageType=NoImage
     mirroring=NoMirror
     resourceName=None
     resourceObject=&MI_UI_CCU_UserCount
     tiling=NoTile
     imageSize=<18 18>
     margin=<ltrb 0 0 0 0>
     outlineSettings {
      bUseBrushTransparency=false
      roundingType=HalfHeightRadius
      width=0
      color {colorUseRule=UseColor_Specified, specifiedColor=<rgba 0 0 0 0>}
      cornerRadii=<0 0 0 0>
     }
     tintColor {colorUseRule=UseColor_Specified, specifiedColor=<rgba 1 1 1 1>}
     uVRegion {bIsValid=false, max=<0 0>, min=<0 0>}
    }
    colorAndOpacity=<rgba 0.5776 0.5395 1 1>
   }
   slot {class=HorizontalBoxSlot, properties {verticalAlignment=VAlign_Center, padding=<ltrb 0 0 3 0>}}
  }
  Text_CCUCount {
   widgetClassPath=&CommonTextBlock
   bIsVariable=true
   parent=HB_CCUCount
   widget=Text_CCUCount
   properties {
    mobileFontSizeMultiplier=1.45
    scrollStyle=&TextScrollStyle-Base_C
    text="{CCU Count}"
    colorAndOpacity {colorUseRule=UseColor_Specified, specifiedColor=<rgba 0.5776 0.5395 1 1>}
    font {
     bForceMonospaced=false
     bMaterialIsStencil=false
     fontMaterial=null
     fontObject=&BurbankSmall
     letterSpacing=0
     monospacedWidth=1
     size=13
     skewAmount=0
     typefaceFontName=Black
     outlineSettings {
      bApplyOutlineToDropShadows=false
      bMiteredCorners=false
      bSeparateFillAlpha=false
      outlineMaterial=null
      outlineSize=0
      outlineColor=<rgba 0 0 0 1>
     }
    }
    shadowOffset=<0 0>
   }
   slot {class=HorizontalBoxSlot, properties {verticalAlignment=VAlign_Center, size {sizeRule=Fill, value=1}}}
  }
  HB_FriendsPlaying {
   widgetClassPath=&HorizontalBox
   bIsVariable=true
   parent=HB_CCUAndFriendsCount
   widget=HB_FriendsPlaying
   properties {visibility=Collapsed}
   slot {class=HorizontalBoxSlot, properties {size {sizeRule=Fill, value=1}}}
  }
  Text_FriendCount {
   widgetClassPath=&CommonTextBlock
   bIsVariable=true
   parent=HB_FriendsPlaying
   widget=Text_FriendCount
   properties {
    mobileFontSizeMultiplier=1.45
    scrollStyle=&TextScrollStyle-Base_C
    text="{Friends Playing Count}"
    colorAndOpacity {colorUseRule=UseColor_Specified, specifiedColor=<rgba 0.5776 0.5395 1 1>}
    font {
     bForceMonospaced=false
     bMaterialIsStencil=false
     fontMaterial=null
     fontObject=&BurbankSmall
     letterSpacing=0
     monospacedWidth=1
     size=13
     skewAmount=0
     typefaceFontName=Black
     outlineSettings {
      bApplyOutlineToDropShadows=false
      bMiteredCorners=false
      bSeparateFillAlpha=false
      outlineMaterial=null
      outlineSize=0
      outlineColor=<rgba 0 0 0 1>
     }
    }
    shadowOffset=<0 0>
   }
   slot {class=HorizontalBoxSlot, properties {horizontalAlignment=HAlign_Left, verticalAlignment=VAlign_Center, size {sizeRule=Fill, value=1}}}
  }
  HB_LikesPercent {
   widgetClassPath=&HorizontalBox
   bIsVariable=true
   parent=HB_CCUAndFriendsCount
   widget=HB_LikesPercent
   slot {class=HorizontalBoxSlot}
  }
  Image_LikeIcon {
   widgetClassPath=&FortMobileImage
   bIsVariable=true
   parent=HB_LikesPercent
   widget=Image_LikeIcon
   properties {
    mobileSizeMultiplier=1.2
    visibility=HitTestInvisible
    brush {
     bIsDynamicallyLoaded=false
     drawAs=Image
     imageType=NoImage
     mirroring=NoMirror
     resourceName=None
     resourceObject=/FortUILibrary/Foundation/Buttons/Materials/Favorite/MI_UI_ThumbsUp_Base_White
     tiling=NoTile
     imageSize=<18 18>
     margin=<ltrb 0 0 0 0>
     outlineSettings {
      bUseBrushTransparency=false
      roundingType=HalfHeightRadius
      width=0
      color {colorUseRule=UseColor_Specified, specifiedColor=<rgba 0 0 0 0>}
      cornerRadii=<0 0 0 0>
     }
     tintColor {colorUseRule=UseColor_Specified, specifiedColor=<rgba 1 1 1 1>}
     uVRegion {bIsValid=false, max=<0 0>, min=<0 0>}
    }
    colorAndOpacity=<rgba 0.5776 0.5395 1 1>
   }
   slot {class=HorizontalBoxSlot, properties {verticalAlignment=VAlign_Center, padding=<ltrb 0 0 3 0>}}
  }
  Text_LikesPercent {
   widgetClassPath=&CommonTextBlock
   bIsVariable=true
   parent=HB_LikesPercent
   widget=Text_LikesPercent
   properties {
    mobileFontSizeMultiplier=1.45
    scrollStyle=&TextScrollStyle-Base_C
    text="{Likes Percent}"
    colorAndOpacity {colorUseRule=UseColor_Specified, specifiedColor=<rgba 0.5776 0.5395 1 1>}
    font {
     bForceMonospaced=false
     bMaterialIsStencil=false
     fontMaterial=null
     fontObject=&BurbankSmall
     letterSpacing=0
     monospacedWidth=1
     size=13
     skewAmount=0
     typefaceFontName=Black
     outlineSettings {
      bApplyOutlineToDropShadows=false
      bMiteredCorners=false
      bSeparateFillAlpha=false
      outlineMaterial=null
      outlineSize=0
      outlineColor=<rgba 0 0 0 1>
     }
    }
    shadowOffset=<0 0>
   }
   slot {class=HorizontalBoxSlot, properties {verticalAlignment=VAlign_Center, size {sizeRule=Fill, value=1}}}
  }
 ]
}
