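# Discovery browser warning icon widget: a material-driven triangle with a "!" text on top and a separate pulse triangle behind it.
# The TriggerWarningAnim event plays Anim_WarningPulse for a caller-supplied number of loops.
# The animation briefly bumps the icon's scale while the normally invisible pulse image expands, fades in and out, and thins its border through an animated material parameter.
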WidgetBlueprint /DiscoveryBrowser/WBP_Discovery_WarningIcon
animations [Anim_WarningPulse {
  binding_count=2
  length_seconds=0.96
  object=&self:Anim_WarningPulse
  sequence {bindings [
    Overlay_0 {
     id=#1
     possessed_class=Overlay
     tracks [{
       type=MovieScene2DTransformTrack
       sections [{
         type=MovieScene2DTransformSection
         end_frame=8
         start_frame=0
         channels [
          Scale.X {type=MovieSceneScriptingFloatChannel, keys [[4,1],[5.6,1.25],[8.8,1]]}
          Scale.Y {type=MovieSceneScriptingFloatChannel, keys [[4,1],[5.6,1.25],[8.8,1]]}
         ]
        }]
      }]
    }
    WarningTriangle_Pulse {
     id=#2
     possessed_class=Image
     tracks [
      {
       type=MovieSceneWidgetMaterialTrack
       sections [{
         type=MovieSceneParameterSection
         channels [Border_Thickness {type=MovieSceneScriptingFloatChannel, keys [[0,0.125],[11.2,0.05],[19.2,0.001]]}]
        }]
      }
      {
       type=MovieScene2DTransformTrack
       sections [{
         type=MovieScene2DTransformSection
         channels [
          Translation.X {type=MovieSceneScriptingFloatChannel, keys [[0,0],[9.6,1]]}
          Translation.Y {type=MovieSceneScriptingFloatChannel, keys [[0,0],[9.6,-7]]}
          Scale.X {type=MovieSceneScriptingFloatChannel, keys [[0,1],[9.6,2.25]]}
          Scale.Y {type=MovieSceneScriptingFloatChannel, keys [[0,1],[9.6,2.25]]}
         ]
        }]
      }
      {
       type=MovieSceneFloatTrack
       display_name=Render Opacity
       sections [{type=MovieSceneFloatSection, channels [None {type=MovieSceneScriptingFloatChannel, keys [[0,0],[9.6,1],[19.2,0]]}]}]
      }
     ]
    }
   ]}
 }]
graphs [EventGraph {nodes [
   n1 = TriggerWarningAnim!() -> (OutputDelegate, then, NumLoopsToPlay:int=1) {is=title K2Node_CustomEvent}
   > PlayAnimation(InAnimation=Get Anim_WarningPulse() -> (Anim_WarningPulse:Widget Animation Object Reference), NumLoopsToPlay=@n1.NumLoopsToPlay)
  ]}]
variables []
widgets {
 info {inheritedWidgetCount=0, parentClass=&UserWidget, rootWidgetClass=&Overlay, widgetCount=5}
 widgets [
  Overlay_1 {widgetClassPath=&Overlay, widget=Overlay_1}
  WarningTriangle_Pulse {
   widgetClassPath=&Image
   parent=Overlay_1
   widget=WarningTriangle_Pulse
   properties {
    renderOpacity=0
    visibility=SelfHitTestInvisible
    brush {
     bIsDynamicallyLoaded=false
     drawAs=Image
     imageType=NoImage
     mirroring=NoMirror
     resourceName=None
     resourceObject=/Game/UI/DiscoveryShared/Art/M_UI_Discovery_WarningTriangle_Pulse
     tiling=NoTile
     imageSize=<40 40>
     margin=<ltrb 0 0 0 0>
     outlineSettings {
      bUseBrushTransparency=false
      roundingType=HalfHeightRadius
      width=0
      color {colorUseRule=UseColor_Specified, specifiedColor=<rgba 0 0 0 0>}
      cornerRadii=<0 0 0 0>
     }
     tintColor {colorUseRule=UseColor_Specified, specifiedColor=<rgba 1 1 1 1>}
     uVRegion {bIsValid=false, max=<0 0>, min=<0 0>}
    }
   }
   slot {class=OverlaySlot, properties {horizontalAlignment=HAlign_Fill, verticalAlignment=VAlign_Fill}}
  }
  Overlay_0 {
   widgetClassPath=&Overlay
   parent=Overlay_1
   widget=Overlay_0
   slot {class=OverlaySlot, properties {horizontalAlignment=HAlign_Fill, verticalAlignment=VAlign_Fill}}
  }
  WarningTriangle {
   widgetClassPath=&Image
   parent=Overlay_0
   widget=WarningTriangle
   properties {
    visibility=SelfHitTestInvisible
    brush {
     bIsDynamicallyLoaded=false
     drawAs=Image
     imageType=NoImage
     mirroring=NoMirror
     resourceName=None
     resourceObject=/Game/UI/DiscoveryShared/Art/M_UI_Discovery_WarningTriangle
     tiling=NoTile
     imageSize=<40 40>
     margin=<ltrb 0 0 0 0>
     outlineSettings {
      bUseBrushTransparency=false
      roundingType=HalfHeightRadius
      width=0
      color {colorUseRule=UseColor_Specified, specifiedColor=<rgba 0 0 0 0>}
      cornerRadii=<0 0 0 0>
     }
     tintColor {colorUseRule=UseColor_Specified, specifiedColor=<rgba 1 1 1 1>}
     uVRegion {bIsValid=false, max=<0 0>, min=<0 0>}
    }
   }
   slot {class=OverlaySlot, properties {horizontalAlignment=HAlign_Fill, verticalAlignment=VAlign_Fill}}
  }
  Exclamation {
   widgetClassPath=&CommonTextBlock
   parent=Overlay_0
   widget=Exclamation
   properties {
    text="!"
    colorAndOpacity {colorUseRule=UseColor_Specified, specifiedColor=<rgba 0 0 0 1>}
    font {
     bForceMonospaced=false
     bMaterialIsStencil=false
     fontMaterial=null
     fontObject=&Burbank
     letterSpacing=0
     monospacedWidth=1
     size=18
     skewAmount=0
     typefaceFontName=BurbankRegular-Black
     outlineSettings {
      bApplyOutlineToDropShadows=false
      bMiteredCorners=false
      bSeparateFillAlpha=false
      outlineMaterial=null
      outlineSize=0
      outlineColor=<rgba 0 0 0 1>
     }
    }
    margin=<ltrb 0 -4 0 -4>
    renderTransform {angle=0, scale=<1 1>, shear=<4 0>, translation=<-1 5>}
    shadowOffset=<0 0>
   }
   slot {class=OverlaySlot, properties {horizontalAlignment=HAlign_Center, verticalAlignment=VAlign_Center}}
  }
 ]
}
