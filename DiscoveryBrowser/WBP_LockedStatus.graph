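# Discovery browser widget that shows a locked icon beside a text highlight stating the lock reason.
# PreConstruct fills the highlight from LockedReason, and Construct hides it at zero opacity with ellipsis overflow.
# The On Lock Hover Select event toggles text scrolling, switches overflow between Clip and Ellipsis, and plays Anim_OnHover forward or in reverse to fade the highlight.
# The UseNewDesign variable is declared but unused in the graphs.
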WidgetBlueprint /DiscoveryBrowser/WBP_LockedStatus
animations [Anim_OnHover {
  binding_count=1
  length_seconds=0.12
  object=&self:Anim_OnHover
  sequence {
   bindings [WBP_UIKit_Highlight {
     id=#1
     possessed_class=WBP_UIKit_Highlight_C
     tracks [{
       type=MovieSceneFloatTrack
       display_name=Render Opacity
       sections [{type=MovieSceneFloatSection, channels [None {type=MovieSceneScriptingFloatChannel, keys [[0,0],[28.808,1]]}]}]
      }]
    }]
   display_rate {numerator=240}
  }
 }]
graphs [
 EventGraph {nodes [
   n1 = On Lock Hover Select!() -> (OutputDelegate, then, bIsHoverSelected:bool=~) {is=title K2Node_CustomEvent}
   n2 = Get Text_Highlight!(self:WBP UIKit Highlight Base Object Reference=Get WBP_UIKit_Highlight() -> (WBP_UIKit_Highlight:WBP UIKit Highlight Object Reference)) -> (Text_Highlight:Common Text Object Reference)
   > SetScrollingEnabled!(
    execute
    self:Common Text Object Reference=@n2
    bInIsScrollingEnabled:bool=Select(Option 0:bool=false, Option 1:bool=true, Index=@n1.bIsHoverSelected) -> (ReturnValue:bool)
   ) -> (then)
   > Set TextOverflowPolicy!(
    execute
    TextOverflowPolicy:ETextOverflowPolicy Enum=Select(Option 0:ETextOverflowPolicy Enum=Ellipsis, Option 1:ETextOverflowPolicy Enum=Clip, Index=@n1.bIsHoverSelected) -> (ReturnValue:ETextOverflowPolicy Enum)
    self:Text Object Reference=@n2
   ) -> (then, Output_Get:ETextOverflowPolicy Enum=Clip)
   > n3 = Branch(Condition=@n1.bIsHoverSelected)
   n4 = Get Anim_OnHover() -> (Anim_OnHover:Widget Animation Object Reference)
   > PlayAnimationForward(InAnimation=@n4)
   PlayAnimationReverse(execute=@n3.else, InAnimation=@n4)
   Event PreConstruct()
   > SetText!(execute, self:WBP UIKit Highlight Base Object Reference=Get WBP_UIKit_Highlight(), Text:Text=Get LockedReason() -> (LockedReason:Text)) -> (then)
   Event Construct()
   > SetDefaultVisualsState()
  ]}
 SetDefaultVisualsState {nodes [
   SetDefaultVisualsState() {is=title K2Node_FunctionEntry}
   > SetRenderOpacity(self=Get WBP_UIKit_Highlight(), InOpacity=0)
   > Set TextOverflowPolicy(TextOverflowPolicy=Ellipsis, self=Get Text_Highlight(self=Get WBP_UIKit_Highlight())) -> (Output_Get=Clip)
  ]}
]
variables [
 LockedReason {type=string, default="{PH} Requires Purchase"}
 UseNewDesign {type=boolean}
]
widgets {
 info {inheritedWidgetCount=0, parentClass=&UserWidget, rootWidgetClass=&HorizontalBox, widgetCount=3}
 widgets [
  HorizontalBox_Content {widgetClassPath=&HorizontalBox, widget=HorizontalBox_Content}
  StatusIndicator_Locked {
   widgetClassPath=/FortUILibrary/UIKit/StatusIndicator/WBP_UIKit_StatusIndicator
   bIsVariable=true
   parent=HorizontalBox_Content
   widget=StatusIndicator_Locked
   properties {type=Utility, icon=/FortUILibrary/Textures/MaterialInstances/MI_UIKit_Icon_Locked, overrideStatusIndicatorVisibility=true}
   slot {class=HorizontalBoxSlot, properties {horizontalAlignment=HAlign_Left, verticalAlignment=VAlign_Center}}
  }
  WBP_UIKit_Highlight {
   widgetClassPath=/FortUILibrary/UIKit/Highlight/WBP_UIKit_Highlight
   bIsVariable=true
   parent=HorizontalBox_Content
   widget=WBP_UIKit_Highlight
   properties {highlightType=Utility}
   slot {
    class=HorizontalBoxSlot
    properties {horizontalAlignment=HAlign_Left, verticalAlignment=VAlign_Center, padding=<ltrb 4 0 0 0>, size {sizeRule=Fill, value=1}}
   }
  }
 ]
}
