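# Discovery UI favorite-icon widget that shows favorited state through layered material images (icon, drop shadow, pulse outline).
# SetFavorite stores the state, then either snaps the color on init or plays the favorite animation (bounce plus expanding pulse) or unfavorite animation (dip).
# Both animations key ColorInterpolationValue, whose matching SetColorInterpolationValue function lerps start and end colors into three channel parameters on the icon's dynamic material.
# PreConstruct grabs that material and applies the initial color.
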WidgetBlueprint /FortniteCoreUI/UI/DiscoveryShared/WBP_Discovery_Favorite
animations [
 OnFavoriteAnim {
  binding_count=3
  length_seconds=0.45
  object=&self:OnFavoriteAnim
  sequence {
   bindings [
    FavoriteIconOverlay {
     id=#1
     possessed_class=Overlay
     tracks [{
       type=MovieScene2DTransformTrack
       sections [{
         type=MovieScene2DTransformSection
         end_frame=45
         start_frame=0
         channels [
          Scale.X {type=MovieSceneScriptingFloatChannel, keys [[0,1],[8,0.875],[16,1.175],[32,1],[45,1]]}
          Scale.Y {type=MovieSceneScriptingFloatChannel, keys [[0,1],[8,0.875],[16,1.175],[32,1],[45,1]]}
         ]
        }]
      }]
    }
    PulseOutline {
     id=#2
     possessed_class=Image
     tracks [
      {
       type=MovieSceneFloatTrack
       display_name=Render Opacity
       sections [{type=MovieSceneFloatSection, channels [None {type=MovieSceneScriptingFloatChannel, keys [[0,0],[15,1],[23,1],[43,0]]}]}]
      }
      {
       type=MovieScene2DTransformTrack
       sections [{
         type=MovieScene2DTransformSection
         end_frame=28
         start_frame=0
         channels [
          Scale.X {type=MovieSceneScriptingFloatChannel, keys [[0,1],[28,2.5]]}
          Scale.Y {type=MovieSceneScriptingFloatChannel, keys [[0,1],[28,2.5]]}
         ]
        }]
      }
     ]
    }
    WBP_Discovery_Favorite_C_63 {
     id=#3
     possessed_class=WBP_Discovery_Favorite_C
     tracks [{
       type=MovieSceneDoubleTrack
       display_name=Color Interpolation Value
       sections [{type=MovieSceneDoubleSection, channels [None {type=MovieSceneScriptingDoubleChannel, keys [[0,0],[20,1]]}]}]
      }]
    }
   ]
   display_rate {numerator=100}
  }
 }
 OnUnfavoriteAnim {
  binding_count=2
  length_seconds=0.24
  object=&self:OnUnfavoriteAnim
  sequence {
   bindings [
    FavoriteIconOverlay {
     id=#1
     possessed_class=Overlay
     tracks [{
       type=MovieScene2DTransformTrack
       sections [{
         type=MovieScene2DTransformSection
         end_frame=24
         channels [
          Scale.X {type=MovieSceneScriptingFloatChannel, keys [[0,1],[12,0.875],[24.0017,1]]}
          Scale.Y {type=MovieSceneScriptingFloatChannel, keys [[0,1],[12,0.875],[24,1]]}
         ]
        }]
      }]
    }
    WBP_Discovery_Favorite_C_63 {
     id=#3
     possessed_class=WBP_Discovery_Favorite_C
     tracks [{
       type=MovieSceneDoubleTrack
       display_name=Color Interpolation Value
       sections [{type=MovieSceneDoubleSection, channels [None {type=MovieSceneScriptingDoubleChannel, keys [[0,1],[24,0]]}]}]
      }]
    }
   ]
   display_rate {numerator=100}
  }
 }
]
graphs [
 EventGraph {nodes [
   Event PreConstruct()
   > n1 = GetDynamicMaterial(self=Get Icon() -> (Icon:Image Object Reference))
   > Set DynamicMaterial(DynamicMaterial:Material Instance Dynamic Object Reference=@n1) -> (Output_Get:Material Instance Dynamic Object Reference)
   > SetColorInterpolationValue!(execute, self, FavoriteRatio:double="To Float (Boolean)"(InBool=Get IsFavorited() -> (IsFavorited:bool))) -> (then)
  ]}
 SetColorInterpolationValue {nodes [
   n1 = SetColorInterpolationValue!() -> (then, FavoriteRatio:double=~) {is=title K2Node_FunctionEntry}
   > SetVectorParameterValue(
    self=Get DynamicMaterial() -> (DynamicMaterial:Material Instance Dynamic Object Reference)
    ParameterName=Get InnerColor() -> (InnerColor:Name=None)
    Value="Lerp (LinearColor)"(
     A=Get StartInnerColor() -> (StartInnerColor:Linear Color Structure)
     B=Get EndInnerColor() -> (EndInnerColor:Linear Color Structure)
     Alpha=@n1
    )
   )
   > SetVectorParameterValue(
    self=Get DynamicMaterial()
    ParameterName=Get OuterColor() -> (OuterColor:Name=None)
    Value="Lerp (LinearColor)"(
     A=Get StartOuterColor() -> (StartOuterColor:Linear Color Structure)
     B=Get EndOuterColor() -> (EndOuterColor:Linear Color Structure)
     Alpha=@n1
    )
   )
   > SetVectorParameterValue(
    self=Get DynamicMaterial()
    ParameterName=Get OuterColor2() -> (OuterColor2:Name=None)
    Value="Lerp (LinearColor)"(A=Get StartOuterColor(), B=Get EndOuterColor(), Alpha=@n1)
   )
  ]}
 SetFavorite {nodes [
   n1 = SetFavorite!() -> (then, IsFavorite:bool=~, IsInit:bool=~) {is=title K2Node_FunctionEntry}
   > Set IsFavorited(IsFavorited:bool=@n1) -> (Output_Get:bool)
   > n2 = Branch(Condition=@n1.IsInit)
   n3 = Get IsFavorited()
   > SetColorInterpolationValue(FavoriteRatio="To Float (Boolean)"(InBool=@n3)) {is=title K2Node_CallFunction}
   n4 = Branch(execute=@n2.else, Condition=@n3)
   > PlayAnimationForward(InAnimation=Get OnFavoriteAnim() -> (OnFavoriteAnim:Widget Animation Object Reference))
   PlayAnimationForward(execute=@n4.else, InAnimation=Get OnUnfavoriteAnim() -> (OnUnfavoriteAnim:Widget Animation Object Reference))
  ]}
]
variables [
 ColorInterpolationValue {type=number}
 DynamicMaterial {type=MaterialInstanceDynamic}
 EndInnerColor {type=LinearColor, default=<rgba 0.7913 0.0086 0.1845 1>}
 EndOuterColor {type=LinearColor, default=<rgba 1 0.3176 0.5843 1>}
 InnerColor {type=string, default=Red Channel Color}
 IsFavorited {type=boolean}
 OuterColor {type=string, default=Green Channel Color}
 OuterColor2 {type=string, default=Blue Channel Color}
 StartInnerColor {type=LinearColor, default=<rgba 0.1746 0.1946 0.2159 1>}
 StartOuterColor {type=LinearColor, default=<rgba 1 1 1 1>}
]
widgets {
 info {inheritedWidgetCount=0, parentClass=&UserWidget, rootWidgetClass=&Overlay, widgetCount=4}
 widgets [
  FavoriteIconOverlay {widgetClassPath=&Overlay, widget=FavoriteIconOverlay}
  PulseOutline {
   widgetClassPath=&Image
   parent=FavoriteIconOverlay
   widget=PulseOutline
   properties {
    visibility=SelfHitTestInvisible
    brush {
     bIsDynamicallyLoaded=false
     drawAs=Image
     imageType=NoImage
     mirroring=NoMirror
     resourceName=None
     resourceObject=/FortUILibrary/Foundation/Buttons/Materials/Favorite/MI_UI_Icon_Favorite_Pulse
     tiling=NoTile
     imageSize=<32 32>
     margin=<ltrb 0 0 0 0>
     outlineSettings {
      bUseBrushTransparency=false
      roundingType=HalfHeightRadius
      width=0
      color {colorUseRule=UseColor_Specified, specifiedColor=<rgba 0 0 0 0>}
      cornerRadii=<0 0 0 0>
     }
     tintColor {colorUseRule=UseColor_Specified, specifiedColor=<rgba 1 1 1 1>}
     uVRegion {bIsValid=false, max=<0 0>, min=<0 0>}
    }
   }
   slot {class=OverlaySlot, properties {horizontalAlignment=HAlign_Fill, verticalAlignment=VAlign_Fill}}
  }
  DropShadow {
   widgetClassPath=&Image
   parent=FavoriteIconOverlay
   widget=DropShadow
   properties {
    visibility=SelfHitTestInvisible
    brush {
     bIsDynamicallyLoaded=false
     drawAs=Image
     imageType=NoImage
     mirroring=NoMirror
     resourceName=None
     resourceObject=/FortUILibrary/Foundation/Buttons/Materials/Favorite/MI_UI_Icon_Favorite_DropShadow
     tiling=NoTile
     imageSize=<32 32>
     margin=<ltrb 0 0 0 0>
     outlineSettings {
      bUseBrushTransparency=false
      roundingType=HalfHeightRadius
      width=0
      color {colorUseRule=UseColor_Specified, specifiedColor=<rgba 0 0 0 0>}
      cornerRadii=<0 0 0 0>
     }
     tintColor {colorUseRule=UseColor_Specified, specifiedColor=<rgba 1 1 1 1>}
     uVRegion {bIsValid=false, max=<0 0>, min=<0 0>}
    }
   }
   slot {class=OverlaySlot, properties {horizontalAlignment=HAlign_Fill, verticalAlignment=VAlign_Fill}}
  }
  Icon {
   widgetClassPath=&Image
   bIsVariable=true
   parent=FavoriteIconOverlay
   widget=Icon
   properties {
    visibility=SelfHitTestInvisible
    brush {
     bIsDynamicallyLoaded=false
     drawAs=Image
     imageType=NoImage
     mirroring=NoMirror
     resourceName=None
     resourceObject=/FortUILibrary/Foundation/Buttons/Materials/Favorite/MI_UI_Icon_Favorite
     tiling=NoTile
     imageSize=<32 32>
     margin=<ltrb 0 0 0 0>
     outlineSettings {
      bUseBrushTransparency=false
      roundingType=HalfHeightRadius
      width=0
      color {colorUseRule=UseColor_Specified, specifiedColor=<rgba 0 0 0 0>}
      cornerRadii=<0 0 0 0>
     }
     tintColor {colorUseRule=UseColor_Specified, specifiedColor=<rgba 1 1 1 1>}
     uVRegion {bIsValid=false, max=<0 0>, min=<0 0>}
    }
   }
   slot {class=OverlaySlot, properties {horizontalAlignment=HAlign_Fill, verticalAlignment=VAlign_Fill}}
  }
 ]
}
