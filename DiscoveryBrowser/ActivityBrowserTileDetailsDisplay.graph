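# Widget Blueprint for an activity tile in the Discovery browser, showing key art, title, creator banner, social counts, favorite, lock, purchase and invalid-activity states.
# Parent-class events for image loads, activity updates and tile activation refresh the visuals, driving material-parameter animations and fades.
# Activation plays hover, pulse and, outside promoted hero rows, grow animations plus tile video.
# Tile sizing derives from grid columns with mobile scaling; clicks on purchase-required tiles trigger warning animations.
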WidgetBlueprint /DiscoveryBrowser/ActivityBrowserTileDetailsDisplay
animations [
 Anim_OnHoverSelect {
  binding_count=7
  length_seconds=0.16
  object=&self:Anim_OnHoverSelect
  sequence {
   bindings [
    TileThumbnail {
     id=#1
     possessed_class=Image
     tracks [{
       type=MovieSceneWidgetMaterialTrack
       sections [{
         type=MovieSceneParameterSection
         channels [
          Banner Offset {type=MovieSceneScriptingFloatChannel}
          Banner Rotation {type=MovieSceneScriptingFloatChannel}
          Thumbnail Zoom {type=MovieSceneScriptingFloatChannel}
          Banner ColorR {type=MovieSceneScriptingFloatChannel, keys [[0,0],[10,1]]}
          Banner ColorG {type=MovieSceneScriptingFloatChannel, keys [[0,0.0941],[10,1]]}
          Banner ColorB {type=MovieSceneScriptingFloatChannel, const [1,2]}
          Banner ColorA {type=MovieSceneScriptingFloatChannel, const [1,2]}
         ]
        }]
      }]
    }
    TileStroke {
     id=#2
     possessed_class=Image
     tracks [{
       type=MovieSceneWidgetMaterialTrack
       sections [{
         type=MovieSceneParameterSection
         end_frame=16
         start_frame=0
         channels [
          Outline Size Ratio {type=MovieSceneScriptingFloatChannel, keys [[0,0],[16,1]]}
          Max Outline Size {type=MovieSceneScriptingFloatChannel, const [0.4,1]}
         ]
        }]
      }]
    }
    BannerText {
     id=#3
     possessed_class=CommonTextBlock
     tracks [{
       type=MovieSceneWidgetMaterialTrack
       display_name=Font.FontMaterial
       sections [{type=MovieSceneParameterSection, channels [Hover_Animate {type=MovieSceneScriptingFloatChannel, keys [[0,0],[16,1]]}]}]
      }]
    }
    OverlaySlot_2 {
     id=#4
     parent_id=#3
     possessed_class=OverlaySlot
     tracks [{
       type=MovieSceneMarginTrack
       display_name=Padding
       sections [{
         type=MovieSceneMarginSection
         end_frame=16
         start_frame=0
         channels [
          Left {type=MovieSceneScriptingFloatChannel, const [0,1]}
          Top {type=MovieSceneScriptingFloatChannel, keys [[0,0],[16,-2]]}
          Right {type=MovieSceneScriptingFloatChannel, const [0,1]}
          Bottom {type=MovieSceneScriptingFloatChannel, const [0,1]}
         ]
        }]
      }]
    }
    OverlaySlot_0 {
     id=#5
     parent_id=#2
     possessed_class=OverlaySlot
     tracks [{
       type=MovieSceneMarginTrack
       display_name=Padding
       sections [{
         type=MovieSceneMarginSection
         end_frame=16
         start_frame=0
         channels [
          Left {type=MovieSceneScriptingFloatChannel, keys [[0,0],[16,-4]]}
          Top {type=MovieSceneScriptingFloatChannel, keys [[0,0],[16,-4]]}
          Right {type=MovieSceneScriptingFloatChannel, keys [[0,0],[16,-4]]}
          Bottom {type=MovieSceneScriptingFloatChannel, keys [[0,0],[16,-4]]}
         ]
        }]
      }]
    }
    ActivityBrowserTileDetailsDisplay_C_5 {
     id=#6
     possessed_class=ActivityBrowserTileDetailsDisplay_C
     tracks [{
       type=MovieSceneDoubleTrack
       display_name=Banner Color Animator
       sections [{type=MovieSceneDoubleSection, end_frame=8, start_frame=0, channels [None {type=MovieSceneScriptingDoubleChannel, keys [[0,0],[8,1]]}]}]
      }]
    }
    WBP_Discovery_WarningIcon {
     id=#7
     possessed_class=WBP_Discovery_WarningIcon_C
     tracks [
      {
       type=MovieSceneFloatTrack
       display_name=Render Opacity
       sections [{type=MovieSceneFloatSection, channels [None {type=MovieSceneScriptingFloatChannel, keys [[0,0],[16,1]]}]}]
      }
      {
       type=MovieScene2DTransformTrack
       sections [{
         type=MovieScene2DTransformSection
         channels [
          Translation.Y {type=MovieSceneScriptingFloatChannel, keys [[0,12],[16,0]]}
          Scale.X {type=MovieSceneScriptingFloatChannel, keys [[0,0.5],[16,1]]}
          Scale.Y {type=MovieSceneScriptingFloatChannel, keys [[0,0.5],[16,1]]}
         ]
        }]
      }
     ]
    }
   ]
   display_rate {numerator=100}
  }
 }
 Anim_OnHoverSelect_GrowNonHeroRowTiles {
  binding_count=7
  length_seconds=0.16
  object=&self:Anim_OnHoverSelect_GrowNonHeroRowTiles
  sequence {
   bindings [
    HorizontalBox_ActivityInfo {id=#8, possessed_class=HorizontalBox, tracks []}
    Ovr_GrowOnHoverSelect {id=#9, possessed_class=Overlay, tracks []}
    OverlaySlot_4 {
     id=#10
     parent_id=#11
     possessed_class=OverlaySlot
     tracks [{
       type=MovieSceneMarginTrack
       display_name=Padding
       sections [{
         type=MovieSceneMarginSection
         end_frame=38
         start_frame=0
         channels [
          Left {type=MovieSceneScriptingFloatChannel, keys [[0,0],[38.4,12]]}
          Top {type=MovieSceneScriptingFloatChannel, const [0,1]}
          Right {type=MovieSceneScriptingFloatChannel, keys [[0,0],[38.4,12]]}
          Bottom {type=MovieSceneScriptingFloatChannel, keys [[0,0],[38.4,-4]]}
         ]
        }]
      }]
    }
    ActivityBrowserSocialProof {id=#11, possessed_class=WBP_ActivityBrowserSocialProof_C, tracks []}
    OverlaySlot_2 {
     id=#12
     parent_id=#8
     possessed_class=OverlaySlot
     tracks [{
       type=MovieSceneMarginTrack
       display_name=Padding
       sections [{
         type=MovieSceneMarginSection
         end_frame=38
         start_frame=0
         channels [
          Left {type=MovieSceneScriptingFloatChannel, keys [[0,0],[38.4,12]]}
          Top {type=MovieSceneScriptingFloatChannel, const [0,1]}
          Right {type=MovieSceneScriptingFloatChannel, keys [[0,0],[38.4,12]]}
          Bottom {type=MovieSceneScriptingFloatChannel, keys [[0,0],[38.4,-4]]}
         ]
        }]
      }]
    }
    SqueegeeSlot_Lock {
     id=#13
     possessed_class=SqueegeeInjectionSlot_C
     tracks [{
       type=MovieScene2DTransformTrack
       sections [{
         type=MovieScene2DTransformSection
         end_frame=38
         start_frame=0
         channels [
          Translation.X {type=MovieSceneScriptingFloatChannel}
          Translation.Y {type=MovieSceneScriptingFloatChannel}
          Angle {type=MovieSceneScriptingFloatChannel}
          Scale.X {type=MovieSceneScriptingFloatChannel, keys [[0,1],[38.4,1.2]]}
          Scale.Y {type=MovieSceneScriptingFloatChannel, keys [[0,1],[38.4,1.2]]}
          Shear.X {type=MovieSceneScriptingFloatChannel}
          Shear.Y {type=MovieSceneScriptingFloatChannel}
         ]
        }]
      }]
    }
    SizeBoxSlot_0 {
     id=#14
     parent_id=#9
     possessed_class=SizeBoxSlot
     tracks [{
       type=MovieSceneMarginTrack
       display_name=Padding
       sections [{
         type=MovieSceneMarginSection
         channels [
          Left {type=MovieSceneScriptingFloatChannel, keys [[0,0],[38.4,-12]]}
          Top {type=MovieSceneScriptingFloatChannel, keys [[0,0],[38.4,-6]]}
          Right {type=MovieSceneScriptingFloatChannel, keys [[0,0],[38.4,-12]]}
          Bottom {type=MovieSceneScriptingFloatChannel, keys [[0,0],[38.4,-6]]}
         ]
        }]
      }]
    }
   ]
   display_rate {numerator=240}
  }
 }
 Anim_OnIntro {
  binding_count=1
  length_seconds=0.24
  object=&self:Anim_OnIntro
  sequence {
   bindings [SizeBox_Primary {
     id=#15
     possessed_class=SizeBox
     tracks [
      {
       type=MovieScene2DTransformTrack
       sections [{type=MovieScene2DTransformSection, channels [Translation.Y {type=MovieSceneScriptingFloatChannel, keys [[0,32],[24,0]]}]}]
      }
      {
       type=MovieSceneFloatTrack
       display_name=Render Opacity
       sections [{type=MovieSceneFloatSection, channels [None {type=MovieSceneScriptingFloatChannel, keys [[0,0],[24,1]]}]}]
      }
     ]
    }]
   display_rate {numerator=100}
  }
 }
 Anim_HoverPulse {
  binding_count=2
  length_seconds=1.12
  object=&self:Anim_HoverPulse
  sequence {
   bindings [
    OverlaySlot_5 {
     id=#16
     parent_id=#17
     possessed_class=OverlaySlot
     tracks [{
       type=MovieSceneMarginTrack
       display_name=Padding
       sections [{
         type=MovieSceneMarginSection
         end_frame=112
         start_frame=0
         channels [
          Left {type=MovieSceneScriptingFloatChannel, keys [[24,0],[64,-9],[104,-9],[112,0]]}
          Top {type=MovieSceneScriptingFloatChannel, keys [[24,0],[64,-9],[104,-9],[112,0]]}
          Right {type=MovieSceneScriptingFloatChannel, keys [[24,0],[64,-9],[104,-9],[112,0]]}
          Bottom {type=MovieSceneScriptingFloatChannel, keys [[24,0],[64,-9],[104,-9],[112,0]]}
         ]
        }]
      }]
    }
    PulseOutline {
     id=#17
     possessed_class=Image
     tracks [
      {
       type=MovieSceneByteTrack
       display_name=Visibility
       sections [{type=MovieSceneByteSection, end_frame=112, start_frame=0, channels [None {type=MovieSceneScriptingByteChannel, keys [[0,3],[112,1]]}]}]
      }
      {
       type=MovieSceneWidgetMaterialTrack
       sections [{
         type=MovieSceneParameterSection
         end_frame=112
         start_frame=0
         channels [
          Outline Size Ratio {type=MovieSceneScriptingFloatChannel, keys [[24,0],[64,1],[104,0]]}
          Outline Opacity {type=MovieSceneScriptingFloatChannel, keys [[24,0],[64,0.5],[104,0]]}
         ]
        }]
      }
     ]
    }
   ]
   display_rate {numerator=100}
  }
 }
 Anim_OnImageLoaded {
  binding_count=1
  length_seconds=0.24
  object=&self:Anim_OnImageLoaded
  sequence {
   bindings [TileThumbnail {
     id=#18
     possessed_class=Image
     tracks [{
       type=MovieSceneWidgetMaterialTrack
       sections [{
         type=MovieSceneParameterSection
         end_frame=24
         start_frame=0
         channels [
          Thumbnail Alpha {type=MovieSceneScriptingFloatChannel, keys [[0,0],[24,1]]}
          Thumbnail Scale {type=MovieSceneScriptingFloatChannel}
          Thumbnail Zoom {type=MovieSceneScriptingFloatChannel, keys [[0,1.05],[24,1]]}
         ]
        }]
      }]
    }]
   display_rate {numerator=100}
  }
 }
 Anim_OnHoverUnhoverRehover {
  binding_count=1
  length_seconds=0.2
  object=&self:Anim_OnHoverUnhoverRehover
  sequence {bindings [TileThumbnail {
     id=#19
     possessed_class=Image
     tracks [{
       type=MovieSceneWidgetMaterialTrack
       sections [{
         type=MovieSceneParameterSection
         channels [
          Thumbnail Scale {type=MovieSceneScriptingFloatChannel}
          Thumbnail Zoom {type=MovieSceneScriptingFloatChannel, keys [[0,1],[4,1.02]]}
         ]
        }]
      }]
    }]}
 }
 Anim_OnDisabledClick {
  binding_count=1
  length_seconds=0.16
  object=&self:Anim_OnDisabledClick
  sequence {
   bindings [DisabledWarningMessage {
     id=#20
     possessed_class=CommonTextBlock
     tracks [
      {
       type=MovieScene2DTransformTrack
       sections [{type=MovieScene2DTransformSection, channels [Translation.Y {type=MovieSceneScriptingFloatChannel, keys [[0,8],[16,0]]}]}]
      }
      {
       type=MovieSceneFloatTrack
       display_name=Render Opacity
       sections [{type=MovieSceneFloatSection, channels [None {type=MovieSceneScriptingFloatChannel, keys [[0,0],[16,1]]}]}]
      }
      {
       type=MovieSceneByteTrack
       display_name=Visibility
       sections [{type=MovieSceneByteSection, channels [None {type=MovieSceneScriptingByteChannel, const [0,1]}]}]
      }
     ]
    }]
   display_rate {numerator=100}
  }
 }
]
graphs [
 ColumnSizeToTileSize {nodes [
   n1 = ColumnSizeToTileSize!() -> (then, NewColumnSize:int) {is=title K2Node_FunctionEntry}
   n2 = float + float(
    A=float * float(A:int=@n1, B=Get GridSingleColumnWidth() -> (GridSingleColumnWidth:double)) {is=title K2Node_PromotableOperator}
    B=float * float(A=int - int(A=@n1), B=Get GridSingleGutterWidth() -> (GridSingleGutterWidth:double))
   )
   n3 = Ceil(A=float * float(A:double=@n2, B=Get Ratio16x9ConvertWidthToHeight() -> (Ratio16x9ConvertWidthToHeight:double))) {is=title K2Node_CallFunction}
   > Return Node!(execute, NewTileSize:Vector 2D Structure=MakeVector2D(X=@n2, Y="To Float (Integer)"(InInt=@n3)))
  ]}
 EventGraph {nodes [
   DisableClickTimerExpired() {is=title K2Node_CustomEvent}
   > PlayAnimationReverse(InAnimation=Get Anim_OnDisabledClick() -> (Anim_OnDisabledClick:Widget Animation Object Reference))
   TestActivityTileRequirePurchaseTrue() {is=title K2Node_CustomEvent}
   > UpdateRequiresPurchase!(execute, self, IsPurchaseRequired:bool=true) -> (then)
   TestActivityTileRequirePurchaseFalse() {is=title K2Node_CustomEvent}
   > UpdateRequiresPurchase(IsPurchaseRequired=false)
   Event On Clicked()
   > n1 = Branch(Condition=DoesActivityRequirePurchase!(self:Fort Activity Tile Details Display Object Reference) -> (ReturnValue:bool))
   n2 = Branch(
    execute=@n1.else
    Condition=Is Timer Active by Handle!(Handle:Timer Handle Structure=Get DisabledClickTimer() -> (DisabledClickTimer:Timer Handle Structure)) -> (ReturnValue:bool)
   )
   TriggerWarningAnim!(
    execute=[@n4.then, @n2.then]
    self:WBP Discovery Warning Icon Object Reference=Get WBP_Discovery_WarningIcon() -> (WBP_Discovery_WarningIcon:WBP Discovery Warning Icon Object Reference)
    NumLoopsToPlay:int=1
   ) -> (then)
   > n3 = Set Timer by Event(Delegate=Create Event(), Time=Get DisabledClickDuration() -> (DisabledClickDuration:double), bLooping=false)
   > Set DisabledClickTimer(DisabledClickTimer:Timer Handle Structure=@n3) -> (Output_Get:Timer Handle Structure)
   n4 = PlayAnimationForward(execute=@n2.else, InAnimation=Get Anim_OnDisabledClick())
   n5 = Event On Hovered()
   Branch(execute=[@n5.then, @n44.then], Condition=Get bAllowThumbnailHoverRehoverAnim() -> (bAllowThumbnailHoverRehoverAnim:bool))
   n6 = Get Anim_OnHoverUnhoverRehover() -> (Anim_OnHoverUnhoverRehover:Widget Animation Object Reference)
   > PlayAnimationForward(InAnimation=@n6)
   n7 = Event On Unhovered()
   Branch(execute=[@n7.then, @n48.then], Condition=Get bAllowThumbnailHoverRehoverAnim())
   > PlayAnimationReverse(InAnimation=@n6)
   n8 = Event UpdateSqueegeeWidgets!() -> (OutputDelegate, then, GameActivity:Fort Game Activity Object Reference) {is=title K2Node_Event}
   > ShowActivityData!(
    execute
    self:Fort Squeegee Widget Injection Slot Object Reference=Get SqueegeeSlot_Lock() -> (SqueegeeSlot_Lock:Squeegee Injection Slot Object Reference)
    GameActivity:Fort Game Activity Object Reference=@n8.GameActivity
   ) -> (then)
   n9 = Event PreConstruct()
   > Branch(Condition=@n9.IsDesignTime)
   > SetHideDetails!(execute, self, Hide Details:bool=Get HideDetails() -> (HideDetails:bool)) -> (then)
   > SetTileSize!(execute, self, Size:Vector 2D Structure=Get TileSize() -> (TileSize:Vector 2D Structure)) -> (then)
   > UpdateEpicActivityStyling!(
    execute
    self
    CreatorDisplayName:String=GetActivityCreatorDisplayText!(self:Fort Activity Tile Details Display Object Reference) -> (ReturnValue:String)
   ) -> (then)
   > Set Game Mode Text Size!(execute, self, FontSize:int=Get DefaultGameModeTextSize() -> (DefaultGameModeTextSize:int=0)) -> (then) {is=title K2Node_CallFunction}
   n10 = Event OnLogoImageChanged!() -> (OutputDelegate, then, bIsLoading:bool, Texture:Texture Object Reference) {is=title K2Node_Event}
   > n11 = Branch(Condition=NOT Boolean(A=@n10.bIsLoading))
   > n12 = Is Valid(InputObject=@n10.Texture)
   n13 = Get Logo_NotReady_Fort554488() -> (Logo_NotReady_Fort554488:Image Object Reference)
   SetVisibility(execute=[@n16.CastFailed, @n11.else, @n12.Is Not Valid], self=@n13, InVisibility=Hidden)
   n14 = Cast To Texture2DDynamic!(execute=@n12.Is Valid, Object=@n10.Texture) -> (then, CastFailed, AsTexture 2DDynamic:Texture 2DDynamic Object Reference) {is=title K2Node_DynamicCast}
   > n15 = SetBrushFromTextureDynamic!(execute, self:Image Object Reference=@n13, Texture:Texture 2DDynamic Object Reference=@n14, bMatchSize:bool=false) -> (then)
   SetVisibility(execute=[@n15.then, @n17.then], self=@n13, InVisibility=Visible)
   n16 = Cast To Texture2D(execute=@n14.CastFailed, Object=@n10.Texture)
   > n17 = SetBrushFromTexture!(execute, self:Image Object Reference=@n13, Texture:Texture 2D Object Reference=@n16, bMatchSize:bool=false) -> (then)
   n18 = Event UpdateCCU!() -> (OutputDelegate, then, CCUCount:int=0) {is=title K2Node_Event}
   > Set CCU Count Data!(
    execute
    self:WBP Activity Browser Social Proof Object Reference=Get ActivityBrowserSocialProof() -> (ActivityBrowserSocialProof:WBP Activity Browser Social Proof Object Reference)
    CCU Count:int=@n18.CCUCount
    Hide:bool=Get HideDetails()
   ) -> (then) {is=title K2Node_CallFunction}
   n19 = Event OnFriendsPlayingChanged!() -> (OutputDelegate, then, NumPlaying:int=0) {is=title K2Node_Event}
   > Set Friends Count Data!(
    execute
    self:WBP Activity Browser Social Proof Object Reference=Get ActivityBrowserSocialProof()
    Friends Count:int=@n19.NumPlaying
    Hide:bool=OR Boolean(A=IsActivityLocked!(self:Fort Activity Tile Details Display Object Reference) -> (ReturnValue:bool), B=Get HideDetails())
   ) -> (then) {is=title K2Node_CallFunction}
   Event Construct()
   > Parent: Construct()
   > n20 = GetDynamicMaterial(self=Get TileThumbnail() -> (TileThumbnail:Image Object Reference))
   > SetScalarParameterValue(self=@n20, ParameterName=Get ThumbnailAlpha() -> (ThumbnailAlpha:Name=None), Value=0)
   > n21 = Get Texture Parameter Value!(execute, self:Material Instance Dynamic Object Reference=@n20, ParameterName:Name=Get TextureParam() -> (TextureParam:Name=None)) -> (then, ReturnValue:Texture Object Reference) {is=title K2Node_CallFunction}
   > Set DefaultImage(DefaultImage:Texture Object Reference=@n21) -> (Output_Get:Texture Object Reference)
   > SetVisibility(self=Get Logo_NotReady_Fort554488(), InVisibility=Hidden)
   > n22 = Branch(Condition=IsMobileGame())
   > SetVisibility(self=Get Spacer_Mobile() -> (Spacer_Mobile:Spacer Object Reference), InVisibility=SelfHitTestInvisible)
   > SetPadding!(
    execute
    self:Overlay Slot Object Reference=SlotAsOverlaySlot(Widget=Get HorizontalBox_ActivityInfo() -> (HorizontalBox_ActivityInfo:Horizontal Box Object Reference))
    InPadding_Left:float=0
    InPadding_Top:float=0
    InPadding_Right:float=0
    InPadding_Bottom:float=-5
   ) -> (then)
   > SetPadding(self=SlotAsOverlaySlot(Widget=Get ActivityBrowserSocialProof()), InPadding_Left=0, InPadding_Top=0, InPadding_Right=0, InPadding_Bottom=-5)
   > Set Game Mode Text Size(FontSize=Get DefaultGameModeTextSize() -> (DefaultGameModeTextSize=0))
   SetVisibility(execute=@n22.else, self=Get Spacer_Mobile(), InVisibility=Collapsed)
   > SetPadding(self=SlotAsOverlaySlot(Widget=Get HorizontalBox_ActivityInfo()), InPadding_Left=0, InPadding_Top=0, InPadding_Right=0, InPadding_Bottom=0)
   > SetPadding(self=SlotAsOverlaySlot(Widget=Get ActivityBrowserSocialProof()), InPadding_Left=0, InPadding_Top=0, InPadding_Right=0, InPadding_Bottom=0)
   > SetVisibility(
    self=Get Button_Details() -> (Button_Details:Common Button Base Object Reference)
    InVisibility=Select(Option 0=Collapsed, Option 1=Visible, Index=Get bShowDetailsButton() -> (bShowDetailsButton:bool))
   )
   > SetFavoriteVisibility!(execute, self, IsFavorited:bool=IsActivityFavorited!(self:Fort Activity Tile Details Display Object Reference) -> (ReturnValue:bool)) -> (then)
   > Set Game Mode Text Size Override() {is=title K2Node_CallFunction}
   Event OnDetailsUpdated() {is=title K2Node_Event}
   > Set IsActivityValid(IsActivityValid:bool=true) -> (Output_Get:bool)
   > Check and Stop Animation!(execute:Exec, Animation:Widget Animation Object Reference=Get Anim_OnHoverUnhoverRehover()) -> (then:Exec) {is=title K2Node_MacroInstance}
   > Check and Stop Animation(Animation=Get Anim_HoverPulse() -> (Anim_HoverPulse:Widget Animation Object Reference))
   > UpdateRequiresPurchase(IsPurchaseRequired=DoesActivityRequirePurchase())
   > ResetOnHoverSelectAnimation()
   > ResetOnImageLoadedAnimation!(execute, self, IsLoaded:bool=Get IsKeyArtValid() -> (IsKeyArtValid:bool)) -> (then)
   > UpdateEpicActivityStyling(CreatorDisplayName=GetActivityCreatorDisplayText())
   > SetVisibility(self=Get SqueegeeSlot_Lock(), InVisibility=Select(Option 0=Collapsed, Option 1=HitTestInvisible, Index=IsActivityLocked()))
   > SetVisibility(
    self=Get Image_MultiMode() -> (Image_MultiMode:Image Object Reference)
    InVisibility=Select(
     Option 0=Collapsed
     Option 1=SelfHitTestInvisible
     Index=IsModeSetActivity!(self:Fort Activity Tile Details Display Object Reference) -> (ReturnValue:bool)
    )
   )
   > n23 = Cast To WBP_Discovery_DetailsButton!(execute, Object=Get Button_Details()) -> (then, CastFailed, AsWBP Discovery Details Button:WBP Discovery Details Button Object Reference) {is=title K2Node_DynamicCast}
   > OnTileReset(self:WBP Discovery Details Button Object Reference=@n23)
   > SetFavoriteVisibility(IsFavorited=IsActivityFavorited())
   n24 = Event OnRequiresPurchaseChanged!() -> (OutputDelegate, then, bRequiresPurchase:bool) {is=title K2Node_Event}
   > UpdateRequiresPurchase(IsPurchaseRequired=@n24.bRequiresPurchase)
   n25 = Event OnIsFavoriteChanged!() -> (OutputDelegate, then, bIsFavorite:bool) {is=title K2Node_Event}
   > SetFavoriteVisibility(IsFavorited=@n25.bIsFavorite)
   n26 = Event OnUpdateColumnSize!() -> (OutputDelegate, then, NewColumnSize:int=0) {is=title K2Node_Event}
   > SetTileSize(Size=ColumnSizeToTileSize!(NewColumnSize:int=@n26.NewColumnSize) -> (NewTileSize:Vector 2D Structure) {is=title K2Node_CallFunction})
   n27 = Event OnPreviewImageChanged!() -> (OutputDelegate, then, bIsLoading:bool, Texture:Texture Object Reference) {is=title K2Node_Event}
   > n28 = Branch(Condition=NOT Boolean(A=@n27.bIsLoading))
   > n29 = Is Valid(InputObject=@n27.Texture)
   n30 = GetDynamicMaterial(execute=[@n29.Is Not Valid, @n28.else], self=Get TileThumbnail())
   > SetScalarParameterValue(self=@n30, ParameterName=Get ThumbnailAlpha() -> (ThumbnailAlpha=None), Value=0)
   > Get Texture Parameter Value(self=@n30, ParameterName=Get TextureParam() -> (TextureParam=None))
   > Set IsKeyArtValid(IsKeyArtValid:bool=false) -> (Output_Get:bool)
   n31 = GetDynamicMaterial(execute=@n29.Is Valid, self=Get TileThumbnail())
   > SetTextureParameterValue(self=@n31, ParameterName=Get TextureParam() -> (TextureParam=None), Value=@n27.Texture)
   > n32 = Branch(Condition=Get IsKeyArtValid())
   PlayAnimation(execute=@n32.else, InAnimation=Get Anim_OnImageLoaded() -> (Anim_OnImageLoaded:Widget Animation Object Reference))
   > Set IsKeyArtValid(IsKeyArtValid=true)
   n33 = Event OnTileActiveSet!() -> (OutputDelegate, then, bIsTileActive:bool) {is=title K2Node_Event}
   > Set IsTileActive(IsTileActive:bool=@n33.bIsTileActive) -> (Output_Get:bool)
   > SetDisabledText()
   n34 = Get IsTileActive() -> (IsTileActive:bool)
   > On Lock Hover Select!(
    execute
    self:WBP Locked Status Object Reference=Get WBP_LockedStatus() -> (WBP_LockedStatus:WBP Locked Status Object Reference)
    bIsHoverSelected:bool=@n34
   ) -> (then)
   > Grow Tile on Select for Non Hero Rows()
   > n35 = Branch(Condition=@n34)
   n36 = ShouldPlayTileVideo!() -> (bOutResult:bool)
   > n37 = Branch(Condition=@n36)
   > n38 = StartTileVideo(self:Fort Activity Tile Details Display Object Reference)
   n39 = Get Anim_OnHoverSelect() -> (Anim_OnHoverSelect:Widget Animation Object Reference)
   PlayAnimation(execute=[@n38.then, @n37.else], InAnimation=@n39)
   n40 = Get Text_ActivityName() -> (Text_ActivityName:Common Text Object Reference)
   > SetScrollingEnabled!(execute, self:Common Text Object Reference=@n40, bInIsScrollingEnabled:bool=true) -> (then)
   n41 = Get Anim_HoverPulse()
   > PlayAnimationForward(InAnimation=@n41)
   > SetDetailsButtonAvailable()
   > n42 = Cast To WBP_Discovery_DetailsButton(Object=Get Button_Details())
   > OnTileFocused(self:WBP Discovery Details Button Object Reference=@n42)
   n43 = IsConsolePlatform!() -> (ReturnValue:bool)
   > n44 = Branch(Condition=@n43)
   n45 = Branch(execute=@n35.else, Condition=@n36)
   > n46 = StopTileVideo(self:Fort Activity Tile Details Display Object Reference)
   PlayAnimation(execute=[@n46.then, @n45.else], InAnimation=@n39, PlayMode=Reverse)
   > SetScrollingEnabled(self=@n40, bInIsScrollingEnabled=false)
   > Check and Stop Animation(Animation=@n41)
   > SetDetailsButtonAvailable()
   > n47 = Cast To WBP_Discovery_DetailsButton(Object=Get Button_Details())
   > OnTileLostFocus(self:WBP Discovery Details Button Object Reference=@n47)
   > n48 = Branch(Condition=@n43)
   PlayAnimationForward(InAnimation=Get Anim_OnIntro() -> (Anim_OnIntro:Widget Animation Object Reference))
   ShouldDecreaseFontSizeInDiscoveryTileTitles!(self:Fort Experiment Context Object Reference=Get FortExperimentContext!() -> (ReturnValue:Fort Experiment Context Object Reference) {is=title K2Node_GetBlueprintContextEx}) -> (ReturnValue:bool)
   Get DecreasedFontSizeInDiscoveryTileTitles!(self:Fort Runtime Options Object Reference=GetRuntimeOptions()) -> (DecreasedFontSizeInDiscoveryTileTitles:int=0)
  ]}
 Grow Tile on Select for Non Hero Rows {nodes [
   Grow Tile on Select for Non Hero Rows() {is=title K2Node_FunctionEntry}
   > Branch(Condition=NOT Boolean(A=Get IsPromotedHeroRow() -> (IsPromotedHeroRow:bool)))
   > PlayAnimation(
    InAnimation=Get Anim_OnHoverSelect_GrowNonHeroRowTiles() -> (Anim_OnHoverSelect_GrowNonHeroRowTiles:Widget Animation Object Reference)
    PlayMode=Select(Option 0:EUMGSequencePlayMode Enum=Reverse, Option 1:EUMGSequencePlayMode Enum, Index=Get IsTileActive()) -> (ReturnValue:EUMGSequencePlayMode Enum)
   )
  ]}
 ResetOnHoverSelectAnimation {nodes [
   ResetOnHoverSelectAnimation() {is=title K2Node_FunctionEntry}
   n1 = Get BannerText() -> (BannerText:Common Text Object Reference)
   > SetPadding!(execute, self=SlotAsOverlaySlot(Widget=@n1), InPadding:Margin Structure=Make Margin(Left=0, Top=0, Right=0, Bottom=0)) -> (then)
   > n2 = GetDynamicFontMaterial!(execute, self:Text Object Reference=@n1) -> (then, ReturnValue:Material Instance Dynamic Object Reference)
   > SetScalarParameterValue(self=@n2, ParameterName=Hover_Animate, Value=0)
   > SetRenderOpacity(self=Get WBP_Discovery_WarningIcon(), InOpacity=0)
   > n3 = GetDynamicMaterial(self=Get TileStroke() -> (TileStroke:Image Object Reference))
   > SetScalarParameterValue(self=@n3, ParameterName=Outline Size Ratio, Value=0)
  ]}
 ResetOnImageLoadedAnimation {nodes [
   n1 = ResetOnImageLoadedAnimation!() -> (then, IsLoaded:bool=~) {is=title K2Node_FunctionEntry}
   > n2 = Branch(Condition=@n1)
   n3 = Get TileThumbnail()
   > n4 = GetDynamicMaterial(self=@n3)
   n5 = Get ThumbnailAlpha() -> (ThumbnailAlpha=None)
   > SetScalarParameterValue(self=@n4, ParameterName=@n5, Value=1)
   n6 = Get ThumbnailScale() -> (ThumbnailScale:Name=None)
   > SetScalarParameterValue(self=@n4, ParameterName=@n6, Value=1)
   n7 = GetDynamicMaterial(execute=@n2.else, self=@n3)
   > SetScalarParameterValue(self=@n7, ParameterName=@n5, Value=0)
   > SetScalarParameterValue(self=@n7, ParameterName=@n6, Value=1.05)
  ]}
 Set Game Mode Text Size {nodes [
   n1 = Set Game Mode Text Size!() -> (then, FontSize:int) {is=title K2Node_FunctionEntry}
   n2 = Get Text_ActivityName()
   n3 = Get!(self:Text Object Reference=@n2) -> (
    Font_FontObject:Object Reference
    Font_FontMaterial:Object Reference
    Font_OutlineSettings:Font Outline Settings Structure
    Font_TypefaceFontName:Name
    Font_Size:float=~
    Font_LetterSpacing:int
    Font_SkewAmount:float=~
    Font_bForceMonospaced:bool=~
    Font_bMaterialIsStencil:bool=~
    Font_MonospacedWidth:float=~
   )
   > SetFont(
    self=@n2
    InFontInfo=MakeSlateFontInfo!(
     FontObject:Object Reference=@n3
     FontMaterial:Object Reference=@n3.Font_FontMaterial
     OutlineSettings:Font Outline Settings Structure=@n3.Font_OutlineSettings
     TypefaceFontName:Name=@n3.Font_TypefaceFontName
     Size:float="To Float (Integer)"(InInt=@n1)
     LetterSpacing:int=@n3.Font_LetterSpacing
     SkewAmount:float=@n3.Font_SkewAmount
     bForceMonospaced:bool=false
     bMaterialIsStencil:bool=false
     MonospacedWidth:float=1
    ) -> (ReturnValue:Slate Font Info Structure)
   )
  ]}
 Set Game Mode Text Size Override {nodes [
   Set Game Mode Text Size Override() {is=title K2Node_FunctionEntry}
   n1 = ShouldDecreaseFontSizeInDiscoveryTileTitles(self=Get FortExperimentContext())
   n2 = GetRuntimeOptions()
   > Set Game Mode Text Size(FontSize=Select(
     Option 0:int=Get DecreasedFontSizeInDiscoveryTileTitles(self=@n2) -> (DecreasedFontSizeInDiscoveryTileTitles=0)
     Option 1:int=Get DefaultGameModeTextSize() -> (DefaultGameModeTextSize=0)
     Index=NOT Boolean(A=@n1)
    ) -> (ReturnValue:int)) {is=title K2Node_CallFunction}
  ]}
 SetBannerColorAnimator {nodes [
   n1 = SetBannerColorAnimator!() -> (then, NewParam:double=~) {is=title K2Node_FunctionEntry}
   > SetColorAndOpacity(
    self:Image Object Reference=Get EpicBannerBg() -> (EpicBannerBg:Image Object Reference)
    InColorAndOpacity:Linear Color Structure="Lerp (LinearColor)"(A=Get StrokeColor() -> (StrokeColor:Linear Color Structure), B=(R=1,G=1,B=1,A=1), Alpha=@n1)
   )
  ]}
 SetDetailsButtonAvailable {nodes [
   SetDetailsButtonAvailable() {is=title K2Node_FunctionEntry}
   > n1 = Branch(Condition=IsMobileGame())
   Switch on ESlateVisibility!(
    execute=@n1.else
    Selection:ESlateVisibility Enum=GetVisibility!(self:Widget Object Reference=Get Button_Details()) -> (ReturnValue:ESlateVisibility Enum=Visible)
   ) -> (Visible:Exec, Collapsed:Exec, Hidden:Exec, HitTestInvisible:Exec, SelfHitTestInvisible:Exec) {is=title K2Node_SwitchEnum}
  ]}
 SetDisabledText {nodes [
   SetDisabledText() {is=title K2Node_FunctionEntry}
   n1 = Get InvalidActivityReason() -> (InvalidActivityReason:EFortInvalidActivityReason Enum=None)
   > "SetText (Text)"(
    self=Get DisabledWarningMessage() -> (DisabledWarningMessage:Common Text Object Reference)
    InText=Select!(
     None:Text
     PartyTooBig:Text="LOCTABLE(\"/Game/UI/DiscoveryShared/StringTable_ActivityBrowser\", \"Warning_PartyTooLarge\")"
     NotPartyLeader:Text="LOCTABLE(\"&StringTable_ActivityBrowser\", \"Warning_NotPartyLeader\")"
     MatchmakingAlready:Text="LOCTABLE(\"&StringTable_ActivityBrowser\", \"Warning_CurrentlyMatchmaking\")"
     PlaylistDisabled:Text="LOCTABLE(\"&StringTable_ActivityBrowser\", \"Warning_ModeDisabled\")"
     PlayerIsLfg:Text="LOCTABLE(\"&StringTable_ActivityBrowser\", \"Warning_PlayerIsLfg\")"
     InvalidData:Text="LOCTABLE(\"&StringTable_ActivityBrowser\", \"Warning_InvalidMode\")"
     ExpiredTournamentLockCache:Text="LOCTABLE(\"&StringTable_ActivityBrowser\", \"Warning_ExpiredTournamentLockCache\")"
     Index:EFortInvalidActivityReason Enum=@n1
    ) -> (ReturnValue:Text)
   )
   > Set IsActivityPreviouslyValid(IsActivityPreviouslyValid:bool=Get IsActivityValid() -> (IsActivityValid:bool)) -> (Output_Get:bool)
   > n2 = Set IsActivityValid(IsActivityValid="Equal (Enum)"(A:EFortInvalidActivityReason Enum=@n1, B:EFortInvalidActivityReason Enum=None))
   n3 = Get HideDetails()
   > n4 = Branch(Condition=AND Boolean!(A=NOT Boolean(A=@n2), B=Get IsTileActive(), C:bool=NOT Boolean(A=@n3)) -> (ReturnValue))
   n5 = Get WBP_Discovery_WarningIcon()
   n6 = Get OvrDisabledWarningMessage() -> (OvrDisabledWarningMessage:Overlay Object Reference)
   > SetVisibility(self=[@n5, @n6], InVisibility=HitTestInvisible)
   SetVisibility(execute=@n4.else, self=[@n5, @n6], InVisibility=Collapsed)
   > n7 = Sequence()
   n8 = Get DisabledWarningMessage()
   Branch(
    execute=@n7.then_1
    Condition="float > float"(A=GetRenderOpacity!(self:Widget Object Reference=@n8) -> (ReturnValue:float), B=0) {is=title K2Node_CallFunction}
   )
   > PlayAnimationReverse(InAnimation=Get Anim_OnDisabledClick())
   > Clear and Invalidate Timer by Handle(Handle=Get DisabledClickTimer())
  ]}
 SetDisabledThumbnailOverlay {nodes [
   SetDisabledThumbnailOverlay() {is=title K2Node_FunctionEntry}
   > n1 = GetDynamicMaterial(self=Get TileThumbnail())
   n2 = IsActivityLocked()
   n3 = "Not Equal (Enum)"!(A:EFortInvalidActivityReason Enum=Get InvalidActivityReason() -> (InvalidActivityReason=None), B:EFortInvalidActivityReason Enum=None) -> (ReturnValue:bool)
   n4 = Get IsPurchaseRequired() -> (IsPurchaseRequired:bool)
   > SetScalarParameterValue(
    self=@n1
    ParameterName=Get ThumbnailDisabledOverlay() -> (ThumbnailDisabledOverlay:Name=None)
    Value="To Float (Boolean)"(InBool=OR Boolean!(A=@n4, B=@n3, C:bool=@n2) -> (ReturnValue))
   )
  ]}
 SetExtraVerticalPaddingOffset {nodes [
   n1 = SetExtraVerticalPaddingOffset!() -> (then, InVerticalTilePadding:double=~) {is=title K2Node_FunctionEntry}
   n2 = SlotAsOverlaySlot(Widget=Get Ovr_Primary() -> (Ovr_Primary:Overlay Object Reference))
   n3 = Break Margin!(Margin:"Margin Structure (by ref)"=Get Padding!(self:Overlay Slot Object Reference=@n2) -> (Padding:Margin Structure)) -> (Left:float=~, Top:float=~, Right:float=~, Bottom:float=~) {is=title K2Node_BreakStruct}
   n4 = Get VerticalTilePaddingOffset() -> (VerticalTilePaddingOffset:double)
   > SetPadding(self=@n2, InPadding=Make Margin(Left=@n3, Top=@n3.Top, Right=@n3.Right, Bottom=float + float(A=@n4, B=@n1)))
  ]}
 SetFavoriteVisibility {nodes [
   n1 = SetFavoriteVisibility!() -> (then, IsFavorited:bool=~) {is=title K2Node_FunctionEntry}
   > SetVisibility(
    self=Get Overlay_Favorite() -> (Overlay_Favorite:Overlay Object Reference)
    InVisibility=Select!(Option 0:ESlateVisibility Enum=Collapsed, Option 1:ESlateVisibility Enum=HitTestInvisible, Index:bool=@n1) -> (ReturnValue:ESlateVisibility Enum)
   )
  ]}
 SetFirstTimeModalStyle {nodes [
   SetFirstTimeModalStyle() {is=title K2Node_FunctionEntry}
   > SetVisibility(
    self=[
     Get WBP_Discovery_WarningIcon()
     Get ActivityBrowserSocialProof()
     Get SqueegeeSlot_Lock()
     Get OvrDisabledWarningMessage()
     Get HorizontalBox_PlayerCount() -> (HorizontalBox_PlayerCount:Horizontal Box Object Reference)
     Get WBP_LockedStatus()
    ]
    InVisibility=Collapsed
   )
   > "SetText (Text)"(self=Get Text_ActivityName(), InText="LOCTEXT(\"Zero Build - Battle Royale\")")
   > SetBrushFromMaterial(self=Get TileStroke(), Material=/FortniteCoreUI/UI/DiscoveryShared/Art/MaterialInstances/MI_UI_Discovery_Tile_Border_FirstTimeModal)
   > SetPadding(self=SlotAsOverlaySlot(Widget=Get TileThumbnail()), InPadding=Make Margin(Left=1.5, Top=1.5, Right=1.5, Bottom=1.5))
   > SetBrushFromMaterial(self=Get TileThumbnail(), Material=/FortniteCoreUI/UI/DiscoveryShared/Art/MaterialInstances/MI_UI_Discovery_Tile_Thumbnail_FirstTimeModal)
   > UpdateEpicActivityStyling()
   > SetFavoriteVisibility(IsFavorited=true) {is=title K2Node_CallFunction}
   > Set bAllowThumbnailHoverRehoverAnim(bAllowThumbnailHoverRehoverAnim:bool=false) -> (Output_Get:bool)
   > SetVisibility(self=Get Button_Details(), InVisibility=Visible)
   > n1 = Cast To WBP_Discovery_DetailsButton(Object=Get Button_Details())
   > Set bAllowDetailsButtonHoverAnim!(execute, bAllowDetailsButtonHoverAnim:bool=false, self:WBP Discovery Details Button Object Reference=@n1) -> (then, Output_Get:bool)
  ]}
 SetHideDetails {nodes [
   n1 = SetHideDetails!() -> (then, Hide Details:bool=~) {is=title K2Node_FunctionEntry}
   > n2 = Set HideDetails(HideDetails:bool=@n1) -> (Output_Get:bool)
   > n3 = Branch(Condition=@n2)
   n4 = Get Overlay_InteractButtons() -> (Overlay_InteractButtons:Overlay Object Reference)
   n5 = Get HorizontalBox_LocksAndHearts() -> (HorizontalBox_LocksAndHearts:Horizontal Box Object Reference)
   n6 = Get Overlay_PurchaseRequired() -> (Overlay_PurchaseRequired:Overlay Object Reference)
   n7 = Get HorizontalBox_ActivityInfo()
   > n8 = SetVisibility(
    self=[@n4, @n5, Get ActivityBrowserSocialProof(), Get OvrDisabledWarningMessage(), Get WBP_Discovery_WarningIcon(), @n6, @n7]
    InVisibility=Collapsed
   )
   UpdateEpicActivityStyling(execute=[@n8.then, @n9.then], CreatorDisplayName=GetActivityCreatorDisplayText())
   n9 = SetVisibility(execute=@n3.else, self=[Get ActivityBrowserSocialProof(), @n4, @n5, @n6, @n7], InVisibility=SelfHitTestInvisible)
  ]}
 SetIsPromotedHero {nodes [
   n1 = SetIsPromotedHero!() -> (then, IsPromotedHeroCarousel:bool=~) {is=title K2Node_FunctionEntry}
   > Set IsPromotedHeroRow(IsPromotedHeroRow:bool=@n1) -> (Output_Get:bool)
  ]}
 SetTileSize {nodes [
   n1 = SetTileSize!() -> (then, Size:Vector 2D Structure) {is=title K2Node_FunctionEntry}
   n2 = BreakVector2D(InVec=@n1)
   n3 = "float > float"(A=@n2, B=500) {is=title K2Node_PromotableOperator}
   n4 = Select(Option 0:double=1.2, Option 1:double=1, Index="float > float"(A=@n2, B=650)) -> (ReturnValue:double)
   n5 = GetMobileDiscoverTileSizeMultiplier!(self:Fort Runtime Options Object Reference=GetRuntimeOptions()) -> (ReturnValue:float)
   n6 = Floor(A=float * float(A=@n2, B=Select(Option 0=@n5, Option 1=@n4, Index=@n3))) {is=title K2Node_CallFunction}
   > SetWidthOverride(
    self=Get SizeBox_Primary() -> (SizeBox_Primary:Size Box Object Reference)
    InWidthOverride=Select(Option 0=@n2, Option 1="To Float (Integer)"(InInt=@n6), Index=IsMobileGame())
   )
   n7 = Select(Option 0=1.2, Option 1=1, Index="float > float"(A=@n2, B=650))
   n8 = Floor(A=float * float(A=@n2.Y, B=Select(Option 0=@n5, Option 1=@n7, Index=@n3)))
   > SetHeightOverride(self=Get SizeBox_Primary(), InHeightOverride=Select(Option 0=@n2.Y, Option 1="To Float (Integer)"(InInt=@n8), Index=IsMobileGame()))
  ]}
 ShouldPlayTileVideo {nodes [
   ShouldPlayTileVideo() {is=title K2Node_FunctionEntry}
   > Return Node!(execute, bOutResult:bool=NOT Boolean(A=Get IsPromotedHeroRow()))
  ]}
 UpdateEpicActivityStyling {nodes [
   n1 = UpdateEpicActivityStyling!() -> (then, CreatorDisplayName:String) {is=title K2Node_FunctionEntry}
   > "SetText (Text)"(self=Get BannerText(), InText=Format Text!(Format="LOCTEXT(\"By {CreatorDisplayName}\")", CreatorDisplayName:String=@n1) -> (Result))
   n2 = AND Boolean(A=NOT Boolean(A=IsEmpty(InString=@n1)), B=NOT Boolean(A=Get HideDetails()), C=NOT Boolean(A=Get IsPromotedHeroRow()))
   > SetVisibility(
    self=Get OriginalContentEpicBanner() -> (OriginalContentEpicBanner:Overlay Object Reference)
    InVisibility=Select(Option 0:ESlateVisibility Enum=Collapsed, Option 1:ESlateVisibility Enum=SelfHitTestInvisible, Index=@n2) -> (ReturnValue:ESlateVisibility Enum)
   )
   n3 = Get BaseColor() -> (BaseColor:Linear Color Structure)
   n4 = Get EpicColor() -> (EpicColor:Linear Color Structure)
   > Set StrokeColor(StrokeColor:Linear Color Structure=SelectColor(
     A=Get DisabledColor() -> (DisabledColor:Linear Color Structure)
     B=Select(Option 0:Linear Color Structure=@n3, Option 1:Linear Color Structure=@n4, Index=@n2) -> (ReturnValue:Linear Color Structure)
     bPickA=DoesActivityRequirePurchase()
    )) -> (Output_Get:Linear Color Structure)
   > SetColorAndOpacity(self=Get EpicBannerBg(), InColorAndOpacity=Get StrokeColor())
  ]}
 UpdateRequiresPurchase {nodes [
   n1 = UpdateRequiresPurchase!() -> (then, IsPurchaseRequired:bool=~) {is=title K2Node_FunctionEntry}
   > Set IsPurchaseRequired(IsPurchaseRequired:bool=@n1) -> (Output_Get:bool)
   n2 = NOT Boolean(A=Get IsPromotedHeroRow())
   n3 = NOT Boolean(A=IsActivityLocked())
   > SetVisibility(
    self=Get Overlay_PurchaseRequired()
    InVisibility=Select(Option 0:ESlateVisibility Enum=Collapsed, Option 1:ESlateVisibility Enum=SelfHitTestInvisible, Index=AND Boolean(A=@n1, B=@n3, C=@n2)) -> (ReturnValue:ESlateVisibility Enum)
   )
   > SetDisabledThumbnailOverlay() {is=title K2Node_CallFunction}
  ]}
]
variables [
 bAllowThumbnailHoverRehoverAnim {type=boolean, default=true}
 BannerColorAnimator {type=number}
 BaseColor {type=LinearColor, default=<rgba 1 1 1 1>}
 DefaultGameModeTextSize {type=integer, default=18}
 DefaultImage {type=Texture, default=null}
 DisabledClickDuration {type=number, default=2.5}
 DisabledClickTimer {type=TimerHandle}
 DisabledColor {type=LinearColor, default=<rgba 0.402 0.6172 0.8879 1>}
 EpicColor {type=LinearColor, default=<rgba 0.6038 1 1 1>}
 FrameColor {type=string, default=Outline Color}
 GridSingleColumnWidth {type=number, default=24}
 GridSingleGutterWidth {type=number, default=24}
 HideDetails {type=boolean}
 IsActivityValid {type=boolean, default=true}
 IsKeyArtValid {type=boolean, default=true}
 IsPromotedHeroRow {type=boolean}
 IsPurchaseRequired {type=boolean}
 IsTileActive {type=boolean}
 Ratio16x9ConvertWidthToHeight {type=number, default=0.5625}
 StrokeColor {type=LinearColor, default=<rgba 0 0 0 0>}
 TextureParam {type=string, default=Thumbnail}
 ThumbnailAlpha {type=string, default=Thumbnail Alpha}
 ThumbnailDisabledOverlay {type=string, default=Overlay Opacity}
 ThumbnailScale {type=string, default=Thumbnail Zoom}
 TileSize {type=Vector2D, default=<408 230>}
 VerticalTilePaddingOffset {type=number, default=-35}
]
widgets {
 info {
  inheritedWidgetCount=8
  parentClass=/Script/DiscoveryBrowserUI.FortActivityTileDetailsDisplay
  rootWidgetClass=&SizeBox
  widgetCount=44
 }
 widgets [
  SizeBox_Primary {
   widgetClassPath=&SizeBox
   bIsVariable=true
   widget=SizeBox_Primary
   properties {bOverride_HeightOverride=true, bOverride_WidthOverride=true, heightOverride=216, widthOverride=384}
  }
  Ovr_GrowOnHoverSelect {widgetClassPath=&Overlay, parent=SizeBox_Primary, widget=Ovr_GrowOnHoverSelect, slot {class=SizeBoxSlot}}
  Ovr_Primary {
   widgetClassPath=&Overlay
   bIsVariable=true
   parent=Ovr_GrowOnHoverSelect
   widget=Ovr_Primary
   slot {class=OverlaySlot, properties {horizontalAlignment=HAlign_Fill, verticalAlignment=VAlign_Fill}}
  }
  PulseOutline {
   widgetClassPath=&Image
   parent=Ovr_Primary
   widget=PulseOutline
   properties {
    visibility=Collapsed
    brush {
     bIsDynamicallyLoaded=false
     drawAs=Box
     imageType=NoImage
     mirroring=NoMirror
     resourceName=None
     resourceObject=/FortniteCoreUI/UI/DiscoveryShared/Art/MaterialInstances/MI_UI_Discovery_Tile_Border
     tiling=NoTile
     imageSize=<32 32>
     margin=<ltrb 0.5 0.5 0.5 0.5>
     outlineSettings {
      bUseBrushTransparency=false
      roundingType=HalfHeightRadius
      width=0
      color {colorUseRule=UseColor_Specified, specifiedColor=<rgba 0 0 0 0>}
      cornerRadii=<0 0 0 0>
     }
     tintColor {colorUseRule=UseColor_Specified, specifiedColor=<rgba 1 1 1 1>}
     uVRegion {bIsValid=false, max=<0 0>, min=<0 0>}
    }
   }
   slot {class=OverlaySlot, properties {horizontalAlignment=HAlign_Fill, verticalAlignment=VAlign_Fill}}
  }
  TileStroke {
   widgetClassPath=&Image
   bIsVariable=true
   parent=Ovr_Primary
   widget=TileStroke
   properties {
    visibility=SelfHitTestInvisible
    brush {
     bIsDynamicallyLoaded=false
     drawAs=Box
     imageType=NoImage
     mirroring=NoMirror
     resourceName=None
     resourceObject=&MI_UI_Discovery_Tile_Border
     tiling=NoTile
     imageSize=<32 32>
     margin=<ltrb 0.5 0.5 0.5 0.5>
     outlineSettings {
      bUseBrushTransparency=false
      roundingType=HalfHeightRadius
      width=0
      color {colorUseRule=UseColor_Specified, specifiedColor=<rgba 0 0 0 0>}
      cornerRadii=<0 0 0 0>
     }
     tintColor {colorUseRule=UseColor_Specified, specifiedColor=<rgba 1 1 1 1>}
     uVRegion {bIsValid=false, max=<0 0>, min=<0 0>}
    }
   }
   slot {class=OverlaySlot, properties {horizontalAlignment=HAlign_Fill, verticalAlignment=VAlign_Fill}}
  }
  Overlay_Interior {
   widgetClassPath=&Overlay
   parent=Ovr_Primary
   widget=Overlay_Interior
   slot {class=OverlaySlot, properties {horizontalAlignment=HAlign_Fill, verticalAlignment=VAlign_Fill}}
  }
  Wrapper {
   widgetClassPath=&GridPanel
   parent=Overlay_Interior
   widget=Wrapper
   properties {columnFill [1], rowFill [0,1]}
   slot {class=OverlaySlot, properties {horizontalAlignment=HAlign_Fill, verticalAlignment=VAlign_Fill}}
  }
  Overlay_1 {widgetClassPath=&Overlay, parent=Wrapper, widget=Overlay_1, slot {class=GridSlot, properties {row=1}}}
  TileThumbnail {
   widgetClassPath=&Image
   bIsVariable=true
   parent=Overlay_1
   widget=TileThumbnail
   properties {
    visibility=SelfHitTestInvisible
    brush {
     bIsDynamicallyLoaded=false
     drawAs=Image
     imageType=NoImage
     mirroring=NoMirror
     resourceName=None
     resourceObject=/FortniteCoreUI/UI/DiscoveryShared/Art/MaterialInstances/MI_UI_Discovery_Tile_Thumbnail
     tiling=NoTile
     imageSize=<32 32>
     margin=<ltrb 0 0 0 0>
     outlineSettings {
      bUseBrushTransparency=false
      roundingType=HalfHeightRadius
      width=0
      color {colorUseRule=UseColor_Specified, specifiedColor=<rgba 0 0 0 0>}
      cornerRadii=<0 0 0 0>
     }
     tintColor {colorUseRule=UseColor_Specified, specifiedColor=<rgba 1 1 1 1>}
     uVRegion {bIsValid=false, max=<0 0>, min=<0 0>}
    }
   }
   slot {class=OverlaySlot, properties {horizontalAlignment=HAlign_Fill, verticalAlignment=VAlign_Fill}}
  }
  Logo_NotReady_Fort554488 {
   widgetClassPath=&Image
   bIsVariable=true
   parent=Overlay_1
   widget=Logo_NotReady_Fort554488
   properties {
    renderOpacity=0
    visibility=Collapsed
    brush {
     bIsDynamicallyLoaded=false
     drawAs=Image
     imageType=NoImage
     mirroring=NoMirror
     resourceName=None
     resourceObject=null
     tiling=NoTile
     imageSize=<64 64>
     margin=<ltrb 0 0 0 0>
     outlineSettings {
      bUseBrushTransparency=false
      roundingType=HalfHeightRadius
      width=0
      color {colorUseRule=UseColor_Specified, specifiedColor=<rgba 0 0 0 0>}
      cornerRadii=<0 0 0 0>
     }
     tintColor {colorUseRule=UseColor_Specified, specifiedColor=<rgba 1 1 1 1>}
     uVRegion {bIsValid=false, max=<0 0>, min=<0 0>}
    }
   }
   slot {class=OverlaySlot, properties {padding=<ltrb 12 12 0 0>}}
  }
  ActivityVideoCycleWidget {
   widgetClassPath=/FortniteCoreUI/UI/DiscoveryShared/ActivityVideoCycleWidget_Tile
   bInherited=true
   bIsVariable=true
   parent=Overlay_1
   widget=ActivityVideoCycleWidget
   properties {movieStartDelay=0.2, visibility=HitTestInvisible}
   slot {class=OverlaySlot, properties {horizontalAlignment=HAlign_Fill, verticalAlignment=VAlign_Fill}}
  }
  HorizontalBox_LocksAndHearts {
   widgetClassPath=&HorizontalBox
   bIsVariable=true
   parent=Overlay_1
   widget=HorizontalBox_LocksAndHearts
   slot {class=OverlaySlot, properties {horizontalAlignment=HAlign_Fill, padding=<ltrb 8 8 6 0>}}
  }
  VerticalBox_FillRemaining {
   widgetClassPath=&VerticalBox
   parent=HorizontalBox_LocksAndHearts
   widget=VerticalBox_FillRemaining
   slot {class=HorizontalBoxSlot, properties {size {sizeRule=Fill, value=1}}}
  }
  Overlay_PurchaseRequired {
   widgetClassPath=&Overlay
   bIsVariable=true
   parent=VerticalBox_FillRemaining
   widget=Overlay_PurchaseRequired
   slot {class=VerticalBoxSlot}
  }
  WBP_LockedStatus {
   widgetClassPath=&/WBP_LockedStatus
   bIsVariable=true
   parent=Overlay_PurchaseRequired
   widget=WBP_LockedStatus
   properties {lockedReason=Requires Purchase}
   slot {class=OverlaySlot, properties {verticalAlignment=VAlign_Fill}}
  }
  Spacer_112 {
   widgetClassPath=&Spacer
   parent=HorizontalBox_LocksAndHearts
   widget=Spacer_112
   properties {size=<4 1>}
   slot {class=HorizontalBoxSlot}
  }
  Overlay_Favorite {
   widgetClassPath=&Overlay
   bIsVariable=true
   parent=HorizontalBox_LocksAndHearts
   widget=Overlay_Favorite
   slot {class=HorizontalBoxSlot, properties {horizontalAlignment=HAlign_Right, verticalAlignment=VAlign_Top}}
  }
  Spacer_76 {
   widgetClassPath=&Spacer
   parent=Overlay_Favorite
   widget=Spacer_76
   properties {size=<38 38>}
   slot {class=OverlaySlot, properties {horizontalAlignment=HAlign_Fill, verticalAlignment=VAlign_Fill}}
  }
  WBP_Discovery_Favorite {
   widgetClassPath=/FortniteCoreUI/UI/DiscoveryShared/WBP_Discovery_Favorite
   bIsVariable=true
   parent=Overlay_Favorite
   widget=WBP_Discovery_Favorite
   properties {isFavorited=true}
   slot {class=OverlaySlot, properties {horizontalAlignment=HAlign_Fill, verticalAlignment=VAlign_Fill}}
  }
  SqueegeeSlot_Lock {
   widgetClassPath=/Game/UI/Plugin/Squeegee/SqueegeeInjectionSlot
   bIsVariable=true
   parent=Overlay_1
   widget=SqueegeeSlot_Lock
   properties {
    pluginId=Squeegee
    pluginWidgetId=Lock Icon
    styleParams {bShowDetails=true, fontSize=15, paddingSize=0.5, iconSize=<26 26>, iconSkewSize=<5 26>}
   }
   slot {class=OverlaySlot, properties {horizontalAlignment=HAlign_Center, verticalAlignment=VAlign_Center}}
  }
  OriginalContentEpicBanner {
   widgetClassPath=&Overlay
   bIsVariable=true
   parent=Wrapper
   widget=OriginalContentEpicBanner
   properties {visibility=Visible}
   slot {class=GridSlot}
  }
  EpicBannerBg {
   widgetClassPath=&Image
   bIsVariable=true
   parent=OriginalContentEpicBanner
   widget=EpicBannerBg
   properties {brush {
     bIsDynamicallyLoaded=false
     drawAs=Image
     imageType=NoImage
     mirroring=NoMirror
     resourceName=None
     resourceObject=null
     tiling=NoTile
     imageSize=<2 2>
     margin=<ltrb 0 0 0 0>
     outlineSettings {
      bUseBrushTransparency=false
      roundingType=HalfHeightRadius
      width=0
      color {colorUseRule=UseColor_Specified, specifiedColor=<rgba 0 0 0 0>}
      cornerRadii=<0 0 0 0>
     }
     tintColor {colorUseRule=UseColor_Specified, specifiedColor=<rgba 1 1 1 1>}
     uVRegion {bIsValid=false, max=<0 0>, min=<0 0>}
    }}
   slot {class=OverlaySlot, properties {horizontalAlignment=HAlign_Fill, verticalAlignment=VAlign_Fill}}
  }
  BannerText {
   widgetClassPath=&CommonTextBlock
   bIsVariable=true
   parent=OriginalContentEpicBanner
   widget=BannerText
   properties {
    justification=Center
    mobileFontSizeMultiplier=1.3
    text=By Epic
    textTransformPolicy=ToUpper
    font {
     bForceMonospaced=false
     bMaterialIsStencil=false
     fontMaterial=/DiscoveryBrowser/Art/MI_UI_Discovery_EpicText
     fontObject=&BurbankSmall
     letterSpacing=220
     monospacedWidth=1
     size=11
     skewAmount=0
     typefaceFontName=Black
     outlineSettings {
      bApplyOutlineToDropShadows=false
      bMiteredCorners=false
      bSeparateFillAlpha=false
      outlineMaterial=null
      outlineSize=0
      outlineColor=<rgba 0 0 0 1>
     }
    }
    margin=<ltrb 0 -4 0 -3.5>
    renderTransform {angle=0, scale=<1 1>, shear=<-8 0>, translation=<0 0>}
    shadowOffset=<0 0>
   }
   slot {class=OverlaySlot, properties {horizontalAlignment=HAlign_Center, verticalAlignment=VAlign_Center}}
  }
  Spacer_15 {
   widgetClassPath=&Spacer
   parent=OriginalContentEpicBanner
   widget=Spacer_15
   properties {size=<1 28>}
   slot {class=OverlaySlot, properties {horizontalAlignment=HAlign_Fill, verticalAlignment=VAlign_Fill}}
  }
  OvrDisabledWarningMessage {
   widgetClassPath=&Overlay
   bIsVariable=true
   parent=Overlay_Interior
   widget=OvrDisabledWarningMessage
   properties {visibility=Collapsed}
   slot {class=OverlaySlot, properties {horizontalAlignment=HAlign_Fill, verticalAlignment=VAlign_Fill}}
  }
  WBP_Discovery_WarningIcon {
   widgetClassPath=&/WBP_Discovery_WarningIcon
   bIsVariable=true
   parent=OvrDisabledWarningMessage
   widget=WBP_Discovery_WarningIcon
   slot {class=OverlaySlot, properties {verticalAlignment=VAlign_Bottom, padding=<ltrb 10 0 0 10>}}
  }
  SizeBox_1 {
   widgetClassPath=&SizeBox
   parent=OvrDisabledWarningMessage
   widget=SizeBox_1
   properties {bOverride_HeightOverride=true, heightOverride=128}
   slot {class=OverlaySlot, properties {horizontalAlignment=HAlign_Fill, verticalAlignment=VAlign_Center}}
  }
  DisabledWarningMessage {
   widgetClassPath=&CommonTextBlock
   bIsVariable=true
   parent=SizeBox_1
   widget=DisabledWarningMessage
   properties {
    justification=Center
    renderOpacity=0
    text="{Disabled warning message. Lorem ipsum.}"
    textTransformPolicy=ToUpper
    visibility=Collapsed
    wrapTextAt=240
    colorAndOpacity {colorUseRule=UseColor_Specified, specifiedColor=<rgba 0.9301 1 0.0097 1>}
    font {
     bForceMonospaced=false
     bMaterialIsStencil=false
     fontMaterial=null
     fontObject=&BurbankSmall
     letterSpacing=0
     monospacedWidth=1
     size=14
     skewAmount=0
     typefaceFontName=Black
     outlineSettings {
      bApplyOutlineToDropShadows=false
      bMiteredCorners=false
      bSeparateFillAlpha=false
      outlineMaterial=null
      outlineSize=0
      outlineColor=<rgba 0 0 0 1>
     }
    }
    shadowOffset=<0 0>
   }
   slot {class=SizeBoxSlot, properties {horizontalAlignment=HAlign_Center, verticalAlignment=VAlign_Center, padding=<ltrb 64 0 64 0>}}
  }
  Overlay_InteractButtons {
   widgetClassPath=&Overlay
   bIsVariable=true
   parent=Overlay_Interior
   widget=Overlay_InteractButtons
   slot {class=OverlaySlot, properties {horizontalAlignment=HAlign_Fill, verticalAlignment=VAlign_Fill}}
  }
  Button_Details {
   widgetClassPath=/FortniteCoreUI/UI/DiscoveryShared/WBP_Discovery_DetailsButton
   bInherited=true
   bIsVariable=true
   parent=Overlay_InteractButtons
   widget=Button_Details
   properties {
    bApplyOffsetAsTransformInsteadOfPaddingToIgnoreLayout=false
    bDisplayIconDropShadow=true
    inputActionWidgetHorizAlignment=HAlign_Right
    keyBind_X_Offset=26
    keyBind_Y_Offset=38
    imageSize=<56 56>
   }
   slot {class=OverlaySlot, properties {horizontalAlignment=HAlign_Right, verticalAlignment=VAlign_Bottom, padding=<ltrb 0 0 8 8>}}
  }
  Text_DebugId {
   widgetClassPath=&TextBlock
   bInherited=true
   parent=Ovr_Primary
   widget=Text_DebugId
   properties {
    autoWrapText=true
    justification=Center
    text=EventId: epicgames_Some_Event_Id EventWindowId: Some_event_window_id EventWindow Playlist: Playlists_name_example
    visibility=Collapsed
    wrapTextAt=400
    wrappingPolicy=AllowPerCharacterWrapping
    font {
     bForceMonospaced=false
     bMaterialIsStencil=false
     fontMaterial=null
     fontObject=&Roboto
     letterSpacing=0
     monospacedWidth=1
     size=16
     skewAmount=0
     typefaceFontName=Bold
     outlineSettings {
      bApplyOutlineToDropShadows=false
      bMiteredCorners=false
      bSeparateFillAlpha=false
      outlineMaterial=null
      outlineSize=3
      outlineColor=<rgba 0 0 0 1>
     }
    }
   }
   slot {class=OverlaySlot, properties {horizontalAlignment=HAlign_Fill, verticalAlignment=VAlign_Center}}
  }
  HorizontalBox_ActivityInfo {
   widgetClassPath=&HorizontalBox
   bIsVariable=true
   parent=Ovr_Primary
   widget=HorizontalBox_ActivityInfo
   properties {renderTransform {angle=0, scale=<1 1>, shear=<0 0>, translation=<0 38>}}
   slot {class=OverlaySlot, properties {horizontalAlignment=HAlign_Fill, verticalAlignment=VAlign_Bottom}}
  }
  Image_MultiMode {
   widgetClassPath=&Image
   bIsVariable=true
   parent=HorizontalBox_ActivityInfo
   widget=Image_MultiMode
   properties {brush {
     bIsDynamicallyLoaded=false
     drawAs=Image
     imageType=NoImage
     mirroring=NoMirror
     resourceName=None
     resourceObject=/DiscoveryBrowser/Art/M_UI_Dicovery_SubModeIcon
     tiling=NoTile
     imageSize=<32 32>
     margin=<ltrb 0 0 0 0>
     outlineSettings {
      bUseBrushTransparency=false
      roundingType=HalfHeightRadius
      width=0
      color {colorUseRule=UseColor_Specified, specifiedColor=<rgba 0 0 0 0>}
      cornerRadii=<0 0 0 0>
     }
     tintColor {colorUseRule=UseColor_Specified, specifiedColor=<rgba 1 1 1 1>}
     uVRegion {bIsValid=false, max=<0 0>, min=<0 0>}
    }}
   slot {class=HorizontalBoxSlot, properties {horizontalAlignment=HAlign_Left, verticalAlignment=VAlign_Center, padding=<ltrb -3 0 5 0>}}
  }
  Spacer_Height {
   widgetClassPath=&Spacer
   parent=HorizontalBox_ActivityInfo
   widget=Spacer_Height
   properties {size=<0 32>}
   slot {class=HorizontalBoxSlot}
  }
  Widget_ClippingArea {
   widgetClassPath=&Overlay
   parent=HorizontalBox_ActivityInfo
   widget=Widget_ClippingArea
   properties {clipping=ClipToBounds}
   slot {class=HorizontalBoxSlot, properties {horizontalAlignment=HAlign_Left, size {sizeRule=Fill, value=1}}}
  }
  Text_ActivityName {
   widgetClassPath=&CommonTextBlock
   bInherited=true
   bIsVariable=true
   parent=Widget_ClippingArea
   widget=Text_ActivityName
   properties {
    bIsScrollingEnabled=false
    mobileFontSizeMultiplier=1.45
    scrollStyle=/FortUILibrary/Styles/Typography/Foundation/TextScrollStyle-NoFade.TextScrollStyle-NoFade_C
    text="{Game Mode}"
    textOverflowPolicy=Ellipsis
    font {
     bForceMonospaced=false
     bMaterialIsStencil=false
     fontMaterial=/FortniteCoreUI/UI/DiscoveryShared/Art/MI_UI_Discovery_EmojiText
     fontObject=&BurbankSmall
     letterSpacing=0
     monospacedWidth=1
     size=12
     skewAmount=0
     typefaceFontName=Bold
     outlineSettings {
      bApplyOutlineToDropShadows=false
      bMiteredCorners=false
      bSeparateFillAlpha=false
      outlineMaterial=null
      outlineSize=0
      outlineColor=<rgba 0 0 0 1>
     }
    }
    shadowOffset=<0 0>
   }
   slot {class=OverlaySlot, properties {verticalAlignment=VAlign_Center}}
  }
  Spacer_Mobile {
   widgetClassPath=&Spacer
   bIsVariable=true
   parent=HorizontalBox_ActivityInfo
   widget=Spacer_Mobile
   properties {visibility=Collapsed, size=<50 45>}
   slot {class=HorizontalBoxSlot, properties {horizontalAlignment=HAlign_Left, verticalAlignment=VAlign_Top}}
  }
  HorizontalBox_PlayerCount {
   widgetClassPath=&HorizontalBox
   bIsVariable=true
   parent=HorizontalBox_ActivityInfo
   widget=HorizontalBox_PlayerCount
   properties {visibility=Collapsed}
   slot {class=HorizontalBoxSlot, properties {verticalAlignment=VAlign_Center}}
  }
  Image_PlayerIcon {
   widgetClassPath=/Script/FortniteUI.FortMobileImage
   parent=HorizontalBox_PlayerCount
   widget=Image_PlayerIcon
   properties {
    mobileSizeMultiplier=1.2
    visibility=HitTestInvisible
    brush {
     bIsDynamicallyLoaded=false
     drawAs=Image
     imageType=NoImage
     mirroring=NoMirror
     resourceName=None
     resourceObject=/Game/UI/DiscoveryShared/Art/T_UI_Discovery_PlayerIcon
     tiling=NoTile
     imageSize=<18 18>
     margin=<ltrb 0 0 0 0>
     outlineSettings {
      bUseBrushTransparency=false
      roundingType=HalfHeightRadius
      width=0
      color {colorUseRule=UseColor_Specified, specifiedColor=<rgba 0 0 0 0>}
      cornerRadii=<0 0 0 0>
     }
     tintColor {colorUseRule=UseColor_Specified, specifiedColor=<rgba 0.2157 0.6588 1 1>}
     uVRegion {bIsValid=false, max=<0 0>, min=<0 0>}
    }
   }
   slot {class=HorizontalBoxSlot, properties {verticalAlignment=VAlign_Center, padding=<ltrb 0 0 2 0>}}
  }
  Text_PlayerCount {
   widgetClassPath=&CommonTextBlock
   bInherited=true
   parent=HorizontalBox_PlayerCount
   widget=Text_PlayerCount
   properties {
    justification=Center
    mobileFontSizeMultiplier=1.45
    text="{COUNT}"
    textTransformPolicy=ToUpper
    colorAndOpacity {colorUseRule=UseColor_Specified, specifiedColor=<rgba 0.2159 0.6584 1 1>}
    font {
     bForceMonospaced=false
     bMaterialIsStencil=false
     fontMaterial=null
     fontObject=&BurbankSmall
     letterSpacing=0
     monospacedWidth=1
     size=13
     skewAmount=0.2
     typefaceFontName=Black
     outlineSettings {
      bApplyOutlineToDropShadows=false
      bMiteredCorners=false
      bSeparateFillAlpha=false
      outlineMaterial=null
      outlineSize=0
      outlineColor=<rgba 0 0 0 1>
     }
    }
    margin=<ltrb 0 -4 0 -4>
    shadowOffset=<0 0>
   }
   slot {class=HorizontalBoxSlot, properties {verticalAlignment=VAlign_Center, size {sizeRule=Fill, value=1}}}
  }
  ActivityBrowserSocialProof {
   widgetClassPath=&/WBP_ActivityBrowserSocialProof
   bIsVariable=true
   parent=Ovr_Primary
   widget=ActivityBrowserSocialProof
   properties {cCU And Friends Text Color=<rgba 0.0319 0.402 1 1>, renderTransform {angle=0, scale=<1 1>, shear=<0 0>, translation=<0 60>}}
   slot {class=OverlaySlot, properties {horizontalAlignment=HAlign_Fill, verticalAlignment=VAlign_Bottom}}
  }
  Button_Favorite {widgetClassPath=&CommonButtonBase, bInherited=true, widget=null}
  ActivityBrowserTag_EpicOriginal {widgetClassPath=/Script/FortniteUI.FortActivityBrowserTag, bInherited=true, widget=null}
  InputActionWidget {widgetClassPath=&CommonActionWidget, bInherited=true, widget=null}
 ]
}
