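# ActivityBrowserTile is the Widget Blueprint for an activity tile in the Discovery browser, based on the native FortActivityBrowserTile class.
# It adds no graph logic or variables, leaving behavior to the parent class and its child widgets.
# The root is an ActivityBrowserTileDetailsDisplay set to show its details button and not simulate hover on touch input.
# An inherited CommonActionWidget slot, InputActionWidget, has no widget assigned.
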WidgetBlueprint /DiscoveryBrowser/ActivityBrowserTile
graphs [EventGraph {links [], nodes []}]
variables []
widgets {
 info {
  inheritedWidgetCount=2
  parentClass=/Script/DiscoveryBrowserUI.FortActivityBrowserTile
  rootWidgetClass=&/ActivityBrowserTileDetailsDisplay.ActivityBrowserTileDetailsDisplay_C
  widgetCount=2
 }
 widgets [
  Display_TileDetails {
   widgetClassPath=&/ActivityBrowserTileDetailsDisplay
   bInherited=true
   bIsVariable=true
   widget=Display_TileDetails
   properties {bShowDetailsButton=true, bSimulateHoverOnTouchInput=false}
  }
  InputActionWidget {widgetClassPath=&CommonActionWidget, bInherited=true, widget=null}
 ]
}
