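# Discovery browser tile for a joinable friend party, showing avatars, names, rich presence, invite or last-partied text, and a private-party badge.
# Parent FortActivityBrowserPlayWithFriendsTile events drive text and visibility, while a dynamic background material handles thumbnail fade-in, invite flare and Arabic mask flip.
# Activation plays hover or touch animations, the former swapping presence text for the CTA button.
# Input method adjusts button padding; mobile touch hides the button and resizes the layout.
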WidgetBlueprint /DiscoveryBrowser/Tiles/JoinableParty/WBP_ActivityBrowserTile_JoinableParty
animations [
 Anim_OnHoverSelect {
  binding_count=3
  length_seconds=0.2505
  object=&self:Anim_OnHoverSelect
  sequence {bindings [
    WBP_CTA_ButtonPrimary {
     id=#1
     possessed_class=WBP_UIKit_Button_Quiet_C
     tracks [
      {
       type=MovieScene2DTransformTrack
       sections [{
         type=MovieScene2DTransformSection
         end_frame=4
         start_frame=0
         channels [
          Translation.X {type=MovieSceneScriptingFloatChannel, const [0,2]}
          Translation.Y {type=MovieSceneScriptingFloatChannel, keys [[0,40],[4,0]]}
          Angle {type=MovieSceneScriptingFloatChannel}
          Scale.X {type=MovieSceneScriptingFloatChannel}
          Scale.Y {type=MovieSceneScriptingFloatChannel}
          Shear.X {type=MovieSceneScriptingFloatChannel}
          Shear.Y {type=MovieSceneScriptingFloatChannel}
         ]
        }]
      }
      {
       type=MovieSceneFloatTrack
       display_name=Render Opacity
       sections [{type=MovieSceneFloatSection, end_frame=4, start_frame=0, channels [None {type=MovieSceneScriptingFloatChannel, keys [[0,0],[4,1]]}]}]
      }
     ]
    }
    HB_PresenceTexts {
     id=#2
     possessed_class=HorizontalBox
     tracks [{
       type=MovieSceneFloatTrack
       display_name=Render Opacity
       sections [{type=MovieSceneFloatSection, end_frame=4, start_frame=0, channels [None {type=MovieSceneScriptingFloatChannel, keys [[0,1],[4,0]]}]}]
      }]
    }
    Bg {
     id=#3
     possessed_class=Image
     tracks [{
       type=MovieSceneWidgetMaterialTrack
       sections [{
         type=MovieSceneParameterSection
         channels [
          Background Color BR {type=MovieSceneScriptingFloatChannel, keys [[0,0.0024],[5.0097,0]]}
          Background Color BG {type=MovieSceneScriptingFloatChannel, keys [[0,0.0513],[5.0097,0.0908]]}
          Background Color BB {type=MovieSceneScriptingFloatChannel, keys [[0,0.3864],[5.0097,0.9911]]}
          Background Color BA {type=MovieSceneScriptingFloatChannel, const [1,2]}
         ]
        }]
      }]
    }
   ]}
 }
 Anim_OnRemove {
  binding_count=1
  length_seconds=0.25
  object=&self:Anim_OnRemove
  sequence {
   bindings [Overlay_RootContent {
     id=#4
     possessed_class=Overlay
     tracks [
      {
       type=MovieSceneFloatTrack
       display_name=Render Opacity
       sections [{type=MovieSceneFloatSection, end_frame=15, start_frame=0, channels [None {type=MovieSceneScriptingFloatChannel, keys [[0,1],[9,0]]}]}]
      }
      {
       type=MovieScene2DTransformTrack
       sections [{
         type=MovieScene2DTransformSection
         channels [
          Scale.X {type=MovieSceneScriptingFloatChannel, keys [[0,1],[15,0]]}
          Scale.Y {type=MovieSceneScriptingFloatChannel, keys [[0,1],[15,0]]}
         ]
        }]
      }
     ]
    }]
   display_rate {numerator=60}
  }
 }
 Anim_OnImageLoaded {
  binding_count=1
  length_seconds=0.2
  object=&self:Anim_OnImageLoaded
  sequence {
   bindings [Bg {
     id=#5
     possessed_class=Image
     tracks [{
       type=MovieSceneWidgetMaterialTrack
       sections [{type=MovieSceneParameterSection, channels [Thumbnail Opacity {type=MovieSceneScriptingFloatChannel, keys [[0,0],[12,1]]}]}]
      }]
    }]
   display_rate {numerator=60}
  }
 }
 Anim_OnTouchSelected {
  binding_count=2
  length_seconds=0.25
  object=&self:Anim_OnTouchSelected
  sequence {
   bindings [
    Bg {
     id=#3
     possessed_class=Image
     tracks [{
       type=MovieSceneWidgetMaterialTrack
       sections [{
         type=MovieSceneParameterSection
         channels [
          Background Color BR {type=MovieSceneScriptingFloatChannel, keys [[0,0.0024],[25,0]]}
          Background Color BG {type=MovieSceneScriptingFloatChannel, keys [[0,0.0513],[25,0.0908]]}
          Background Color BB {type=MovieSceneScriptingFloatChannel, keys [[0,0.3864],[25,0.9911]]}
          Background Color BA {type=MovieSceneScriptingFloatChannel, const [1,2]}
         ]
        }]
      }]
    }
    PulseOutline {
     id=#6
     possessed_class=Image
     tracks [
      {
       type=MovieSceneWidgetMaterialTrack
       sections [{
         type=MovieSceneParameterSection
         end_frame=25
         start_frame=0
         channels [
          Outline Size Ratio {type=MovieSceneScriptingFloatChannel, keys [[5,0],[14,1]]}
          Outline Opacity {type=MovieSceneScriptingFloatChannel, keys [[5,0],[25,0.5]]}
         ]
        }]
      }
      {
       type=MovieSceneByteTrack
       display_name=Visibility
       sections [{type=MovieSceneByteSection, end_frame=25, start_frame=0, channels [None {type=MovieSceneScriptingByteChannel, const [3,2]}]}]
      }
     ]
    }
   ]
   display_rate {numerator=100}
  }
 }
]
graphs [
 EventGraph {nodes [
   "On Clicked (WBP_CTA_ButtonPrimary)"!() -> (OutputDelegate, then, Button:Common Button Base Object Reference) {is=title K2Node_ComponentBoundEvent}
   > HandleCTAButtonClicked(self:Fort Activity Browser Play With Friends Tile Object Reference)
   n1 = Event PreConstruct()
   > n2 = GetDynamicMaterial(self=Get Bg() -> (Bg:Image Object Reference))
   > Set BackgroundDynamicMaterialInstance(BackgroundDynamicMaterialInstance:Material Instance Dynamic Object Reference=@n2) -> (Output_Get:Material Instance Dynamic Object Reference)
   > SetColorAndOpacity(
    self=[
     Get Text_SingleUserName() -> (Text_SingleUserName:Common Text Object Reference)
     Get Text_RichPresence() -> (Text_RichPresence:Common Text Object Reference)
    ]
    InColorAndOpacity=Make Slate Color(SpecifiedColor=Get ColorSharedTextAndImages() -> (ColorSharedTextAndImages:Linear Color Structure))
   )
   > SetTileSize!(execute, self, Size:Vector 2D Structure=Get TileSize() -> (TileSize:Vector 2D Structure)) -> (then)
   Event Construct()
   > Set Right Left Background() {is=title K2Node_CallFunction}
   > SetRenderOpacity(self=Get WBP_CTA_ButtonPrimary() -> (WBP_CTA_ButtonPrimary:WBP UIKit Button Quiet Object Reference), InOpacity=0)
   > SetThumbnail!(execute, self, Texture:Texture Object Reference, IsLoaded:bool=false) -> (then)
   n3 = AND Boolean(A=IsMobileGame(), B=IsUsingTouch(PlayerController=GetOwningPlayer()))
   > n4 = Branch(Condition=@n3)
   > n5 = SetVisibility(self=Get WBP_CTA_ButtonPrimary(), InVisibility=Collapsed)
   SetSize!(
    execute=[@n5.then, @n4.else]
    self:Spacer Object Reference=Get Spacer_TileSize() -> (Spacer_TileSize:Spacer Object Reference)
    InSize:Vector 2D Structure=Select(Option 0:Vector 2D Structure=Get TileSize(), Option 1:Vector 2D Structure=(X=810,Y=210), Index=@n3) -> (ReturnValue:Vector 2D Structure)
   ) -> (then)
   > SetHeightOverride(self=Get SizeBox_Icons() -> (SizeBox_Icons:Size Box Object Reference), InHeightOverride=SelectFloat(A=80, B=48, bPickA=@n3))
   > SetSize(self=Get Spacer_72() -> (Spacer_72:Spacer Object Reference), InSize=Select(Option 0=(X=1,Y=2), Option 1=(X=1,Y=5), Index=@n3))
   > Set CTAButton Padding!(execute, self, InputType:ECommonInputType Enum=GetCurrentInputType(self=Get CommonInputSubsystem())) -> (then) {is=title K2Node_CallFunction}
   > SetIsInteractionEnabled!(execute, self:Common Button Base Object Reference=Get WBP_CTA_ButtonPrimary(), bInIsInteractionEnabled:bool=false) -> (then)
   n6 = Event UpdateIslandThumbnail!() -> (OutputDelegate, then, ThumbnailTexture:Texture Object Reference) {is=title K2Node_Event}
   > SetThumbnail(Texture=@n6.ThumbnailTexture, IsLoaded=true)
   n7 = Event UpdateOtherPlayersSubText!() -> (OutputDelegate, then, OtherPlayersSubText:"Text (by ref)") {is=title K2Node_Event}
   > SetVisibility(
    self=Get Text_JoinableParty() -> (Text_JoinableParty:Common Text Object Reference)
    InVisibility=Select(Option 0:ESlateVisibility Enum=Collapsed, Option 1:ESlateVisibility Enum=HitTestInvisible, Index=TextIsEmpty(InText=@n7.OtherPlayersSubText)) -> (ReturnValue:ESlateVisibility Enum)
   )
   > "SetText (Text)"(self=Get Text_OtherPlayers() -> (Text_OtherPlayers:Common Text Object Reference), InText=@n7.OtherPlayersSubText)
   n8 = Event UpdateRichPresence!() -> (OutputDelegate, then, RichPresence:"Text (by ref)") {is=title K2Node_Event}
   > "SetText (Text)"(self=Get Text_RichPresence(), InText=@n8.RichPresence)
   n9 = Event UpdatePartyMemberNames!() -> (OutputDelegate, then, Names:"Text (by ref)") {is=title K2Node_Event}
   n10 = TextIsEmpty(InText=@n9.Names)
   > SetVisibility(
    self=Get RichText_DisplayNamesOrSingleFriendName() -> (RichText_DisplayNamesOrSingleFriendName:Common Rich Text Block Object Reference)
    InVisibility=Select(Option 0=HitTestInvisible, Option 1=Collapsed, Index=@n10)
   )
   > n11 = Branch(Condition=@n10)
   SetText(execute=@n11.else, self=Get RichText_DisplayNamesOrSingleFriendName(), InText=@n9.Names)
   n12 = Event OnTileActiveChanged!() -> (OutputDelegate, then, bIsTileActive:bool) {is=title K2Node_Event}
   > n13 = Branch(Condition=@n12.bIsTileActive)
   n14 = Select(
    Option 0:Widget Animation Object Reference=Get Anim_OnHoverSelect() -> (Anim_OnHoverSelect:Widget Animation Object Reference)
    Option 1:Widget Animation Object Reference=Get Anim_OnTouchSelected() -> (Anim_OnTouchSelected:Widget Animation Object Reference)
    Index=IsUsingTouch(PlayerController=GetOwningPlayer())
   ) -> (ReturnValue:Widget Animation Object Reference)
   > PlayAnimationForward(InAnimation=@n14)
   > OnBaseTileHover(self:WBP Tile Focus Outline Object Reference=Get FocusOutline() -> (FocusOutline:WBP Tile Focus Outline Object Reference))
   PlayAnimationReverse(execute=@n13.else, InAnimation=@n14)
   > OnBaseTileUnhovered(self:WBP Tile Focus Outline Object Reference=Get FocusOutline())
   n15 = Event UpdateSingleFriendName!() -> (OutputDelegate, then, SingleFriendName:"Text (by ref)") {is=title K2Node_Event}
   > "SetText (Text)"(self=Get Text_SingleUserName(), InText=@n15.SingleFriendName)
   n16 = Event UpdateLastInteraction!() -> (OutputDelegate, then, LastInteraction:"Text (by ref)") {is=title K2Node_Event}
   n17 = TextIsEmpty(InText=@n16.LastInteraction)
   > SetVisibility(
    self=Get Text_TimeAgoPartied() -> (Text_TimeAgoPartied:Common Text Object Reference)
    InVisibility=Select(Option 0:ESlateVisibility Enum=HitTestInvisible, Option 1:ESlateVisibility Enum=Collapsed, Index=@n17) -> (ReturnValue:ESlateVisibility Enum)
   )
   > n18 = Branch(Condition=@n17)
   "SetText (Text)"(execute=@n18.else, self=Get Text_TimeAgoPartied(), InText=@n16.LastInteraction)
   n19 = Event On Input Method Changed!() -> (OutputDelegate, then, CurrentInputType:ECommonInputType Enum) {is=title K2Node_Event}
   > n20 = Set Current Input Type(Current Input Type:ECommonInputType Enum=@n19.CurrentInputType) -> (Output_Get:ECommonInputType Enum)
   > Set CTAButton Padding(InputType=@n20)
   n21 = AND Boolean(A="Equal (Enum)"(A=@n20, B=Touch), B=IsMobileGame())
   > n22 = Branch(Condition=@n21)
   n23 = Get WBP_CTA_ButtonPrimary()
   > n24 = SetVisibility(self=@n23, InVisibility=Collapsed)
   SetSize(
    execute=[@n25.then, @n24.then]
    self=Get Spacer_TileSize()
    InSize=Select(Option 0:Vector 2D Structure=Get TileSize(), Option 1:Vector 2D Structure=(X=810,Y=210), Index=@n21) -> (ReturnValue:Vector 2D Structure)
   )
   > SetHeightOverride(self=Get SizeBox_Icons(), InHeightOverride=SelectFloat(A=80, B=48, bPickA=@n21))
   > SetSize(self=Get Spacer_72(), InSize=Select(Option 0=(X=1,Y=2), Option 1=(X=1,Y=5), Index=@n21))
   n25 = SetVisibility(execute=@n22.else, self=@n23, InVisibility=SelfHitTestInvisible)
   n26 = Event BP_OnSelectionChanged!() -> (OutputDelegate, then, bIsSelected:bool) {is=title K2Node_Event}
   > n27 = Branch(Condition=@n26.bIsSelected)
   n28 = Get Anim_OnTouchSelected()
   > PlayAnimation(InAnimation=@n28)
   PlayAnimation(execute=@n27.else, InAnimation=@n28, PlayMode=Reverse)
   n29 = Event Party Information Updated!() -> (OutputDelegate, then, bInIsTileSelected:bool) {is=title K2Node_Event}
   > SetSinglePlayerVisuals()
   > SetInviteVisuals()
   > SetPrivacyVisuals()
   n30 = Select(
    Option 0:Widget Animation Object Reference=Get Anim_OnHoverSelect()
    Option 1:Widget Animation Object Reference=Get Anim_OnTouchSelected()
    Index=IsUsingTouch(PlayerController=GetOwningPlayer())
   ) -> (ReturnValue:Widget Animation Object Reference)
   > SetAnimationCurrentTime!(
    execute
    self:User Widget Object Reference
    InAnimation:Widget Animation Object Reference=@n30
    InTime:float=Select(
     Option 0:float=GetStartTime!(self:Widget Animation Object Reference=@n30) -> (ReturnValue:float)
     Option 1:float=GetEndTime(self=@n30)
     Index=@n29.bInIsTileSelected
    ) -> (ReturnValue:float)
   ) -> (then)
   > Set CTAButton Padding(InputType=GetCurrentInputType(self=Get CommonInputSubsystem()))
   Event OnTextureBeginLoading() {is=title K2Node_Event}
   > SetThumbnail(IsLoaded=false)
   Event On Entry Released() {is=title K2Node_Event}
   > Set CTAButton Padding(InputType=GetCurrentInputType(self=Get CommonInputSubsystem()))
   > SetRenderOpacity(self=Get WBP_CTA_ButtonPrimary(), InOpacity=0)
   n31 = Event OnTextureLoadingComplete!() -> (OutputDelegate, then, ThumbnailTexture:Texture Object Reference) {is=title K2Node_Event}
   > SetThumbnail(Texture=@n31.ThumbnailTexture, IsLoaded=true)
   Branch(Condition=@n1.IsDesignTime)
   > SetInviteVisuals!(execute, self, InIsActiveInvite:bool=Get DebugDesignTime_IsActiveInvite() -> (DebugDesignTime_IsActiveInvite:bool)) -> (then)
   > SetSinglePlayerVisuals!(execute, self, InPartyCount:int=Get DebugDesignTime_PartyCount() -> (DebugDesignTime_PartyCount:int=0)) -> (then)
   > n32 = Branch(Condition="integer > integer"(A=Get DebugDesignTime_PartyCount() -> (DebugDesignTime_PartyCount=0), B=1))
   > n33 = SetText(self=Get RichText_DisplayNamesOrSingleFriendName(), InText=Format Text(Format="INVTEXT(\"(PH) Top Friend Name 01, Top Friend Name 02\")"))
   SetPrivacyVisuals!(execute=[@n33.then, @n34.then], self, InIsPartyPrivate:bool=Get DebugDesignTime_IsPrivateParty() -> (DebugDesignTime_IsPrivateParty:bool)) -> (then)
   n34 = SetText(execute=@n32.else, self=Get RichText_DisplayNamesOrSingleFriendName(), InText=Format Text(Format="INVTEXT(\"(PH) Single Nickname\")"))
  ]}
 Is Arabic Language {nodes [
   Is Arabic Language() {is=title K2Node_FunctionEntry}
   > Return Node!(execute, IsArabic:bool="Equal Exactly (String)"(A=GetCurrentCulture!() -> (ReturnValue:String), B=ar))
  ]}
 Set CTAButton Padding {nodes [
   n1 = Set CTAButton Padding!() -> (then, InputType:ECommonInputType Enum) {is=title K2Node_FunctionEntry}
   n2 = Get WBP_CTA_ButtonPrimary()
   n3 = Break Margin!(Margin:"Margin Structure (by ref)"=Get Padding!(self:Overlay Slot Object Reference=SlotAsOverlaySlot(Widget=@n2)) -> (Padding:Margin Structure)) -> (Left:float=~, Top:float=~, Right:float=~, Bottom:float=~) {is=title K2Node_BreakStruct}
   n4 = GetIsEnabled!(self:Widget Object Reference=Get WBP_CTA_ButtonPrimary()) -> (ReturnValue:bool)
   n5 = SelectFloat(A=40, B=0, bPickA=@n4)
   n6 = SelectFloat(A=24, B=0, bPickA=@n4)
   > SetPadding!(
    execute
    self:Overlay Slot Object Reference=SlotAsOverlaySlot(Widget=Get WBP_CTA_ButtonPrimary())
    InPadding:Margin Structure=Select!(
     MouseAndKeyboard:Margin Structure=Make Margin(Left=@n6, Top=@n3.Top, Right=@n3.Right, Bottom=@n3.Bottom)
     Gamepad:Margin Structure=Make Margin(Left=@n5, Top=@n3.Top, Right=@n3.Right, Bottom=@n3.Bottom)
     Touch:Margin Structure
     Count:Margin Structure
     Index:ECommonInputType Enum=@n1
    ) -> (ReturnValue:Margin Structure)
   ) -> (then)
  ]}
 Set Right Left Background {nodes [
   Set Right Left Background() {is=title K2Node_FunctionEntry}
   > n1 = GetDynamicMaterial(self=Get Bg())
   > SetScalarParameterValue(
    self=@n1
    ParameterName=Mask Direction
    Value=Select!(Option 0:float=0.5, Option 1:float=-0.5, Index:bool=Is Arabic Language!() -> (IsArabic:bool) {is=title K2Node_CallFunction}) -> (ReturnValue:float)
   )
  ]}
 SetInviteVisuals {nodes [
   SetInviteVisuals() {is=title K2Node_FunctionEntry}
   > SetScalarParameterValue(
    self=Get BackgroundDynamicMaterialInstance() -> (BackgroundDynamicMaterialInstance:Material Instance Dynamic Object Reference)
    ParameterName=Get FlareOpacity() -> (FlareOpacity:Name=None)
    Value=Select(Option 0=0, Option 1=1, Index=Get bIsActiveInvite() -> (bIsActiveInvite:bool))
   )
   > SetVisibility(
    self=Get Text_InviteToParty() -> (Text_InviteToParty:Common Text Object Reference)
    InVisibility=Select(Option 0:ESlateVisibility Enum=Collapsed, Option 1:ESlateVisibility Enum=SelfHitTestInvisible, Index=Get bIsActiveInvite()) -> (ReturnValue:ESlateVisibility Enum)
   )
   > Branch(Condition=Get bIsActiveInvite())
   > SetVisibility(self=[Get Text_JoinableParty(), Get Text_OtherPlayers(), Get Text_TimeAgoPartied()], InVisibility=Collapsed)
  ]}
 SetPrivacyVisuals {nodes [
   SetPrivacyVisuals() {is=title K2Node_FunctionEntry}
   > SetVisibility(
    self=Get WBP_JoinableParty_State_Private() -> (WBP_JoinableParty_State_Private:WBP Joinable Party State Private Object Reference)
    InVisibility=Select(Option 0=Collapsed, Option 1=HitTestInvisible, Index=Get bIsPartyPrivate() -> (bIsPartyPrivate:bool))
   )
  ]}
 SetSinglePlayerVisuals {nodes [
   SetSinglePlayerVisuals() {is=title K2Node_FunctionEntry}
   n1 = "integer > integer"(A=Get CurrentPartySize() -> (CurrentPartySize:int=0), B=1)
   > SetVisibility(self=Get Text_SingleUserName(), InVisibility=Select(Option 0=HitTestInvisible, Option 1=Collapsed, Index=@n1))
   > SetVisibility(self=Get Text_OtherPlayers(), InVisibility=Select(Option 0=Collapsed, Option 1=HitTestInvisible, Index=@n1))
  ]}
 SetThumbnail {nodes [
   n1 = SetThumbnail!() -> (then, Texture:Texture Object Reference, IsLoaded:bool=~) {is=title K2Node_FunctionEntry}
   > n2 = Branch(Condition=@n1.IsLoaded)
   n3 = Get BackgroundDynamicMaterialInstance()
   > SetTextureParameterValue(self=@n3, ParameterName=Get ThumbnailTexture() -> (ThumbnailTexture:Name=None), Value=@n1)
   > PlayAnimationForward(InAnimation=Get Anim_OnImageLoaded() -> (Anim_OnImageLoaded:Widget Animation Object Reference))
   SetScalarParameterValue(execute=@n2.else, self=@n3, ParameterName=Get ThumbnailOpacity() -> (ThumbnailOpacity:Name=None), Value=0)
  ]}
 SetTileSize {nodes [
   n1 = SetTileSize!() -> (then, Size:Vector 2D Structure) {is=title K2Node_FunctionEntry}
   n2 = GetMobileDiscoverTileSizeMultiplier!(self:Fort Runtime Options Object Reference=GetRuntimeOptions()) -> (ReturnValue:float)
   n3 = BreakVector2D(InVec=@n1)
   n4 = "To Float (Integer)"(InInt=Floor(A=float * float(A=@n3, B=@n2) {is=title K2Node_PromotableOperator}) {is=title K2Node_CallFunction})
   n5 = IsMobileGame()
   n6 = "To Float (Integer)"(InInt=Floor(A=float * float(A=@n3.Y, B=@n2)))
   > SetSize(
    self=Get Spacer_TileSize()
    InSize=MakeVector2D(X=Select(Option 0:double=@n3, Option 1:double=@n4, Index=@n5) -> (ReturnValue:double), Y=Select(Option 0=@n3.Y, Option 1=@n6, Index=@n5))
   )
  ]}
]
variables [
 BackgroundDynamicMaterialInstance {type=MaterialInstanceDynamic}
 ColorSharedTextAndImages {type=LinearColor, default=<rgba 0.6038 1 1 1>}
 Current Input Type {type=ECommonInputType, default=MouseAndKeyboard}
 DebugDesignTime_IsActiveInvite {type=boolean, default=true}
 DebugDesignTime_IsPrivateParty {type=boolean}
 DebugDesignTime_PartyCount {type=integer, default=2}
 FlareOpacity {type=string, default=Flare Opacity}
 ThumbnailOpacity {type=string, default=Thumbnail Opacity}
 ThumbnailTexture {type=string, default=Thumbnail}
 TileSize {type=Vector2D, default=<648 196>}
]
widgets {
 info {
  inheritedWidgetCount=2
  parentClass=/Script/DiscoveryBrowserUI.FortActivityBrowserPlayWithFriendsTile
  rootWidgetClass=&Overlay
  widgetCount=27
 }
 widgets [
  Overlay_Root {widgetClassPath=&Overlay, widget=Overlay_Root}
  Spacer_TileSize {
   widgetClassPath=&Spacer
   bIsVariable=true
   parent=Overlay_Root
   widget=Spacer_TileSize
   properties {size=<648 196>}
   slot {class=OverlaySlot, properties {horizontalAlignment=HAlign_Fill, verticalAlignment=VAlign_Fill}}
  }
  Overlay_RootContent {
   widgetClassPath=&Overlay
   parent=Overlay_Root
   widget=Overlay_RootContent
   slot {class=OverlaySlot, properties {horizontalAlignment=HAlign_Fill, verticalAlignment=VAlign_Fill}}
  }
  GridPanel_Root {
   widgetClassPath=&GridPanel
   parent=Overlay_RootContent
   widget=GridPanel_Root
   properties {columnFill [1], rowFill [0,1]}
   slot {class=OverlaySlot, properties {horizontalAlignment=HAlign_Fill, verticalAlignment=VAlign_Fill}}
  }
  Bg {
   widgetClassPath=&Image
   bIsVariable=true
   parent=GridPanel_Root
   widget=Bg
   properties {
    visibility=SelfHitTestInvisible
    brush {
     bIsDynamicallyLoaded=false
     drawAs=Box
     imageType=NoImage
     mirroring=NoMirror
     resourceName=None
     resourceObject=/DiscoveryBrowser/Tiles/JoinableParty/MaterialInstances/MI_UI_JoinableParty_Bg
     tiling=NoTile
     imageSize=<32 32>
     margin=<ltrb 0.5 0.5 0.5 0.5>
     outlineSettings {
      bUseBrushTransparency=false
      roundingType=HalfHeightRadius
      width=0
      color {colorUseRule=UseColor_Specified, specifiedColor=<rgba 0 0 0 0>}
      cornerRadii=<0 0 0 0>
     }
     tintColor {colorUseRule=UseColor_Specified, specifiedColor=<rgba 1 1 1 1>}
     uVRegion {bIsValid=false, max=<0 0>, min=<0 0>}
    }
   }
   slot {class=GridSlot, properties {rowSpan=2}}
  }
  FocusOutline {
   widgetClassPath=&/WBP_Tile_FocusOutline
   bIsVariable=true
   parent=GridPanel_Root
   widget=FocusOutline
   slot {class=GridSlot, properties {rowSpan=2}}
  }
  Grid_TopInfo {
   widgetClassPath=&VerticalBox
   parent=GridPanel_Root
   widget=Grid_TopInfo
   slot {class=GridSlot, properties {padding=<ltrb 24 16 24 0>}}
  }
  SizeBox_Icons {
   widgetClassPath=&SizeBox
   bIsVariable=true
   parent=Grid_TopInfo
   widget=SizeBox_Icons
   properties {bOverride_HeightOverride=true, heightOverride=48}
   slot {class=VerticalBoxSlot, properties {horizontalAlignment=HAlign_Left, verticalAlignment=VAlign_Top}}
  }
  ScaleBox_0 {widgetClassPath=&ScaleBox, parent=SizeBox_Icons, widget=ScaleBox_0, properties {stretch=ScaleToFitY}, slot {class=SizeBoxSlot}}
  PartyMembersAvatarsDisplay {
   widgetClassPath=/Game/UI/Sidebar/SidebarJoinablePartyPortraits
   bInherited=true
   bIsVariable=true
   parent=ScaleBox_0
   widget=PartyMembersAvatarsDisplay
   properties {maxDisplayedUsers=3}
   slot {class=ScaleBoxSlot}
  }
  Spacer_72 {
   widgetClassPath=&Spacer
   bIsVariable=true
   parent=Grid_TopInfo
   widget=Spacer_72
   properties {size=<1 2>}
   slot {class=VerticalBoxSlot}
  }
  HB_Names {
   widgetClassPath=&HorizontalBox
   parent=Grid_TopInfo
   widget=HB_Names
   slot {class=VerticalBoxSlot, properties {horizontalAlignment=HAlign_Left}}
  }
  RichText_DisplayNamesOrSingleFriendName {
   widgetClassPath=&CommonRichTextBlock
   bIsVariable=true
   parent=HB_Names
   widget=RichText_DisplayNamesOrSingleFriendName
   properties {
    bOverrideDefaultStyle=true
    bTintInlineIcon=true
    mobileTextBlockScale=1.5
    text="{PH} Top Friend Name"
    defaultTextStyleOverride {
     overflowPolicy=Clip
     transformPolicy=None
     colorAndOpacity {colorUseRule=UseColor_Specified, specifiedColor=<rgba 1 1 1 1>}
     font {
      bForceMonospaced=false
      bMaterialIsStencil=false
      fontMaterial=null
      fontObject=&BurbankSmall
      letterSpacing=0
      monospacedWidth=1
      size=16
      skewAmount=0
      typefaceFontName=Black
      outlineSettings {
       bApplyOutlineToDropShadows=false
       bMiteredCorners=false
       bSeparateFillAlpha=false
       outlineMaterial=null
       outlineSize=0
       outlineColor=<rgba 0 0 0 1>
      }
     }
     highlightColor {colorUseRule=UseColor_Specified, specifiedColor=<rgba 0 0 0 1>}
     highlightShape {
      bIsDynamicallyLoaded=false
      drawAs=Image
      imageType=NoImage
      mirroring=NoMirror
      resourceName=None
      resourceObject=null
      tiling=NoTile
      imageSize=<32 32>
      margin=<ltrb 0 0 0 0>
      outlineSettings {
       bUseBrushTransparency=false
       roundingType=HalfHeightRadius
       width=0
       color {colorUseRule=UseColor_Specified, specifiedColor=<rgba 0 0 0 0>}
       cornerRadii=<0 0 0 0>
      }
      tintColor {colorUseRule=UseColor_Specified, specifiedColor=<rgba 1 1 1 1>}
      uVRegion {bIsValid=false, max=<0 0>, min=<0 0>}
     }
     selectedBackgroundColor {colorUseRule=UseColor_Foreground, specifiedColor=<rgba 1 0 1 1>}
     shadowColorAndOpacity=<rgba 0 0 0 1>
     shadowOffset=<0 0>
     strikeBrush {
      bIsDynamicallyLoaded=false
      drawAs=Image
      imageType=NoImage
      mirroring=NoMirror
      resourceName=None
      resourceObject=null
      tiling=NoTile
      imageSize=<32 32>
      margin=<ltrb 0 0 0 0>
      outlineSettings {
       bUseBrushTransparency=false
       roundingType=HalfHeightRadius
       width=0
       color {colorUseRule=UseColor_Specified, specifiedColor=<rgba 0 0 0 0>}
       cornerRadii=<0 0 0 0>
      }
      tintColor {colorUseRule=UseColor_Specified, specifiedColor=<rgba 1 1 1 1>}
      uVRegion {bIsValid=false, max=<0 0>, min=<0 0>}
     }
     underlineBrush {
      bIsDynamicallyLoaded=false
      drawAs=Image
      imageType=NoImage
      mirroring=NoMirror
      resourceName=None
      resourceObject=null
      tiling=NoTile
      imageSize=<32 32>
      margin=<ltrb 0 0 0 0>
      outlineSettings {
       bUseBrushTransparency=false
       roundingType=HalfHeightRadius
       width=0
       color {colorUseRule=UseColor_Specified, specifiedColor=<rgba 0 0 0 0>}
       cornerRadii=<0 0 0 0>
      }
      tintColor {colorUseRule=UseColor_Specified, specifiedColor=<rgba 1 1 1 1>}
      uVRegion {bIsValid=false, max=<0 0>, min=<0 0>}
     }
    }
   }
   slot {class=HorizontalBoxSlot, properties {padding=<ltrb 0 0 8 0>}}
  }
  Text_SingleUserName {
   widgetClassPath=&CommonTextBlock
   bIsVariable=true
   parent=HB_Names
   widget=Text_SingleUserName
   properties {
    mobileFontSizeMultiplier=1.5
    text="(PH) Single Name"
    colorAndOpacity {colorUseRule=UseColor_Specified, specifiedColor=<rgba 0.6038 1 1 1>}
    font {
     bForceMonospaced=false
     bMaterialIsStencil=false
     fontMaterial=null
     fontObject=&BurbankSmall
     letterSpacing=0
     monospacedWidth=1
     size=16
     skewAmount=0
     typefaceFontName=Bold
     outlineSettings {
      bApplyOutlineToDropShadows=false
      bMiteredCorners=false
      bSeparateFillAlpha=false
      outlineMaterial=null
      outlineSize=0
      outlineColor=<rgba 0 0 0 1>
     }
    }
   }
   slot {class=HorizontalBoxSlot, properties {verticalAlignment=VAlign_Center}}
  }
  HorizontalBox_0 {widgetClassPath=&HorizontalBox, parent=Grid_TopInfo, widget=HorizontalBox_0, slot {class=VerticalBoxSlot}}
  Text_OtherPlayers {
   widgetClassPath=&CommonTextBlock
   bIsVariable=true
   parent=HorizontalBox_0
   widget=Text_OtherPlayers
   properties {
    mobileFontSizeMultiplier=1.75
    text="(PH) +2 others"
    visibility=Collapsed
    colorAndOpacity {colorUseRule=UseColor_Specified, specifiedColor=<rgba 0.0331 0.7011 1 1>}
    font {
     bForceMonospaced=false
     bMaterialIsStencil=false
     fontMaterial=null
     fontObject=&BurbankSmall
     letterSpacing=0
     monospacedWidth=1
     size=13
     skewAmount=0
     typefaceFontName=Bold
     outlineSettings {
      bApplyOutlineToDropShadows=false
      bMiteredCorners=false
      bSeparateFillAlpha=false
      outlineMaterial=null
      outlineSize=0
      outlineColor=<rgba 0 0 0 1>
     }
    }
    shadowOffset=<0 0>
    shapedTextOptions {bOverride_TextFlowDirection=true, bOverride_TextShapingMethod=false, textFlowDirection=Culture, textShapingMethod=Auto}
   }
   slot {class=HorizontalBoxSlot}
  }
  Text_JoinableParty {
   widgetClassPath=&CommonTextBlock
   bIsVariable=true
   parent=HorizontalBox_0
   widget=Text_JoinableParty
   properties {
    mobileFontSizeMultiplier=1.75
    text=In a joinable party
    visibility=Collapsed
    colorAndOpacity {colorUseRule=UseColor_Specified, specifiedColor=<rgba 0.0331 0.7011 1 1>}
    font {
     bForceMonospaced=false
     bMaterialIsStencil=false
     fontMaterial=null
     fontObject=&BurbankSmall
     letterSpacing=0
     monospacedWidth=1
     size=13
     skewAmount=0
     typefaceFontName=Bold
     outlineSettings {
      bApplyOutlineToDropShadows=false
      bMiteredCorners=false
      bSeparateFillAlpha=false
      outlineMaterial=null
      outlineSize=0
      outlineColor=<rgba 0 0 0 1>
     }
    }
    shadowOffset=<0 0>
    shapedTextOptions {bOverride_TextFlowDirection=true, bOverride_TextShapingMethod=false, textFlowDirection=Culture, textShapingMethod=Auto}
   }
   slot {class=HorizontalBoxSlot}
  }
  Text_TimeAgoPartied {
   widgetClassPath=&CommonTextBlock
   bIsVariable=true
   parent=HorizontalBox_0
   widget=Text_TimeAgoPartied
   properties {
    mobileFontSizeMultiplier=1.75
    text="(PH) Partied 24h ago"
    visibility=Collapsed
    colorAndOpacity {colorUseRule=UseColor_Specified, specifiedColor=<rgba 0 0.6183 1 1>}
    font {
     bForceMonospaced=false
     bMaterialIsStencil=false
     fontMaterial=null
     fontObject=&BurbankSmall
     letterSpacing=0
     monospacedWidth=1
     size=13
     skewAmount=0
     typefaceFontName=Bold
     outlineSettings {
      bApplyOutlineToDropShadows=false
      bMiteredCorners=false
      bSeparateFillAlpha=false
      outlineMaterial=null
      outlineSize=0
      outlineColor=<rgba 0 0 0 1>
     }
    }
    shadowOffset=<0 0>
    shapedTextOptions {bOverride_TextFlowDirection=true, bOverride_TextShapingMethod=false, textFlowDirection=Culture, textShapingMethod=Auto}
   }
   slot {class=HorizontalBoxSlot}
  }
  Text_InviteToParty {
   widgetClassPath=&CommonTextBlock
   bIsVariable=true
   parent=HorizontalBox_0
   widget=Text_InviteToParty
   properties {
    mobileFontSizeMultiplier=1.75
    text=has invited you to their party
    visibility=Collapsed
    colorAndOpacity {colorUseRule=UseColor_Specified, specifiedColor=<rgba 0 0.6183 1 1>}
    font {
     bForceMonospaced=false
     bMaterialIsStencil=false
     fontMaterial=null
     fontObject=&BurbankSmall
     letterSpacing=0
     monospacedWidth=1
     size=13
     skewAmount=0
     typefaceFontName=Bold
     outlineSettings {
      bApplyOutlineToDropShadows=false
      bMiteredCorners=false
      bSeparateFillAlpha=false
      outlineMaterial=null
      outlineSize=0
      outlineColor=<rgba 0 0 0 1>
     }
    }
    shadowOffset=<0 0>
    shapedTextOptions {bOverride_TextFlowDirection=true, bOverride_TextShapingMethod=false, textFlowDirection=Culture, textShapingMethod=Auto}
   }
   slot {class=HorizontalBoxSlot}
  }
  Grid_BottomInfo_Clips {
   widgetClassPath=&Overlay
   parent=GridPanel_Root
   widget=Grid_BottomInfo_Clips
   slot {class=GridSlot, properties {row=1, verticalAlignment=VAlign_Bottom}}
  }
  HB_PresenceTexts {
   widgetClassPath=&HorizontalBox
   parent=Grid_BottomInfo_Clips
   widget=HB_PresenceTexts
   slot {class=OverlaySlot, properties {padding=<ltrb 24 16 0 16>}}
  }
  SizeBox_MaxWidthForLongPresenceToForceEllipses {
   widgetClassPath=&SizeBox
   parent=HB_PresenceTexts
   widget=SizeBox_MaxWidthForLongPresenceToForceEllipses
   properties {bOverride_MaxDesiredWidth=true, maxDesiredWidth=540}
   slot {class=HorizontalBoxSlot, properties {horizontalAlignment=HAlign_Left}}
  }
  Text_RichPresence {
   widgetClassPath=&CommonTextBlock
   bIsVariable=true
   parent=SizeBox_MaxWidthForLongPresenceToForceEllipses
   widget=Text_RichPresence
   properties {
    clipping=ClipToBounds
    mobileFontSizeMultiplier=1.35
    scrollStyle=&TextScrollStyle-Base_C
    text="(PH) Main Lobby"
    textOverflowPolicy=Ellipsis
    colorAndOpacity {colorUseRule=UseColor_Specified, specifiedColor=<rgba 0.6038 1 1 1>}
    font {
     bForceMonospaced=false
     bMaterialIsStencil=false
     fontMaterial=null
     fontObject=&BurbankSmall
     letterSpacing=0
     monospacedWidth=1
     size=15
     skewAmount=0
     typefaceFontName=Bold
     outlineSettings {
      bApplyOutlineToDropShadows=false
      bMiteredCorners=false
      bSeparateFillAlpha=false
      outlineMaterial=null
      outlineSize=0
      outlineColor=<rgba 0 0 0 1>
     }
    }
    shapedTextOptions {bOverride_TextFlowDirection=true, bOverride_TextShapingMethod=false, textFlowDirection=Culture, textShapingMethod=Auto}
   }
   slot {class=SizeBoxSlot}
  }
  WBP_CTA_ButtonPrimary {
   widgetClassPath=/FortUILibrary/UIKit/Buttons/WBP_UIKit_Button_Quiet
   bIsVariable=true
   parent=Grid_BottomInfo_Clips
   widget=WBP_CTA_ButtonPrimary
   properties {minHeight=48, text=Quiet Button, triggeringInputAction {dataTable=&InputActionDataTable, rowName=Accept}}
   slot {class=OverlaySlot, properties {padding=<ltrb 24 16 0 16>}}
  }
  WBP_JoinableParty_State_Private {
   widgetClassPath=&/WBP_JoinableParty_State_Private
   bIsVariable=true
   parent=Overlay_RootContent
   widget=WBP_JoinableParty_State_Private
   properties {visibility=Collapsed}
   slot {class=OverlaySlot, properties {horizontalAlignment=HAlign_Right, padding=<ltrb 0 4 4 0>}}
  }
  PulseOutline {
   widgetClassPath=&Image
   parent=Overlay_Root
   widget=PulseOutline
   properties {
    visibility=Collapsed
    brush {
     bIsDynamicallyLoaded=false
     drawAs=Box
     imageType=NoImage
     mirroring=NoMirror
     resourceName=None
     resourceObject=/FortniteCoreUI/UI/DiscoveryShared/Art/MaterialInstances/MI_UI_Discovery_Tile_Border
     tiling=NoTile
     imageSize=<32 32>
     margin=<ltrb 0.5 0.5 0.5 0.5>
     outlineSettings {
      bUseBrushTransparency=false
      roundingType=HalfHeightRadius
      width=0
      color {colorUseRule=UseColor_Specified, specifiedColor=<rgba 0 0 0 0>}
      cornerRadii=<0 0 0 0>
     }
     tintColor {colorUseRule=UseColor_Specified, specifiedColor=<rgba 1 1 1 1>}
     uVRegion {bIsValid=false, max=<0 0>, min=<0 0>}
    }
   }
   slot {class=OverlaySlot, properties {horizontalAlignment=HAlign_Fill, verticalAlignment=VAlign_Fill}}
  }
  InputActionWidget {widgetClassPath=&CommonActionWidget, bInherited=true, widget=null}
 ]
}
