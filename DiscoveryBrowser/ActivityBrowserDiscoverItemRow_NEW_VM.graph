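# Discovery browser category row widget on FortDiscoverItemBrowserRow, with a title, tile list, paging arrows and loading-more indicator.
# Row events drive animations for active, peek and move states through tint, opacity and font-material parameters, while UpdateNewRow resets states for changed categories.
# Active changes propagate to arrow buttons, and a timer delays the loading indicator.
# Input method sets row visibility, and a parameter collection sizes the title, with mobile scaling.
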WidgetBlueprint /DiscoveryBrowser/ActivityBrowserDiscoverItemRow_NEW_VM.ActivityBrowserDiscoverItemRow
animations [
 OnRowActiveInactiveAnim {
  binding_count=3
  length_seconds=0.16
  object=&self:OnRowActiveInactiveAnim
  sequence {
   bindings [
    Text_CategoryName {
     id=#1
     possessed_class=CommonTextBlock
     tracks [{
       type=MovieSceneWidgetMaterialTrack
       display_name=Font.FontMaterial
       sections [{type=MovieSceneParameterSection, channels [Pressed_Animate {type=MovieSceneScriptingFloatChannel, keys [[0,0],[16,1]]}]}]
      }]
    }
    InactiveDarken {
     id=#2
     possessed_class=CommonBorder
     tracks [{
       type=MovieSceneColorTrack
       display_name=Content Color and Opacity
       sections [{
         type=MovieSceneColorSection
         channels [
          Color.R {type=MovieSceneScriptingFloatChannel, keys [[0,0.625],[16,1]]}
          Color.G {type=MovieSceneScriptingFloatChannel, keys [[0,0.875],[16,1]]}
          Color.B {type=MovieSceneScriptingFloatChannel, const [1,2]}
          Color.A {type=MovieSceneScriptingFloatChannel}
         ]
        }]
      }]
    }
    InactiveOverlayDim {
     id=#3
     possessed_class=Image
     tracks [{
       type=MovieSceneFloatTrack
       display_name=Render Opacity
       sections [{type=MovieSceneFloatSection, channels [None {type=MovieSceneScriptingFloatChannel, keys [[0,0.375],[16,0]]}]}]
      }]
    }
   ]
   display_rate {numerator=100}
  }
 }
 OnRowMoveUpOutOfViewAnim {
  binding_count=2
  length_seconds=0.16
  object=&self:OnRowMoveUpOutOfViewAnim
  sequence {
   bindings [
    VBContent {
     id=#4
     possessed_class=VerticalBox
     tracks [{
       type=MovieSceneFloatTrack
       display_name=Render Opacity
       sections [{type=MovieSceneFloatSection, channels [None {type=MovieSceneScriptingFloatChannel, keys [[0,1],[16,0]]}]}]
      }]
    }
    Text_CategoryName {
     id=#5
     possessed_class=CommonTextBlock
     tracks [{
       type=MovieSceneWidgetMaterialTrack
       display_name=Font.FontMaterial
       sections [{type=MovieSceneParameterSection, channels [Pressed_Animate {type=MovieSceneScriptingFloatChannel, keys [[0,1],[16,0]]}]}]
      }]
    }
   ]
   display_rate {numerator=100}
  }
 }
 OnRowMoveDownIntoView {
  binding_count=2
  length_seconds=0.24
  object=&self:OnRowMoveDownIntoView
  sequence {
   bindings [
    Text_CategoryName {
     id=#6
     possessed_class=CommonTextBlock
     tracks [{
       type=MovieSceneWidgetMaterialTrack
       display_name=Font.FontMaterial
       sections [{type=MovieSceneParameterSection, channels [Pressed_Animate {type=MovieSceneScriptingFloatChannel, keys [[0,0],[24,1]]}]}]
      }]
    }
    VBContent {
     id=#7
     possessed_class=VerticalBox
     tracks [{
       type=MovieSceneFloatTrack
       display_name=Render Opacity
       sections [{type=MovieSceneFloatSection, channels [None {type=MovieSceneScriptingFloatChannel, keys [[0,0],[24,1]]}]}]
      }]
    }
   ]
   display_rate {numerator=100}
  }
 }
 OnPeek {
  binding_count=3
  length_seconds=0.24
  object=&self:OnPeek
  sequence {
   bindings [
    InactiveOverlayDim {
     id=#8
     possessed_class=Image
     tracks [{
       type=MovieSceneFloatTrack
       display_name=Render Opacity
       sections [{type=MovieSceneFloatSection, channels [None {type=MovieSceneScriptingFloatChannel, keys [[0,0.375],[24,0]]}]}]
      }]
    }
    InactiveDarken {
     id=#9
     possessed_class=CommonBorder
     tracks [{
       type=MovieSceneColorTrack
       display_name=Content Color and Opacity
       sections [{
         type=MovieSceneColorSection
         channels [
          Color.R {type=MovieSceneScriptingFloatChannel, keys [[0,0.625],[24,1]]}
          Color.G {type=MovieSceneScriptingFloatChannel, keys [[0,0.875],[24,1]]}
          Color.B {type=MovieSceneScriptingFloatChannel, const [1,2]}
          Color.A {type=MovieSceneScriptingFloatChannel}
         ]
        }]
      }]
    }
    Text_CategoryName {
     id=#10
     possessed_class=CommonTextBlock
     tracks [{
       type=MovieSceneWidgetMaterialTrack
       display_name=Font.FontMaterial
       sections [{type=MovieSceneParameterSection, channels [Hover_Animate {type=MovieSceneScriptingFloatChannel, keys [[0,0],[24,1]]}]}]
      }]
    }
   ]
   display_rate {numerator=100}
  }
 }
 OnLoadingMore {
  binding_count=1
  length_seconds=0.24
  object=&self:OnLoadingMore
  sequence {
   bindings [LoadingMoreTilesHB {
     id=#11
     possessed_class=HorizontalBox
     tracks [
      {
       type=MovieSceneFloatTrack
       display_name=Render Opacity
       sections [{type=MovieSceneFloatSection, channels [None {type=MovieSceneScriptingFloatChannel, keys [[0,0],[24,1]]}]}]
      }
      {
       type=MovieScene2DTransformTrack
       sections [{type=MovieScene2DTransformSection, channels [Translation.X {type=MovieSceneScriptingFloatChannel, keys [[0,16],[24,0]]}]}]
      }
     ]
    }]
   display_rate {numerator=100}
  }
 }
]
graphs [
 EventGraph {nodes [
   n1 = HandleInputMethodChanged!() -> (OutputDelegate, then, NewInputType:ECommonInputType Enum) {is=title K2Node_CustomEvent}
   > SetVisibility(InVisibility=Select!(
     MouseAndKeyboard:ESlateVisibility Enum=Visible
     Gamepad:ESlateVisibility Enum=SelfHitTestInvisible
     Touch:ESlateVisibility Enum=Visible
     Count:ESlateVisibility Enum=SelfHitTestInvisible
     Index:ECommonInputType Enum=@n1.NewInputType
    ) -> (ReturnValue))
   n2 = FinishedLoadingMoreDelay() {is=title K2Node_CustomEvent}
   > Branch(Condition=Get IsLoadingMoreQueryActive() -> (IsLoadingMoreQueryActive:bool))
   > n3 = Branch(Condition=Get IsLoadingMoreVisible() -> (IsLoadingMoreVisible:bool))
   Set IsLoadingMoreVisible(execute=@n3.else, IsLoadingMoreVisible:bool=true) -> (Output_Get:bool)
   > PlayAnimationForward(InAnimation=Get OnLoadingMore() -> (OnLoadingMore:Widget Animation Object Reference))
   > SetVisibility(self=Get LoadingMoreTilesHB() -> (LoadingMoreTilesHB:Horizontal Box Object Reference), InVisibility=SelfHitTestInvisible)
   n4 = Event OnCategoryItemChanged!() -> (OutputDelegate, then, bPlayAnimation:bool, CategoryPanelName:"Name (by ref)"=None) {is=title K2Node_Event}
   > UpdateNewRow!(
    execute
    self
    IsActive:bool=GetIsActive!(self:Fort Activity Browser Row Object Reference) -> (ReturnValue:bool)
    PlayAnimation:bool=@n4.bPlayAnimation
   ) -> (then)
   n5 = Event OnRowPeekStateChanged!() -> (OutputDelegate, then, bIsInPeekState:bool) {is=title K2Node_Event}
   n6 = Get OnRowActiveInactiveAnim() -> (OnRowActiveInactiveAnim:Widget Animation Object Reference)
   > n7 = Branch(Condition=OR Boolean(A=IsAnimationPlaying(InAnimation=@n6), B=GetIsActive()))
   n8 = Branch(execute=@n7.else, Condition=@n5.bIsInPeekState)
   n9 = Get OnPeek() -> (OnPeek:Widget Animation Object Reference)
   > PlayAnimationForward(InAnimation=@n9)
   PlayAnimationReverse(execute=@n8.else, InAnimation=@n9)
   n10 = Event OnRowMoveUp!() -> (OutputDelegate, then, bMovingOffscreen:bool) {is=title K2Node_Event}
   > Branch(Condition=@n10.bMovingOffscreen)
   > Check and Stop Animation!(
    execute:Exec
    Animation:Widget Animation Object Reference=Get OnRowMoveDownIntoView() -> (OnRowMoveDownIntoView:Widget Animation Object Reference)
   ) -> (then:Exec) {is=title K2Node_MacroInstance}
   > PlayAnimation(InAnimation=Get OnRowMoveUpOutOfViewAnim() -> (OnRowMoveUpOutOfViewAnim:Widget Animation Object Reference))
   n11 = Event OnRowIsActiveChanged!() -> (OutputDelegate, then, bIsActiveRow:bool) {is=title K2Node_Event}
   > Check and Stop Animation(Animation=Get OnPeek())
   > n12 = Sequence()
   > n13 = Branch(Condition=@n11.bIsActiveRow)
   > SetVisibility(InVisibility=SelfHitTestInvisible)
   n14 = Get OnRowActiveInactiveAnim()
   > PlayAnimation(
    InAnimation=@n14
    StartAtTime=GetAnimationCurrentTime!(self:User Widget Object Reference, InAnimation:Widget Animation Object Reference=@n14) -> (ReturnValue:float)
   )
   SetVisibility(execute=@n13.else, InVisibility=HitTestInvisible)
   > n15 = "Play Animation with Finished event (legacy)"!(
    execute:Exec
    Widget:User Widget Object Reference=Self-Reference()
    InAnimation:Widget Animation Object Reference=@n14
    StartAtTime:float=GetAnimationCurrentTime(InAnimation=@n14)
    NumLoopsToPlay:int=1
    PlayMode:EUMGSequencePlayMode Enum=Reverse
    PlaybackSpeed:float=1
   ) -> (then:Exec, Result:UMGSequence Player Object Reference, Finished:Exec) {is=title K2Node_PlayAnimation}
   > Set IsWaitingForInactiveAnim(IsWaitingForInactiveAnim:bool=true) -> (Output_Get:bool)
   n16 = Branch(
    execute=@n15.Finished
    Condition=AND Boolean(
     A=GetIsInPeekState!(self:Fort Activity Browser Row Object Reference) -> (ReturnValue:bool)
     B=Get IsWaitingForInactiveAnim() -> (IsWaitingForInactiveAnim:bool)
    )
   )
   > n17 = PlayAnimationForward(InAnimation=Get OnPeek())
   Set IsWaitingForInactiveAnim(execute=[@n17.then, @n16.else], IsWaitingForInactiveAnim=false)
   n18 = For Each Loop(
    Exec=@n12.then_1
    Array:Array of Activity Browser Arrow Button Object References=Get ArrowButtons() -> (ArrowButtons:Array of Activity Browser Arrow Button Object References)
   ) -> (Array Element:Activity Browser Arrow Button Object Reference)
   > OnRowActiveChanged!(execute, self:Activity Browser Arrow Button Object Reference=@n18, IsActive:bool=@n11.bIsActiveRow) -> (then)
   Event OnRowMoveDown!() -> (OutputDelegate, then, bMovingOffscreen:bool) {is=title K2Node_Event}
   > Check and Stop Animation(Animation=Get OnRowMoveUpOutOfViewAnim())
   n19 = Get VBContent() -> (VBContent:Vertical Box Object Reference)
   > Branch(Condition="float < float"!(A:double=GetRenderOpacity!(self:Widget Object Reference=@n19) -> (ReturnValue:float), B:double=1) -> (ReturnValue:bool) {is=title K2Node_CallFunction})
   > PlayAnimation(InAnimation=Get OnRowMoveDownIntoView())
   Event Construct()
   > Parent: Construct()
   n20 = Create Event()
   n21 = Get CommonInputSubsystem()
   > Bind Event to On Input Method Changed(self=@n21, Delegate=@n20)
   > HandleInputMethodChanged!(execute, self, NewInputType:ECommonInputType Enum=GetCurrentInputType(self=@n21)) -> (then) {is=title K2Node_CallFunction}
   > Set Font Category Size() {is=title K2Node_CallFunction}
   > n22 = Sequence()
   > n23 = Cast To ActivityBrowserArrowButton!(execute, Object=Get Button_PageLeft() -> (Button_PageLeft:Common Button Base Object Reference)) -> (then, CastFailed, AsActivity Browser Arrow Button:Activity Browser Arrow Button Object Reference) {is=title K2Node_DynamicCast}
   > Add(
    TargetArray:Array of Activity Browser Arrow Button Object References=Get ArrowButtons()
    NewItem:"Activity Browser Arrow Button Object Reference (by ref)"=@n23
   ) {is=title K2Node_CallArrayFunction}
   n24 = Cast To ActivityBrowserArrowButton(execute=@n22.then_1, Object=Get Button_PageRight() -> (Button_PageRight:Common Button Base Object Reference))
   > Add(TargetArray=Get ArrowButtons(), NewItem=@n24)
   Event Destruct()
   > Parent: Destruct()
   > Unbind Event from On Input Method Changed(self=@n21, Delegate=@n20)
   > Clear(TargetArray:Array of Activity Browser Arrow Button Object References=Get ArrowButtons())
   K2Node_Event_7 = OnQueryStatusChanged_1!() -> (OutputDelegate, then, bIsActive:bool=~) {is=title K2Node_CustomEvent}
   > n25 = Set IsLoadingMoreQueryActive(IsLoadingMoreQueryActive:bool=@K2Node_Event_7.bIsActive) -> (Output_Get:bool)
   > n26 = Branch(Condition=@n25)
   > n27 = Set Timer by Event(Delegate=@n2, Time=Get LoadingMoreDisplayDelay() -> (LoadingMoreDisplayDelay:double), bLooping=false)
   > Set LoadingMoreDelayTimer(LoadingMoreDelayTimer:Timer Handle Structure=@n27) -> (Output_Get:Timer Handle Structure)
   Clear and Invalidate Timer by Handle(execute=@n26.else, Handle=Get LoadingMoreDelayTimer() -> (LoadingMoreDelayTimer:Timer Handle Structure))
   > Branch(Condition=Get IsLoadingMoreVisible())
   > Set IsLoadingMoreVisible(IsLoadingMoreVisible=false)
   n28 = Get OnLoadingMore()
   n29 = GetAnimationCurrentTime(InAnimation=@n28)
   n30 = GetEndTime(self=@n28)
   > n31 = "Play Animation with Finished event (legacy)"(
    Widget=Self-Reference()
    InAnimation=@n28
    StartAtTime=SelectFloat(A=float - float(A=@n30, B=@n29) {is=title K2Node_CallFunction}, B=0, bPickA=IsAnimationPlaying(InAnimation=@n28))
    NumLoopsToPlay=1
    PlayMode=Reverse
    PlaybackSpeed=1
   )
   n32 = Branch(execute=@n31.Finished, Condition=Get IsLoadingMoreVisible())
   SetVisibility(execute=@n32.else, self=Get LoadingMoreTilesHB(), InVisibility=Collapsed)
  ]}
 ResetOnPeekAnimation {nodes [
   ResetOnPeekAnimation() {is=title K2Node_FunctionEntry}
   > SetContentColorAndOpacity!(
    execute
    self:Border Object Reference=Get InactiveDarken() -> (InactiveDarken:Common Border Object Reference)
    InContentColorAndOpacity:Linear Color Structure=Make Linear Color!(R:float=1, G:float=1, B:float=1, A:float=1) -> (LinearColor:Linear Color Structure) {is=title K2Node_MakeStruct}
   ) -> (then)
   > SetRenderOpacity(self=Get InactiveOverlayDim() -> (InactiveOverlayDim:Image Object Reference), InOpacity=0)
   > n1 = GetDynamicFontMaterial!(execute, self:Text Object Reference=Get Text_CategoryName() -> (Text_CategoryName:Common Text Object Reference)) -> (then, ReturnValue:Material Instance Dynamic Object Reference)
   > SetScalarParameterValue(self=@n1, ParameterName=Get FontHoverAnimateParam() -> (FontHoverAnimateParam:Name=None), Value=1)
  ]}
 ResetOnRowActiveInactiveAnimation {nodes [
   n1 = ResetOnRowActiveInactiveAnimation!() -> (then, IsRowActive:bool=~) {is=title K2Node_FunctionEntry}
   > n2 = Branch(Condition=@n1)
   n3 = Get InactiveDarken()
   > SetContentColorAndOpacity(self=@n3, InContentColorAndOpacity=Make Linear Color(R=1, G=1, B=1, A=1))
   n4 = Get InactiveOverlayDim()
   > SetRenderOpacity(self=@n4, InOpacity=0)
   n5 = Get Text_CategoryName()
   > n6 = GetDynamicFontMaterial(self=@n5)
   n7 = Get FontPressedAnimateParam() -> (FontPressedAnimateParam:Name=None)
   > SetScalarParameterValue(self=@n6, ParameterName=@n7, Value=1)
   SetContentColorAndOpacity(execute=@n2.else, self=@n3, InContentColorAndOpacity=Make Linear Color(R=0.625, G=0.875, B=1, A=1))
   > SetRenderOpacity(self=@n4, InOpacity=0.375)
   > n8 = GetDynamicFontMaterial(self=@n5)
   > SetScalarParameterValue(self=@n8, ParameterName=@n7, Value=0)
  ]}
 Set Font Category Size {nodes [
   Set Font Category Size() {is=title K2Node_FunctionEntry}
   > n1 = GetScalarParameterValue!(
    execute
    Collection:Material Parameter Collection Object Reference=/DiscoveryBrowser/DataAssets/DiscoverySharedValues
    ParameterName:Name=DiscoveryRowHeaderFontSize
   ) -> (then, ReturnValue:float) {is=title K2Node_CallMaterialParameterCollectionFunction}
   n2 = NOT Boolean(A=IsMobileGame())
   n3 = float * float(A=GetMobileFontSizeMultiplier!(self:Common Text Object Reference=Get Text_CategoryName()) -> (ReturnValue:float), B=@n1) {is=title K2Node_PromotableOperator}
   n4 = String To Integer(InString="To String (Float)"(InDouble=Select!(Option 0:double=@n3, Option 1:double=@n1, Index:bool=@n2) -> (ReturnValue:double)))
   n5 = Get Text_CategoryName()
   n6 = Get!(self:Text Object Reference=@n5) -> (
    Font_FontObject:Object Reference
    Font_FontMaterial:Object Reference
    Font_OutlineSettings:Font Outline Settings Structure
    Font_TypefaceFontName:Name
    Font_Size:float=~
    Font_LetterSpacing:int
    Font_SkewAmount:float=~
    Font_bForceMonospaced:bool=~
    Font_bMaterialIsStencil:bool=~
    Font_MonospacedWidth:float=~
   )
   > SetFont(
    self=@n5
    InFontInfo=MakeSlateFontInfo!(
     FontObject:Object Reference=@n6
     FontMaterial:Object Reference=@n6.Font_FontMaterial
     OutlineSettings:Font Outline Settings Structure=@n6.Font_OutlineSettings
     TypefaceFontName:Name=@n6.Font_TypefaceFontName
     Size:float="To Float (Integer)"(InInt=@n4)
     LetterSpacing:int=@n6.Font_LetterSpacing
     SkewAmount:float=@n6.Font_SkewAmount
     bForceMonospaced:bool=false
     bMaterialIsStencil:bool=false
     MonospacedWidth:float=1
    ) -> (ReturnValue:Slate Font Info Structure)
   )
  ]}
 SnapToEndOfAnimation {nodes [
   n1 = SnapToEndOfAnimation!() -> (then, InAnimation:Widget Animation Object Reference, PlayMode:EUMGSequencePlayMode Enum=Forward) {is=title K2Node_FunctionEntry}
   > PlayAnimation(InAnimation=@n1, StartAtTime=GetEndTime(self=@n1), PlayMode=@n1.PlayMode)
  ]}
 UpdateNewRow {nodes [
   UpdateNewRow!() -> (then, IsActive:bool=~, PlayAnimation:bool=~) {is=title K2Node_FunctionEntry}
   > Set IsWaitingForInactiveAnim(IsWaitingForInactiveAnim=false)
   > n1 = Sequence()
   > Check and Stop Animation(Animation=Get OnRowMoveUpOutOfViewAnim())
   > n2 = Branch(Condition=Get IsActive() -> (IsActive:bool))
   > Check and Stop Animation(Animation=Get OnPeek())
   > n3 = Branch(Condition=Get PlayAnimation() -> (PlayAnimation:bool))
   > SetRenderOpacity(self=Get VBContent(), InOpacity=0)
   > n4 = PlayAnimation(InAnimation=Get OnRowMoveDownIntoView())
   SetVisibility(execute=[@n4.then, @n6.then], InVisibility=SelfHitTestInvisible)
   > ResetOnRowActiveInactiveAnimation!(execute, self, IsRowActive:bool=true) -> (then) {is=title K2Node_CallFunction}
   > n5 = GetDynamicFontMaterial(self=Get Text_CategoryName())
   > SetScalarParameterValue(self=@n5, ParameterName=Get FontPressedAnimateParam() -> (FontPressedAnimateParam=None), Value=1)
   n6 = SetRenderOpacity(execute=@n3.else, self=Get VBContent(), InOpacity=1)
   Check and Stop Animation(execute=@n2.else, Animation=Get OnRowMoveDownIntoView())
   > SetRenderOpacity(self=Get VBContent(), InOpacity=1)
   > SetVisibility(InVisibility=HitTestInvisible)
   > n7 = Branch(Condition=GetIsInPeekState())
   > Check and Stop Animation(Animation=Get OnRowActiveInactiveAnim())
   > ResetOnPeekAnimation() {is=title K2Node_CallFunction}
   > n8 = GetDynamicFontMaterial(self=Get Text_CategoryName())
   > SetScalarParameterValue(self=@n8, ParameterName=Get FontPressedAnimateParam() -> (FontPressedAnimateParam=None), Value=0)
   Check and Stop Animation(execute=@n7.else, Animation=Get OnPeek())
   > ResetOnRowActiveInactiveAnimation(IsRowActive=false)
   n9 = For Each Loop(Exec=@n1.then_1, Array=Get ArrowButtons())
   > OnRowReset!(execute, self:Activity Browser Arrow Button Object Reference=@n9, IsActive:bool=Get IsActive()) -> (then)
  ]}
]
variables [
 ArrowButtons {type="array<ActivityBrowserArrowButton_C>", default []}
 FontHoverAnimateParam {type=string, default=Hover_Animate}
 FontPressedAnimateParam {type=string, default=Pressed_Animate}
 IsLoadingMoreQueryActive {type=boolean}
 IsLoadingMoreVisible {type=boolean}
 IsWaitingForInactiveAnim {type=boolean}
 LoadingMoreDelayTimer {type=TimerHandle}
 LoadingMoreDisplayDelay {type=number, default=0.2}
]
widgets {
 info {
  inheritedWidgetCount=4
  parentClass=/Script/DiscoveryBrowserUI.FortDiscoverItemBrowserRow
  rootWidgetClass=&Overlay
  widgetCount=17
 }
 widgets [
  Overlay_1 {widgetClassPath=&Overlay, widget=Overlay_1}
  VBContent {
   widgetClassPath=&VerticalBox
   bIsVariable=true
   parent=Overlay_1
   widget=VBContent
   slot {class=OverlaySlot, properties {horizontalAlignment=HAlign_Fill, verticalAlignment=VAlign_Fill}}
  }
  SZCategoryName {
   widgetClassPath=&SafeZone
   parent=VBContent
   widget=SZCategoryName
   properties {padBottom=false, padTop=false}
   slot {class=VerticalBoxSlot}
  }
  HorizontalBox_0 {widgetClassPath=&HorizontalBox, parent=SZCategoryName, widget=HorizontalBox_0, slot {class=SafeZoneSlot}}
  Text_CategoryName {
   widgetClassPath=&CommonTextBlock
   bInherited=true
   bIsVariable=true
   parent=HorizontalBox_0
   widget=Text_CategoryName
   properties {
    mobileFontSizeMultiplier=1.2
    scrollStyle=&TextScrollStyle-Base_C
    text="{Activity Category}"
    textTransformPolicy=ToUpper
    font {
     bForceMonospaced=false
     bMaterialIsStencil=false
     fontMaterial=null
     fontObject=&BurbankSmall
     letterSpacing=10
     monospacedWidth=1
     size=18
     skewAmount=0
     typefaceFontName=Bold
     outlineSettings {
      bApplyOutlineToDropShadows=false
      bMiteredCorners=false
      bSeparateFillAlpha=false
      outlineMaterial=null
      outlineSize=0
      outlineColor=<rgba 0 0 0 1>
     }
    }
    shadowOffset=<0 0>
   }
   slot {class=HorizontalBoxSlot}
  }
  LoadingMoreTilesHB {
   widgetClassPath=&HorizontalBox
   bIsVariable=true
   parent=HorizontalBox_0
   widget=LoadingMoreTilesHB
   properties {visibility=Collapsed}
   slot {class=HorizontalBoxSlot, properties {horizontalAlignment=HAlign_Right, verticalAlignment=VAlign_Bottom, size {sizeRule=Fill, value=1}}}
  }
  Spinner {
   widgetClassPath=&Image
   parent=LoadingMoreTilesHB
   widget=Spinner
   properties {
    visibility=HitTestInvisible
    brush {
     bIsDynamicallyLoaded=false
     drawAs=Image
     imageType=NoImage
     mirroring=NoMirror
     resourceName=None
     resourceObject=/FortUILibrary/UIKit/Throbber/MI_UIKit_Monochrome_Throbber
     tiling=NoTile
     imageSize=<32 32>
     margin=<ltrb 0 0 0 0>
     outlineSettings {
      bUseBrushTransparency=false
      roundingType=HalfHeightRadius
      width=0
      color {colorUseRule=UseColor_Specified, specifiedColor=<rgba 0 0 0 0>}
      cornerRadii=<0 0 0 0>
     }
     tintColor {colorUseRule=UseColor_Specified, specifiedColor=<rgba 1 1 1 1>}
     uVRegion {bIsValid=false, max=<0 0>, min=<0 0>}
    }
   }
   slot {class=HorizontalBoxSlot, properties {horizontalAlignment=HAlign_Center, verticalAlignment=VAlign_Center, padding=<ltrb 0 -8 0 -8>}}
  }
  Text_LoadingMore {
   widgetClassPath=&CommonTextBlock
   parent=LoadingMoreTilesHB
   widget=Text_LoadingMore
   properties {
    mobileFontSizeMultiplier=1.2
    text=Loading more
    textTransformPolicy=ToUpper
    colorAndOpacity {colorUseRule=UseColor_Specified, specifiedColor=<rgba 0 0.0908 1 1>}
    font {
     bForceMonospaced=false
     bMaterialIsStencil=false
     fontMaterial=null
     fontObject=&Burbank
     letterSpacing=0
     monospacedWidth=1
     size=18
     skewAmount=0
     typefaceFontName=Burbank-PlayerPower
     outlineSettings {
      bApplyOutlineToDropShadows=false
      bMiteredCorners=false
      bSeparateFillAlpha=false
      outlineMaterial=null
      outlineSize=0
      outlineColor=<rgba 0 0 0 1>
     }
    }
    margin=<ltrb 0 -6 0 -6>
    renderTransform {angle=0, scale=<1 1>, shear=<-8 0>, translation=<0 0>}
    shadowOffset=<0 0>
   }
   slot {class=HorizontalBoxSlot, properties {verticalAlignment=VAlign_Center, padding=<ltrb 8 0 0 0>, size {sizeRule=Fill, value=1}}}
  }
  Overlay_0 {widgetClassPath=&Overlay, parent=VBContent, widget=Overlay_0, slot {class=VerticalBoxSlot, properties {padding=<ltrb 0 16 0 48>}}}
  InactiveDarken {
   widgetClassPath=&CommonBorder
   bIsVariable=true
   parent=Overlay_0
   widget=InactiveDarken
   properties {padding=<ltrb 0 0 0 0>}
   slot {class=OverlaySlot, properties {horizontalAlignment=HAlign_Fill, verticalAlignment=VAlign_Fill}}
  }
  SZTiles {
   widgetClassPath=&SafeZone
   parent=InactiveDarken
   widget=SZTiles
   properties {padBottom=false, padTop=false}
   slot {class=BorderSlot, properties {padding=<ltrb 0 0 0 0>}}
  }
  ListView_Tiles {
   widgetClassPath=/Script/DiscoveryBrowserUI.FortDiscoverItemListView
   bInherited=true
   bIsVariable=true
   parent=SZTiles
   widget=ListView_Tiles
   properties {
    activityEntryWidgetClass=&/ActivityBrowserTile.ActivityBrowserTile_C
    bEnableRightClickScrolling=false
    bEnableScrollAnimation=true
    clipping=Inherit
    entrySpacing=24
    entryWidgetClass=/DiscoveryBrowser/Tiles/JoinableParty/WBP_ActivityBrowserTile_JoinableParty.WBP_ActivityBrowserTile_JoinableParty_C
    playWithFriendsEntryWidgetClass=&WBP_ActivityBrowserTile_JoinableParty_C
   }
   slot {class=SafeZoneSlot}
  }
  InactiveOverlayDim {
   widgetClassPath=&Image
   bIsVariable=true
   parent=Overlay_0
   widget=InactiveOverlayDim
   properties {
    renderOpacity=0
    visibility=HitTestInvisible
    brush {
     bIsDynamicallyLoaded=false
     drawAs=Image
     imageType=NoImage
     mirroring=NoMirror
     resourceName=None
     resourceObject=/DiscoveryBrowser/Art/MI_UI_Discovery_BackgroundMaster_TranslucentDim
     tiling=NoTile
     imageSize=<32 32>
     margin=<ltrb 0 0 0 0>
     outlineSettings {
      bUseBrushTransparency=false
      roundingType=HalfHeightRadius
      width=0
      color {colorUseRule=UseColor_Specified, specifiedColor=<rgba 0 0 0 0>}
      cornerRadii=<0 0 0 0>
     }
     tintColor {colorUseRule=UseColor_Specified, specifiedColor=<rgba 1 1 1 1>}
     uVRegion {bIsValid=false, max=<0 0>, min=<0 0>}
    }
   }
   slot {class=OverlaySlot, properties {horizontalAlignment=HAlign_Fill, verticalAlignment=VAlign_Fill, padding=<ltrb -128 -8 -128 -8>}}
  }
  SZButton_PageRight {
   widgetClassPath=&SafeZone
   parent=Overlay_0
   widget=SZButton_PageRight
   properties {padBottom=false, padLeft=false, padTop=false}
   slot {class=OverlaySlot, properties {horizontalAlignment=HAlign_Right, verticalAlignment=VAlign_Fill}}
  }
  Button_PageRight {
   widgetClassPath=&/ActivityBrowserArrowButton
   bInherited=true
   bIsVariable=true
   parent=SZButton_PageRight
   widget=Button_PageRight
   properties {
    minHeight=64
    minWidth=64
    onButtonBaseClicked=(&self:WidgetTree.Button_PageRight.HandleOnButtonClicked)
    onButtonBaseDoubleClicked=(&self:WidgetTree.Button_PageRight.HandleOnButtonDoubleClicked)
    onButtonBaseHovered=(&self:WidgetTree.Button_PageRight.HandleOnButtonHovered)
    onButtonBaseUnhovered=(&self:WidgetTree.Button_PageRight.HandleOnButtonUnhovered)
    onSelectedChangedBase=(&self:WidgetTree.Button_PageRight.HandleOnSelectedChanged)
   }
   slot {class=SafeZoneSlot, properties {hAlign=HAlign_Right, vAlign=VAlign_Center, padding=<ltrb 0 -64 -64 -64>}}
  }
  SZButton_PageLeft {
   widgetClassPath=&SafeZone
   parent=Overlay_0
   widget=SZButton_PageLeft
   properties {padBottom=false, padRight=false, padTop=false}
   slot {class=OverlaySlot, properties {verticalAlignment=VAlign_Fill}}
  }
  Button_PageLeft {
   widgetClassPath=&ActivityBrowserArrowButton
   bInherited=true
   bIsVariable=true
   parent=SZButton_PageLeft
   widget=Button_PageLeft
   properties {
    flip=true
    minHeight=64
    minWidth=64
    onButtonBaseClicked=(&self:WidgetTree.Button_PageLeft.HandleOnButtonClicked)
    onButtonBaseDoubleClicked=(&self:WidgetTree.Button_PageLeft.HandleOnButtonDoubleClicked)
    onButtonBaseHovered=(&self:WidgetTree.Button_PageLeft.HandleOnButtonHovered)
    onButtonBaseUnhovered=(&self:WidgetTree.Button_PageLeft.HandleOnButtonUnhovered)
    onSelectedChangedBase=(&self:WidgetTree.Button_PageLeft.HandleOnSelectedChanged)
    visibility=Collapsed
   }
   slot {class=SafeZoneSlot, properties {hAlign=HAlign_Left, vAlign=VAlign_Center, padding=<ltrb -64 -64 0 -64>}}
  }
 ]
}
