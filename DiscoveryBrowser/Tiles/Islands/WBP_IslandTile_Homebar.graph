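# Homebar variant of the minimal island tile in the Discovery browser, marking the current activity with a selection indicator.
# Construct sets download progress visibility and builds the indicator image in code inside a CommonVisualAttachment, enlarged on mobile.
# Anim_Indicator animates a value that slides and fades the indicator, which hides while tile selection is active.
# Parent overrides also add or remove a divider and restyle the live/countdown label by timer state.
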WidgetBlueprint /DiscoveryBrowser/Tiles/Islands/WBP_IslandTile_Homebar
animations [Anim_Indicator {
  binding_count=1
  length_seconds=0.2
  object=&self:Anim_Indicator
  sequence {
   bindings [WBP_IslandTile_Homebar_C_16 {
     id=#1
     possessed_class=WBP_IslandTile_Homebar_C
     tracks [{
       type=MovieSceneDoubleTrack
       display_name=Indicator Anim Value
       sections [{type=MovieSceneDoubleSection, end_frame=12, start_frame=0, channels [None {type=MovieSceneScriptingDoubleChannel, keys [[0,0],[12,1]]}]}]
      }]
    }]
   display_rate {numerator=60}
  }
 }]
graphs [
 EventGraph {nodes [
   n1 = Event PreConstruct()
   > Parent: PreConstruct(IsDesignTime=@n1.IsDesignTime)
   Event Construct()
   > Parent: Construct()
   > SetDownloadProgressVisibility!(
    execute
    self:WBP Island Tile Minimal Object Reference
    IsVisible:bool=ShouldShowHomebarDownloadProgress!(self:Fort Activity Tile Details Display Object Reference) -> (ReturnValue:bool)
   ) -> (then)
   n2 = Construct Object from Class(Class=&CommonVisualAttachment) -> (ReturnValue:Common Visual Attachment Object Reference)
   > n3 = Set SelectionIndicator(SelectionIndicator:Common Visual Attachment Object Reference=@n2) -> (Output_Get:Common Visual Attachment Object Reference)
   > Set ContentAnchor!(execute, ContentAnchor:Vector 2D Structure=(X=0.5,Y=1), self:Common Visual Attachment Object Reference=@n3) -> (then, Output_Get:Vector 2D Structure)
   > n4 = Construct Object from Class(Class=&Image) -> (ReturnValue:Image Object Reference)
   > n5 = Set IndicatorImage(IndicatorImage:Image Object Reference=@n4) -> (Output_Get:Image Object Reference)
   > SetBrush!(
    execute
    self=@n5
    InBrush_TintColor:Slate Color Structure=(SpecifiedColor=(R=1,G=1,B=1,A=1),ColorUseRule=UseColor_Specified)
    InBrush_DrawAs:ESlateBrushDrawType Enum=Image
    InBrush_Tiling:ESlateBrushTileType Enum=NoTile
    InBrush_Mirroring:ESlateBrushMirrorType Enum=NoMirror
    InBrush_ImageSize:Vector 2D Structure=(X=32,Y=32)
    InBrush_Margin:Margin Structure=(Left=0,Top=0,Right=0,Bottom=0)
    InBrush_ResourceObject:Object Reference=/DiscoveryBrowser/Tiles/Islands/Materials/MI_UI_IslandTile_State_SelectedGame
    InBrush_OutlineSettings:Slate Brush Outline Settings Structure=(CornerRadii=(X=0,Y=0,Z=0,W=0),Color=(SpecifiedColor=(R=0,G=0,B=0,A=0),ColorUseRule=UseColor_Specified),Width=0,RoundingType=HalfHeightRadius,bUseBrushTransparency=False)
   ) -> (then)
   > n6 = AddChild(
    self=Get GridRoot_9Grid() -> (GridRoot_9Grid:Grid Panel Object Reference)
    Content=Get SelectionIndicator() -> (SelectionIndicator:Common Visual Attachment Object Reference)
   )
   > n7 = Cast To GridSlot!(execute, Object=@n6) -> (then, CastFailed, AsGrid Slot:Grid Slot Object Reference) {is=title K2Node_DynamicCast}
   > Set HorizontalAlignment!(execute, HorizontalAlignment:EHorizontalAlignment Enum=HAlign_Center, self:Grid Slot Object Reference=@n7) -> (then, Output_Get:EHorizontalAlignment Enum=HAlign_Fill)
   > SetGridSlotColumnAndRow!(execute, Widget:Widget Object Reference=Get SelectionIndicator(), Row:int=1, Row Span:int=1, Column:int=1, Column Span:int=1) -> (then)
   > Set Layer!(execute, Layer:int=-10, self:Grid Slot Object Reference=@n7) -> (then, Output_Get:int=0)
   > AddChild(self=Get SelectionIndicator(), Content=Get IndicatorImage() -> (IndicatorImage:Image Object Reference))
   > SetMobileStyle() {is=title K2Node_CallFunction}
  ]}
 OnTileSelectionActive {nodes [
   OnTileSelectionActive() {is=title K2Node_FunctionEntry}
   > Parent: OnTileSelectionActive() {is=title K2Node_CallParentFunction}
   > n1 = Branch(Condition=Get IsCurrentActivity() -> (IsCurrentActivity:bool))
   > ToggleSelectIndicator!(execute, self, Should Show Indicator:bool=false) -> (then)
   SetIndicatorAnimValue!(execute=@n1.else, self, NewParam:double=0) -> (then)
  ]}
 OnTileSelectionNotActive {nodes [
   OnTileSelectionNotActive() {is=title K2Node_FunctionEntry}
   > Parent: OnTileSelectionNotActive() {is=title K2Node_CallParentFunction}
   > n1 = Branch(Condition=Get IsCurrentActivity())
   > ToggleSelectIndicator(Should Show Indicator=Select(Option 0:bool=false, Option 1:bool=true, Index=Get IsCurrentActivity()) -> (ReturnValue:bool))
   SetIndicatorAnimValue(execute=@n1.else, NewParam=0)
  ]}
 SetIndicatorAnimValue {nodes [
   SetIndicatorAnimValue!() -> (then, NewParam:double=~) {is=title K2Node_FunctionEntry}
   n1 = Get NewParam() -> (NewParam:double)
   > SetRenderTranslation!(execute, self=Get IndicatorImage(), Translation_X:double=0, Translation_Y:double=Lerp(A=48, B=0, Alpha=@n1)) -> (then)
   > Set RenderOpacity!(execute, RenderOpacity:float=@n1, self:Widget Object Reference=Get IndicatorImage()) -> (then, Output_Get:float)
  ]}
 SetLiveLabelTextAndType {nodes [
   n1 = SetLiveLabelTextAndType!() -> (then, TimerState:ECountdownTimerState Enum=None) {is=title K2Node_FunctionEntry}
   > Parent: SetLiveLabelTextAndType!(execute, TimerState:ECountdownTimerState Enum=@n1) -> (then) {is=title K2Node_CallParentFunction}
   n2 = Get CurrentLabelText() -> (CurrentLabelText:Text)
   n3 = Get Highlight_Label() -> (Highlight_Label:WBP UIKit Highlight Object Reference)
   > SetText!(
    execute
    self:WBP UIKit Highlight Base Object Reference=@n3
    Text:Text=Select!(None:Text=@n2, Countdown:Text=@n2, Live:Text="LOCTEXT(\"Live\")", Countup:Text, Index:ECountdownTimerState Enum=@n1) -> (ReturnValue:Text)
   ) -> (then)
   n4 = Byte to Enum E_UI_HighlightType!(Byte:Byte=MakeLiteralByte!(Value:Byte=Get CurrentLabelType() -> (CurrentLabelType:EFortDiscoverLabelStyle Enum=Utility)) -> (ReturnValue:Byte=0)) -> (ReturnValue:E_UI_HighlightType Enum) {is=title K2Node_CastByteToEnum}
   > SetType!(
    execute
    self:WBP UIKit Highlight Base Object Reference=@n3
    Type:E_UI_HighlightType Enum=Select(
     None:E_UI_HighlightType Enum=@n4
     Countdown:E_UI_HighlightType Enum=@n4
     Live:E_UI_HighlightType Enum=NewEnumerator6
     Countup:E_UI_HighlightType Enum=NewEnumerator3
     Index=@n1
    ) -> (ReturnValue:E_UI_HighlightType Enum)
   ) -> (then)
  ]}
 SetMobileStyle {nodes [
   SetMobileStyle() {is=title K2Node_FunctionEntry}
   > Branch(Condition=IsMobileGame())
   n1 = Get IndicatorImage()
   n2 = Get!(self:Image Object Reference=@n1) -> (
    Brush_TintColor:Slate Color Structure
    Brush_DrawAs:ESlateBrushDrawType Enum
    Brush_Tiling:ESlateBrushTileType Enum
    Brush_Mirroring:ESlateBrushMirrorType Enum
    Brush_ImageSize:Vector 2D Structure
    Brush_Margin:Margin Structure
    Brush_ResourceObject:Object Reference
    Brush_OutlineSettings:Slate Brush Outline Settings Structure
   )
   > SetBrush!(
    execute
    self=@n1
    InBrush_TintColor=@n2
    InBrush_DrawAs=@n2.Brush_DrawAs
    InBrush_Tiling=@n2.Brush_Tiling
    InBrush_Mirroring=@n2.Brush_Mirroring
    InBrush_ImageSize_X:double=64
    InBrush_ImageSize_Y:double=64
    InBrush_Margin=@n2.Brush_Margin
    InBrush_ResourceObject=@n2.Brush_ResourceObject
    InBrush_OutlineSettings=@n2.Brush_OutlineSettings
   ) -> (then)
   > SetClipping!(execute, self:Widget Object Reference=@n1, InClipping:EWidgetClipping Enum=ClipToBoundsWithoutIntersecting) -> (then)
  ]}
 ToggleSelect {nodes [
   n1 = ToggleSelect!() -> (then, IsCurrentActivity:bool) {is=title K2Node_FunctionEntry}
   > Parent: ToggleSelect!(execute, IsCurrentActivity:bool=@n1) -> (then) {is=title K2Node_CallParentFunction}
   > Set IsCurrentActivity(IsCurrentActivity:bool=@n1) -> (Output_Get:bool)
   > SetIndicatorAnimValue(NewParam="To Float (Boolean)"(InBool=Get IsCurrentActivity())) {is=title K2Node_CallFunction}
  ]}
 ToggleSelectIndicator {nodes [
   n1 = ToggleSelectIndicator!() -> (then, Should Show Indicator:bool=~) {is=title K2Node_FunctionEntry}
   > n2 = Branch(Condition=@n1)
   n3 = Get Anim_Indicator() -> (Anim_Indicator:Widget Animation Object Reference)
   > PlayAnimationForward(InAnimation=@n3)
   PlayAnimationReverse(execute=@n2.else, InAnimation=@n3)
  ]}
 ToggleSpacer {nodes [
   n1 = ToggleSpacer!() -> (then, ShouldShowSpacer:bool) {is=title K2Node_FunctionEntry}
   > Parent: ToggleSpacer!(execute, ShouldShowSpacer:bool=@n1) -> (then) {is=title K2Node_CallParentFunction}
   > n2 = Branch(Condition=@n1)
   > n3 = Construct Object from Class(Class=&/WBP_UIKit_Divider_Homebar.WBP_UIKit_Divider_Homebar_C) -> (ReturnValue:WBP UIKit Divider Homebar Object Reference)
   > n4 = Set Divider(Divider:WBP UIKit Divider Homebar Object Reference=@n3) -> (Output_Get:WBP UIKit Divider Homebar Object Reference)
   > SetPadding!(
    execute
    self:User Widget Object Reference=@n4
    InPadding_Left:float=24
    InPadding_Top:float=0
    InPadding_Right:float=0
    InPadding_Bottom:float=0
   ) -> (then)
   n5 = Get Divider() -> (Divider:WBP UIKit Divider Homebar Object Reference)
   n6 = Get GridRoot_9Grid()
   > AddChildToGrid!(execute, self:Grid Panel Object Reference=@n6, Content:Widget Object Reference=@n5, InRow:int=1, InColumn:int=2) -> (then, ReturnValue:Grid Slot Object Reference)
   RemoveChild(execute=@n2.else, self=@n6, Content=@n5)
  ]}
]
variables [
 Divider {type=WBP_UIKit_Divider_Homebar_C, default=null}
 IndicatorAnimValue {type=number}
 IndicatorImage {type=Image, default=null}
 IsCurrentActivity {type=boolean}
 IsTileHovered {type=boolean}
 SelectionIndicator {type=CommonVisualAttachment, default=null}
]
widgets {
 info {
  inheritedWidgetCount=0
  parentClass=&/WBP_IslandTile_Minimal.WBP_IslandTile_Minimal_C
  rootWidgetClass=null
  widgetCount=0
 }
 widgets []
}
