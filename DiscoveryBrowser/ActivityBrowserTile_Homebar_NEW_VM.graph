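# Homebar activity browser tile for the Discovery browser, a FortActivityBrowserTile that wraps an island tile widget (WBP_IslandTile_Homebar) in an overlay.
# It forwards the nested button group selection view model get and set interface calls to that inner tile.
# When the list entry is released, it casts the inner tile to WBP_IslandTile_Minimal to clean up its context menu.
# BroadcastClickToRow has an empty body.
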WidgetBlueprint /DiscoveryBrowser/ActivityBrowserTile_Homebar_NEW_VM
graphs [
 BroadcastClickToRow {nodes [BroadcastClickToRow() {is=title K2Node_FunctionEntry}]}
 EventGraph {nodes [
   Event On Entry Released() {is=title K2Node_Event}
   > n1 = Cast To WBP_IslandTile_Minimal!(execute, Object=Get Display_TileDetails() -> (Display_TileDetails:Fort Activity Tile Details Display Object Reference)) -> (then, CastFailed, AsWBP Island Tile Minimal:WBP Island Tile Minimal Object Reference) {is=title K2Node_DynamicCast}
   > WrapperCleanupContextMenuOnRelease(self:WBP Island Tile Minimal Object Reference=@n1)
   n2 = Event SetNestedButtonGroupSelectionVM!() -> (OutputDelegate, then, InVM:Fort Nested Button Group Selection VM Object Reference) {is=title K2Node_Event}
   > SetNestedButtonGroupSelectionVM!(
    execute
    self:Fort Nested Button Group VMItem Interface=Get Display_TileDetails()
    InVM:Fort Nested Button Group Selection VM Object Reference=@n2.InVM
   ) -> (then)
  ]}
 GetNestedButtonGroupSelectionVM {nodes [
   GetNestedButtonGroupSelectionVM() {is=title K2Node_FunctionEntry}
   > Return Node!(
    execute
    ReturnValue:Fort Nested Button Group Selection VM Object Reference=GetNestedButtonGroupSelectionVM!(self:Fort Nested Button Group VMItem Interface=Get Display_TileDetails()) -> (ReturnValue:Fort Nested Button Group Selection VM Object Reference) {is=title K2Node_CallFunction}
   )
  ]}
]
variables [BroadcastClickToRow {}]
widgets {
 info {inheritedWidgetCount=2, parentClass=/Script/DiscoveryBrowserUI.FortActivityBrowserTile, rootWidgetClass=&Overlay, widgetCount=3}
 widgets [
  Overlay_0 {widgetClassPath=&Overlay, widget=Overlay_0}
  Display_TileDetails {
   widgetClassPath=/DiscoveryBrowser/Tiles/Islands/WBP_IslandTile_Homebar
   bInherited=true
   bIsVariable=true
   parent=Overlay_0
   widget=Display_TileDetails
   properties {touchMethod=PreciseTap}
   slot {class=OverlaySlot}
  }
  InputActionWidget {widgetClassPath=&CommonActionWidget, bInherited=true, widget=null}
 ]
}
