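# Category tile widget for the Discovery browser, based on FortActivityCategoryTile, showing a centered category title over a material-driven image.
# The TileHover animation raises the tile material's Hover_Animate parameter and fades the title color from white to black.
# OnTileActiveSet plays it forward or in reverse, and On Entry Released plays it reversed with an end-time offset to reset the tile.
# Construct sets NoImage_Animate to 1 on the image's dynamic material.
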WidgetBlueprint /DiscoveryBrowser/ActivityBrowserCategoryTile
animations [TileHover {
  binding_count=2
  length_seconds=0.16
  object=&self:TileHover
  sequence {bindings [
    Image_Tile {
     id=#1
     possessed_class=Image
     tracks [{
       type=MovieSceneWidgetMaterialTrack
       sections [{type=MovieSceneParameterSection, channels [Hover_Animate {type=MovieSceneScriptingFloatChannel, keys [[0,0],[3.2,1]]}]}]
      }]
    }
    Text_CategoryTitle {
     id=#2
     possessed_class=CommonTextBlock
     tracks [{
       type=MovieSceneColorTrack
       display_name=Color and Opacity
       sections [{
         type=MovieSceneColorSection
         channels [
          Color.R {type=MovieSceneScriptingFloatChannel, keys [[0,1],[1.6,0]]}
          Color.G {type=MovieSceneScriptingFloatChannel, keys [[0,1],[1.6,0]]}
          Color.B {type=MovieSceneScriptingFloatChannel, keys [[0,1],[1.6,0]]}
          Color.A {type=MovieSceneScriptingFloatChannel, const [1,1]}
         ]
        }]
      }]
    }
   ]}
 }]
graphs [EventGraph {nodes [
   n1 = Event OnTileActiveSet!() -> (OutputDelegate, then, bIsTileActive:bool) {is=title K2Node_Event}
   > n2 = Branch(Condition=@n1.bIsTileActive)
   n3 = Get TileHover() -> (TileHover:Widget Animation Object Reference)
   > PlayAnimation(InAnimation=@n3)
   PlayAnimation(execute=@n2.else, InAnimation=@n3, PlayMode=Reverse)
   Event Construct()
   > Parent: Construct()
   > n4 = GetDynamicMaterial(self=Get Image_Tile() -> (Image_Tile:Image Object Reference))
   > SetScalarParameterValue(self=@n4, ParameterName=NoImage_Animate, Value=1)
   Event On Entry Released() {is=title K2Node_Event}
   n5 = Get TileHover()
   > PlayAnimation(InAnimation=@n5, StartAtTime=GetEndTime(self=@n5), PlayMode=Reverse)
  ]}]
variables []
widgets {
 info {inheritedWidgetCount=2, parentClass=/Script/DiscoveryBrowserUI.FortActivityCategoryTile, rootWidgetClass=&SizeBox, widgetCount=5}
 widgets [
  SizeBox_22 {
   widgetClassPath=&SizeBox
   widget=SizeBox_22
   properties {bOverride_HeightOverride=true, bOverride_WidthOverride=true, heightOverride=216, widthOverride=384}
  }
  Overlay_1 {widgetClassPath=&Overlay, parent=SizeBox_22, widget=Overlay_1, slot {class=SizeBoxSlot}}
  Image_Tile {
   widgetClassPath=&Image
   bIsVariable=true
   parent=Overlay_1
   widget=Image_Tile
   properties {brush {
     bIsDynamicallyLoaded=false
     drawAs=Image
     imageType=NoImage
     mirroring=NoMirror
     resourceName=None
     resourceObject=/DiscoveryBrowser/Art/MI_UI_Discovery_CategoryTile
     tiling=NoTile
     imageSize=<128 128>
     margin=<ltrb 0.5 0.5 0.5 0.5>
     outlineSettings {
      bUseBrushTransparency=false
      roundingType=HalfHeightRadius
      width=0
      color {colorUseRule=UseColor_Specified, specifiedColor=<rgba 0 0 0 0>}
      cornerRadii=<0 0 0 0>
     }
     tintColor {colorUseRule=UseColor_Specified, specifiedColor=<rgba 1 1 1 1>}
     uVRegion {bIsValid=false, max=<0 0>, min=<0 0>}
    }}
   slot {class=OverlaySlot, properties {horizontalAlignment=HAlign_Fill, verticalAlignment=VAlign_Fill}}
  }
  Text_CategoryTitle {
   widgetClassPath=&CommonTextBlock
   bInherited=true
   parent=Overlay_1
   widget=Text_CategoryTitle
   properties {
    autoWrapText=true
    justification=Center
    text="{Category}"
    font {
     bForceMonospaced=false
     bMaterialIsStencil=false
     fontMaterial=null
     fontObject=&BurbankSmall
     letterSpacing=0
     monospacedWidth=1
     size=18
     skewAmount=0
     typefaceFontName=Bold
     outlineSettings {
      bApplyOutlineToDropShadows=false
      bMiteredCorners=false
      bSeparateFillAlpha=false
      outlineMaterial=null
      outlineSize=0
      outlineColor=<rgba 0 0 0 1>
     }
    }
    shadowOffset=<0 0>
   }
   slot {class=OverlaySlot, properties {horizontalAlignment=HAlign_Fill, verticalAlignment=VAlign_Center, padding=<ltrb 7 7 7 7>}}
  }
  InputActionWidget {widgetClassPath=&CommonActionWidget, bInherited=true, widget=null}
 ]
}
