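# Base Discovery browser panel widget that shows an uppercase category title above a tile view of activity category tiles.
# EntryWidth and EntryHeight variables size tiles on initialization and design-time PreConstruct, which also collapses the scrollbar.
# SetActive plays a title font-material animation forward or backward via Pressed_Animate, which Construct resets.
# Move events fade content out and back with one animation; another event toggles the tile view's mouse-wheel input.
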WidgetBlueprint /DiscoveryBrowser/ActivityCategoryTilePanelBase
animations [
 ActiveInactivate {
  binding_count=1
  length_seconds=0.16
  object=&self:ActiveInactivate
  sequence {bindings [Text_Title {
     id=#1
     possessed_class=CommonTextBlock
     tracks [{
       type=MovieSceneWidgetMaterialTrack
       display_name=Font.FontMaterial
       sections [{type=MovieSceneParameterSection, channels [Pressed_Animate {type=MovieSceneScriptingFloatChannel, keys [[0,0],[3.2,1]]}]}]
      }]
    }]}
 }
 OnMoveUpOutOfView {
  binding_count=1
  length_seconds=0.25
  object=&self:OnMoveUpOutOfView
  sequence {bindings [VerticalBox_Content {
     id=#2
     possessed_class=VerticalBox
     tracks [{
       type=MovieSceneFloatTrack
       display_name=Render Opacity
       sections [{type=MovieSceneFloatSection, channels [None {type=MovieSceneScriptingFloatChannel, keys [[0,1],[5,0]]}]}]
      }]
    }]}
 }
]
graphs [EventGraph {nodes [
   MoveDownIntoView() {is=title K2Node_CustomEvent}
   n1 = Get OnMoveUpOutOfView() -> (OnMoveUpOutOfView:Widget Animation Object Reference)
   > PlayAnimationReverse(InAnimation=@n1)
   MoveUpOutOfView() {is=title K2Node_CustomEvent}
   > PlayAnimationForward(InAnimation=@n1)
   n2 = EnableMouseWheelInput!() -> (OutputDelegate, then, bListenForInput:bool=~) {is=title K2Node_CustomEvent}
   > SetListenForMouseWheelInput!(
    execute
    self:Fort Activity Tile View Object Reference=Get TileView_Categories() -> (TileView_Categories:Fort Activity Tile View Object Reference)
    bListenForInput:bool=@n2.bListenForInput
   ) -> (then)
   n3 = SetActive!() -> (OutputDelegate, then, IsActive:bool=~) {is=title K2Node_CustomEvent}
   > n4 = Branch(Condition=@n3.IsActive)
   n5 = Get ActiveInactivate() -> (ActiveInactivate:Widget Animation Object Reference)
   > PlayAnimationForward(InAnimation=@n5)
   PlayAnimationReverse(execute=@n4.else, InAnimation=@n5)
   n6 = Event PreConstruct()
   n7 = Get TileView_Categories()
   > SetScrollbarVisibility!(execute, self:List View Base Object Reference=@n7, InVisibility:ESlateVisibility Enum=Collapsed) -> (then)
   > n8 = Branch(Condition=@n6.IsDesignTime)
   SetEntryHeight!(execute=[@n10.then, @n8.then], self:Tile View Object Reference=@n7, NewHeight:float=Get EntryHeight() -> (EntryHeight:double)) -> (then)
   > SetEntryWidth!(execute, self:Tile View Object Reference=@n7, NewWidth:float=Get EntryWidth() -> (EntryWidth:double)) -> (then)
   Event Construct()
   > Parent: Construct()
   > SetRenderOpacity(self=Get VerticalBox_Content() -> (VerticalBox_Content:Vertical Box Object Reference), InOpacity=1)
   > n9 = GetDynamicFontMaterial!(execute, self:Text Object Reference=Get Text_Title() -> (Text_Title:Common Text Object Reference)) -> (then, ReturnValue:Material Instance Dynamic Object Reference)
   > SetScalarParameterValue(self=@n9, ParameterName=Pressed_Animate, Value=0)
   n10 = Event OnInitialized()
  ]}]
variables [
 EntryHeight {type=number, default=159}
 EntryWidth {type=number, default=408}
]
widgets {
 info {
  inheritedWidgetCount=2
  parentClass=/Script/DiscoveryBrowserUI.FortActivityCategoryTilePanel
  rootWidgetClass=&VerticalBox
  widgetCount=3
 }
 widgets [
  VerticalBox_Content {widgetClassPath=&VerticalBox, bIsVariable=true, widget=VerticalBox_Content}
  Text_Title {
   widgetClassPath=&CommonTextBlock
   bInherited=true
   parent=VerticalBox_Content
   widget=Text_Title
   properties {
    text="{PanelName}"
    textTransformPolicy=ToUpper
    font {
     bForceMonospaced=false
     bMaterialIsStencil=false
     fontMaterial=/DiscoveryBrowser/Art/MI_UI_Discovery_CategoryHeaderText
     fontObject=&BurbankSmall
     letterSpacing=0
     monospacedWidth=1
     size=18
     skewAmount=0
     typefaceFontName=Bold
     outlineSettings {
      bApplyOutlineToDropShadows=false
      bMiteredCorners=false
      bSeparateFillAlpha=false
      outlineMaterial=null
      outlineSize=0
      outlineColor=<rgba 0 0 0 1>
     }
    }
    shadowOffset=<0 0>
   }
   slot {class=VerticalBoxSlot, properties {padding=<ltrb 10 0 0 0>}}
  }
  TileView_Categories {
   widgetClassPath=/Script/DiscoveryBrowserUI.FortActivityTileView
   bInherited=true
   bIsVariable=true
   parent=VerticalBox_Content
   widget=TileView_Categories
   properties {
    bEnableFixedLineOffset=true
    bEnableRightClickScrolling=false
    entryHeight=159
    entryWidgetClass=&/ActivityBrowserCategoryTile.ActivityBrowserCategoryTile_C
    entryWidth=407
    horizontalEntrySpacing=24
    numDesignerPreviewEntries=8
    tileAlignment=LeftAligned
    verticalEntrySpacing=24
   }
   slot {class=VerticalBoxSlot, properties {size {sizeRule=Fill, value=1}}}
  }
 ]
}
