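; Blueprint function library for Discover browser analytics that decides which tiles and rows are on screen and starts or stops their visibility impressions.
; ComputeItemVisibility tests a widget's absolute rect against given bounds, allowing overhang set by runtime-option minimum visible percentages.
; A runtime option instead uses visible percentages calculated during paint on minimal island tiles.
; ComputeRowVisibility updates each displayed list entry's tile view model and counts the row visible when any tile is.
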(fn ComputeItemVisibility
  (params (Widget :type Widget%20Object%20Reference) (TopLeft :type Vector%202D%20Structure) (BottomRight :type Vector%202D%20Structure))
  (bind _returnvalue (Options|GetRuntimeOptions))
  (bind _returnvalue_2 (Class|FortRuntimeOptions|GetDiscoverAnalyticMinimumHorizontalPercentageforVisibility _returnvalue))
  (bind (_x _y) (Math|Vector2D|BreakVector2D TopLeft))
  (bind (_x_2 _y_2) (Math|Vector2D|BreakVector2D BottomRight))
  (bind _returnvalue_3 (Class|FortRuntimeOptions|GetDiscoverAnalyticMinimumVerticalPercentageforVisibility _returnvalue))
  (Utilities|FlowControl|Sequence
    (:then_2
      (if (Class|FortRuntimeOptions|GetUseVisibilityCalculationsMadeDuringPaintforAnalytics _returnvalue)
        (bind _aswbp_island_tile_minimal (Utilities|Casting|CastToWBP_IslandTile_Minimal Widget)
          (:then
            (return (and (>= (Variables|/Script/DiscoveryBrowserUI.FortActivityTileDetailsDisplay|GetVisiblePercentage_Height _aswbp_island_tile_minimal) _returnvalue_3) (>= (Variables|/Script/DiscoveryBrowserUI.FortActivityTileDetailsDisplay|GetVisiblePercentage_Width _aswbp_island_tile_minimal) _returnvalue_2))))
          (:CastFailed
            _))
        (else
          _))
      (bind (_topleft _bottomright _size) (Default|GetWidgetAbsoluteRect Widget))
      (return (Math|Boolean|ANDBoolean (>= (Math|Vector2D|BreakVector2D _topleft) (- _x (* (Math|Vector2D|BreakVector2D _size) (- 1.0 (/ _returnvalue_2 100))))) (>= (.Y (Math|Vector2D|BreakVector2D _topleft)) (- _y (* (.Y (Math|Vector2D|BreakVector2D _size)) (- 1.0 (/ _returnvalue_3 100))))) (<= (Math|Vector2D|BreakVector2D _bottomright) (+ (* (Math|Vector2D|BreakVector2D _size) (- 1.0 (/ _returnvalue_2 100))) _x_2)) (<= (.Y (Math|Vector2D|BreakVector2D _bottomright)) (+ (* (.Y (Math|Vector2D|BreakVector2D _size)) (- 1.0 (/ _returnvalue_3 100))) _y_2)))))))

(fn ComputeRowVisibility
  (params (ListView :type List%20View%20Object%20Reference) (Provider%20VM :type Discover%20Provider%20View%20Model%20Object%20Reference) (TopLeft :type Vector%202D%20Structure) (BottomRight :type Vector%202D%20Structure))
  (for (_array_element _array_index) (ListViewBase|GetDisplayedEntryWidgets ListView)
    (bind _visible (Default|ComputeItemVisibility _array_element TopLeft BottomRight))
    (bind _asdiscover_tile_item_view_model (Utilities|Casting|CastToFortDiscoverTileItemVM (Utilities|Array|Get(acopy) (ListView|GetListItems ListView) _array_index)))
    (if _visible
      (Class|DiscoverTileItemViewModel|StartVisibilityImpression _asdiscover_tile_item_view_model)
      ; When applying visibility to individual tiles, if at least one tile is visible, the row is considered as visible too.
      (LocalVariables|SetIsRowVisible true)
      (else
        (Class|DiscoverTileItemViewModel|StopVisibilityImpression _asdiscover_tile_item_view_model))))
  (if (LocalVariables|GetIsRowVisible)
    (Class|DiscoverProviderViewModel|StartVisibleImpressions Provider%20VM)
    (else
      (Class|DiscoverProviderViewModel|StopVisibleImpression Provider%20VM))))

(fn GetWidgetAbsoluteRect
  (params (Widget :type Widget%20Object%20Reference))
  (bind (_x _y) (Math|Vector2D|BreakVector2D (UserInterface|Geometry|LocaltoAbsolute(Coordinates) (Widget|GetCachedGeometry Widget))))
  (bind (_x_2 _y_2) (Math|Vector2D|BreakVector2D (UserInterface|Geometry|GetAbsoluteSize (Widget|GetCachedGeometry Widget))))
  ; For debugging purposes - might get removed at some point
  (return (UserInterface|Geometry|LocaltoAbsolute(Coordinates) (Widget|GetCachedGeometry Widget)) (Math|Vector2D|MakeVector2D (+ _x _x_2) (+ _y _y_2)) (UserInterface|Geometry|GetAbsoluteSize (Widget|GetCachedGeometry Widget)) (Utilities|String|Append (Utilities|GetDisplayName Widget) " X1/Y1 : " (Utilities|String|ToString(Vector2d) (UserInterface|Geometry|LocaltoAbsolute(Coordinates) (Widget|GetCachedGeometry Widget))) " X2/Y2 : " (Utilities|String|ToString(Vector2d) (Math|Vector2D|MakeVector2D (+ _x _x_2) (+ _y _y_2))) " Size : " (Utilities|String|ToString(Vector2d) (UserInterface|Geometry|GetAbsoluteSize (Widget|GetCachedGeometry Widget))))))
