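; Function set for the Discovery browser UI that picks themed background, header, top-edge and footer materials for an activity's link category (LEGO, Fall Guys, TMNT, Epic Games and others).
; Each getter is a Select on the category enum, with creator-page or nebulous default materials for None.
; Top edge and footer share each brand's FadeMask instance.
; Only the footer wraps it in a dynamic instance, setting Fade Rotation to zero.
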(fn GetBackgroundMaterial
  (params (LinkCategory :type EFortActivityLinkCategory%20Enum))
  (return (Utilities|Select :Index LinkCategory :None "/DiscoveryBrowser/CreatorProfile/MaterialInstances/MI_CreatorPage_Bg.MI_CreatorPage_Bg" :LEGO "/DiscoveryBrowser/Art/MaterialInstances/MI_UI_Discovery_Background_Lego_Background_B.MI_UI_Discovery_Background_Lego_Background_B" :RocketRacing "/DiscoveryBrowser/Art/MaterialInstances/MI_UI_Discovery_Background_RocketRacing_Background.MI_UI_Discovery_Background_RocketRacing_Background" :FallGuys "/DiscoveryBrowser/Art/MaterialInstances/MI_UI_Discovery_Background_FallGuys_Background.MI_UI_Discovery_Background_FallGuys_Background" :TMNT "/DiscoveryBrowser/Art/MaterialInstances/MI_UI_Discovery_Background_NinjaTurtles_Background.MI_UI_Discovery_Background_NinjaTurtles_Background" :TheWalkingDeadUniverse "/DiscoveryBrowser/Art/MaterialInstances/MI_UI_Discovery_Background_DarkVortex_Background.MI_UI_Discovery_Background_DarkVortex_Background" :SolidWave "/DiscoveryBrowser/Art/MaterialInstances/SolidWave/MI_UI_Discovery_Background_SolidWave_Background.MI_UI_Discovery_Background_SolidWave_Background" :Akita "/DiscoveryBrowser/Art/MaterialInstances/MI_UI_Discovery_Background_Akita_Background.MI_UI_Discovery_Background_Akita_Background" :KPOPDemonHunters "/DiscoveryBrowser/Art/MaterialInstances/MI_UI_Discovery_Background_Refrigerator_Background.MI_UI_Discovery_Background_Refrigerator_Background" :EpicGames "/DiscoveryBrowser/Art/MaterialInstances/EpicGames/MI_UI_Discovery_IP_EpicGames_Background.MI_UI_Discovery_IP_EpicGames_Background" :Creatures "/DiscoveryBrowser/Art/MaterialInstances/MI_UI_Discovery_Background_Creatures_Background.MI_UI_Discovery_Background_Creatures_Background")))

(fn GetFooterMaterial
  (params (LinkCategory :type EFortActivityLinkCategory%20Enum))
  (bind _returnvalue (Rendering|Material|CreateDynamicMaterialInstance 0 (Utilities|Select :Index LinkCategory :None "/DiscoveryBrowser/TabScreens/DiscoverHome/Materials/MI_UI_Discovery_NebulousBackgroundWithBottomMask.MI_UI_Discovery_NebulousBackgroundWithBottomMask" :LEGO "/DiscoveryBrowser/Art/MaterialInstances/MI_UI_Discovery_Background_Lego_FadeMask_B.MI_UI_Discovery_Background_Lego_FadeMask_B" :RocketRacing "/DiscoveryBrowser/Art/MaterialInstances/MI_UI_Discovery_Background_RocketRacing_FadeMask.MI_UI_Discovery_Background_RocketRacing_FadeMask" :FallGuys "/DiscoveryBrowser/Art/MaterialInstances/MI_UI_Discovery_Background_FallGuys_Background_FadeMask.MI_UI_Discovery_Background_FallGuys_Background_FadeMask" :TMNT "/DiscoveryBrowser/Art/MaterialInstances/MI_UI_Discovery_Background_NinjaTurtles_FadeMask.MI_UI_Discovery_Background_NinjaTurtles_FadeMask" :TheWalkingDeadUniverse "/DiscoveryBrowser/Art/MaterialInstances/MI_UI_Discovery_Background_DarkVortex_FadeMask.MI_UI_Discovery_Background_DarkVortex_FadeMask" :SolidWave "/DiscoveryBrowser/Art/MaterialInstances/SolidWave/MI_UI_Discovery_Background_SolidWave_FadeMask.MI_UI_Discovery_Background_SolidWave_FadeMask" :Akita "/DiscoveryBrowser/Art/MaterialInstances/MI_UI_Discovery_Background_Akita_FadeMask.MI_UI_Discovery_Background_Akita_FadeMask" :KPOPDemonHunters "/DiscoveryBrowser/Art/MaterialInstances/MI_UI_Discovery_Background_Refrigerator_FadeMask.MI_UI_Discovery_Background_Refrigerator_FadeMask" :EpicGames "/DiscoveryBrowser/Art/MaterialInstances/EpicGames/MI_UI_Discovery_IP_EpicGames_FadeMask.MI_UI_Discovery_IP_EpicGames_FadeMask" :Creatures "/DiscoveryBrowser/Art/MaterialInstances/MI_UI_Discovery_Background_Creatures_Background_FadeMask.MI_UI_Discovery_Background_Creatures_Background_FadeMask") "None" "None" :!class /Script/Engine.KismetMaterialLibrary))
  (Rendering|Material|SetScalarParameterValue _returnvalue "Fade Rotation" 0.0 :!class /Script/Engine.MaterialInstanceDynamic)
  (return _returnvalue))

(fn GetHeaderImageMaterial
  (params (LinkCategory :type EFortActivityLinkCategory%20Enum))
  (return (Utilities|Select :Index LinkCategory :None "/DiscoveryBrowser/CreatorProfile/MaterialInstances/MI_CreatorPage_Header.MI_CreatorPage_Header" :LEGO "/DiscoveryBrowser/Art/MaterialInstances/MI_UI_Discovery_Background_Lego_Header_B.MI_UI_Discovery_Background_Lego_Header_B" :RocketRacing "/DiscoveryBrowser/Art/MaterialInstances/MI_UI_Discovery_Background_RocketRacing_Header.MI_UI_Discovery_Background_RocketRacing_Header" :FallGuys "/DiscoveryBrowser/Art/MaterialInstances/MI_UI_Discovery_Background_FallGuys_Header.MI_UI_Discovery_Background_FallGuys_Header" :TMNT "/DiscoveryBrowser/Art/MaterialInstances/MI_UI_Discovery_Background_NinjaTurtles_Header.MI_UI_Discovery_Background_NinjaTurtles_Header" :TheWalkingDeadUniverse "/DiscoveryBrowser/Art/MaterialInstances/MI_UI_Discovery_Background_DarkVortex_Header.MI_UI_Discovery_Background_DarkVortex_Header" :SolidWave "/DiscoveryBrowser/Art/MaterialInstances/SolidWave/MI_UI_Discovery_Background_SolidWave_Header.MI_UI_Discovery_Background_SolidWave_Header" :Akita "/DiscoveryBrowser/Art/MaterialInstances/MI_UI_Discovery_Background_Akita_Header.MI_UI_Discovery_Background_Akita_Header" :KPOPDemonHunters "/DiscoveryBrowser/Art/MaterialInstances/MI_UI_Discovery_Background_Refrigerator_Header.MI_UI_Discovery_Background_Refrigerator_Header" :EpicGames "/DiscoveryBrowser/Art/MaterialInstances/EpicGames/MI_UI_Discovery_IP_EpicGames_Header.MI_UI_Discovery_IP_EpicGames_Header" :Creatures "/DiscoveryBrowser/Art/MaterialInstances/MI_UI_Discovery_Background_Creatures_Header.MI_UI_Discovery_Background_Creatures_Header")))

(fn GetTopEdgeMaterial
  (params (LinkCategory :type EFortActivityLinkCategory%20Enum))
  (return (Utilities|Select :Index LinkCategory :None "/DiscoveryBrowser/TabScreens/DiscoverHome/Materials/MI_UI_Discovery_NebulousBackgroundWithTopMask.MI_UI_Discovery_NebulousBackgroundWithTopMask" :LEGO "/DiscoveryBrowser/Art/MaterialInstances/MI_UI_Discovery_Background_Lego_FadeMask_B.MI_UI_Discovery_Background_Lego_FadeMask_B" :RocketRacing "/DiscoveryBrowser/Art/MaterialInstances/MI_UI_Discovery_Background_RocketRacing_FadeMask.MI_UI_Discovery_Background_RocketRacing_FadeMask" :FallGuys "/DiscoveryBrowser/Art/MaterialInstances/MI_UI_Discovery_Background_FallGuys_Background_FadeMask.MI_UI_Discovery_Background_FallGuys_Background_FadeMask" :TMNT "/DiscoveryBrowser/Art/MaterialInstances/MI_UI_Discovery_Background_NinjaTurtles_FadeMask.MI_UI_Discovery_Background_NinjaTurtles_FadeMask" :TheWalkingDeadUniverse "/DiscoveryBrowser/Art/MaterialInstances/MI_UI_Discovery_Background_DarkVortex_FadeMask.MI_UI_Discovery_Background_DarkVortex_FadeMask" :SolidWave "/DiscoveryBrowser/Art/MaterialInstances/SolidWave/MI_UI_Discovery_Background_SolidWave_FadeMask.MI_UI_Discovery_Background_SolidWave_FadeMask" :Akita "/DiscoveryBrowser/Art/MaterialInstances/MI_UI_Discovery_Background_Akita_FadeMask.MI_UI_Discovery_Background_Akita_FadeMask" :KPOPDemonHunters "/DiscoveryBrowser/Art/MaterialInstances/MI_UI_Discovery_Background_Refrigerator_FadeMask.MI_UI_Discovery_Background_Refrigerator_FadeMask" :EpicGames "/DiscoveryBrowser/Art/MaterialInstances/EpicGames/MI_UI_Discovery_IP_EpicGames_FadeMask.MI_UI_Discovery_IP_EpicGames_FadeMask" :Creatures "/DiscoveryBrowser/Art/MaterialInstances/MI_UI_Discovery_Background_Creatures_Background_FadeMask.MI_UI_Discovery_Background_Creatures_Background_FadeMask")))
