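# Blank tile entry widget for the Discovery Browser, built on CommonUserWidget, that shows only a faint rounded-corner backplate.
# The sole widget is a UIKit backplate at low render opacity and medium brightness.
# SetBackplateShape passes a corner radius enum through to the backplate's SetCornerRadius, so callers control the shape.
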WidgetBlueprint /DiscoveryBrowser/WBP_Discovery_BlankTileEntry_RoundedCorners
graphs [
 EventGraph {nodes [Event Construct()]}
 SetBackplateShape {nodes [
   n1 = SetBackplateShape!() -> (then, BackplateCornerRadiusType:E_UI_BackplateCornerRadius Enum=NewEnumerator0) {is=title K2Node_FunctionEntry}
   > SetCornerRadius!(
    execute
    self:WBP UIKit Backplate Base Object Reference=Get WBP_UIKit_Backplate() -> (WBP_UIKit_Backplate:WBP UIKit Backplate Object Reference)
    Corner Radius:E_UI_BackplateCornerRadius Enum=@n1
   ) -> (then)
  ]}
]
variables []
widgets {
 info {
  inheritedWidgetCount=0
  parentClass=&CommonUserWidget
  rootWidgetClass=/FortUILibrary/UIKit/Backplates/WBP_UIKit_Backplate.WBP_UIKit_Backplate_C
  widgetCount=1
 }
 widgets [WBP_UIKit_Backplate {
   widgetClassPath=&WBP_UIKit_Backplate
   bIsVariable=true
   widget=WBP_UIKit_Backplate
   properties {brightness=Medium, renderOpacity=0.2}
  }]
}
