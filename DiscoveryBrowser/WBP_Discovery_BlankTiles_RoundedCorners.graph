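# Discovery browser empty-state widget that shows a no-results message over a non-interactive grid of blank rounded-corner tiles.
# PreConstruct sets the message text and fills the TileView with placeholder objects, using the stored size and count.
# An entry-initialized callback, bound on Construct and cleared on Destruct, applies the corner shape to each tile.
# Named slots allow custom content around the message, and SetMaxRowsOverride caps grid height through a SizeBox.
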WidgetBlueprint /DiscoveryBrowser/WBP_Discovery_BlankTiles_RoundedCorners
graphs [
 CreateBlankTileEntries {nodes [
   n1 = CreateBlankTileEntries!() -> (then, EntryShape:E_UI_BackplateCornerRadius Enum=NewEnumerator0, EntrySize:Vector 2D Structure) {is=title K2Node_FunctionEntry}
   > SetEntryTileSize!(execute, self, EntrySize:Vector 2D Structure=@n1.EntrySize) -> (then)
   > Set EntryShape(EntryShape:E_UI_BackplateCornerRadius Enum=@n1) -> (Output_Get:E_UI_BackplateCornerRadius Enum=NewEnumerator0)
   > ClearListItems(self=Get TileView_BlankTiles() -> (TileView_BlankTiles:Tile View Object Reference))
   n2 = Get Entries Count() -> (Entries Count:int=0)
   > Branch(Condition="integer > integer"(A=@n2))
   > For Loop(LastIndex="Max (Integer)"(A=int - int(A=@n2), B=0))
   n3 = Get TileView_BlankTiles()
   > n4 = Construct Object from Class(Class=&/WBP_Discovery_BlankTileEntry_RoundedCorners.WBP_Discovery_BlankTileEntry_RoundedCorners_C, self=@n3) -> (ReturnValue:WBP Discovery Blank Tile Entry Rounded Corners Object Reference)
   > AddItem!(execute, self:List View Object Reference=@n3, Item:Object Reference=@n4) -> (then)
  ]}
 EventGraph {nodes [
   n1 = BlankEntryInitialized!() -> (OutputDelegate, then, Item:Object Reference, Widget:User Widget Object Reference) {is=title K2Node_CustomEvent}
   > n2 = Cast To WBP_Discovery_BlankTileEntry_RoundedCorners!(execute, Object=@n1.Widget) -> (then, CastFailed, AsWBP Discovery Blank Tile Entry Rounded Corners:WBP Discovery Blank Tile Entry Rounded Corners Object Reference) {is=title K2Node_DynamicCast}
   > SetBackplateShape!(
    execute
    self:WBP Discovery Blank Tile Entry Rounded Corners Object Reference=@n2
    BackplateCornerRadiusType:E_UI_BackplateCornerRadius Enum=Get EntryShape() -> (EntryShape:E_UI_BackplateCornerRadius Enum=NewEnumerator0)
   ) -> (then)
   n3 = Event PreConstruct()
   > n4 = Sequence()
   > Branch(Condition=@n3.IsDesignTime)
   n5 = Get Slot_BelowCustomContent() -> (Slot_BelowCustomContent:Named Slot Object Reference)
   > SetVisibility(
    self=@n5
    InVisibility=Select(Option 0=Collapsed, Option 1=SelfHitTestInvisible, Index=HasAnyChildren!(self:Panel Widget Object Reference=@n5) -> (ReturnValue:bool))
   )
   n6 = Get Slot_AboveCustomContent() -> (Slot_AboveCustomContent:Named Slot Object Reference)
   > SetVisibility(self=@n6, InVisibility=Select(Option 0=Collapsed, Option 1=SelfHitTestInvisible, Index=HasAnyChildren(self=@n6)))
   SetMessageTextOverrides!(
    execute=@n4.then_1
    self
    InTitle:Text=Get FlavorTitleOverride() -> (FlavorTitleOverride:Text)
    InSubTitle:Text=Get FlavorSubtitleOverride() -> (FlavorSubtitleOverride:Text)
   ) -> (then)
   > CreateBlankTileEntries!(
    execute
    self
    EntryShape:E_UI_BackplateCornerRadius Enum=NewEnumerator1
    EntrySize:Vector 2D Structure=Get EntrySize() -> (EntrySize:Vector 2D Structure)
   ) -> (then) {is=title K2Node_CallFunction}
   > SetScrollbarVisibility!(execute, self:List View Base Object Reference=Get TileView_BlankTiles(), InVisibility:ESlateVisibility Enum=Collapsed) -> (then)
   Event Construct()
   n7 = Get TileView_BlankTiles()
   > Bind Event to On Entry Initialized(self:List View Object Reference=@n7, Delegate=Create Event()) {is=title K2Node_AddDelegate}
   Event Destruct()
   > Unbind all Events from On Entry Initialized!(execute:Exec, self:List View Object Reference=@n7) -> (then:Exec) {is=title K2Node_ClearDelegate}
  ]}
 SetEntryTileSize {nodes [
   n1 = SetEntryTileSize!() -> (then, EntrySize:Vector 2D Structure) {is=title K2Node_FunctionEntry}
   > n2 = Set EntrySize(EntrySize:Vector 2D Structure=@n1) -> (Output_Get:Vector 2D Structure)
   n3 = BreakVector2D(InVec=@n2)
   n4 = Get TileView_BlankTiles()
   > SetEntryWidth!(execute, self:Tile View Object Reference=@n4, NewWidth:float=@n3) -> (then)
   > SetEntryHeight!(execute, self:Tile View Object Reference=@n4, NewHeight:float=@n3.Y) -> (then)
  ]}
 SetMaxRowsOverride {nodes [
   n1 = SetMaxRowsOverride!() -> (then, MaxRows:int) {is=title K2Node_FunctionEntry}
   > Branch(Condition="integer > integer"(A=@n1))
   n2 = Get TileView_BlankTiles()
   n3 = float * float(A=GetEntryHeight!(self:Tile View Object Reference=@n2) -> (ReturnValue:float), B:int=@n1) {is=title K2Node_PromotableOperator}
   n4 = float * float(A=GetVerticalEntrySpacing!(self:List View Object Reference=@n2) -> (ReturnValue:float), B=@n1)
   > SetMaxDesiredHeight!(
    execute
    self:Size Box Object Reference=Get SizeBox_OptionallyCapRowHeight() -> (SizeBox_OptionallyCapRowHeight:Size Box Object Reference)
    InMaxDesiredHeight:float=float - float(A=float + float(A=@n3, B=@n4), B=1)
   ) -> (then)
  ]}
 SetMessageTextOverrides {nodes [
   n1 = SetMessageTextOverrides!() -> (then, InTitle:Text, InSubTitle:Text) {is=title K2Node_FunctionEntry}
   > n2 = Set FlavorTitleOverride(FlavorTitleOverride:Text=@n1) -> (Output_Get:Text)
   > n3 = Set FlavorSubtitleOverride(FlavorSubtitleOverride:Text=@n1.InSubTitle) -> (Output_Get:Text)
   > Set Message Text Overrides!(
    execute
    self:WBP Discovery No Results Message Object Reference=Get WBP_Discovery_NoResultsMessage() -> (WBP_Discovery_NoResultsMessage:WBP Discovery No Results Message Object Reference)
    InTitle:Text=@n2
    InMessage:Text=@n3
   ) -> (then) {is=title K2Node_CallFunction}
  ]}
]
variables [
 Entries Count {type=integer, default=12}
 EntryShape {type=E_UI_BackplateCornerRadius, default=Backplate}
 EntrySize {type=Vector2D, default=<312 176>}
 FlavorSubtitleOverride {type=string, default=Your favorites from Discover will appear here.}
 FlavorTitleOverride {type=string, default=No Results}
]
widgets {
 info {inheritedWidgetCount=0, parentClass=&UserWidget, rootWidgetClass=&Overlay, widgetCount=7}
 widgets [
  Overlay_BlankTiles {widgetClassPath=&Overlay, widget=Overlay_BlankTiles}
  SizeBox_OptionallyCapRowHeight {
   widgetClassPath=&SizeBox
   bIsVariable=true
   parent=Overlay_BlankTiles
   widget=SizeBox_OptionallyCapRowHeight
   slot {class=OverlaySlot, properties {horizontalAlignment=HAlign_Fill}}
  }
  TileView_BlankTiles {
   widgetClassPath=/Script/UMG.TileView
   bIsVariable=true
   parent=SizeBox_OptionallyCapRowHeight
   widget=TileView_BlankTiles
   properties {
    allowOverscroll=false
    bAllowDragging=false
    bEnableRightClickScrolling=false
    bEnableTouchScrolling=false
    bEntrySizeIncludesEntrySpacing=false
    bIsFocusable=false
    consumeMouseWheel=Never
    entryHeight=176
    entryWidgetClass=&WBP_Discovery_BlankTileEntry_RoundedCorners_C
    entryWidth=312
    horizontalEntrySpacing=24
    numDesignerPreviewEntries=20
    selectionMode=None
    tileAlignment=LeftAligned
    verticalEntrySpacing=88
    visibility=HitTestInvisible
    wheelScrollMultiplier=0
   }
   slot {class=SizeBoxSlot}
  }
  VerticalBox_0 {
   widgetClassPath=&VerticalBox
   parent=Overlay_BlankTiles
   widget=VerticalBox_0
   slot {class=OverlaySlot, properties {horizontalAlignment=HAlign_Center, verticalAlignment=VAlign_Center}}
  }
  Slot_AboveCustomContent {
   widgetClassPath=&NamedSlot
   bIsVariable=true
   parent=VerticalBox_0
   widget=Slot_AboveCustomContent
   slot {class=VerticalBoxSlot, properties {verticalAlignment=VAlign_Top}}
  }
  WBP_Discovery_NoResultsMessage {
   widgetClassPath=&/WBP_Discovery_MessageNoResults.WBP_Discovery_NoResultsMessage
   bIsVariable=true
   parent=VerticalBox_0
   widget=WBP_Discovery_NoResultsMessage
   slot {class=VerticalBoxSlot}
  }
  Slot_BelowCustomContent {
   widgetClassPath=&NamedSlot
   bIsVariable=true
   parent=VerticalBox_0
   widget=Slot_BelowCustomContent
   slot {class=VerticalBoxSlot, properties {verticalAlignment=VAlign_Top}}
  }
 ]
}
