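# Activatable Library screen of the Discover browser, showing a player's library in a nested side nav and continuous list template.
# The Discover surface view model feeds a panel tree view model that builds the nav and list, with an empty-library placeholder.
# When storage is enabled, a Storage tab is appended that swaps in alternate content and reroutes focus.
# It also fires view analytics, fades the header edge on scroll and toggles back buttons for homebar placement.
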WidgetBlueprint /DiscoveryBrowser/TabScreens/Library/DiscoverLibraryView_V2
graphs [
 BP_GetDesiredFocusTarget {nodes [
   Get Desired Focus Target()
   > Return Node!(
    execute
    ReturnValue:Widget Object Reference=Get WBP_Template_NestedSideNav_AndContinuousList() -> (WBP_Template_NestedSideNav_AndContinuousList:WBP Template Nested Side Nav And Continuous List Object Reference)
   )
  ]}
 EventGraph {nodes [
   "On Clicked (Button_DeeplinkToDiscover)"!() -> (OutputDelegate, then, Button:Common Button Base Object Reference) {is=title K2Node_ComponentBoundEvent}
   n1 = GetOwningPlayer()
   n2 = Self-Reference()
   n3 = GetIsLibraryTabInTopNavigationBarEnabled!(self:Fort Runtime Options Object Reference=GetRuntimeOptions(), ContextWidget:Widget Object Reference=@n2) -> (ReturnValue:bool)
   > n4 = Branch(Condition=AND Boolean(
     A=IsLibraryOnHomebar!(ContextWidget:Widget Object Reference=@n2, PlayerController:Player Controller Object Reference=@n1) -> (ReturnValue:bool)
     B=NOT Boolean(A=@n3)
    ))
   DeactivateWidget(execute=[@n5.then, @n4.then])
   Reroute Node(InputPin=@n4.else)
   n5 = "On Clicked (WBP_BottomBarDecoyButton_LibraryBack)"!() -> (OutputDelegate, then, Button:Common Button Base Object Reference) {is=title K2Node_ComponentBoundEvent}
   n6 = OnMegaListScrolled!() -> (OutputDelegate, then, ItemOffset:float=~, DistanceRemaining:float=~) {is=title K2Node_CustomEvent}
   n7 = "Absolute (Float)"(A=float - float(A=@n6.DistanceRemaining, B=1))
   > SetRenderOpacity(
    self=Get Image_HeaderSoftEdge!(self:WBP Template Nested Side Nav And Continuous List Object Reference=Get WBP_Template_NestedSideNav_AndContinuousList()) -> (Image_HeaderSoftEdge:Image Object Reference)
    InOpacity="To Float (Integer)"(InInt=Ceil(A=@n7) {is=title K2Node_CallFunction})
   )
   HandleNoStorageNavigableWidgets() {is=title K2Node_CustomEvent}
   > UpdateStorageNavigableWidget!(execute, self, bIsAlternateContent:bool=true) -> (then)
   > GiveFocusToSideNav(self:WBP Template Nested Side Nav And Continuous List Object Reference=Get WBP_Template_NestedSideNav_AndContinuousList())
   n8 = OnInputMethodChanged!() -> (OutputDelegate, then, bNewInputType:ECommonInputType Enum) {is=title K2Node_CustomEvent}
   SetBottomButtonIfHomeLibraryEnabled(execute=[@n28.then, @n8.then])
   n9 = OnMegaListArrayFinishedAddStorageTabToEnd!() -> (OutputDelegate, then, LastIndex:int) {is=title K2Node_CustomEvent}
   > n10 = IsLibraryStorageEnabled!(execute) -> (then, ReturnValue:bool)
   > Branch(Condition=@n10)
   > AddSideNavigationTab!(
    execute
    self:WBP UIKit Side Navigation Base Object Reference=Get WBP_UIKit_SideNavigation!(self:WBP Template Nested Side Nav And Continuous List Object Reference=Get WBP_Template_NestedSideNav_AndContinuousList()) -> (WBP_UIKit_SideNavigation:WBP UIKit Side Navigation Object Reference)
    TabViewModel:MVVM Base Viewmodel Object Reference
    TabID:Name=Storage
    Index:int=int + int(A=@n9.LastIndex)
   ) -> (then)
   n11 = Get WBP_Template_NestedSideNav_AndContinuousList()
   > n12 = Cast To WBP_UIKit_SideNavigation_Button!(
    execute
    Object=GetTabButtonBaseByID!(self:Common Tab List Widget Base Object Reference=Get WBP_UIKit_SideNavigation(self=@n11), TabNameID:Name=Storage) -> (ReturnValue:Common Button Base Object Reference)
   ) -> (then, CastFailed, AsWBP UIKit Side Navigation Button:WBP UIKit Side Navigation Button Object Reference) {is=title K2Node_DynamicCast}
   > SetText!(execute, self:WBP UIKit Button Generic Base Object Reference=@n12, Text:Text="LOCTEXT(\"Storage\")") -> (then)
   n13 = OnSideNavSelectionChanged!() -> (OutputDelegate, then, TabID:Name, InnerTabID:Name, TabViewModel:MVVM Base Viewmodel Object Reference) {is=title K2Node_CustomEvent}
   > SendSelectedSubViewEvent!(execute, self, TabID:Name=@n13.TabID, InnerTabID:Name=@n13.InnerTabID) -> (then)
   n14 = "Equal (Name)"(A=@n13.TabID, B=Storage)
   > UpdateStorageNavigableWidget(bIsAlternateContent=@n14)
   > SwitchMainContent(self:WBP Template Nested Side Nav And Continuous List Object Reference=Get WBP_Template_NestedSideNav_AndContinuousList())
   > SetVisibility(
    self=Get WBP_Library_StorageDeviceInfo() -> (WBP_Library_StorageDeviceInfo:WBP Library Storage Device Info Object Reference)
    InVisibility=Select(Option 0=Collapsed, Option 1=SelfHitTestInvisible, Index=@n14)
   )
   Event Construct()
   n15 = Get FortDiscoverSurfaceViewModel() -> (FortDiscoverSurfaceViewModel:Discover Surface View Model Object Reference)
   > n16 = Branch(Condition=ShouldQueryOnConstruct!(self:Discover Surface View Model Object Reference=@n15) -> (ReturnValue:bool))
   > n17 = Set SurfaceName!(execute, SurfaceName:Name=CreativeDiscoverySurface_Library_V2, self:Discover Surface View Model Object Reference=@n15) -> (then, Output_Get:Name=None)
   SetBottomButtonVisibility!(
    execute=[@n17.then, @n16.else]
    self:WBP Bottom Bar Decoy Button Object Reference=Get WBP_BottomBarDecoyButton_LibraryBack() -> (WBP_BottomBarDecoyButton_LibraryBack:WBP Bottom Bar Decoy Button Object Reference)
    ShouldButtonBeVisible:bool=IsLibraryOnHomebar(ContextWidget=Self-Reference(), PlayerController=GetOwningPlayer())
   ) -> (then)
   n18 = Create Event()
   n19 = Get WBP_Template_NestedSideNav_AndContinuousList()
   > Bind Event to On Mega List Array Finished Recursive Creation(self:WBP Template Nested Side Nav And Continuous List Object Reference=@n19, Delegate=@n18) {is=title K2Node_AddDelegate}
   n20 = Create Event()
   n21 = Get WBP_UIKit_SideNavigation(self=@n19)
   > Bind Event to On Side Navigation Selection Changed(self:WBP UIKit Side Navigation Base Object Reference=@n21, Delegate=@n20) {is=title K2Node_AddDelegate}
   n22 = Get ListView!(self:Fort Polymorphic List View Object Reference=Get WBP_Poly_MegaList!(self:WBP Template Nested Side Nav And Continuous List Object Reference=Get WBP_Template_NestedSideNav_AndContinuousList()) -> (WBP_Poly_MegaList:WBP Poly Mega List Object Reference)) -> (ListView:Common List View Object Reference)
   > Bind Event to On List View Scrolled(self:List View Object Reference=@n22, Delegate=Create Event()) {is=title K2Node_AddDelegate}
   n23 = Create Event()
   n24 = Get WBP_Library_StorageList() -> (WBP_Library_StorageList:WBP Library Storage List Object Reference)
   > Bind Event to On No Navigable Widgets(self:WBP Library Storage List Object Reference=@n24, Delegate=@n23) {is=title K2Node_AddDelegate}
   n25 = Self-Reference()
   n26 = IsLibraryOnHomebar(ContextWidget=@n25, PlayerController=GetOwningPlayer())
   n27 = NOT Boolean(A=GetIsLibraryTabInTopNavigationBarEnabled(self=GetRuntimeOptions(), ContextWidget=@n25))
   > SetVisibility(
    self=Get Button_DeeplinkToDiscover() -> (Button_DeeplinkToDiscover:WBP UIKit Button Loud Object Reference)
    InVisibility=Select(Option 0=Collapsed, Option 1=SelfHitTestInvisible, Index=AND Boolean(A=@n26, B=@n27))
   )
   > n28 = Bind Event to On Input Method Changed(self=Get CommonInputSubsystem(), Delegate=@n8)
   Event On Activated()
   > ActivateWidget(self=Get WBP_Template_NestedSideNav_AndContinuousList())
   n29 = Get FortDiscoverSurfaceViewModel()
   > n30 = Branch(Condition=ShouldQueryOnConstruct(self=@n29))
   StartAnalyticsImpressions(execute=[@n30.then, @n34.then, @n33.else], self:Discover Surface View Model Object Reference=Get FortDiscoverSurfaceViewModel())
   > Fire Event Main View Opened!(execute, View:Widget Object Reference=Self-Reference(), AdditionalInfo:Flow Analytics Additional Info Structure, bIsOverlay:bool=false) -> (then)
   n31 = Get FortDiscoverSurfaceViewModel()
   > UpdateCachedSurfaceData!(execute, self:Discover Surface View Model Object Reference=@n31, bShouldForceUpdate:bool=ShouldQueryOnConstruct(self=@n31)) -> (then)
   > n32 = Switch on ECommonInputType!(execute, Selection:ECommonInputType Enum=GetCurrentInputType(self=Get CommonInputSubsystem())) -> (MouseAndKeyboard:Exec, Gamepad:Exec, Touch:Exec, Count:Exec) {is=title K2Node_SwitchEnum}
   RequestRefreshFocus(execute=@n32.Gamepad, self:Common Activatable Widget Object Reference)
   n33 = Branch(
    execute=@n30.else
    Condition="Equal (Name)"(A=Get SurfaceName!(self:Discover Surface View Model Object Reference=@n29) -> (SurfaceName:Name=None))
   )
   > n34 = Set SurfaceName(SurfaceName=CreativeDiscoverySurface_Library_V2, self=@n29) -> (Output_Get=None)
   Event Destruct()
   > Unbind Event from On Mega List Array Finished Recursive Creation(self:WBP Template Nested Side Nav And Continuous List Object Reference=@n19, Delegate=@n18) {is=title K2Node_RemoveDelegate}
   > Unbind Event from On Side Navigation Selection Changed(self:WBP UIKit Side Navigation Base Object Reference=@n21, Delegate=@n20) {is=title K2Node_RemoveDelegate}
   > Unbind all Events from On List View Scrolled!(execute:Exec, self:List View Object Reference=@n22) -> (then:Exec) {is=title K2Node_ClearDelegate}
   > Unbind Event from On No Navigable Widgets(self:WBP Library Storage List Object Reference=@n24, Delegate=@n23) {is=title K2Node_RemoveDelegate}
   Event On Deactivated()
   > DeactivateWidget(self=Get WBP_Template_NestedSideNav_AndContinuousList())
   > FireExitedImpressionsFromSurface(self:Discover Surface View Model Object Reference=Get FortDiscoverSurfaceViewModel())
  ]}
 On Update Flat Surface {nodes [
   n1 = On Update Flat Surface!() -> (then, ProvidersVM:Array of Discover Provider View Model Object References) {is=title K2Node_FunctionEntry}
   > Branch(Condition=Is Not Empty!(TargetArray:Array of Discover Provider View Model Object References=@n1) -> (ReturnValue:bool))
   n2 = Get FortDiscoverPanelTreeViewModel() -> (FortDiscoverPanelTreeViewModel:Discover Panel Tree View Model Object Reference)
   > Initialize!(
    execute
    self:Discover Panel Tree View Model Object Reference=@n2
    InSurfaceViewModel:Discover Surface View Model Object Reference=Get FortDiscoverSurfaceViewModel()
    InPanelViewModel:Discover Provider View Model Object Reference="Get (a copy)"(Array:Array of Discover Provider View Model Object References=@n1) -> (Output:Discover Provider View Model Object Reference)
    InReferenceLinkVM:Fort Discover Surface Tile Item VM Object Reference
   ) -> (then)
   > SetupSideNavAndMegaListFromViewModel!(
    execute
    self:WBP Template Nested Side Nav And Continuous List Object Reference=Get WBP_Template_NestedSideNav_AndContinuousList()
    Tree View Model:Discover Panel Tree View Model Object Reference=@n2
   ) -> (then)
  ]}
 SendSelectedSubViewEvent {nodes [
   n1 = SendSelectedSubViewEvent!() -> (then, TabID:Name, InnerTabID:Name) {is=title K2Node_FunctionEntry}
   n2 = Get WBP_Template_NestedSideNav_AndContinuousList()
   > n3 = Cast To FortPlayerController(Object=GetOwningPlayer(self=Get WBP_UIKit_SideNavigation(self=@n2)))
   n4 = "Equal (Name)"(A=@n1.InnerTabID)
   > "Fire Event Sub View Opened (Non-Localized)"!(
    execute
    FortPC:Fort Player Controller Object Reference=@n3
    ViewName:String="To String (Name)"(InName=SelectName!(A:Name=@n1, B:Name=@n1.InnerTabID, bPickA:bool=@n4) -> (ReturnValue:Name=None))
    AdditionalInfo:Flow Analytics Additional Info Structure
    bIsOverlay:bool=false
   ) -> (then)
  ]}
 SetBottomButtonIfHomeLibraryEnabled {nodes [
   SetBottomButtonIfHomeLibraryEnabled() {is=title K2Node_FunctionEntry}
   > SetBottomButtonVisibility(
    self=Get WBP_BottomBarDecoyButton_LibraryBack()
    ShouldButtonBeVisible=IsLibraryOnHomebar(ContextWidget=Self-Reference(), PlayerController=GetOwningPlayer())
   )
  ]}
 UpdateStorageNavigableWidget {nodes [
   UpdateStorageNavigableWidget!() -> (then, bIsAlternateContent:bool=~) {is=title K2Node_FunctionEntry}
   > n1 = GetNavigableWidget!(execute, self:WBP Library Storage List Object Reference=Get WBP_Library_StorageList()) -> (then, NavigableWidget:Widget Object Reference)
   > SetAlternateContentNavigableWidget!(
    execute
    self:WBP Template Nested Side Nav And Continuous List Object Reference=Get WBP_Template_NestedSideNav_AndContinuousList()
    NavigableWidget:Widget Object Reference=@n1
    bIsAlternateContent:bool=Get bIsAlternateContent() -> (bIsAlternateContent:bool)
   ) -> (then)
  ]}
]
variables [IntroWhooshSound {type=SoundBase, default=/Game/Sounds/Fort_UI_Sounds/FrontEnd/Tabs/UI_Whoosh_TopBar_Tab_Home_Cue}]
widgets {
 info {
  inheritedWidgetCount=0
  namedSlotCount=5
  parentClass=&CommonActivatableWidget
  rootWidgetClass=&GridPanel
  widgetCount=12
 }
 widgets [
  RootGrid_Library {widgetClassPath=&GridPanel, widget=RootGrid_Library, properties {columnFill [1], rowFill [1]}}
  WBP_CaptureForPostBufferUpdate {
   widgetClassPath=&WBP_CaptureForPostBufferUpdate
   bIsVariable=true
   parent=RootGrid_Library
   slot {class=GridSlot, properties {layer=-20}}
  }
  LibraryBackground {
   widgetClassPath=&Image
   parent=RootGrid_Library
   widget=LibraryBackground
   properties {brush {
     bIsDynamicallyLoaded=false
     drawAs=Image
     imageType=FullColor
     mirroring=NoMirror
     resourceName=None
     resourceObject=/DiscoveryBrowser/TabScreens/Library/MaterialInstances/MI_UI_Library_Background
     tiling=NoTile
     imageSize=<32 32>
     margin=<ltrb 0 0 0 0>
     outlineSettings {
      bUseBrushTransparency=false
      roundingType=HalfHeightRadius
      width=0
      color {colorUseRule=UseColor_Specified, specifiedColor=<rgba 0 0 0 0>}
      cornerRadii=<0 0 0 0>
     }
     tintColor {colorUseRule=UseColor_Specified, specifiedColor=<rgba 1 1 1 1>}
     uVRegion {bIsValid=false, max=<0 0>, min=<0 0>}
    }}
   slot {class=GridSlot, properties {layer=-10}}
  }
  WBP_Template_NestedSideNav_AndContinuousList {
   widgetClassPath=/DiscoveryBrowser/WBP_Template_NestedSideNav_AndContinuousList
   bIsVariable=true
   parent=RootGrid_Library
   widget=WBP_Template_NestedSideNav_AndContinuousList
   properties {
    bIsFocusable=true
    noResultsSubText="Please try again later.  SURFACE EMPTY"
    sideNav_Button_NestedClass=&/WBP_UIKit_SideNavigation_ButtonNested_Library.WBP_UIKit_SideNavigation_ButtonNested_Library_C
    sideNav_Button_TopLevelClass=&/WBP_UIKit_SideNavigation_Button_Library.WBP_UIKit_SideNavigation_Button_Library_C
    desiredFocusWidget {widgetName=None}
    megaListPadding=<ltrb -38 0 104 0>
   }
   slot {class=GridSlot}
  }
  WBP_Discovery_BlankTiles_RoundedCorners {
   widgetClassPath=/DiscoveryBrowser/WBP_Discovery_BlankTiles_RoundedCorners
   bIsVariable=true
   namedSlotHost=WBP_Template_NestedSideNav_AndContinuousList
   widget=WBP_Discovery_BlankTiles_RoundedCorners
   properties {
    entries Count=0
    flavorSubtitleOverride=Your recently played and favorites will appear here. Explore Discover to get started
    flavorTitleOverride=Your Library is empty
   }
  }
  Overlay_0 {widgetClassPath=&Overlay, namedSlotHost=WBP_Discovery_BlankTiles_RoundedCorners, widget=Overlay_0}
  IconLibrary {
   widgetClassPath=&Image
   bIsVariable=true
   parent=Overlay_0
   widget=IconLibrary
   properties {brush {
     bIsDynamicallyLoaded=false
     drawAs=Image
     imageType=FullColor
     mirroring=NoMirror
     resourceName=None
     resourceObject=/FortUILibrary/Textures/MaterialInstances/MI_UIKit_Icon_Library
     tiling=NoTile
     imageSize=<96 96>
     margin=<ltrb 0 0 0 0>
     outlineSettings {
      bUseBrushTransparency=false
      roundingType=HalfHeightRadius
      width=0
      color {colorUseRule=UseColor_Specified, specifiedColor=<rgba 0 0 0 0>}
      cornerRadii=<0 0 0 0>
     }
     tintColor {colorUseRule=UseColor_Specified, specifiedColor=<rgba 1 1 1 1>}
     uVRegion {bIsValid=false, max=<0 0>, min=<0 0>}
    }}
   slot {class=OverlaySlot, properties {horizontalAlignment=HAlign_Center, padding=<ltrb 0 0 0 16>}}
  }
  Overlay_2 {widgetClassPath=&Overlay, namedSlotHost=WBP_Discovery_BlankTiles_RoundedCorners, widget=Overlay_2}
  Button_DeeplinkToDiscover {
   widgetClassPath=/FortUILibrary/UIKit/Buttons/WBP_UIKit_Button_Loud
   bIsVariable=true
   parent=Overlay_2
   widget=Button_DeeplinkToDiscover
   properties {text=Back to Discover, triggeringInputAction {dataTable=/FortUILibrary/Foundation/ActivityBrowserInputActionTable, rowName=Select}}
   slot {class=OverlaySlot, properties {horizontalAlignment=HAlign_Center, padding=<ltrb 0 32 0 0>}}
  }
  WBP_Library_StorageList {
   widgetClassPath=&/WBP_Library_StorageList
   bIsVariable=true
   namedSlotHost=WBP_Template_NestedSideNav_AndContinuousList
   widget=WBP_Library_StorageList
   properties {padding=<ltrb 40 -28 0 0>}
  }
  WBP_Library_StorageDeviceInfo {
   widgetClassPath=&/WBP_Library_StorageDeviceInfo
   bIsVariable=true
   namedSlotHost=WBP_Template_NestedSideNav_AndContinuousList
   widget=WBP_Library_StorageDeviceInfo
   properties {visibility=Collapsed}
  }
  WBP_BottomBarDecoyButton_LibraryBack {
   widgetClassPath=/Game/UI/Frontend/BottomBar/WBP_BottomBarDecoyButton
   bIsVariable=true
   parent=RootGrid_Library
   widget=WBP_BottomBarDecoyButton_LibraryBack
   properties {triggeringInputAction {dataTable=&InputActionDataTable, rowName=Back}}
   slot {class=GridSlot, properties {horizontalAlignment=HAlign_Right, verticalAlignment=VAlign_Bottom}}
  }
 ]
}
