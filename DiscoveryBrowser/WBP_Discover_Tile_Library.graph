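# Library tile widget for the Discover browser, a button that asynchronously pushes the Library view and fires a clicked dispatcher; entry release fires another.
# On construct it applies mobile-only compact sizing, hides the button's text and, if library storage is enabled, initializes its view model from the current download.
# Helper functions format progress as a percentage and show it only with minimal lobby enabled, a downloading or updating status and progress above a threshold.
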WidgetBlueprint /DiscoveryBrowser/WBP_Discover_Tile_Library
graphs [
 EventGraph {nodes [
   "On Clicked (Button_Library)"!() -> (OutputDelegate, then, Button:Common Button Base Object Reference) {is=title K2Node_ComponentBoundEvent}
   > PushPrimaryContentClassAsync!(
    execute:Exec
    WidgetClass:Common Activatable Widget Class Reference=/DiscoveryBrowser/TabScreens/Library/DiscoverLibraryView_V2.DiscoverLibraryView_V2_C
   ) -> (then:Exec, OnPrimaryContentPushed:Exec, ActivatableWidget:Common Activatable Widget Object Reference) {is=title K2Node_AsyncAction}
   > Call On Library Tile Clicked!(execute:Exec, self:WBP Discover Tile Library Object Reference) -> (then:Exec) {is=title K2Node_CallDelegate}
   Event Construct()
   > SetCompactUIStyle() {is=title K2Node_CallFunction}
   > SetVisibility(
    self=Get Block_Text!(self:WBP UIKit Button Generic Base Object Reference=Get Button_Library() -> (Button_Library:WBP Discover Button Generic Object Reference)) -> (Block_Text:WBP UIKit Block Text Object Reference)
    InVisibility=Collapsed
   )
   > n1 = IsLibraryStorageEnabled!(execute) -> (then, ReturnValue:bool)
   > Branch(Condition=@n1)
   > InitializeUsingCurrentDownload(self:Fort Library Storage Item View Model Object Reference=Get FortLibraryStorageItemViewModel() -> (FortLibraryStorageItemViewModel:Fort Library Storage Item View Model Object Reference))
   Event On Entry Released() {is=title K2Node_Event}
   > Call On Library Tile Released!(execute:Exec, self:WBP Discover Tile Library Object Reference) -> (then:Exec) {is=title K2Node_CallDelegate}
   "On Hovered (Button_Library)"!() -> (OutputDelegate, then, Button:Common Button Base Object Reference) {is=title K2Node_ComponentBoundEvent}
   Event PreConstruct()
  ]}
 GetDownloadProgressVisibility {nodes [
   n1 = GetDownloadProgressVisibility!() -> (then, DownloadStatus:ELibraryStorageItemDownloadStatus Enum=NotDownloaded, DownloadProgress:int) {is=title K2Node_FunctionEntry}
   > n2 = IsMinimalLobbyEnabled!(execute) -> (then, ReturnValue:bool)
   n3 = "integer > integer"(A=@n1.DownloadProgress)
   n4 = Select!(
    NotDownloaded:bool=false
    Downloaded:bool=false
    Downloading:bool=true
    UpdateRequired:bool=false
    Updating:bool=true
    Queued:bool=false
    Paused:bool=false
    PendingRemoval:bool=false
    Removing:bool=false
    Index:ELibraryStorageItemDownloadStatus Enum=@n1
   ) -> (ReturnValue:bool)
   > Return Node!(
    execute
    Visibility:ESlateVisibility Enum=Select!(
     Option 0:ESlateVisibility Enum=Collapsed
     Option 1:ESlateVisibility Enum=SelfHitTestInvisible
     Index:bool=AND Boolean!(A=@n2, B=@n4, C:bool=@n3) -> (ReturnValue)
    ) -> (ReturnValue:ESlateVisibility Enum)
   )
  ]}
 OnLibraryTileClicked {nodes [OnLibraryTileClicked() {is=title K2Node_FunctionEntry}]}
 OnLibraryTileReleased {nodes [OnLibraryTileReleased() {is=title K2Node_FunctionEntry}]}
 SetCompactUIStyle {nodes [
   SetCompactUIStyle() {is=title K2Node_FunctionEntry}
   > Branch(Condition=IsMobileGame())
   n1 = Get Button_Library()
   > SetIconSize!(execute, self:WBP UIKit Button Generic Base Object Reference=@n1, Icon:Vector 2D Structure=(X=64,Y=64)) -> (then)
   > SetButtonPadding!(
    execute
    self:WBP UIKit Button Generic Base Object Reference=@n1
    Padding_Left:float=16
    Padding_Top:float=0
    Padding_Right:float=16
    Padding_Bottom:float=0
   ) -> (then)
  ]}
 UpdateDownloadPercent {nodes [
   n1 = UpdateDownloadPercent!() -> (then, DownloadProgress:int) {is=title K2Node_FunctionEntry}
   > SetText!(
    execute
    self:WBP UIKit Block Text Object Reference=Get Block_Text(self=Get Button_Library())
    Text:Text=Format Text!(Format="LOCTEXT(\"{DownloadProgress}%\")", DownloadProgress:Text="To Text (Integer)"(Value=@n1)) -> (Result)
   ) -> (then)
  ]}
]
variables [
 OnLibraryTileClicked {}
 OnLibraryTileReleased {}
]
widgets {
 info {
  inheritedWidgetCount=2
  parentClass=/Script/DiscoveryBrowserUI.FortActivityBrowserLibraryTile
  rootWidgetClass=&Overlay
  widgetCount=5
 }
 widgets [
  Overlay_0 {widgetClassPath=&Overlay, widget=Overlay_0}
  WBP_CaptureForPostBufferUpdate {
   widgetClassPath=&WBP_CaptureForPostBufferUpdate
   bIsVariable=true
   parent=Overlay_0
   slot {class=OverlaySlot, properties {horizontalAlignment=HAlign_Fill, verticalAlignment=VAlign_Fill}}
  }
  Button_Library {
   widgetClassPath=/DiscoveryBrowser/Tiles/WBP_Discover_Button_Generic
   bIsVariable=true
   parent=Overlay_0
   widget=Button_Library
   properties {
    bHideInputAction=true
    iconMaterial=/FortUILibrary/Textures/MaterialInstances/MI_UIKit_Icon_Library
    iconVerticalAlignment=VAlign_Bottom
    minHeight=141
    minWidth=90
    mobileMinHeight=172
    mobileMinWidth=90
    text=100%
    textBlockAlignment=Bottom
    buttonPadding=<ltrb 0 0 0 0>
    font {
     bForceMonospaced=false
     bMaterialIsStencil=false
     fontMaterial=/FortUILibrary/UIKit/Buttons/MaterialInstances/CTAButtons/MI_UIKit_TextColor_Quiet
     fontObject=&HeadingNow
     letterSpacing=48
     monospacedWidth=1
     size=15.75
     skewAmount=0
     typefaceFontName=74 Regular
     outlineSettings {
      bApplyOutlineToDropShadows=false
      bMiteredCorners=false
      bSeparateFillAlpha=false
      outlineMaterial=null
      outlineSize=0
      outlineColor=<rgba 0 0 0 1>
     }
    }
    textPadding=<ltrb 0 4 0 0>
   }
   slot {class=OverlaySlot, properties {verticalAlignment=VAlign_Center}}
  }
  OnboardingTooltip {widgetClassPath=/Script/FortniteUI.FortSidebarOnboardTooltipWidget, bInherited=true, widget=null}
  InputActionWidget {widgetClassPath=&CommonActionWidget, bInherited=true, widget=null}
 ]
}
