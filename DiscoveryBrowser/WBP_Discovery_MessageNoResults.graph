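# Discovery browser widget that shows a "no results" message as a centered title with a dimmer, wrapping description beneath it.
# PreConstruct passes the FlavorTitleBase and FlavorMessageBase defaults into Set Message Text Overrides, which writes them to the two CommonTextBlocks.
# The same function takes title and message inputs, so other text can replace the defaults.
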WidgetBlueprint /DiscoveryBrowser/WBP_Discovery_MessageNoResults.WBP_Discovery_NoResultsMessage
graphs [
 EventGraph {nodes [
   Event PreConstruct()
   > Set Message Text Overrides!(
    execute
    self
    InTitle:Text=Get FlavorTitleBase() -> (FlavorTitleBase:Text)
    InMessage:Text=Get FlavorMessageBase() -> (FlavorMessageBase:Text)
   ) -> (then) {is=title K2Node_CallFunction}
  ]}
 Set Message Text Overrides {nodes [
   n1 = Set Message Text Overrides!() -> (then, InTitle:Text, InMessage:Text) {is=title K2Node_FunctionEntry}
   > "SetText (Text)"(self=Get Text_FlavorTitle() -> (Text_FlavorTitle:Common Text Object Reference), InText=@n1)
   > Set Text(Text=@n1.InMessage, self=Get Description() -> (Description:Common Text Object Reference))
  ]}
]
variables [
 FlavorMessageBase {type=string, default=It looks like you don't have any results yet.}
 FlavorTitleBase {type=string, default=No Results}
]
widgets {
 info {inheritedWidgetCount=0, parentClass=&UserWidget, rootWidgetClass=&VerticalBox, widgetCount=3}
 widgets [
  VBMessage {widgetClassPath=&VerticalBox, widget=VBMessage}
  Text_FlavorTitle {
   widgetClassPath=&CommonTextBlock
   bIsVariable=true
   parent=VBMessage
   widget=Text_FlavorTitle
   properties {
    justification=Center
    text=No Results
    textTransformPolicy=ToUpper
    font {
     bForceMonospaced=false
     bMaterialIsStencil=false
     fontMaterial=null
     fontObject=&Burbank
     letterSpacing=0
     monospacedWidth=1
     size=32
     skewAmount=0.2
     typefaceFontName=BurbankRegular-Black
     outlineSettings {
      bApplyOutlineToDropShadows=false
      bMiteredCorners=false
      bSeparateFillAlpha=false
      outlineMaterial=null
      outlineSize=0
      outlineColor=<rgba 0 0 0 1>
     }
    }
    shadowOffset=<0 0>
   }
   slot {class=VerticalBoxSlot, properties {horizontalAlignment=HAlign_Center, padding=<ltrb 0 0 0 8>}}
  }
  Description {
   widgetClassPath=&CommonTextBlock
   bIsVariable=true
   parent=VBMessage
   widget=Description
   properties {
    justification=Center
    renderOpacity=0.7
    text=No Results
    textTransformPolicy=ToUpper
    wrapTextAt=800
    font {
     bForceMonospaced=false
     bMaterialIsStencil=false
     fontMaterial=null
     fontObject=&Burbank
     letterSpacing=0
     monospacedWidth=1
     size=32
     skewAmount=0.2
     typefaceFontName=BurbankRegular-Black
     outlineSettings {
      bApplyOutlineToDropShadows=false
      bMiteredCorners=false
      bSeparateFillAlpha=false
      outlineMaterial=null
      outlineSize=0
      outlineColor=<rgba 0 0 0 1>
     }
    }
    shadowOffset=<0 0>
   }
   slot {class=VerticalBoxSlot, properties {horizontalAlignment=HAlign_Center}}
  }
 ]
}
