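# Discovery browser row widget, derived from WBP_Row_Poly_Base, that shows a billboard tile in a grid below a row header.
# SetVMData casts each item to an activity or surface-tile view model, then creates or reuses the matching cached tile and updates it.
# The tile receives focus, description visibility from the provider view model, and forwards outer-row navigation through the row's dispatcher.
# Entry release calls StopVisibilityOnEntryReleased, and Destruct unbinds tile navigation events.
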WidgetBlueprint /DiscoveryBrowser/Rows/WBP_Row_Billboard_Wrapper
graphs [
 CreateBillboardActivityTile {nodes [
   n1 = CreateBillboardActivityTile!() -> (then, FortActivityViewmodel:Discover Activity View Model Object Reference) {is=title K2Node_FunctionEntry}
   > n2 = Get!(execute:Exec) -> (then:Exec, else:Exec, WBP_DiscoverTileBillboard:WBP Discovery Tile Billboard Object Reference)
   n3 = Create Widget!(
    execute:Exec=@n5.then
    Class:User Widget Class Reference=/DiscoveryBrowser/Tiles/Billboard/WBP_Discovery_Tile_Billboard.WBP_Discovery_Tile_Billboard_C
    OwningPlayer:Player Controller Object Reference
    FortActivityViewModel:Discover Activity View Model Object Reference
    Style:Common Button Style Class Reference=/FortUILibrary/UIKit/Buttons/Blocks/ButtonStyle_UIKit_Empty.ButtonStyle_UIKit_Empty_C
    bApplyAlphaOnDisable:bool=True
    bLocked:bool=False
    bSelectable:bool=False
    bShouldSelectUponReceivingFocus:bool=False
    bInteractableWhenSelected:bool=False
    bToggleable:bool=False
    bTriggerClickedAfterSelection:bool=False
    bDisplayInputActionWhenNotInteractable:bool=True
    bHideInputActionWithKeyboard:bool=False
    bShouldUseFallbackDefaultInputAction:bool=True
    bRequiresHold:bool=False
    HoldData:Common UIHold Data Class Reference
    ClickMethod:EButtonClickMethod Enum=DownAndUp
    InputPriority:int=0
    TriggeringInputAction:Data Table Row Handle Structure
   ) -> (then:Exec, ReturnValue:WBP Discovery Tile Billboard Object Reference)
   > n4 = Set WBP_DiscoverTileBillboard(WBP_DiscoverTileBillboard:WBP Discovery Tile Billboard Object Reference=@n3) -> (Output_Get:WBP Discovery Tile Billboard Object Reference)
   OnActivityViewModelSet!(
    execute=[@n6.then, @n2.then]
    self:WBP Discovery Tile Billboard Object Reference=[Get WBP_DiscoverTileBillboard() -> (WBP_DiscoverTileBillboard:WBP Discovery Tile Billboard Object Reference), @n4]
    ActivityViewModel:Discover Activity View Model Object Reference=Get FortActivityViewmodel() -> (FortActivityViewmodel:Discover Activity View Model Object Reference)
   ) -> (then)
   > SetDesiredFocusWidget!(execute, self:User Widget Object Reference, Widget:Widget Object Reference=@n4) -> (then, ReturnValue:bool)
   > Bind Event to On Navigate to New Outer Row(self:WBP Discovery Tile Billboard Object Reference=@n4, Delegate=Create Event()) {is=title K2Node_AddDelegate}
   > SetDescriptionVisibility!(
    execute
    self
    Widget:Widget Object Reference=Get Billboard_Description!(self:WBP Discovery Tile Billboard Object Reference=Get WBP_DiscoverTileBillboard()) -> (Billboard_Description:Common Text Object Reference)
   ) -> (then)
   n5 = Remove Widget!(execute:Exec=@n2.else, Widget:Widget Object Reference=@n2) -> (then:Exec) {is=title K2Node_MacroInstance}
   Reroute Node(InputPin:Discover Activity View Model Object Reference=@n1) -> (OutputPin:Discover Activity View Model Object Reference)
   n6 = SetContent!(execute=@n4.then, self, Content:Widget Object Reference=@n4) -> (then)
  ]}
 CreateBillboardSurfaceTile {nodes [
   n1 = CreateBillboardSurfaceTile!() -> (then, FortDiscoverSurfaceTileIPViewmodel:Fort Discover Surface Tile Item VM Object Reference) {is=title K2Node_FunctionEntry}
   > n2 = Get!(execute) -> (then, else, WBP_DiscoverIPTileBillboard:WBP Discovery IP Tile Billboard Object Reference)
   n3 = OnDiscoverSurfaceTileIPViewModelSet!(
    execute=[@n7.then, @n2.then]
    self:WBP Discovery IP Tile Billboard Object Reference=Get WBP_DiscoverIPTileBillboard() -> (WBP_DiscoverIPTileBillboard:WBP Discovery IP Tile Billboard Object Reference)
    SurfaceTileViewModel:Fort Discover Surface Tile Item VM Object Reference=Get FortDiscoverSurfaceTileIPViewmodel() -> (FortDiscoverSurfaceTileIPViewmodel:Fort Discover Surface Tile Item VM Object Reference)
   ) -> (then)
   n4 = Create Widget!(
    execute=@n6.then
    Class=/DiscoveryBrowser/Tiles/Billboard/WBP_Discovery_IP_Tile_Billboard.WBP_Discovery_IP_Tile_Billboard_C
    OwningPlayer
    FortDiscoverSurfaceTileItemVM:Fort Discover Surface Tile Item VM Object Reference
    Style=&ButtonStyle_UIKit_Empty_C
    bApplyAlphaOnDisable=True
    bLocked=False
    bSelectable=False
    bShouldSelectUponReceivingFocus=False
    bInteractableWhenSelected=False
    bToggleable=False
    bTriggerClickedAfterSelection=False
    bDisplayInputActionWhenNotInteractable=True
    bHideInputActionWithKeyboard=False
    bShouldUseFallbackDefaultInputAction=True
    bRequiresHold=False
    HoldData
    ClickMethod=DownAndUp
    InputPriority=0
    TriggeringInputAction
   ) -> (then, ReturnValue:WBP Discovery IP Tile Billboard Object Reference)
   > n5 = Set WBP_DiscoverIPTileBillboard(WBP_DiscoverIPTileBillboard:WBP Discovery IP Tile Billboard Object Reference=@n4) -> (Output_Get:WBP Discovery IP Tile Billboard Object Reference)
   SetDesiredFocusWidget(execute=@n3.then, Widget=@n5)
   > Bind Event to On Navigate to New Outer Row(self:WBP Discovery IP Tile Billboard Object Reference=@n5, Delegate=Create Event())
   > SetDescriptionVisibility(Widget=Get Billboard_Description(self:WBP Discovery IP Tile Billboard Object Reference=Get WBP_DiscoverIPTileBillboard()))
   n6 = Remove Widget(execute=@n2.else, Widget=@n2)
   Reroute Node(InputPin:Fort Discover Surface Tile Item VM Object Reference=@n1) -> (OutputPin:Fort Discover Surface Tile Item VM Object Reference)
   n7 = SetContent(execute=@n5.then, Content=@n5)
  ]}
 EventGraph {nodes [
   n1 = OnNavigateToNewRow!() -> (
    OutputDelegate
    then
    IsDirectionUp:bool=~
    PreviousSubrowSelectedIndex:int
    PreviousNumItemsPerSubrow:int
    PreviousRowWasViewAllTileSelected:bool=~
   ) {is=title K2Node_CustomEvent}
   > Call On Navigate to New Outer Row!(
    execute:Exec
    self:WBP Row Poly Base Object Reference
    IsDirectionUp:bool=@n1.IsDirectionUp
    PreviousSubrowSelectedIndex:int=@n1.PreviousSubrowSelectedIndex
    PreviousNumItemsPerSubrow:int=@n1.PreviousNumItemsPerSubrow
    PreviousRowWasViewAllTileSelected:bool=@n1.PreviousRowWasViewAllTileSelected
   ) -> (then:Exec) {is=title K2Node_CallDelegate}
   Event On Entry Released() {is=title K2Node_Event}
   > StopVisibilityOnEntryReleased!(
    execute
    self:WBP Row Poly Base Object Reference
    InProviderVM:Discover Provider View Model Object Reference=Get FortDiscoverProviderViewModel() -> (FortDiscoverProviderViewModel:Discover Provider View Model Object Reference)
   ) -> (then)
   n2 = Event OnListItemObjectSet()
   > n3 = Cast To FortDiscoverProviderViewModel!(execute, Object=@n2.ListItemObject) -> (then, CastFailed, AsDiscover Provider View Model:Discover Provider View Model Object Reference) {is=title K2Node_DynamicCast}
   > Set FortDiscoverProviderViewModel(FortDiscoverProviderViewModel:Discover Provider View Model Object Reference=@n3) -> (Output_Get:Discover Provider View Model Object Reference)
   Event Destruct()
   > UnbindEvents()
  ]}
 GetNewSubrowIndexData {nodes [
   GetNewSubrowIndexData() {is=title K2Node_FunctionEntry}
   > Return Node!(
    execute
    NewSubrowLinesCount:int=0
    NewSubrowNumItemsPerRow:int=0
    DisplayedEntries:int=0
    NewSubrowFirstVisibleItemIndex:int=0
    SelectedItemIndex:int=0
   )
  ]}
 SetContent {nodes [
   n1 = SetContent!() -> (then, Content:Widget Object Reference) {is=title K2Node_FunctionEntry}
   > n2 = AddChildToGrid!(
    execute
    self:Grid Panel Object Reference=Get GP_Root() -> (GP_Root:Grid Panel Object Reference)
    Content:Widget Object Reference=@n1
    InRow:int=1
    InColumn:int=0
   ) -> (then, ReturnValue:Grid Slot Object Reference)
   > SetRowSpan!(execute, self:Grid Slot Object Reference=@n2, InRowSpan:int=1) -> (then)
   > SetHorizontalAlignment(self:Grid Slot Object Reference=@n2, InHorizontalAlignment=HAlign_Fill)
   > SetVerticalAlignment(self:Grid Slot Object Reference=@n2, InVerticalAlignment=VAlign_Top)
   > SetPadding!(execute, self:Grid Slot Object Reference=@n2, InPadding:Margin Structure=Get TilePadding() -> (TilePadding:Margin Structure)) -> (then)
  ]}
 SetDescriptionVisibility {nodes [
   n1 = SetDescriptionVisibility!() -> (then, Widget:Widget Object Reference) {is=title K2Node_FunctionEntry}
   n2 = Get FortDiscoverProviderViewModel()
   > SetVisibility(
    self=@n1
    InVisibility="To Visibility (Boolean)"!(
     bIsVisible:bool=Get bShowActivityDescription!(self:Discover Provider View Model Object Reference=@n2) -> (bShowActivityDescription:bool)
     TrueVisibility:ESlateVisibility Enum=SelfHitTestInvisible
     FalseVisibility:ESlateVisibility Enum=Collapsed
    ) -> (ReturnValue:ESlateVisibility Enum=Visible)
   )
  ]}
 SetVMData {nodes [
   n1 = SetVMData!() -> (then, Items:Array of Discover Tile Item View Model Object References) {is=title K2Node_FunctionEntry}
   > n2 = For Each Loop(Array:Array of Discover Tile Item View Model Object References=@n1) -> (Array Element:Discover Tile Item View Model Object Reference)
   > n3 = Cast To FortActivityViewModel!(execute, Object=@n2) -> (then, CastFailed, AsDiscover Activity View Model:Discover Activity View Model Object Reference) {is=title K2Node_DynamicCast}
   > CreateBillboardActivityTile!(execute, self, FortActivityViewmodel:Discover Activity View Model Object Reference=@n3) -> (then) {is=title K2Node_CallFunction}
   n4 = Cast To FortDiscoverSurfaceTileItemVM!(execute=@n3.CastFailed, Object=@n2) -> (then, CastFailed, AsFort Discover Surface Tile Item VM:Fort Discover Surface Tile Item VM Object Reference) {is=title K2Node_DynamicCast}
   > CreateBillboardSurfaceTile!(execute, self, FortDiscoverSurfaceTileIPViewmodel:Fort Discover Surface Tile Item VM Object Reference=@n4) -> (then) {is=title K2Node_CallFunction}
  ]}
 UnbindEvents {nodes [
   UnbindEvents() {is=title K2Node_FunctionEntry}
   > n1 = Get!(execute) -> (then, else, WBP_DiscoverTileBillboard:WBP Discovery Tile Billboard Object Reference)
   > n2 = Unbind Event from On Navigate to New Outer Row(self:WBP Discovery Tile Billboard Object Reference=@n1, Delegate=Create Event()) {is=title K2Node_RemoveDelegate}
   n3 = Get!(execute=[@n2.then, @n1.else]) -> (then, else, WBP_DiscoverIPTileBillboard:WBP Discovery IP Tile Billboard Object Reference)
   > Unbind Event from On Navigate to New Outer Row(self:WBP Discovery IP Tile Billboard Object Reference=@n3, Delegate=Create Event())
  ]}
]
variables [
 TilePadding {type=Margin, default=<ltrb 80 24 80 48>}
 WBP_DiscoverIPTileBillboard {type=WBP_Discovery_IP_Tile_Billboard_C, default=null}
 WBP_DiscoverTileBillboard {type=WBP_Discovery_Tile_Billboard_C, default=null}
]
widgets {
 info {
  inheritedWidgetCount=0
  namedSlotCount=1
  parentClass=&/WBP_Row_Poly_Base.WBP_Row_Poly_Base_C
  rootWidgetClass=null
  widgetCount=2
 }
 widgets [
  GP_Root {widgetClassPath=&GridPanel, bIsVariable=true, widget=GP_Root, properties {columnFill [1], rowFill [0,1]}}
  WBP_Row_Header {
   widgetClassPath=&/WBP_Row_Header
   bIsVariable=true
   parent=GP_Root
   widget=WBP_Row_Header
   slot {class=GridSlot, properties {rowSpan=1, verticalAlignment=VAlign_Top}}
  }
 ]
}
