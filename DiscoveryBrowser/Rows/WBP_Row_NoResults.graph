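# Discovery browser row widget shown for an empty category, with blank placeholder tiles and an empty-state message behind a focusable, text-less button.
# SetPanelType substring-matches the panel name to pick the message, falling back to the provider view model's display name, and builds one row of circular creator, wide collection or standard tiles.
# Tile entry generate/release delegates, bound while constructed, drive resizing the button's SizeBox to span the displayed tiles.
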WidgetBlueprint /DiscoveryBrowser/Rows/WBP_Row_NoResults
graphs [
 EventGraph {nodes [
   OnCalculateOutlineWidthWhenEntryChanges!() -> (OutputDelegate, then, Widget:User Widget Object Reference) {is=title K2Node_CustomEvent}
   n1 = Get TileView_BlankTiles!(self:WBP Discovery Blank Tiles Rounded Corners Object Reference=Get WBP_Discovery_BlankTiles_RoundedCorners() -> (WBP_Discovery_BlankTiles_RoundedCorners:WBP Discovery Blank Tiles Rounded Corners Object Reference)) -> (TileView_BlankTiles:Tile View Object Reference)
   n2 = Length(TargetArray:Array of User Widget Object References=GetDisplayedEntryWidgets!(self:List View Base Object Reference=@n1) -> (ReturnValue:Array of User Widget Object References))
   n3 = GetEntryWidth!(self:Tile View Object Reference=@n1) -> (ReturnValue:float)
   n4 = GetHorizontalEntrySpacing!(self:List View Object Reference=@n1) -> (ReturnValue:float)
   > SetWidthOverride(
    self=Get SizeBox_MatchWidthOfTileView() -> (SizeBox_MatchWidthOfTileView:Size Box Object Reference)
    InWidthOverride=float + float(A=float * float(A=@n4, B:int=@n2) {is=title K2Node_PromotableOperator}, B=float * float(A=@n3, B=@n2))
   )
   Event Construct()
   n5 = Get TileView_BlankTiles(self=Get WBP_Discovery_BlankTiles_RoundedCorners())
   > Bind Event to On Entry Generated(self:List View Base Object Reference=@n5, Delegate=Create Event()) {is=title K2Node_AddDelegate}
   > Bind Event to On Entry Released(self:List View Base Object Reference=@n5, Delegate=Create Event()) {is=title K2Node_AddDelegate}
   > SetDesiredFocusWidget!(
    execute
    self:User Widget Object Reference
    Widget:Widget Object Reference=Get WBP_Discover_Button_Generic() -> (WBP_Discover_Button_Generic:WBP Discover Button Generic Object Reference)
   ) -> (then, ReturnValue:bool)
   Event Destruct()
   > Unbind all Events from On Entry Generated!(execute:Exec, self:List View Base Object Reference=@n5) -> (then:Exec) {is=title K2Node_ClearDelegate}
   > Unbind all Events from On Entry Released!(execute:Exec, self:List View Base Object Reference=@n5) -> (then:Exec) {is=title K2Node_ClearDelegate}
   n6 = Event OnListItemObjectSet()
   > n7 = Cast To FortDiscoverProviderViewModel!(execute, Object=@n6.ListItemObject) -> (then, CastFailed, AsDiscover Provider View Model:Discover Provider View Model Object Reference) {is=title K2Node_DynamicCast}
   > Set FortDiscoverProviderViewModel(FortDiscoverProviderViewModel:Discover Provider View Model Object Reference=@n7) -> (Output_Get:Discover Provider View Model Object Reference)
  ]}
 SetPanelType {nodes [
   n1 = SetPanelType!() -> (then, PanelName:String) {is=title K2Node_FunctionEntry}
   > Set MakeCircularTiles(MakeCircularTiles:bool=false) -> (Output_Get:bool)
   > Set MakeCollectionTiles(MakeCollectionTiles:bool=false) -> (Output_Get:bool)
   > n2 = Branch(Condition="Not Equal (Integer)"(A=FindSubstring(SearchIn=@n1, Substring=creator), B=-1))
   n3 = Get WBP_Discovery_BlankTiles_RoundedCorners()
   > SetMessageTextOverrides!(
    execute
    self:WBP Discovery Blank Tiles Rounded Corners Object Reference=@n3
    InTitle:Text="LOCTEXT(\"No Creators Favorited\")"
    InSubTitle:Text="LOCTEXT(\"Creators that you favorite in Discover will appear here\")"
   ) -> (then)
   > n4 = Set MakeCircularTiles(MakeCircularTiles=true)
   n5 = Get MakeCircularTiles() -> (MakeCircularTiles:bool)
   n6 = Get WBP_Discovery_BlankTiles_RoundedCorners()
   CreateBlankTileEntries!(
    execute=[@n16.then, @n12.then, @n14.then, @n4.then, @n8.then, @n10.then]
    self:WBP Discovery Blank Tiles Rounded Corners Object Reference=@n6
    EntryShape:E_UI_BackplateCornerRadius Enum=Select(Option 0:E_UI_BackplateCornerRadius Enum=NewEnumerator1, Option 1:E_UI_BackplateCornerRadius Enum=NewEnumerator4, Index=@n5) -> (ReturnValue:E_UI_BackplateCornerRadius Enum)
    EntrySize:Vector 2D Structure=Select(
     Option 0:Vector 2D Structure=Select(Option 0=(X=256,Y=144), Option 1=(X=192,Y=192), Index=@n5)
     Option 1:Vector 2D Structure=(X=396,Y=144)
     Index=Get MakeCollectionTiles() -> (MakeCollectionTiles:bool)
    ) -> (ReturnValue:Vector 2D Structure)
   ) -> (then)
   > SetMaxRowsOverride!(execute, self:WBP Discovery Blank Tiles Rounded Corners Object Reference=@n6, MaxRows:int=1) -> (then)
   n7 = Branch(execute=@n2.else, Condition="Not Equal (Integer)"(A=FindSubstring(SearchIn=@n1, Substring=favoritecollaborations), B=-1))
   > SetMessageTextOverrides(self=@n3, InTitle="LOCTEXT(\"No Favorited Collaborations\")", InSubTitle="LOCTEXT(\"Collaborations you favorited will appear here\")")
   > n8 = Set MakeCollectionTiles(MakeCollectionTiles=true)
   n9 = Branch(execute=@n7.else, Condition="Not Equal (Integer)"(A=FindSubstring(SearchIn=@n1, Substring=favoritecategories), B=-1))
   > SetMessageTextOverrides(self=@n3, InTitle="LOCTEXT(\"No Favorited Categories\")", InSubTitle="LOCTEXT(\"Categories you favorited will appear here\")")
   > n10 = Set MakeCollectionTiles(MakeCollectionTiles=true)
   n11 = Branch(execute=@n9.else, Condition="Not Equal (Integer)"(A=FindSubstring(SearchIn=@n1, Substring=favorite), B=-1))
   > n12 = SetMessageTextOverrides(self=@n3, InTitle="LOCTEXT(\"No Games Favorited\")", InSubTitle="LOCTEXT(\"Games you favorite will appear here\")")
   n13 = Branch(execute=@n11.else, Condition="Not Equal (Integer)"(A=FindSubstring(SearchIn=@n1, Substring=played), B=-1))
   > n14 = SetMessageTextOverrides(self=@n3, InTitle="LOCTEXT(\"No Recently Played Games\")", InSubTitle="LOCTEXT(\"Games you have recently played will appear here\")")
   n15 = Get FortDiscoverProviderViewModel() -> (FortDiscoverProviderViewModel:Discover Provider View Model Object Reference)
   n16 = SetMessageTextOverrides(
    execute=@n13.else
    self=@n3
    InTitle=Format Text!(
     Format="LOCTEXT(\"No {PanelDisplayName}\")"
     PanelDisplayName:Text=Get PanelDisplayName!(self:Discover Provider View Model Object Reference=@n15) -> (PanelDisplayName:Text)
    ) -> (Result)
    InSubTitle="LOCTEXT(\"This category is currently empty\")"
   )
  ]}
]
variables []
widgets {
 info {
  inheritedWidgetCount=0
  namedSlotCount=1
  parentClass=&/WBP_Row_Poly_Base.WBP_Row_Poly_Base_C
  rootWidgetClass=null
  widgetCount=4
 }
 widgets [
  Overlay_0 {widgetClassPath=&Overlay, widget=Overlay_0}
  SizeBox_MatchWidthOfTileView {
   widgetClassPath=&SizeBox
   bIsVariable=true
   parent=Overlay_0
   widget=SizeBox_MatchWidthOfTileView
   properties {bOverride_WidthOverride=true, widthOverride=1728}
   slot {class=OverlaySlot, properties {verticalAlignment=VAlign_Fill, padding=<ltrb 64 28 0 28>}}
  }
  WBP_Discover_Button_Generic {
   widgetClassPath=/DiscoveryBrowser/Tiles/WBP_Discover_Button_Generic
   bIsVariable=true
   parent=SizeBox_MatchWidthOfTileView
   widget=WBP_Discover_Button_Generic
   properties {
    bHideInputAction=true
    backgroundMaterial=/FortUILibrary/UIKit/Buttons/MaterialInstances/CTAButtons/MI_UIKit_BgTab_Enhanced
    outlineMaterial=/FortUILibrary/UIKit/Buttons/MaterialInstances/CTAButtons/MI_UIKit_ButtonOutline
    text=""
    touchMethod=PreciseTap
    useTextBlock=false
   }
   slot {class=SizeBoxSlot}
  }
  WBP_Discovery_BlankTiles_RoundedCorners {
   widgetClassPath=/DiscoveryBrowser/WBP_Discovery_BlankTiles_RoundedCorners
   bIsVariable=true
   parent=Overlay_0
   widget=WBP_Discovery_BlankTiles_RoundedCorners
   slot {class=OverlaySlot, properties {horizontalAlignment=HAlign_Fill, padding=<ltrb 64 0 0 0>}}
  }
 ]
}
