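# Discovery browser row widget, derived from WBP_Row_Poly_Base, that wraps a horizontal scrolling tile list.
# It takes a provider view model as its list item, clears tiles when that changes, and requests next pages as entries initialize.
# Generated tiles share a nested selection view model, which touch input can deselect on row release.
# Up/down navigation broadcasts the visible-relative column for nearest-neighbor selection across rows, also relaying billboard tile navigation.
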WidgetBlueprint /DiscoveryBrowser/Rows/WBP_Row_Poly_HorizontalScrollWrapper
graphs [
 DoCustomNavigation {nodes [
   n1 = DoCustomNavigation!() -> (then, Navigation:EUINavigation Enum) {is=title K2Node_FunctionEntry}
   > n2 = Switch on EUINavigation!(execute, Selection:EUINavigation Enum=@n1) -> (Left:Exec, Right:Exec, Up:Exec, Down:Exec, Next:Exec, Previous:Exec, Invalid:Exec) {is=title K2Node_SwitchEnum}
   n3 = Get WBP_Row_HorizontalScrollingList() -> (WBP_Row_HorizontalScrollingList:WBP Row Horizontal Scrolling List Object Reference)
   n4 = Get ListView!(self:Fort Polymorphic List View Object Reference=@n3) -> (ListView:Common List View Object Reference)
   n5 = Length(TargetArray:Array of User Widget Object References=GetDisplayedEntryWidgets!(self:List View Base Object Reference=@n4) -> (ReturnValue:Array of User Widget Object References))
   n6 = GetFirstVisibleItemIndex!(self:WBP Row Horizontal Scrolling List Object Reference=@n3) -> (FirstVisibleItemIndex:int=0)
   n7 = GetIndexForItem!(
    self:List View Object Reference=@n4
    Item:Object Reference=Get Selected Item!(self:List View Object Reference=@n4) -> (ReturnValue:Object Reference) {is=title K2Node_CallFunction}
   ) -> (ReturnValue:int=0)
   Call On Navigate to New Outer Row!(
    execute:Exec=[@n2.Down, @n2.Up]
    self:WBP Row Poly Base Object Reference
    IsDirectionUp:bool="Equal (Enum)"(A:EUINavigation Enum=Up, B:EUINavigation Enum=@n1)
    PreviousSubrowSelectedIndex:int="Min (Integer)"(A=int - int(A=@n7, B=@n6), B=int - int(A=@n5)) -> (ReturnValue=0)
    PreviousNumItemsPerSubrow:int=@n5
    PreviousRowWasViewAllTileSelected:bool
   ) -> (then:Exec) {is=title K2Node_CallDelegate}
   > Return Node!(execute, ReturnValue:Widget Object Reference)
  ]}
 EventGraph {nodes [
   n1 = OnEntryInit!() -> (OutputDelegate, then, Item:Object Reference, Widget:User Widget Object Reference) {is=title K2Node_CustomEvent}
   n2 = Get WBP_Row_HorizontalScrollingList()
   > CheckAndFetchNextPage!(
    execute
    self:Discover Provider View Model Object Reference=Get FortDiscoverProviderViewModel() -> (FortDiscoverProviderViewModel:Discover Provider View Model Object Reference)
    SelectionIndex:int=GetIndexForItem(self=Get ListView(self=@n2), Item=@n1.Item) -> (ReturnValue=0)
    bForceFetchNextPage:bool=false
   ) -> (then)
   n3 = HandleInputMethodChanged!() -> (OutputDelegate, then, bNewInputType:ECommonInputType Enum) {is=title K2Node_CustomEvent}
   > n4 = Switch on ECommonInputType!(execute, Selection:ECommonInputType Enum=@n3.bNewInputType) -> (MouseAndKeyboard:Exec, Gamepad:Exec, Touch:Exec, Count:Exec) {is=title K2Node_SwitchEnum}
   Clear Selection(execute=@n4.Touch, self:List View Object Reference=Get ListView(self=Get WBP_Row_HorizontalScrollingList()))
   n5 = OnTileItemGenerated!() -> (OutputDelegate, then, Widget:User Widget Object Reference) {is=title K2Node_CustomEvent}
   > SetNestedButtonGroupSelectionVM!(
    execute:Exec
    self:Object Reference=@n5.Widget
    InVM:Fort Nested Button Group Selection VM Object Reference=Get FortNestedButtonGroupSelectionVM() -> (FortNestedButtonGroupSelectionVM:Fort Nested Button Group Selection VM Object Reference)
   ) -> (then:Exec) {is=title K2Node_Message}
   > n6 = Cast To WBP_Discovery_Tile_Billboard!(execute, Object=@n5.Widget) -> (then, CastFailed, AsWBP Discovery Tile Billboard:WBP Discovery Tile Billboard Object Reference) {is=title K2Node_DynamicCast}
   n7 = NavigateToNewOuterRow!() -> (
    OutputDelegate
    then
    IsDirectionUp:bool=~
    PreviousSubrowSelectedIndex:int
    PreviousNumItemsPerSubrow:int
    PreviousRowWasViewAllTileSelected:bool=~
   ) {is=title K2Node_CustomEvent}
   Bind Event to On Navigate to New Outer Row(execute=@n6.then, self:WBP Discovery Tile Billboard Object Reference=@n6, Delegate=@n7) {is=title K2Node_AddDelegate}
   Event Construct()
   n8 = Get ListView(self=Get WBP_Row_HorizontalScrollingList())
   > Bind Event to On Entry Initialized(self:List View Object Reference=@n8, Delegate=Create Event()) {is=title K2Node_AddDelegate}
   n9 = Create Event()
   n10 = Get CommonInputSubsystem()
   > Bind Event to On Input Method Changed(self=@n10, Delegate=@n9)
   n11 = Create Event()
   n12 = Get ListView(self=Get WBP_Row_HorizontalScrollingList())
   > Bind Event to On Entry Generated(self:List View Base Object Reference=@n12, Delegate=@n11) {is=title K2Node_AddDelegate}
   > SetDesiredFocusWidget!(execute, self:User Widget Object Reference, Widget:Widget Object Reference=Get WBP_Row_HorizontalScrollingList()) -> (then, ReturnValue:bool)
   Event Destruct()
   > Unbind all Events from On Entry Initialized!(execute:Exec, self:List View Object Reference=@n8) -> (then:Exec) {is=title K2Node_ClearDelegate}
   > Unbind Event from On Input Method Changed(self=@n10, Delegate=@n9)
   > Unbind Event from On Entry Generated(self:List View Base Object Reference=@n12, Delegate=@n11) {is=title K2Node_RemoveDelegate}
   n13 = Event NavigateToNearestNeighborSubrowIndex!() -> (OutputDelegate, then, Was Last Selection View All:bool, NearestNeighborIndex:int=0) {is=title K2Node_Event}
   n14 = Get ListView(self=Get WBP_Row_HorizontalScrollingList())
   > SetSelectedIndex!(execute, self:List View Object Reference=@n14, Index:int=@n13.NearestNeighborIndex) -> (then)
   > NavigateToIndex!(execute, self:List View Object Reference=@n14, Index:int=@n13.NearestNeighborIndex) -> (then)
   Event On Entry Released() {is=title K2Node_Event}
   > StopVisibilityOnEntryReleased!(
    execute
    self:WBP Row Poly Base Object Reference
    InProviderVM:Discover Provider View Model Object Reference=Get FortDiscoverProviderViewModel()
   ) -> (then)
   > n15 = Cast To UserWidget!(
    execute
    Object:Fort Nested Button Group VMItem Interface=Get CurrentSelectedButton!(self:Fort Nested Button Group Selection VM Object Reference=Get FortNestedButtonGroupSelectionVM()) -> (CurrentSelectedButton:Fort Nested Button Group VMItem Interface)
   ) -> (then, CastFailed, AsUser Widget:User Widget Object Reference) {is=title K2Node_DynamicCast}
   > Is Valid(InputObject=Get FortNestedButtonGroupSelectionVM())
   n16 = GetDisplayedEntryWidgets(self=Get ListView(self=Get WBP_Row_HorizontalScrollingList()))
   n17 = GetOwningPlayer()
   > Branch(Condition=AND Boolean(
     A=IsUsingTouch(PlayerController=@n17)
     B=Contains Item(TargetArray:Array of User Widget Object References=@n16, ItemToFind:"User Widget Object Reference (by ref)"=@n15)
    ))
   n18 = Get FortNestedButtonGroupSelectionVM()
   > SetIsSelected!(
    execute
    self:Fort Nested Button Group Selection VM Object Reference=@n18
    InButton:Fort Nested Button Group VMItem Interface=Get CurrentSelectedButton(self=@n18)
    bInSelected:bool=false
   ) -> (then)
   n19 = Event OnListItemObjectSet()
   > n20 = Cast To FortDiscoverProviderViewModel!(execute, Object=@n19.ListItemObject) -> (then, CastFailed, AsDiscover Provider View Model:Discover Provider View Model Object Reference) {is=title K2Node_DynamicCast}
   > K2Node_Composite_1 = Clear Listview If View Model Changed!(execute, New Fort Discover Provider View Model:Object Reference=@n20) -> (exec:Exec) {is=title K2Node_Composite}
   > Set FortDiscoverProviderViewModel(FortDiscoverProviderViewModel:Discover Provider View Model Object Reference=@n20) -> (Output_Get:Discover Provider View Model Object Reference)
   n21 = Event On Item Selection Changed!() -> (OutputDelegate, then, bIsSelected:bool) {is=title K2Node_Event}
   > Set IsRowSelected!(execute, IsRowSelected:bool=@n21.bIsSelected, self:WBP Row Horizontal Scrolling List Object Reference=Get WBP_Row_HorizontalScrollingList()) -> (then, Output_Get:bool)
   n22 = Event ComputeRowVisibilityBase!() -> (OutputDelegate, then, Top Left:Vector 2D Structure, Bottom Right:Vector 2D Structure) {is=title K2Node_Event}
   > ComputeRowVisibility!(
    execute
    ListView:List View Object Reference=Get ListView(self=Get WBP_Row_HorizontalScrollingList())
    Provider VM:Discover Provider View Model Object Reference=Get FortDiscoverProviderViewModel()
    TopLeft:Vector 2D Structure=@n22.Top Left
    BottomRight:Vector 2D Structure=@n22.Bottom Right
   ) -> (then)
   Event UpdateRowPadding() {is=title K2Node_Event}
   n23 = Get bIsTopTenRow!(self:Discover Provider View Model Object Reference=Get FortDiscoverProviderViewModel()) -> (bIsTopTenRow:bool)
   > UpdateRowPadding!(execute, self:WBP Row Horizontal Scrolling List Object Reference=Get WBP_Row_HorizontalScrollingList(), IsTopTenRow:bool=@n23) -> (then)
   > SetLeftArrowPadding!(execute, self:WBP Row Horizontal Scrolling List Object Reference=Get WBP_Row_HorizontalScrollingList(), IsTop10:bool=@n23) -> (then)
   n24 = Event SetArrowBackgroundFadeVisibility!() -> (OutputDelegate, then, isVisible:bool) {is=title K2Node_Event}
   > SetArrowBackgroundFadeVisibility!(execute, self:WBP Row Horizontal Scrolling List Object Reference=Get WBP_Row_HorizontalScrollingList(), IsVisible:bool=@n24.isVisible) -> (then)
   n25 = Event UpdateNestedButtonGroupVM!() -> (OutputDelegate, then, NestedVM:Fort Nested Button Group Selection VM Object Reference) {is=title K2Node_Event}
   > Set FortNestedButtonGroupSelectionVM(FortNestedButtonGroupSelectionVM:Fort Nested Button Group Selection VM Object Reference=@n25.NestedVM) -> (Output_Get:Fort Nested Button Group Selection VM Object Reference)
   n26 = Event SetOuterItemRowIndex!() -> (OutputDelegate, then, Index:int=0) {is=title K2Node_Event}
   > Set OuterItemRowIndex!(execute, OuterItemRowIndex:int=@n26.Index, self:WBP Row Horizontal Scrolling List Object Reference=Get WBP_Row_HorizontalScrollingList()) -> (then, Output_Get:int=0)
   Call On Navigate to New Outer Row(
    execute=@n7.then
    IsDirectionUp=@n7.IsDirectionUp
    PreviousSubrowSelectedIndex=@n7.PreviousSubrowSelectedIndex
    PreviousNumItemsPerSubrow=@n7.PreviousNumItemsPerSubrow
    PreviousRowWasViewAllTileSelected=@n7.PreviousRowWasViewAllTileSelected
   )
  ]}
 Clear Listview If View Model Changed {
  path=EventGraph.K2Node_Composite_1.Clear Listview If View Model Changed
  nodes [
   n1 = Inputs!() -> (execute, New Fort Discover Provider View Model:Object Reference)
   > n2 = Branch(Condition="Not Equal (Object)"(A:Object Reference=@n1, B:Object Reference=Get FortDiscoverProviderViewModel()) -> (ReturnValue:bool=~) {is=title K2Node_PromotableOperator})
   > n3 = ClearListItems(self=Get ListView(self=Get WBP_Row_HorizontalScrollingList()))
   Outputs!(exec:Exec=[@n3.then, @n2.else])
  ]
 }
 GetNewSubrowIndexData {nodes [
   GetNewSubrowIndexData() {is=title K2Node_FunctionEntry}
   n1 = Get WBP_Row_HorizontalScrollingList()
   n2 = Length(TargetArray=GetDisplayedEntryWidgets(self=Get ListView(self=@n1)))
   n3 = Get ListView(self=Get WBP_Row_HorizontalScrollingList())
   > Return Node!(
    execute
    NewSubrowLinesCount:int=1
    NewSubrowNumItemsPerRow:int=@n2
    DisplayedEntries:int=@n2
    NewSubrowFirstVisibleItemIndex:int=GetFirstVisibleItemIndex(self=Get WBP_Row_HorizontalScrollingList()) -> (FirstVisibleItemIndex=0)
    SelectedItemIndex:int=GetIndexForItem(self=@n3, Item=Get Selected Item(self=@n3)) -> (ReturnValue=0)
   )
  ]}
]
variables [FortNestedButtonGroupSelectionVM {type=FortNestedButtonGroupSelectionVM, default=null}]
widgets {
 info {
  inheritedWidgetCount=0
  namedSlotCount=1
  parentClass=&/WBP_Row_Poly_Base.WBP_Row_Poly_Base_C
  rootWidgetClass=null
  widgetCount=1
 }
 widgets [WBP_Row_HorizontalScrollingList {
   widgetClassPath=&/WBP_Row_HorizontalScrollingList
   bIsVariable=true
   widget=WBP_Row_HorizontalScrollingList
   properties {bIsFocusable=true}
  }]
}
