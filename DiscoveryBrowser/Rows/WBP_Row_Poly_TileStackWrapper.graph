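# Discovery browser list entry widget that wraps WBP_Row_PanelList's tile view and View All button as one row, backed by a discover provider view model.
# Custom up/down navigation computes sub-rows from items per row and hands off to adjacent rows through parent delegates.
# Focus and hover drive tile selection; focus also updates outer list scrolling.
# It passes a nested button selection view model to generated tiles and forwards visibility, padding and View All requests.
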WidgetBlueprint /DiscoveryBrowser/Rows/WBP_Row_Poly_TileStackWrapper
graphs [
 DoCustomNavigation {nodes [
   n1 = DoCustomNavigation!() -> (then, Navigation:EUINavigation Enum) {is=title K2Node_FunctionEntry}
   > n2 = Switch on EUINavigation!(execute, Selection:EUINavigation Enum=@n1) -> (Left:Exec, Right:Exec, Up:Exec, Down:Exec, Next:Exec, Previous:Exec, Invalid:Exec) {is=title K2Node_SwitchEnum}
   n3 = Branch(execute=@n2.Down, Condition=ShouldNavigateDownToNextOuterRow!() -> (ShouldNavigateToLowerList:bool))
   n4 = Get WBP_Row_PanelList() -> (WBP_Row_PanelList:WBP Row Panel List Object Reference)
   n5 = Get TileView!(self:Fort Polymorphic Tile View Object Reference=@n4) -> (TileView:Fort Tile View Object Reference)
   Call On Navigate to New Outer Row!(
    execute:Exec=[@n3.then, @n2.Up]
    self:WBP Row Poly Base Object Reference
    IsDirectionUp:bool="Equal (Enum)"(A:EUINavigation Enum=Up, B:EUINavigation Enum=@n1)
    PreviousSubrowSelectedIndex:int=GetSelectedItemIndex!() -> (SelectedItemIndex:int=0)
    PreviousNumItemsPerSubrow:int=GetNumberItemsPerRow!(self:Fort Tile View Object Reference=@n5) -> (ReturnValue:int=0)
    PreviousRowWasViewAllTileSelected:bool=Get IsViewAllSelected!(self:WBP Row Panel List Object Reference=@n4) -> (IsViewAllSelected:bool)
   ) -> (then:Exec) {is=title K2Node_CallDelegate}
   > OnRowUnselected(self:WBP Row Panel List Object Reference=@n4)
   > Return Node!(execute, ReturnValue:Widget Object Reference)
   Set Selected Item!(
    execute=@n3.else
    self:List View Object Reference=@n5
    Item:Object Reference=GetItemAt!(self:List View Object Reference=@n5, Index:int=GetLastItemIndex!() -> (LastItemIndex:int=0)) -> (ReturnValue:Object Reference)
   ) -> (then) {is=title K2Node_CallFunction}
   > Navigate To Item!(
    execute
    self:List View Object Reference=@n5
    Item:Object Reference=Get Selected Item!(self:List View Object Reference=@n5) -> (ReturnValue:Object Reference) {is=title K2Node_CallFunction}
   ) -> (then)
   > Return Node()
  ]}
 EventGraph {nodes [
   OnViewAllButtonFocused() {is=title K2Node_CustomEvent}
   > Call On Update Outer List Scroll Offset!(
    execute:Exec
    self:WBP Row Poly Base Object Reference
    RowIndex:int=Get OuterItemRowIndex!(self:WBP Row Panel List Object Reference=Get WBP_Row_PanelList()) -> (OuterItemRowIndex:int=0)
    SubRowIndex:int=GetSelectedSubRowIndex!() -> (SelectedSubRowZeroBased:int=0)
    IsViewAll:bool=true
   ) -> (then:Exec) {is=title K2Node_CallDelegate}
   n1 = OnItemFocused!() -> (OutputDelegate, then, ItemFocused:int) {is=title K2Node_CustomEvent}
   n2 = Get TileView(self=Get WBP_Row_PanelList())
   > SetSelectedIndex!(execute, self:List View Object Reference=@n2, Index:int=@n1.ItemFocused) -> (then)
   > Call On Update Outer List Scroll Offset(
    RowIndex=Get OuterItemRowIndex(self=Get WBP_Row_PanelList()) -> (OuterItemRowIndex=0)
    SubRowIndex=GetSelectedSubRowIndex() -> (SelectedSubRowZeroBased=0)
   )
   n3 = OnItemHovered!() -> (OutputDelegate, then, Item:Object Reference, bIsHovered:bool=~) {is=title K2Node_CustomEvent}
   > Branch(Condition=@n3.bIsHovered)
   > SetSelectedIndex(self=@n2, Index=GetIndexForItem!(self:List View Object Reference=@n2, Item:Object Reference=@n3.Item) -> (ReturnValue:int=0))
   n4 = HandleInputMethodChanged!() -> (OutputDelegate, then, bNewInputType:ECommonInputType Enum) {is=title K2Node_CustomEvent}
   > n5 = Switch on ECommonInputType!(execute, Selection:ECommonInputType Enum=@n4.bNewInputType) -> (MouseAndKeyboard:Exec, Gamepad:Exec, Touch:Exec, Count:Exec) {is=title K2Node_SwitchEnum}
   Clear Selection(execute=@n5.Touch, self:List View Object Reference=Get TileView(self=Get WBP_Row_PanelList()))
   OnSendProviderViewModel() {is=title K2Node_CustomEvent}
   > OpenViewAllScreen!(
    execute
    self:WBP Row Panel List Object Reference=Get WBP_Row_PanelList()
    ProviderViewModel:Discover Provider View Model Object Reference=Get FortDiscoverProviderViewModel() -> (FortDiscoverProviderViewModel:Discover Provider View Model Object Reference)
   ) -> (then)
   n6 = OnTileItemGenerated!() -> (OutputDelegate, then, Widget:User Widget Object Reference) {is=title K2Node_CustomEvent}
   > SetNestedButtonGroupSelectionVM!(
    execute:Exec
    self:Object Reference=@n6.Widget
    InVM:Fort Nested Button Group Selection VM Object Reference=Get NestedButtonGroupSelectionVM() -> (NestedButtonGroupSelectionVM:Fort Nested Button Group Selection VM Object Reference)
   ) -> (then:Exec) {is=title K2Node_Message}
   Event Construct()
   n7 = Create Event()
   n8 = Get CommonInputSubsystem()
   > Bind Event to On Input Method Changed(self=@n8, Delegate=@n7)
   n9 = Get WBP_Row_PanelList()
   > Bind Event to On Tile Focus Gained(self:WBP Row Panel List Object Reference=@n9, Delegate=Create Event()) {is=title K2Node_AddDelegate}
   n10 = Get TileView(self=@n9)
   > Bind Event to On Item Is Hovered Changed(self:List View Object Reference=@n10, Delegate=Create Event()) {is=title K2Node_AddDelegate}
   > Bind Event to On Entry Generated(self:List View Base Object Reference=@n10, Delegate=Create Event()) {is=title K2Node_AddDelegate}
   > Bind Event to On View All Send Provider View Model(self:WBP Row Panel List Object Reference=@n9, Delegate=Create Event()) {is=title K2Node_AddDelegate}
   > Bind Event to On View All Focus Gained(self:WBP Row Panel List Object Reference=@n9, Delegate=Create Event()) {is=title K2Node_AddDelegate}
   > SetDesiredFocusWidget!(execute, self:User Widget Object Reference, Widget:Widget Object Reference=Get WBP_Row_PanelList()) -> (then, ReturnValue:bool)
   n11 = Event NavigateToNearestNeighborSubrowIndex!() -> (OutputDelegate, then, Was Last Selection View All:bool, NearestNeighborIndex:int=0) {is=title K2Node_Event}
   > Set Was Last Selection View All(Was Last Selection View All:bool=@n11.Was Last Selection View All) -> (Output_Get:bool)
   > SetSelectedIndex(self=Get TileView(self=Get WBP_Row_PanelList()), Index=@n11.NearestNeighborIndex)
   n12 = Get WBP_Row_PanelList()
   > n13 = Branch(Condition=AND Boolean(
     A=HasViewAllButton!(self:WBP Row Panel List Object Reference=@n12) -> (HasViewAll:bool)
     B=Get Was Last Selection View All() -> (Was Last Selection View All:bool)
    ))
   n14 = Get WBP_Row_PanelList()
   > Set IsViewAllSelected!(execute, IsViewAllSelected:bool=true, self:WBP Row Panel List Object Reference=@n14) -> (then, Output_Get:bool)
   > SetFocus(self=Get Button_ViewAll_PaddingHalfOfTileEntrySpacing!(self:WBP Row Panel List Object Reference=Get WBP_Row_PanelList()) -> (Button_ViewAll_PaddingHalfOfTileEntrySpacing:WBP Discover Button Generic Object Reference))
   NavigateToIndex!(execute=@n13.else, self:List View Object Reference=Get TileView(self=@n14), Index:int=@n11.NearestNeighborIndex) -> (then)
   n15 = Event SetOuterItemRowIndex!() -> (OutputDelegate, then, Index:int=0) {is=title K2Node_Event}
   > Set OuterItemRowIndex!(execute, OuterItemRowIndex:int=@n15.Index, self:WBP Row Panel List Object Reference=Get WBP_Row_PanelList()) -> (then, Output_Get:int=0)
   n16 = Event OnListItemObjectSet()
   > n17 = Cast To FortDiscoverProviderViewModel!(execute, Object=@n16.ListItemObject) -> (then, CastFailed, AsDiscover Provider View Model:Discover Provider View Model Object Reference) {is=title K2Node_DynamicCast}
   > Set FortDiscoverProviderViewModel(FortDiscoverProviderViewModel:Discover Provider View Model Object Reference=@n17) -> (Output_Get:Discover Provider View Model Object Reference)
   Event Destruct()
   > Unbind Event from On Input Method Changed(self=@n8, Delegate=@n7)
   > Unbind Event from On Entry Generated(self:List View Base Object Reference=Get TileView(self=Get WBP_Row_PanelList()), Delegate=Create Event()) {is=title K2Node_RemoveDelegate}
   Event On Entry Released() {is=title K2Node_Event}
   > StopVisibilityOnEntryReleased!(
    execute
    self:WBP Row Poly Base Object Reference
    InProviderVM:Discover Provider View Model Object Reference=Get FortDiscoverProviderViewModel()
   ) -> (then)
   > n18 = Cast To UserWidget!(
    execute
    Object:Fort Nested Button Group VMItem Interface=Get CurrentSelectedButton!(self:Fort Nested Button Group Selection VM Object Reference=Get NestedButtonGroupSelectionVM()) -> (CurrentSelectedButton:Fort Nested Button Group VMItem Interface)
   ) -> (then, CastFailed, AsUser Widget:User Widget Object Reference) {is=title K2Node_DynamicCast}
   > Is Valid(InputObject=Get NestedButtonGroupSelectionVM())
   n19 = GetDisplayedEntryWidgets!(self:List View Base Object Reference=Get TileView(self=Get WBP_Row_PanelList())) -> (ReturnValue:Array of User Widget Object References)
   n20 = GetOwningPlayer()
   > Branch(Condition=AND Boolean(
     A=IsUsingTouch(PlayerController=@n20)
     B=Contains Item(TargetArray:Array of User Widget Object References=@n19, ItemToFind:"User Widget Object Reference (by ref)"=@n18)
    ))
   n21 = Get NestedButtonGroupSelectionVM()
   > SetIsSelected!(
    execute
    self:Fort Nested Button Group Selection VM Object Reference=@n21
    InButton:Fort Nested Button Group VMItem Interface=Get CurrentSelectedButton(self=@n21)
    bInSelected:bool=false
   ) -> (then)
   n22 = Event ComputeRowVisibilityBase!() -> (OutputDelegate, then, Top Left:Vector 2D Structure, Bottom Right:Vector 2D Structure) {is=title K2Node_Event}
   > ComputeRowVisibility!(
    execute
    ListView:List View Object Reference=Get TileView(self=Get WBP_Row_PanelList())
    Provider VM:Discover Provider View Model Object Reference=Get FortDiscoverProviderViewModel()
    TopLeft:Vector 2D Structure=@n22.Top Left
    BottomRight:Vector 2D Structure=@n22.Bottom Right
   ) -> (then)
   Event UpdateRowPadding() {is=title K2Node_Event}
   > UpdateRowPadding!(
    execute
    self:WBP Row Panel List Object Reference=Get WBP_Row_PanelList()
    IsTopTenRow:bool=Get bIsTopTenRow!(self:Discover Provider View Model Object Reference=Get FortDiscoverProviderViewModel()) -> (bIsTopTenRow:bool)
   ) -> (then)
   n23 = Event UpdateNestedButtonGroupVM!() -> (OutputDelegate, then, NestedVM:Fort Nested Button Group Selection VM Object Reference) {is=title K2Node_Event}
   > Set NestedButtonGroupSelectionVM(NestedButtonGroupSelectionVM:Fort Nested Button Group Selection VM Object Reference=@n23.NestedVM) -> (Output_Get:Fort Nested Button Group Selection VM Object Reference)
   Event On Item Selection Changed!() -> (OutputDelegate, then, bIsSelected:bool) {is=title K2Node_Event}
  ]}
 GetLastItemIndex {nodes [
   GetLastItemIndex() {is=title K2Node_FunctionEntry}
   n1 = GetDisplayedEntryWidgets(self=Get TileView(self=Get WBP_Row_PanelList()))
   > Return Node!(execute, LastItemIndex:int=int - int(A=Length(TargetArray:Array of User Widget Object References=@n1)))
  ]}
 GetNewSubrowIndexData {nodes [
   GetNewSubrowIndexData() {is=title K2Node_FunctionEntry}
   n1 = Get TileView(self=Get WBP_Row_PanelList())
   n2 = Get TileView(self=Get WBP_Row_PanelList())
   > Return Node!(
    execute
    NewSubrowLinesCount:int=GetNumberOfDisplayedRows!() -> (NumberOfDisplayedRows:int=0)
    NewSubrowNumItemsPerRow:int=GetNumberItemsPerRow(self=@n1) -> (ReturnValue=0)
    DisplayedEntries:int=Length(TargetArray=GetDisplayedEntryWidgets(self=@n1))
    NewSubrowFirstVisibleItemIndex:int=0
    SelectedItemIndex:int=GetIndexForItem(self=@n2, Item=Get Selected Item(self=@n2)) -> (ReturnValue=0)
   )
  ]}
 GetNumberOfDisplayedRows {nodes [
   GetNumberOfDisplayedRows() {is=title K2Node_FunctionEntry}
   n1 = Get TileView(self=Get WBP_Row_PanelList())
   n2 = Length(TargetArray=GetDisplayedEntryWidgets(self=@n1))
   n3 = GetNumberItemsPerRow(self=@n1) -> (ReturnValue=0)
   > Return Node!(execute, NumberOfDisplayedRows:int=Ceil(A=float / float(A:int=@n2, B:int=@n3)) {is=title K2Node_CallFunction})
  ]}
 GetSelectedItemIndex {nodes [
   GetSelectedItemIndex() {is=title K2Node_FunctionEntry}
   n1 = Get TileView(self=Get WBP_Row_PanelList())
   > Return Node!(execute, SelectedItemIndex:int=GetIndexForItem(self=@n1, Item=Get Selected Item(self=@n1)) -> (ReturnValue=0))
  ]}
 GetSelectedSubRowIndex {nodes [
   GetSelectedSubRowIndex() {is=title K2Node_FunctionEntry}
   n1 = Get TileView(self=Get WBP_Row_PanelList())
   n2 = int / int(A=GetIndexForItem(self=@n1, Item=Get Selected Item(self=@n1)) -> (ReturnValue=0), B=GetNumberItemsPerRow(self=@n1) -> (ReturnValue=0))
   > Return Node!(execute, SelectedSubRowZeroBased:int=Floor(A="To Float (Integer)"(InInt=@n2)) {is=title K2Node_CallFunction})
  ]}
 ShouldNavigateDownToNextOuterRow {nodes [
   ShouldNavigateDownToNextOuterRow() {is=title K2Node_FunctionEntry}
   n1 = Get WBP_Row_PanelList()
   n2 = Get TileView(self=@n1)
   n3 = GetNumberItemsPerRow(self=@n2) -> (ReturnValue=0)
   n4 = int + int(A=GetIndexForItem(self=@n2, Item=Get Selected Item(self=@n2)) -> (ReturnValue=0), B=@n3)
   n5 = int - int(A=int * int(A=@n3, B=GetNumberOfDisplayedRows() -> (NumberOfDisplayedRows=0) {is=title K2Node_CallFunction}))
   > Return Node!(execute, ShouldNavigateToLowerList:bool=OR Boolean(A="integer > integer"(A=@n4, B=@n5), B=Get IsViewAllSelected(self=@n1)))
  ]}
]
variables [
 NestedButtonGroupSelectionVM {type=FortNestedButtonGroupSelectionVM, default=null}
 Was Last Selection View All {type=boolean}
]
widgets {
 info {
  inheritedWidgetCount=0
  namedSlotCount=1
  parentClass=&/WBP_Row_Poly_Base.WBP_Row_Poly_Base_C
  rootWidgetClass=null
  widgetCount=1
 }
 widgets [WBP_Row_PanelList {widgetClassPath=&/WBP_Row_PanelList, bIsVariable=true, widget=WBP_Row_PanelList}]
}
