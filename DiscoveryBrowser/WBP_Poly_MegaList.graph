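# Discovery browser's vertical list of rows, choosing tile-stack, horizontal-scroll, billboard or no-results entry widgets from each provider view model.
# Rows share a nested button group view model and report navigation, letting the list match nearest sub-row items and signal exits past either end.
# Sub-rows map to fractional scroll offsets for focus and touch swipes, with much legacy navigation gated by a rollback cvar.
# PostPaint passes list bounds to rows for impression visibility.
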WidgetBlueprint /DiscoveryBrowser/WBP_Poly_MegaList
graphs [
 ComputeVisibleImpressionItems {nodes [
   ComputeVisibleImpressionItems() {is=title K2Node_FunctionEntry}
   > n1 = GetWidgetAbsoluteRect!(execute, Widget:Widget Object Reference=Self-Reference()) -> (then, TopLeft:Vector 2D Structure, BottomRight:Vector 2D Structure, Size:Vector 2D Structure, DebugStringOutput:String)
   > n2 = For Each Loop(Array:Array of User Widget Object References=GetDisplayedEntryWidgets!(self:List View Base Object Reference=Get ListView() -> (ListView:Common List View Object Reference)) -> (ReturnValue:Array of User Widget Object References)) -> (Array Element:User Widget Object Reference)
   > ComputeRowVisibilityBase!(execute:Exec, self:Object Reference=@n2, Top Left:Vector 2D Structure=@n1, Bottom Right:Vector 2D Structure=@n1.BottomRight) -> (then:Exec) {is=title K2Node_Message}
   PrintString!(
    execute
    InString=Append!(
     A="Main Rect X1/Y1 : "
     B="To String (Vector2d)"!(InVec:Vector 2D Structure) -> (ReturnValue:String)
     C:String=" X2/Y2 : "
     D:String="To String (Vector2d)"()
    ) -> (ReturnValue)
    bPrintToScreen
    bPrintToLog
    TextColor
    Duration
    Key
   ) -> (then)
  ]}
 DoCustomNavigation {nodes [
   n1 = DoCustomNavigation!() -> (then, Navigation:EUINavigation Enum) {is=title K2Node_FunctionEntry}
   > n2 = Switch on EUINavigation!(execute, Selection:EUINavigation Enum=@n1) -> (Left:Exec, Right:Exec, Up:Exec, Down:Exec, Next:Exec, Previous:Exec, Invalid:Exec) {is=title K2Node_SwitchEnum}
   Call Navigate Down Out Of Mega List!(execute:Exec=@n2.Down, self:WBP Poly Mega List Object Reference) -> (then:Exec) {is=title K2Node_CallDelegate}
   > Return Node!(execute, ReturnValue:Widget Object Reference)
   Call Navigate Up Out Of Mega List!(execute:Exec=@n2.Up, self:WBP Poly Mega List Object Reference) -> (then:Exec) {is=title K2Node_CallDelegate}
   > Return Node()
  ]}
 EventGraph {nodes [
   n1 = "On Entry Initialized (ListView)"!() -> (OutputDelegate, then, Item:Object Reference, Widget:User Widget Object Reference) {is=title K2Node_ComponentBoundEvent}
   > SetOuterItemRowIndex!(
    execute:Exec
    self:Object Reference=@n1.Widget
    Index:int=GetIndexForItem!(self:List View Object Reference=Get ListView(), Item:Object Reference=@n1.Item) -> (ReturnValue:int=0)
   ) -> (then:Exec) {is=title K2Node_Message}
   n2 = "On Entry Released (ListView)"!() -> (OutputDelegate, then, Widget:User Widget Object Reference) {is=title K2Node_ComponentBoundEvent}
   > n3 = Cast To WBP_Row_Poly_Base!(execute, Object=@n2.Widget) -> (then, CastFailed, AsWBP Row Poly Base:WBP Row Poly Base Object Reference) {is=title K2Node_DynamicCast}
   > Unbind all Events from On Navigate to New Outer Row!(execute:Exec, self:WBP Row Poly Base Object Reference=@n3) -> (then:Exec) {is=title K2Node_ClearDelegate}
   > Unbind all Events from On Update Outer List Scroll Offset!(execute:Exec, self:WBP Row Poly Base Object Reference=@n3) -> (then:Exec) {is=title K2Node_ClearDelegate}
   n4 = "On Entry Generated (ListView)"!() -> (OutputDelegate, then, Widget:User Widget Object Reference) {is=title K2Node_ComponentBoundEvent}
   > UpdateRowPadding!(execute:Exec, self:Object Reference=@n4.Widget) -> (then:Exec) {is=title K2Node_Message}
   > UpdateNestedButtonGroupVM!(
    execute:Exec
    self:Object Reference=@n4.Widget
    NestedVM:Fort Nested Button Group Selection VM Object Reference=Get NestedButtonGroupSelectionVM() -> (NestedButtonGroupSelectionVM:Fort Nested Button Group Selection VM Object Reference)
   ) -> (then:Exec) {is=title K2Node_Message}
   > SetArrowBackgroundFadeVisibility!(
    execute:Exec
    self:Object Reference=@n4.Widget
    isVisible:bool=Get UseDarkShadowBehindHorizontalScrollArrows() -> (UseDarkShadowBehindHorizontalScrollArrows:bool)
   ) -> (then:Exec) {is=title K2Node_Message}
   > n5 = Cast To WBP_Row_Poly_Base(Object=@n4.Widget)
   > Bind Event to On Navigate to New Outer Row(self:WBP Row Poly Base Object Reference=@n5, Delegate=Create Event()) {is=title K2Node_AddDelegate}
   > Bind Event to On Update Outer List Scroll Offset(self:WBP Row Poly Base Object Reference=@n5, Delegate=Create Event()) {is=title K2Node_AddDelegate}
   n6 = "On Item Scrolled Into View (ListView)"!() -> (OutputDelegate, then, Item:Object Reference, Widget:User Widget Object Reference) {is=title K2Node_ComponentBoundEvent}
   > n7 = Branch Natural Nav Rollback Cvar!(In:Exec) -> (Use Natural Nav:Exec, Rollback:Exec) {is=title K2Node_MacroInstance}
   UpdateLastNavigationDirectionOnScroll!(execute=@n7.Rollback, self, Item:Object Reference=@n6.Item) -> (then)
   > n8 = GetNewSubrowIndexData!(execute:Exec, self:Object Reference=@n6.Widget) -> (
    then:Exec
    NewSubrowLinesCount:int=0
    NewSubrowNumItemsPerRow:int=0
    DisplayedEntries:int=0
    NewSubrowFirstVisibleItemIndex:int=0
    SelectedItemIndex:int=0
   ) {is=title K2Node_Message}
   > n9 = GetNearestNeighborSubrowIndex!(
    execute
    self
    NavDirection:E_UI_DirectionUpDownNone Enum=Get LastNavDirection() -> (LastNavDirection:E_UI_DirectionUpDownNone Enum=NewEnumerator0)
    NewSubrowLinesCount:int=@n8
    NewSubrowNumItemsPerRow:int=@n8.NewSubrowNumItemsPerRow
    DisplayedEntries:int=@n8.DisplayedEntries
    NewSubrowFirstVisibleItemIndex:int=@n8.NewSubrowFirstVisibleItemIndex
    SelectedItemIndex:int=@n8.SelectedItemIndex
   ) -> (then, NearestNeighborIndex:int=0)
   > NavigateToNearestNeighborSubrowIndex!(
    execute:Exec
    self:Object Reference=@n6.Widget
    Was Last Selection View All:bool=Get WasLastSelectionViewAll() -> (WasLastSelectionViewAll:bool)
    NearestNeighborIndex:int=@n9
   ) -> (then:Exec) {is=title K2Node_Message}
   n10 = OnNavigateToNewOuterRow!() -> (
    OutputDelegate
    then
    IsDirectionUp:bool=~
    PreviousSubrowSelectedIndex:int
    PreviousNumItemsPerSubrow:int
    PreviousRowWasViewAllTileSelected:bool=~
   ) {is=title K2Node_CustomEvent}
   > n11 = Branch Natural Nav Rollback Cvar()
   Set OldSubrowSelectedIndex(execute=@n11.Rollback, OldSubrowSelectedIndex:int=@n10.PreviousSubrowSelectedIndex) -> (Output_Get:int=0)
   > Set OldSubrowNumItemsPerRow(OldSubrowNumItemsPerRow:int=@n10.PreviousNumItemsPerSubrow) -> (Output_Get:int=0)
   > Set WasLastSelectionViewAll(WasLastSelectionViewAll:bool=@n10.PreviousRowWasViewAllTileSelected) -> (Output_Get:bool)
   > n12 = Set LastNavigationDirectionWasUp(LastNavigationDirectionWasUp:bool=@n10.IsDirectionUp) -> (Output_Get:bool)
   > n13 = Branch(Condition=@n12)
   n14 = Get ListView()
   n15 = int - int(A=GetNumItems!(self:List View Object Reference=@n14) -> (ReturnValue:int=0))
   n16 = GetSelectedItemIndex!(Target:List View Object Reference=@n14) -> (SelectedItemIndex:int=0)
   > n17 = Set NewlyDesiredOuterRowIndex(NewlyDesiredOuterRowIndex:int="Clamp (Integer)"(Value=int - int(A=@n16), Min=0, Max=@n15) -> (ReturnValue=0)) -> (Output_Get:int=0)
   > Call On User Navigated Mega List!(
    execute:Exec
    self:WBP Poly Mega List Object Reference
    IsDirectionUp:bool=Get LastNavigationDirectionWasUp() -> (LastNavigationDirectionWasUp:bool)
    NewDesiredOuterIndex:int=@n17
   ) -> (then:Exec) {is=title K2Node_CallDelegate}
   n18 = Get NewlyDesiredOuterRowIndex() -> (NewlyDesiredOuterRowIndex:int=0)
   > n19 = Branch(Condition=AND Boolean(A="Equal (Integer)"(A=@n16), B="Equal (Integer)"(A=@n18)))
   > Call Navigate Up Out Of Mega List()
   n20 = Get ListView()
   Set Selected Item!(
    execute=[@n22.else, @n19.else]
    self:List View Object Reference=@n20
    Item:Object Reference=GetItemAt!(self:List View Object Reference=@n20, Index:int=Get NewlyDesiredOuterRowIndex() -> (NewlyDesiredOuterRowIndex=0)) -> (ReturnValue:Object Reference)
   ) -> (then) {is=title K2Node_CallFunction}
   > Navigate To Item!(
    execute
    self:List View Object Reference=@n20
    Item:Object Reference=Get Selected Item!(self:List View Object Reference=@n20) -> (ReturnValue:Object Reference) {is=title K2Node_CallFunction}
   ) -> (then)
   n21 = Set NewlyDesiredOuterRowIndex(execute=@n13.else, NewlyDesiredOuterRowIndex="Clamp (Integer)"(Value=int + int(A=@n16), Min=0, Max=@n15) -> (ReturnValue=0)) -> (Output_Get=0)
   > Call On User Navigated Mega List(IsDirectionUp=Get LastNavigationDirectionWasUp(), NewDesiredOuterIndex=@n21)
   > n22 = Branch(Condition="Equal (Integer)"(A=@n16, B=@n15))
   > Call Navigate Down Out Of Mega List()
   n23 = OnListItemFocused!() -> (OutputDelegate, then, RowIndex:int, SubRowIndex:int, IsViewAll:bool=~) {is=title K2Node_CustomEvent}
   > n24 = Branch Natural Nav Rollback Cvar()
   Set WasLastSelectionViewAll(execute=@n24.Rollback, WasLastSelectionViewAll=@n23.IsViewAll)
   > Set RowIndex(RowIndex:int=@n23.RowIndex) -> (Output_Get:int=0)
   > Set OldSubRowIndex(OldSubRowIndex:int=@n23.SubRowIndex) -> (Output_Get:int=0)
   n25 = Get CommonInputSubsystem()
   > Branch(Condition="Not Equal (Enum)"!(A:ECommonInputType Enum=GetCurrentInputType(self=@n25), B:ECommonInputType Enum=MouseAndKeyboard) -> (ReturnValue:bool))
   > Branch(Condition=NOT Boolean(A=Get bUseScrollOffsetOverrideRedux() -> (bUseScrollOffsetOverrideRedux:bool)))
   > UpdateScrollOffsetMegalist()
   n26 = ItemSelectionChanged!() -> (OutputDelegate, then, Item:Object Reference, bIsSelected:bool=~) {is=title K2Node_CustomEvent}
   > Branch Natural Nav Rollback Cvar()
   > Branch(Condition=@n26.bIsSelected)
   n27 = GetIndexForItem(self=Get ListView(), Item=@n26.Item) -> (ReturnValue=0)
   n28 = "integer <= integer"(A=@n27, B=Get OldSelectedIndex() -> (OldSelectedIndex:int=0))
   > n29 = Branch(Condition=AND Boolean(A="Equal (Integer)"(A=@n27, B=1), B=NOT Boolean(A=@n28)))
   > n30 = Call On User Navigated Mega List()
   Set OldSelectedIndex(execute=[@n31.then, @n30.then], OldSelectedIndex:int=@n27) -> (Output_Get:int=0)
   n31 = Call On User Navigated Mega List(
    execute=@n29.else
    IsDirectionUp=@n28
    NewDesiredOuterIndex=Select(Option 0:int=1, Option 1:int=0, Index="Equal (Integer)"(A=@n27)) -> (ReturnValue:int)
   )
   Event OnInitialized()
   > n32 = Construct Object from Class(Class=/Script/FortniteUI.FortNestedButtonGroupSelectionVM) -> (ReturnValue:Fort Nested Button Group Selection VM Object Reference)
   > Set NestedButtonGroupSelectionVM(NestedButtonGroupSelectionVM:Fort Nested Button Group Selection VM Object Reference=@n32) -> (Output_Get:Fort Nested Button Group Selection VM Object Reference)
   > SetIsTouchScrollingEnabled!(
    execute
    self:List View Base Object Reference=Get ListView()
    bInIsTouchScrollingEnabled:bool=GetDiscoveryReduxExperimentValue!(FortPlayerController:Player Controller Object Reference=GetOwningPlayer()) -> (ReturnValue:bool)
   ) -> (then)
   Event Destruct()
   n33 = Create Event()
   > Unbind Event from On Item Selection Changed(self:List View Object Reference=Get ListView(), Delegate=@n33) {is=title K2Node_RemoveDelegate}
   Event Construct()
   > Bind Event to On Item Selection Changed(self:List View Object Reference=Get ListView(), Delegate=@n33) {is=title K2Node_AddDelegate}
   > SetScrollbarVisibility!(execute, self:List View Base Object Reference=Get ListView(), InVisibility:ESlateVisibility Enum=Collapsed) -> (then)
   > n34 = Branch Natural Nav Rollback Cvar()
   SetAllNavigationRules!(execute=@n34.Rollback, self:Widget Object Reference=Get ListView(), Rule:EUINavigationRule Enum=Escape, WidgetToFocus:Name=None) -> (then)
   > Set bIsFocusable(bIsFocusable:bool=false) -> (Output_Get:bool)
   > SetDesiredFocusWidget!(execute, self:User Widget Object Reference, Widget:Widget Object Reference=Get ListView()) -> (then, ReturnValue:bool)
   > SetReturnFocusToSelection!(execute, self:List View Object Reference=Get ListView(), bEnabled:bool=true) -> (then)
   > SetClipping!(execute, self:Widget Object Reference=Get GridRoot() -> (GridRoot:Grid Panel Object Reference), InClipping:EWidgetClipping Enum=ClipToBounds) -> (then)
   > SetEnableProximateEntryNavigation!(execute, self:List View Object Reference=Get ListView(), bNewValue:bool=false) -> (then)
  ]}
 GetDesiredEntryClassForItem {nodes [
   n1 = GetDesiredEntryClassForItem!() -> (then, Item:Object Reference) {is=title K2Node_FunctionEntry}
   > n2 = Cast To FortDiscoverProviderViewModel!(execute, Object=@n1) -> (then, CastFailed, AsDiscover Provider View Model:Discover Provider View Model Object Reference) {is=title K2Node_DynamicCast}
   n3 = Get Items!(self:Discover Provider View Model Object Reference=@n2) -> (Items:Array of Discover Tile Item View Model Object References)
   n4 = Select(
    Option 0:User Widget Class Reference=/DiscoveryBrowser/Rows/WBP_Row_Poly_TileStackWrapper.WBP_Row_Poly_TileStackWrapper_C
    Option 1:User Widget Class Reference=/DiscoveryBrowser/Rows/WBP_Row_Poly_HorizontalScrollWrapper.WBP_Row_Poly_HorizontalScrollWrapper_C
    Index=Get bUsesHorizontalScrolling!(self:Discover Provider View Model Object Reference=@n2) -> (bUsesHorizontalScrolling:bool)
   ) -> (ReturnValue:User Widget Class Reference)
   n5 = Get bIsBannerItemRow!(self:Discover Provider View Model Object Reference=@n2) -> (bIsBannerItemRow:bool)
   > Return Node(ReturnValue:User Widget Class Reference=Select(
     Option 0=Select(Option 0=@n4, Option 1=/DiscoveryBrowser/Rows/WBP_Row_Billboard_Wrapper.WBP_Row_Billboard_Wrapper_C, Index=@n5)
     Option 1=/DiscoveryBrowser/Rows/WBP_Row_NoResults.WBP_Row_NoResults_C
     Index=Is Empty!(TargetArray:Array of Discover Tile Item View Model Object References=@n3) -> (ReturnValue:bool)
    ))
  ]}
 GetLastVisibleRowCount {nodes [
   GetLastVisibleRowCount() {is=title K2Node_FunctionEntry}
   > n1 = GetWidgetAbsoluteRect!(execute, Widget=Self-Reference()) -> (then, TopLeft, BottomRight, Size_X:double, Size_Y:double, DebugStringOutput)
   > Set L_SelfContainerSize(L_SelfContainerSize:double=@n1.Size_Y) -> (Output_Get:double)
   > n2 = For Each Loop with Break!(Exec:Exec, Array:Array of User Widget Object References=GetDisplayedEntryWidgets(self=Get ListView()), Break:Exec=@n5.else) -> (LoopBody:Exec, Array Element:User Widget Object Reference, Array Index:int, Completed:Exec)
   > n3 = GetWidgetAbsoluteRect(Widget=@n2)
   n4 = float + float(A=@n3.Size_Y, B=Get L_ListViewContainerAbsoluteSize() -> (L_ListViewContainerAbsoluteSize:double))
   > n5 = Branch(Condition="float < float"(A=@n4, B=Get L_SelfContainerSize() -> (L_SelfContainerSize:double)))
   > Set L_ListViewContainerAbsoluteSize(L_ListViewContainerAbsoluteSize:double=@n4) -> (Output_Get:double)
   > Set L_TotalRowCount(L_TotalRowCount:int=@n2.Array Index) -> (Output_Get:int=0)
   Return Node!(execute=@n2.Completed, Count:int=Get L_TotalRowCount() -> (L_TotalRowCount:int=0))
  ]}
 GetNearestNeighborSubrowIndex {nodes [
   n1 = GetNearestNeighborSubrowIndex!() -> (
    then
    NavDirection:E_UI_DirectionUpDownNone Enum=NewEnumerator0
    NewSubrowLinesCount:int
    NewSubrowNumItemsPerRow:int
    DisplayedEntries:int
    NewSubrowFirstVisibleItemIndex:int
    SelectedItemIndex:int
   ) {is=title K2Node_FunctionEntry}
   > n2 = Switch on E_UI_DirectionUpDownNone!(execute, Selection:E_UI_DirectionUpDownNone Enum=@n1) -> (NewEnumerator0:Exec, NewEnumerator2:Exec, NewEnumerator1:Exec, NewEnumerator3:Exec) {is=title K2Node_SwitchEnum}
   n3 = int - int(A=Get DisplayedEntries() -> (DisplayedEntries:int=0))
   n4 = Get NewSubrowFirstVisibleItemIndex() -> (NewSubrowFirstVisibleItemIndex:int=0)
   n5 = "% (Integer)"(A=Get OldSubrowSelectedIndex() -> (OldSubrowSelectedIndex:int=0), B=Get OldSubrowNumItemsPerRow() -> (OldSubrowNumItemsPerRow:int=0)) -> (ReturnValue=0)
   Return Node!(
    execute=@n2.NewEnumerator2
    NearestNeighborIndex:int="Min (Integer)"(A=int + int(A=@n5, B=@n4), B=int + int(A=@n4, B=@n3)) -> (ReturnValue=0)
   )
   Return Node(execute=@n2.NewEnumerator3, NearestNeighborIndex=0)
   Return Node(execute=@n2.NewEnumerator1, NearestNeighborIndex=Get SelectedItemIndex() -> (SelectedItemIndex:int=0))
   n6 = int - int(A=Get DisplayedEntries() -> (DisplayedEntries=0))
   n7 = Get NewSubrowFirstVisibleItemIndex() -> (NewSubrowFirstVisibleItemIndex=0)
   n8 = int * int(A=Get NewSubrowNumItemsPerRow() -> (NewSubrowNumItemsPerRow:int=0), B=int - int(A=Get NewSubrowLinesCount() -> (NewSubrowLinesCount:int=0)))
   n9 = Get OldSubrowSelectedIndex() -> (OldSubrowSelectedIndex=0)
   > Return Node(NearestNeighborIndex="Min (Integer)"(A=int + int(A=@n6, B=@n7), B=int + int!(A=@n7, B=@n8, C:int=@n9) -> (ReturnValue)) -> (ReturnValue=0))
  ]}
 GetNewSubrowIndexData {nodes [
   GetNewSubrowIndexData() {is=title K2Node_FunctionEntry}
   > Return Node!(
    execute
    NewSubrowLinesCount:int=0
    NewSubrowNumItemsPerRow:int=0
    DisplayedEntries:int=0
    NewSubrowFirstVisibleItemIndex:int=0
    SelectedItemIndex:int=0
   )
  ]}
 GetNumSubRowLinesWithinOuterRow {nodes [
   n1 = GetNumSubRowLinesWithinOuterRow!() -> (then, OuterRowIndex:int) {is=title K2Node_FunctionEntry}
   > n2 = Cast To FortDiscoverProviderViewModel(Object=GetItemAt(self=Get ListView(), Index=@n1))
   > n3 = Branch(Condition=NOT Boolean(A=Get bUsesHorizontalScrolling(self=@n2)))
   > Return Node!(execute, NumSubRows:int=Get MaxVisibleRows!(self:Discover Provider View Model Object Reference=@n2) -> (MaxVisibleRows:int=0))
   Return Node(execute=[@n3.else, @n2.CastFailed], NumSubRows=1)
  ]}
 GetRowIsFullyVisible {nodes [
   n1 = GetRowIsFullyVisible!() -> (then, RowIndex:int) {is=title K2Node_FunctionEntry}
   n2 = Get ListView()
   n3 = Ceil(A=GetScrollOffset!(self:List View Base Object Reference=Get ListView()) -> (ReturnValue:float)) {is=title K2Node_CallFunction}
   > Return Node!(execute, ReturnValue:bool="integer >= integer"(A=int + int!(A=@n1, B=@n3) -> (ReturnValue), B=GetNumItems(self=@n2) -> (ReturnValue=0)))
  ]}
 GetScrollOffsetForRowAndSubRow {nodes [
   n1 = GetScrollOffsetForRowAndSubRow!() -> (then, OuterRowIndex:int, SubRowIndex:int) {is=title K2Node_FunctionEntry}
   n2 = GetNumItems(self=Get ListView()) -> (ReturnValue=0)
   > Set OuterRowIdx(OuterRowIdx:int="Clamp (Integer)"(Value=@n1, Min=0, Max=int - int(A=@n2)) -> (ReturnValue=0)) -> (Output_Get:int=0)
   n3 = GetNumSubRowLinesWithinOuterRow!(OuterRowIndex:int=Get OuterRowIdx() -> (OuterRowIdx:int=0)) -> (NumSubRows:int=0) {is=title K2Node_CallFunction}
   > Set SubRowIdx(SubRowIdx:int="Clamp (Integer)"(Value=@n1.SubRowIndex, Min=0, Max=int - int(A=@n3)) -> (ReturnValue=0)) -> (Output_Get:int=0)
   n4 = Get SubRowIdx() -> (SubRowIdx:int=0)
   > Set NewSubRowOffset(NewSubRowOffset:double=SafeDivide(A="To Float (Integer)"(InInt=@n4), B="To Float (Integer)"(InInt=@n3))) -> (Output_Get:double)
   > n5 = Branch(Condition="Equal (Integer)"(A=Get SubRowIdx() -> (SubRowIdx=0)))
   n6 = GetNumSubRowLinesWithinOuterRow(OuterRowIndex=int - int(A=Get OuterRowIdx() -> (OuterRowIdx=0))) -> (NumSubRows=0)
   > n7 = Set PreviousSubrowHalfOffset(PreviousSubrowHalfOffset:double=SafeDivide(A=0.5, B="To Float (Integer)"(InInt=@n6))) -> (Output_Get:double)
   n8 = "To Float (Integer)"(InInt=Get OuterRowIdx() -> (OuterRowIdx=0))
   n9 = float - float(A=Get NewSubRowOffset() -> (NewSubRowOffset:double), B=Get PreviousSubrowHalfOffset() -> (PreviousSubrowHalfOffset:double))
   Return Node!(execute=[@n11.then, @n7.then], NewScrollOffsetForOuterList:double="Max (Float)"(A=0, B=float + float(A=@n8, B=@n9)))
   n10 = GetNumSubRowLinesWithinOuterRow(OuterRowIndex=Get OuterRowIdx() -> (OuterRowIdx=0)) -> (NumSubRows=0)
   n11 = Set PreviousSubrowHalfOffset(execute=@n5.else, PreviousSubrowHalfOffset=SafeDivide(A=0.5, B="To Float (Integer)"(InInt=@n10)))
  ]}
 GetScrollOffsetForSubrowRedux {nodes [
   n1 = GetScrollOffsetForSubrowRedux!() -> (then, OuterRowIndex:int, SubRowIndex:int) {is=title K2Node_FunctionEntry}
   n2 = "To Float (Integer)"(InInt=Get OuterRowIndex() -> (OuterRowIndex:int=0))
   n3 = GetNumSubRowLinesWithinOuterRow(OuterRowIndex=int - int(A=Get OuterRowIndex() -> (OuterRowIndex=0))) -> (NumSubRows=0)
   n4 = Get OuterRowIndex() -> (OuterRowIndex=0)
   n5 = GetNumSubRowLinesWithinOuterRow(OuterRowIndex=@n1.SubRowIndex) -> (NumSubRows=0)
   n6 = float - float(A=SafeDivide(A="To Float (Integer)"(InInt=@n4), B="To Float (Integer)"(InInt=@n5)), B=SafeDivide(A=0.5, B="To Float (Integer)"(InInt=@n3)))
   > Return Node(NewScrollOffsetForOuterList="Max (Float)"(A=0, B=float + float(A=@n2, B=@n6)))
  ]}
 GetSelectedItemIndex {nodes [
   n1 = GetSelectedItemIndex!() -> (then, Target:List View Object Reference) {is=title K2Node_FunctionEntry}
   > Return Node!(execute, SelectedItemIndex:int=GetIndexForItem(self=@n1, Item=Get Selected Item(self=@n1)) -> (ReturnValue=0))
  ]}
 NavigateDownOutOfMegaList {nodes [NavigateDownOutOfMegaList() {is=title K2Node_FunctionEntry}]}
 NavigateUpOutOfMegaList {nodes [NavigateUpOutOfMegaList() {is=title K2Node_FunctionEntry}]}
 OnNavigatedMegaListRow {nodes [OnNavigatedMegaListRow!() -> (then, LastNavigationDirectionWasUp:bool=~, CurrentRowIndex:int, CurrentSubRowIndex:int) {is=title K2Node_FunctionEntry}]}
 OnUserNavigatedMegaList {nodes [OnUserNavigatedMegaList!() -> (then, IsDirectionUp:bool=~, NewDesiredOuterIndex:int) {is=title K2Node_FunctionEntry}]}
 PopulateSpacer {nodes [
   PopulateSpacer!() -> (then, SpacerSize:Vector 2D Structure) {is=title K2Node_FunctionEntry}
   n1 = Get GridRoot()
   n2 = GetChildAt!(self:Panel Widget Object Reference=@n1, Index:int=1) -> (ReturnValue:Widget Object Reference)
   > n3 = Is Valid(InputObject=@n2)
   > n4 = RemoveChild(self=@n1, Content=@n2)
   n5 = Construct Object from Class(execute=[@n4.then, @n3.Is Not Valid], Class=&Spacer) -> (ReturnValue:Spacer Object Reference)
   > SetSize!(execute, self:Spacer Object Reference=@n5, InSize:Vector 2D Structure=Get SpacerSize() -> (SpacerSize:Vector 2D Structure)) -> (then)
   > AddChildToGrid!(execute, self:Grid Panel Object Reference=Get GridRoot(), Content:Widget Object Reference=@n5, InRow:int=1, InColumn:int=0) -> (then, ReturnValue:Grid Slot Object Reference)
  ]}
 PostPaint {nodes [
   PostPaint() {is=title K2Node_FunctionEntry}
   > ComputeVisibleImpressionItems() {is=title K2Node_CallFunction}
  ]}
 SwipeToNextRow_Touch {nodes [
   SwipeToNextRow_Touch() {is=title K2Node_FunctionEntry}
   > SyncSwipeRowAndSubrowIndex()
   n1 = GetNumSubRowLinesWithinOuterRow(OuterRowIndex=Get CurrentSwipeRowIdx() -> (CurrentSwipeRowIdx:int=0)) -> (NumSubRows=0)
   > n2 = Branch(Condition="integer < integer"(A=Get CurrentSwipeSubrowIdx() -> (CurrentSwipeSubrowIdx:int=0), B=int - int(A=@n1)))
   > n3 = Increment Int(Value=Get CurrentSwipeSubrowIdx() -> (CurrentSwipeSubrowIdx=0))
   n4 = GetScrollOffsetForRowAndSubRow!(
    execute=[@n3."  ", @n5.then]
    self
    OuterRowIndex:int=Get CurrentSwipeRowIdx() -> (CurrentSwipeRowIdx=0)
    SubRowIndex:int=Get CurrentSwipeSubrowIdx() -> (CurrentSwipeSubrowIdx=0)
   ) -> (then, NewScrollOffsetForOuterList:double) {is=title K2Node_CallFunction}
   > SetScrollOffset!(execute, self:List View Base Object Reference=Get ListView(), InScrollOffset:float=@n4) -> (then)
   Increment Int(" "=@n2.else, Value=Get CurrentSwipeRowIdx() -> (CurrentSwipeRowIdx=0))
   > n5 = Set CurrentSwipeSubrowIdx(CurrentSwipeSubrowIdx:int=0) -> (Output_Get:int=0)
  ]}
 SwipeToPreviousRow_Touch {nodes [
   SwipeToPreviousRow_Touch() {is=title K2Node_FunctionEntry}
   > SyncSwipeRowAndSubrowIndex()
   > n1 = Branch(Condition="integer > integer"(A=Get CurrentSwipeSubrowIdx() -> (CurrentSwipeSubrowIdx=0)))
   > n2 = Decrement Int(Value=Get CurrentSwipeSubrowIdx() -> (CurrentSwipeSubrowIdx=0))
   n3 = GetScrollOffsetForRowAndSubRow(
    execute=[@n5.then, @n2."  "]
    OuterRowIndex=Get CurrentSwipeRowIdx() -> (CurrentSwipeRowIdx=0)
    SubRowIndex=Get CurrentSwipeSubrowIdx() -> (CurrentSwipeSubrowIdx=0)
   )
   > SetScrollOffset(self=Get ListView(), InScrollOffset=@n3)
   Decrement Int(" "=@n1.else, Value=Get CurrentSwipeRowIdx() -> (CurrentSwipeRowIdx=0))
   n4 = Get CurrentSwipeRowIdx() -> (CurrentSwipeRowIdx=0)
   > n5 = Set CurrentSwipeSubrowIdx(CurrentSwipeSubrowIdx=int - int(A=GetNumSubRowLinesWithinOuterRow(OuterRowIndex=@n4) -> (NumSubRows=0))) -> (Output_Get=0)
  ]}
 SyncSwipeRowAndSubrowIndex {nodes [
   SyncSwipeRowAndSubrowIndex() {is=title K2Node_FunctionEntry}
   > n1 = GetScrollOffsetForRowAndSubRow(OuterRowIndex=Get CurrentSwipeRowIdx() -> (CurrentSwipeRowIdx=0), SubRowIndex=Get CurrentSwipeSubrowIdx() -> (CurrentSwipeSubrowIdx=0))
   n2 = GetScrollOffset(self=Get ListView())
   > n3 = Branch(Condition="Nearly Equal (Float)"(A=@n1, B=@n2, ErrorTolerance=0))
   > Return Node!(execute)
   n4 = Set CurrentScrollOffset(execute=@n3.else, CurrentScrollOffset:double=@n2) -> (Output_Get:double)
   n5 = Truncate(A=@n4)
   n6 = int + int(A=@n5)
   n7 = int - int(A=GetNumItems(self=Get ListView()) -> (ReturnValue=0))
   > n8 = For Loop(FirstIndex="Max (Integer)"(A=int - int(A=@n5), B=0), LastIndex="Max (Integer)"(A="Min (Integer)"(A=@n6, B=@n7) -> (ReturnValue=0), B=0))
   > n9 = Set CurrentSwipeRowIdx(CurrentSwipeRowIdx:int=@n8) -> (Output_Get:int=0)
   n10 = GetNumSubRowLinesWithinOuterRow(OuterRowIndex=@n9) -> (NumSubRows=0)
   > n11 = For Loop(LastIndex="Max (Integer)"(A=int - int(A=@n10), B=0))
   > n12 = Set CurrentSwipeSubrowIdx(CurrentSwipeSubrowIdx=@n11) -> (Output_Get=0)
   > n13 = GetScrollOffsetForRowAndSubRow(OuterRowIndex=Get CurrentSwipeRowIdx() -> (CurrentSwipeRowIdx=0), SubRowIndex=@n12)
   n14 = Get CurrentScrollOffset() -> (CurrentScrollOffset:double)
   > Branch(Condition=OR Boolean(A="Nearly Equal (Float)"(A=@n13, B=@n14, ErrorTolerance=0), B="float > float"(A=@n13, B=@n14)))
   > Return Node()
  ]}
 UpdateLastNavigationDirectionOnScroll {nodes [
   UpdateLastNavigationDirectionOnScroll!() -> (then, Item:Object Reference) {is=title K2Node_FunctionEntry}
   n1 = GetIndexForItem(self=Get ListView(), Item=Get Item() -> (Item:Object Reference)) -> (ReturnValue=0)
   > n2 = Branch(Condition=AND Boolean(A=Get bFirstRowSnapIsZeroIndex() -> (bFirstRowSnapIsZeroIndex:bool), B="Equal (Integer)"(A=@n1)))
   > Set LastNavDirection(LastNavDirection:E_UI_DirectionUpDownNone Enum=NewEnumerator3) -> (Output_Get:E_UI_DirectionUpDownNone Enum=NewEnumerator0)
   > Set bFirstRowSnapIsZeroIndex(bFirstRowSnapIsZeroIndex:bool=false) -> (Output_Get:bool)
   n3 = Get OldOuterRowIndex() -> (OldOuterRowIndex:int=0)
   n4 = Branch(execute=@n2.else, Condition="Equal (Integer)"(A=@n3, B=@n1))
   > n5 = Set LastNavDirection(LastNavDirection=NewEnumerator1) -> (Output_Get=NewEnumerator0)
   Set OldOuterRowIndex(execute=[@n8.then, @n5.then, @n7.then], OldOuterRowIndex:int=@n1) -> (Output_Get:int=0)
   n6 = Branch(execute=@n4.else, Condition="integer > integer"(A=@n3, B=@n1))
   > n7 = Set LastNavDirection(LastNavDirection=NewEnumerator0) -> (Output_Get=NewEnumerator0)
   n8 = Set LastNavDirection(execute=@n6.else, LastNavDirection=NewEnumerator2) -> (Output_Get=NewEnumerator0)
  ]}
 UpdateScrollOffset {nodes [
   UpdateScrollOffset!() -> (then, IsDirectionUp:bool=~, RowIndex:int) {is=title K2Node_FunctionEntry}
   > n1 = Branch Natural Nav Rollback Cvar()
   SetScrollIntoViewAlignment!(execute=@n1.Rollback, self:List View Object Reference=Get ListView(), NewScrollIntoViewAlignment:EScrollIntoViewAlignment Enum=CenterAligned) -> (then)
   n2 = GetScrollOffset(self=Get ListView())
   > Set FirstVisibleRow(FirstVisibleRow:int="To Integer (Integer64)"!(InInt:Integer64=Ceil to Integer64!(A:double=@n2) -> (ReturnValue:Integer64=0)) -> (ReturnValue:int=0)) -> (Output_Get:int=0)
   > n3 = Branch(Condition=Get IsDirectionUp() -> (IsDirectionUp:bool))
   n4 = Get FirstVisibleRow() -> (FirstVisibleRow:int=0)
   > Branch(Condition="Equal (Integer)"(A=int - int(A=@n4), B=Get RowIndex() -> (RowIndex:int=0)))
   n5 = Get RowIndex() -> (RowIndex=0)
   > n6 = Branch(Condition="integer > integer"(A=GetNumSubRowLinesWithinOuterRow(OuterRowIndex=@n5) -> (NumSubRows=0), B=1))
   > n7 = GetScrollOffsetForSubrowRedux!(execute, self, OuterRowIndex:int=Get RowIndex() -> (RowIndex=0), SubRowIndex:int=Get OldSubRowIndex() -> (OldSubRowIndex:int=0)) -> (then, NewScrollOffsetForOuterList:double) {is=title K2Node_CallFunction}
   > n8 = SetScrollOffset(self=Get ListView(), InScrollOffset=@n7)
   n9 = Get CurrentPage() -> (CurrentPage:int=0)
   n10 = Set CurrentPage(execute=[@n13.then, @n8.then], CurrentPage:int="Max (Integer)"(A=0, B=int - int(A=@n9))) -> (Output_Get:int=0)
   n11 = "float < float"(A=GetScrollOffset(self=Get ListView()), B=2)
   n12 = "Equal (Integer)"(A=@n10)
   > Return Node!(execute, CurrentPage:int=SelectInt(A=0, B=1, bPickA=AND Boolean(A=@n12, B=@n11)))
   n13 = SetScrollIntoViewAlignment(execute=@n6.else, self=Get ListView(), NewScrollIntoViewAlignment=BottomOrRight)
   n14 = GetLastVisibleRowCount!() -> (Count:int=0) {is=title K2Node_CallFunction}
   n15 = Get FirstVisibleRow() -> (FirstVisibleRow=0)
   Branch(execute=@n3.else, Condition="Equal (Integer)"(A=Get RowIndex() -> (RowIndex=0), B=int + int(A=@n14, B=@n15)))
   > SetScrollIntoViewAlignment(self=Get ListView(), NewScrollIntoViewAlignment=TopOrLeft)
   > n16 = Set CurrentPage(CurrentPage=int + int(A=Get CurrentPage() -> (CurrentPage=0))) -> (Output_Get=0)
   > Return Node(CurrentPage=@n16)
  ]}
 UpdateScrollOffsetMegalist {nodes [
   UpdateScrollOffsetMegalist() {is=title K2Node_FunctionEntry}
   n1 = "Equal (Integer)"(A=Get OldSubRowIndex() -> (OldSubRowIndex=0), B=Get SubRowIndex() -> (SubRowIndex:int=0))
   n2 = GetRowIsFullyVisible!(RowIndex:int=Get RowIndex() -> (RowIndex=0)) -> (ReturnValue:bool) {is=title K2Node_CallFunction}
   > n3 = Branch(Condition=AND Boolean(A=NOT Boolean(A=@n1), B=NOT Boolean(A=@n2)))
   > n4 = Branch(Condition=Get LastNavigationDirectionWasUp())
   n5 = Set SubRowIndex(execute=[@n4.then, @n3.else], SubRowIndex:int=Get OldSubRowIndex() -> (OldSubRowIndex=0)) -> (Output_Get:int=0)
   > n6 = GetScrollOffsetForRowAndSubRow(OuterRowIndex=Get RowIndex() -> (RowIndex=0), SubRowIndex=@n5)
   > n7 = SetScrollOffset(self=Get ListView(), InScrollOffset=@n6)
   Set CurrentSwipeRowIdx(execute=[@n7.then, @n4.else], CurrentSwipeRowIdx=Get RowIndex() -> (RowIndex=0)) -> (Output_Get=0)
   > Set CurrentSwipeSubrowIdx(CurrentSwipeSubrowIdx=Get OldSubRowIndex() -> (OldSubRowIndex=0)) -> (Output_Get=0)
   > Call On Navigated Mega List Row!(
    execute:Exec
    self:WBP Poly Mega List Object Reference
    LastNavigationDirectionWasUp:bool=Get LastNavigationDirectionWasUp()
    CurrentRowIndex:int=Get RowIndex() -> (RowIndex=0)
    CurrentSubRowIndex:int=Get OldSubRowIndex() -> (OldSubRowIndex=0)
   ) -> (then:Exec) {is=title K2Node_CallDelegate}
  ]}
]
variables [
 bFirstRowSnapIsZeroIndex {type=boolean}
 bUseScrollOffsetOverrideRedux {type=boolean}
 CurrentPage {type=integer}
 CurrentSwipeRowIdx {type=integer}
 CurrentSwipeSubrowIdx {type=integer}
 FirstRow_TileView {type=FortTileView, default=null}
 LastNavDirection {type=E_UI_DirectionUpDownNone, default=E_UI_MAX}
 LastNavigationDirectionWasUp {type=boolean}
 NavigateDownOutOfMegaList {}
 NavigateUpOutOfMegaList {}
 NestedButtonGroupSelectionVM {type=FortNestedButtonGroupSelectionVM, default=null}
 NewlyDesiredOuterRowIndex {type=integer}
 OldOuterRowIndex {type=integer}
 OldSelectedIndex {type=integer, default=-1}
 OldSubRowIndex {type=integer}
 OldSubrowNumItemsPerRow {type=integer}
 OldSubrowSelectedIndex {type=integer}
 OnNavigatedMegaListRow {}
 OnUserNavigatedMegaList {}
 RowIndex {type=integer}
 SubRowIndex {type=integer}
 UseDarkShadowBehindHorizontalScrollArrows {type=boolean, default=true}
 WasLastSelectionViewAll {type=boolean}
]
widgets {
 info {inheritedWidgetCount=1, parentClass=/Script/FortniteUI.FortPolymorphicListView, rootWidgetClass=&GridPanel, widgetCount=2}
 widgets [
  GridRoot {widgetClassPath=&GridPanel, bIsVariable=true, widget=GridRoot, properties {columnFill [1], rowFill [1]}}
  ListView {
   widgetClassPath=&CommonListView
   bInherited=true
   bIsVariable=true
   parent=GridRoot
   widget=ListView
   properties {
    bEnableProximateEntryNavigation=true
    bEnableTouchScrolling=false
    clipping=Inherit
    consumeMouseWheel=Never
    entryWidgetClass=&WBP_Row_Poly_TileStackWrapper_C
    navigation=&self:WidgetTree.ListView.WidgetNavigation_0
    numDesignerPreviewEntries=20
    wheelScrollMultiplier=3
   }
   slot {class=GridSlot}
  }
 ]
}
