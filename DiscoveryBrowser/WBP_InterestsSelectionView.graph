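# Activatable Discovery screen where players pick interests, then save or skip.
# The interests surface view model handles save, skip and impression analytics; providers fill a mega list whose rows are restyled as they initialize.
# Skip opens a confirmation dialog if favorites are pending; processing state shows a loading view or closes the parent modal.
# Custom focus and navigation functions link tile grids to footer buttons, with swipe row paging and mobile layout tweaks.
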WidgetBlueprint /DiscoveryBrowser/WBP_InterestsSelectionView
graphs [
 BP_GetDesiredFocusTarget {nodes [
   Get Desired Focus Target()
   > Branch(Condition=Get IsListFilledWithData() -> (IsListFilledWithData:bool))
   > n1 = Branch(Condition=Get IsSkipModalDeactivated() -> (IsSkipModalDeactivated:bool))
   > Return Node!(execute, ReturnValue:Widget Object Reference=Get Button_Skip() -> (Button_Skip:WBP UIKit Button Regular Object Reference))
   n2 = Get ListView!(self:Fort Polymorphic List View Object Reference=Get WBP_Poly_MegaList() -> (WBP_Poly_MegaList:WBP Poly Mega List Object Reference)) -> (ListView:Common List View Object Reference)
   n3 = Cast To WBP_Row_Poly_TileStackWrapper!(
    execute=@n1.else
    Object="Get (a copy)"(Array:Array of User Widget Object References=GetDisplayedEntryWidgets!(self:List View Base Object Reference=@n2) -> (ReturnValue:Array of User Widget Object References)) -> (Output:User Widget Object Reference)
   ) -> (then, CastFailed, AsWBP Row Poly Tile Stack Wrapper:WBP Row Poly Tile Stack Wrapper Object Reference) {is=title K2Node_DynamicCast}
   n4 = Get TileView!(self:Fort Polymorphic Tile View Object Reference=Get WBP_Row_PanelList!(self:WBP Row Poly Tile Stack Wrapper Object Reference=@n3) -> (WBP_Row_PanelList:WBP Row Panel List Object Reference)) -> (TileView:Fort Tile View Object Reference)
   > SetSelectedIndex!(execute, self:List View Object Reference=@n4, Index:int=0) -> (then)
   > Navigate To Item!(
    execute
    self:List View Object Reference=@n4
    Item:Object Reference=Get Selected Item!(self:List View Object Reference=@n4) -> (ReturnValue:Object Reference) {is=title K2Node_CallFunction}
   ) -> (then)
   > Return Node(ReturnValue=@n4)
  ]}
 EventGraph {nodes [
   n1 = "On User Navigated Mega List (WBP_Poly_MegaList)"!() -> (OutputDelegate, then, IsDirectionUp:bool, NewDesiredOuterIndex:int=0) {is=title K2Node_ComponentBoundEvent}
   > n2 = Branch(Condition=@n1.IsDirectionUp)
   n3 = Branch(
    execute=@n2.else
    Condition=Get bIsEnabled!(self:Widget Object Reference=Get Button_SaveSelections() -> (Button_SaveSelections:WBP UIKit Button Loud Object Reference)) -> (bIsEnabled:bool)
   )
   > SetFocus(self=Get Button_SaveSelections())
   SetFocus(execute=@n3.else, self=Get Button_Skip())
   "On Clicked (Button_SaveSelections)"!() -> (OutputDelegate, then, Button:Common Button Base Object Reference) {is=title K2Node_ComponentBoundEvent}
   > SaveInterestSelections(self:Discover Interests Surface View Model Object Reference=Get FortDiscoverInterestsSurfaceViewModel() -> (FortDiscoverInterestsSurfaceViewModel:Discover Interests Surface View Model Object Reference))
   "On Swipe Up (SwipePanel)"() {is=title K2Node_ComponentBoundEvent}
   n4 = Get WBP_Poly_MegaList()
   > SwipeToPreviousRow_Touch(self:WBP Poly Mega List Object Reference=@n4)
   "On Swipe Down (SwipePanel)"() {is=title K2Node_ComponentBoundEvent}
   > SwipeToNextRow_Touch(self:WBP Poly Mega List Object Reference=@n4)
   "On Clicked (Button_Skip)"!() -> (OutputDelegate, then, Button:Common Button Base Object Reference) {is=title K2Node_ComponentBoundEvent}
   n5 = Get FortDiscoverInterestsSurfaceViewModel()
   > n6 = Branch(Condition=Get ContainsPendingFavorites!(self:Discover Interests Surface View Model Object Reference=@n5) -> (ContainsPendingFavorites:bool))
   > n7 = Construct Object from Class(Class=/DiscoveryBrowser/Dialogs/DialogVM_InterestsSkip.DialogVM_InterestsSkip_C) -> (ReturnValue:Dialog VM Interests Skip Object Reference)
   > n8 = Create Widget!(
    execute:Exec
    Class:User Widget Class Reference=/FortUILibrary/UIKit/Dialogs/WBP_UIKit_Dialog.WBP_UIKit_Dialog_C
    OwningPlayer:Player Controller Object Reference
    UIKitDialogViewModel:UI Kit Dialog Viewmodel Object Reference=@n7
   ) -> (then:Exec, ReturnValue:WBP UIKit Dialog Object Reference)
   n9 = Get Dialog_Internal!(self:WBP UIKit Dialog Object Reference=@n8) -> (Dialog_Internal:WBP UIKit Dialog Internal Object Reference)
   > n10 = Set Buttons!(
    execute
    Buttons:S UI Dialog Buttons Structure=Get ModalButtons() -> (ModalButtons:S UI Dialog Buttons Structure)
    self:WBP UIKit Dialog Internal Object Reference=@n9
   ) -> (then, Output_Get:S UI Dialog Buttons Structure)
   n11 = OnModalButtonsClicked!() -> (OutputDelegate, then, ButtonIndex:int) {is=title K2Node_CustomEvent}
   n12 = Bind Event to Dialog Button Clicked Dispatcher(execute=@n10.then, self:WBP UIKit Dialog Internal Object Reference=@n9, Delegate=@n11) {is=title K2Node_AddDelegate}
   n13 = OnModalSkipDeactivated() {is=title K2Node_CustomEvent}
   Assign On Widget Deactivated!(execute:Exec=@n12.then, self:Common Activatable Widget Object Reference=@n8, Delegate:"Delegate (by ref)"=@n13) -> (then:Exec) {is=title K2Node_AssignDelegate}
   > TryPushWidgetAsModalPanel!(execute, PushWidget:User Widget Object Reference=@n8) -> (then)
   SkipInterestSelection(execute=@n6.else, self:Discover Interests Surface View Model Object Reference=@n5)
   n14 = OnListViewInitialized!() -> (OutputDelegate, then, Item:Object Reference, Widget:User Widget Object Reference) {is=title K2Node_CustomEvent}
   > UpdateRowHeaderStyle!(execute, self, Object:Object Reference=@n14.Widget) -> (then)
   OnListViewEntryGenerated!() -> (OutputDelegate, then, Widget:User Widget Object Reference) {is=title K2Node_CustomEvent}
   > Set IsListFilledWithData(IsListFilledWithData:bool=true) -> (Output_Get:bool)
   > RequestRefreshFocus(self:Common Activatable Widget Object Reference)
   > SetSelectedIndex(self=Get ListView(self=Get WBP_Poly_MegaList()), Index=0)
   Event PreConstruct()
   > SetVisibility(self=Get ListView(self=Get WBP_Poly_MegaList()), InVisibility=SelfHitTestInvisible)
   > Set IsListFilledWithData(IsListFilledWithData=false)
   Event Construct()
   > SetMobileStyle() {is=title K2Node_CallFunction}
   n15 = OR Boolean(
    A=GetForcedInterestsShouldUseSafeZoneForNonMobile!(self:Discover Interests Surface View Model Object Reference=Get FortDiscoverInterestsSurfaceViewModel()) -> (ReturnValue:bool)
    B=IsMobileGame()
   )
   > SetSidesToPad!(
    execute
    self:Safe Zone Object Reference=Get SafeZone_Mobile() -> (SafeZone_Mobile:Safe Zone Object Reference)
    InPadLeft:bool=@n15
    InPadRight:bool=@n15
    InPadTop:bool=@n15
    InPadBottom:bool=@n15
   ) -> (then)
   n16 = Get FortDiscoverInterestsSurfaceViewModel()
   > Branch(Condition=ShouldQueryOnConstruct!(self:Discover Surface View Model Object Reference=@n16) -> (ReturnValue:bool))
   > Set SurfaceName!(execute, SurfaceName:Name=CreativeDiscoverySurface_Interests, self:Discover Surface View Model Object Reference=@n16) -> (then, Output_Get:Name=None)
   Event On Activated()
   > Broadcast Event!(
    execute
    self:Gameplay Event Router Component Object Reference=GetGlobalEventRouter()
    EventData:BPS Set Text Chat Active State Structure=Make BPS Set Text Chat Active State!(bSetActive_5_#1:bool=False) -> (BPS_SetTextChatActiveState:BPS Set Text Chat Active State Structure) {is=title K2Node_MakeStruct}
    EventContext:Object Reference=GetOwningPlayer()
   ) -> (then, ReturnValue:bool)
   n17 = Get FortDiscoverInterestsSurfaceViewModel()
   > n18 = Branch(Condition=ShouldQueryOnConstruct(self=@n17))
   StartAnalyticsImpressions(execute=[@n18.then, @n21.then, @n20.else], self:Discover Surface View Model Object Reference=Get FortDiscoverInterestsSurfaceViewModel())
   n19 = Get FortDiscoverInterestsSurfaceViewModel()
   > UpdateCachedSurfaceData!(execute, self:Discover Surface View Model Object Reference=@n19, bShouldForceUpdate:bool=ShouldQueryOnConstruct(self=@n19)) -> (then)
   n20 = Branch(
    execute=@n18.else
    Condition="Equal (Name)"(A=Get SurfaceName!(self:Discover Surface View Model Object Reference=@n17) -> (SurfaceName:Name=None))
   )
   > n21 = Set SurfaceName(SurfaceName=CreativeDiscoverySurface_Interests, self=@n17) -> (Output_Get=None)
   Event OnInitialized()
   n22 = Get ListView(self=Get WBP_Poly_MegaList())
   > Bind Event to On Entry Initialized(self:List View Object Reference=@n22, Delegate=Create Event()) {is=title K2Node_AddDelegate}
   > Assign On Entry Generated!(execute:Exec, self:List View Base Object Reference=@n22, Delegate:"Delegate (by ref)"=Create Event()) -> (then:Exec) {is=title K2Node_AssignDelegate}
   Event On Deactivated()
   > Broadcast Event(
    self=GetGlobalEventRouter()
    EventData=Make BPS Set Text Chat Active State(bSetActive_5_#1=true)
    EventContext=GetOwningPlayer()
   )
   n23 = Get FortDiscoverInterestsSurfaceViewModel()
   > StopVisibleImpression(self:Discover Provider View Model Object Reference="Get (a copy)"(Array:Array of Discover Provider View Model Object References=Get Providers!(self:Discover Surface View Model Object Reference=@n23) -> (Providers:Array of Discover Provider View Model Object References)) -> (Output:Discover Provider View Model Object Reference))
   > FireExitedImpressionsFromSurface(self:Discover Surface View Model Object Reference=Get FortDiscoverInterestsSurfaceViewModel())
   Unbind all Events from Dialog Button Clicked Dispatcher!(execute:Exec=[@n24.then, @n25.then], self:WBP UIKit Dialog Internal Object Reference=@n9) -> (then:Exec) {is=title K2Node_ClearDelegate}
   Get FortDiscoverSurfaceViewModel() -> (FortDiscoverSurfaceViewModel:Discover Surface View Model Object Reference)
   Set IsSkipModalDeactivated(execute=@n13.then, IsSkipModalDeactivated:bool=false) -> (Output_Get:bool)
   n24 = RequestRefreshFocus(execute=@n26.then)
   SkipInterestSelection(execute=@n27.0, self=Get FortDiscoverInterestsSurfaceViewModel())
   > n25 = DeactivateWidget(self=@n8)
   DeactivateWidget(execute=@n27.1, self=@n8)
   > n26 = Set IsSkipModalDeactivated(IsSkipModalDeactivated=true)
   n27 = Switch on Int!(execute:Exec=@n11.then, Selection:int=@n11.ButtonIndex) -> (0:Exec, 1:Exec) {bHasDefaultPin=false}
  ]}
 FromSaveButton {nodes [
   FromSaveButton!() -> (then, Navigation:EUINavigation Enum) {is=title K2Node_FunctionEntry}
   > Return Node(ReturnValue=Get Button_Skip())
  ]}
 Interest Selection Complete {nodes [Interest Selection Complete() {is=title K2Node_FunctionEntry}]}
 OnProcessingInterestStateChanged {nodes [
   n1 = OnProcessingInterestStateChanged!() -> (then, ProcessingState:EProcessingInterestsState Enum=None) {is=title K2Node_FunctionEntry}
   > n2 = Switch on EProcessingInterestsState!(execute, Selection:EProcessingInterestsState Enum=@n1) -> (None:Exec, Loading:Exec, Completed:Exec, Failed:Exec) {is=title K2Node_SwitchEnum}
   n3 = FindParentWidgetOfType!(
    execute=[@n2.Completed, @n2.Failed]
    StartingWidget:Widget Object Reference=Self-Reference()
    Type:Widget Class Reference=/FortniteCoreUI/Frontend/WBP_InterestsSelectionModal.WBP_InterestsSelectionModal_C
   ) -> (then, ReturnValue:WBP Interests Selection Modal Object Reference)
   > Is Valid(InputObject=@n3)
   > DeactivateWidget(self=@n3)
   SetActiveWidget(
    execute=@n2.None
    self=Get S_Content() -> (S_Content:Common Visibility Switcher Object Reference)
    Widget=Get VB_Content() -> (VB_Content:Vertical Box Object Reference)
   )
   SetActiveWidget(execute=@n2.Loading, self=Get S_Content(), Widget=Get VB_Loading() -> (VB_Loading:Vertical Box Object Reference))
  ]}
 OnUpdateProviders {nodes [
   n1 = OnUpdateProviders!() -> (then, Provider VMs:Array of Discover Provider View Model Object References) {is=title K2Node_FunctionEntry}
   n2 = Get ListView(self=Get WBP_Poly_MegaList())
   > ClearListItems(self=@n2)
   > Branch(Condition=Is Not Empty!(TargetArray:Array of Discover Provider View Model Object References=@n1) -> (ReturnValue:bool))
   > Set List Items!(execute, self:List View Object Reference=@n2, InListItems:Array of Object References=@n1) -> (then) {is=title K2Node_CallFunction}
  ]}
 SetMobileStyle {nodes [
   SetMobileStyle() {is=title K2Node_FunctionEntry}
   > Branch(Condition=IsMobileGame())
   > SetPadding!(
    execute
    self:Vertical Box Slot Object Reference=SlotAsVerticalBoxSlot(Widget=Get VerticalBox_Footer() -> (VerticalBox_Footer:Vertical Box Object Reference))
    InPadding_Left:float=80
    InPadding_Top:float=48
    InPadding_Right:float=48
    InPadding_Bottom:float=0
   ) -> (then)
  ]}
 SetSaveButtonEnabled {nodes [
   n1 = SetSaveButtonEnabled!() -> (then, IsEnabled:bool=~) {is=title K2Node_FunctionEntry}
   > SetIsEnabled(self=Get Button_SaveSelections(), bInIsEnabled=@n1)
  ]}
 ToListView {nodes [
   ToListView!() -> (then, Navigation:EUINavigation Enum) {is=title K2Node_FunctionEntry}
   n1 = Get ListView(self=Get WBP_Poly_MegaList())
   > n2 = Cast To WBP_Row_Poly_TileStackWrapper(Object="Get (a copy)"(Array:Array of User Widget Object References=GetDisplayedEntryWidgets(self=@n1)) -> (Output:User Widget Object Reference))
   n3 = Get Button_Skip()
   n4 = Get TileView(self=Get WBP_Row_PanelList(self=@n2))
   n5 = "To Float (Integer)"(InInt=GetNumItems!(self:List View Object Reference=@n4) -> (ReturnValue:int=0))
   n6 = float - float(A=@n5, B=1)
   n7 = GetNumberItemsPerRow!(self:Fort Tile View Object Reference=@n4) -> (ReturnValue:int=0)
   n8 = "Division (Whole and Remainder)"!(Dividend:double=@n5, Divisor:double="To Float (Integer)"(InInt=@n7)) -> (Remainder:double, ReturnValue:int=0)
   n9 = int * int(A=@n8.ReturnValue, B=@n7)
   n10 = int * int(A=int - int(A=@n8.ReturnValue), B=@n7)
   n11 = "Equal (Float)"(A=@n8)
   > SetSelectedIndex(
    self=@n4
    Index=Select(
     Option 0:int=Select(Option 0=@n9, Option 1=@n10, Index=@n11)
     Option 1:int=Floor(A=@n6) {is=title K2Node_CallFunction}
     Index=IsHovered(self=@n3)
    ) -> (ReturnValue:int)
   )
   > Navigate To Item(self=@n4, Item=Get Selected Item(self=@n4))
   > Return Node(ReturnValue=@n4)
  ]}
 ToSaveButton {nodes [
   ToSaveButton!() -> (then, Navigation:EUINavigation Enum) {is=title K2Node_FunctionEntry}
   > Branch(Condition=GetIsEnabled!(self:Widget Object Reference=Get Button_SaveSelections()) -> (ReturnValue:bool))
   > Return Node(ReturnValue=Get Button_SaveSelections())
  ]}
 UpdateRowHeaderStyle {nodes [
   n1 = UpdateRowHeaderStyle!() -> (then, Object:Object Reference) {is=title K2Node_FunctionEntry}
   > n2 = Cast To WBP_Row_Poly_TileStackWrapper(Object=@n1)
   n3 = Get WBP_Row_Header!(self:WBP Row Panel List Object Reference=Get WBP_Row_PanelList(self=@n2)) -> (WBP_Row_Header:WBP Row Header Object Reference)
   > UpdateWidgetStyleTags!(
    execute
    InWidget:Widget Object Reference=Get Text_HeaderRow!(self:WBP Row Header Object Reference=@n3) -> (Text_HeaderRow:Common Text Object Reference)
    InAddedTags:Array of Gameplay Tag Structures=Get AddedTagsHeader() -> (AddedTagsHeader:Array of Gameplay Tag Structures)
    InRemovedTags:Array of Gameplay Tag Structures=Get RemovedTagsHeader() -> (RemovedTagsHeader:Array of Gameplay Tag Structures)
   ) -> (then, ReturnValue:bool)
   > UpdateWidgetStyleTags(
    InWidget=Get SubText_HeaderRow!(self:WBP Row Header Object Reference=@n3) -> (SubText_HeaderRow:Common Text Object Reference)
    InAddedTags=Get AddedTagsSubHeader() -> (AddedTagsSubHeader:Array of Gameplay Tag Structures)
    InRemovedTags=Get RemovedTagsSubHeader() -> (RemovedTagsSubHeader:Array of Gameplay Tag Structures)
   )
   n4 = Get WBP_Row_PanelList(self=@n2)
   > SetGridSlotPadding!(
    execute
    Widget:Widget Object Reference=Get OverlayRoot!(self:WBP Row Panel List Object Reference=@n4) -> (OverlayRoot:Overlay Object Reference)
    Padding_Left:float=0
    Padding_Top:float=SelectFloat(A=48, B=72, bPickA=IsMobileGame())
    Padding_Right:float=0
    Padding_Bottom:float=0
   ) -> (then)
   > Set ForcedMinColumnCount!(
    execute
    ForcedMinColumnCount:int=GetForcedInterestsMinColumnCount!(self:Discover Interests Surface View Model Object Reference=Get FortDiscoverInterestsSurfaceViewModel()) -> (ReturnValue:int=0)
    self:WBP Row Panel List Object Reference=@n4
   ) -> (then, Output_Get:int=0)
   > SetPadding!(
    execute
    self:Overlay Slot Object Reference=SlotAsOverlaySlot(Widget=Get SizeBox_CapMaxRows!(self:WBP Row Panel List Object Reference=@n4) -> (SizeBox_CapMaxRows:Size Box Object Reference))
    InPadding:Margin Structure=Make Margin(Left=0, Top=0, Right=0, Bottom=0)
   ) -> (then)
  ]}
]
variables [
 Interest Selection Complete {}
 IsListFilledWithData {type=boolean}
 IsSkipModalDeactivated {type=boolean}
 ModalButtons {
  type=S_UI_DialogButtons
  default {
   each Button Horizontal Fill=false
   each Button Min Width=200
   each Button Size=Medium
   focuse Button=1
   space Between Buttons=24
   cTA Buttons Types [Regular,Quiet]
   group Placement {
    horizontal Alignment=HAlign_Fill
    vertical Alignment=VAlign_Bottom
    grid {column=1, column Span=0, layer=0, row=2, row Span=0, nudge=<0 0>}
    padding=<ltrb 80 64 80 48>
   }
  }
 }
]
widgets {
 info {
  inheritedWidgetCount=0
  parentClass=&CommonActivatableWidget
  rootWidgetClass=/Script/FortniteUI.FortSwipePanel
  widgetCount=16
 }
 widgets [
  SwipePanel {
   widgetClassPath=&FortSwipePanel
   bIsVariable=true
   widget=SwipePanel
   properties {bBeginSwipeOnPointerEnter=false, bConsumePointerInput=false, thresholdUnit=Inches, swipeThreshold=<0 0.25>}
  }
  GridPanel_0 {
   widgetClassPath=&GridPanel
   parent=SwipePanel
   widget=GridPanel_0
   properties {columnFill [1], rowFill [1]}
   slot {class=FortSwipePanelSlot}
  }
  NebulousBackground {
   widgetClassPath=&Image
   parent=GridPanel_0
   widget=NebulousBackground
   properties {
    visibility=HitTestInvisible
    brush {
     bIsDynamicallyLoaded=false
     drawAs=Image
     imageType=FullColor
     mirroring=NoMirror
     resourceName=None
     resourceObject=/DiscoveryBrowser/TabScreens/DiscoverHome/Materials/MI_UI_Discovery_ImmersiveBackground
     tiling=NoTile
     imageSize=<32 32>
     margin=<ltrb 0 0 0 0>
     outlineSettings {
      bUseBrushTransparency=false
      roundingType=HalfHeightRadius
      width=0
      color {colorUseRule=UseColor_Specified, specifiedColor=<rgba 0 0 0 0>}
      cornerRadii=<0 0 0 0>
     }
     tintColor {colorUseRule=UseColor_Specified, specifiedColor=<rgba 1 1 1 1>}
     uVRegion {bIsValid=false, max=<0 0>, min=<0 0>}
    }
   }
   slot {class=GridSlot}
  }
  WBP_CaptureForPostBufferUpdate {
   widgetClassPath=&WBP_CaptureForPostBufferUpdate
   bIsVariable=true
   parent=GridPanel_0
   slot {class=GridSlot, properties {layer=5}}
  }
  SafeZone_Mobile {
   widgetClassPath=&SafeZone
   bIsVariable=true
   parent=GridPanel_0
   widget=SafeZone_Mobile
   slot {class=GridSlot, properties {layer=10}}
  }
  S_Content {
   widgetClassPath=&CommonVisibilitySwitcher
   bIsVariable=true
   parent=SafeZone_Mobile
   widget=S_Content
   slot {class=SafeZoneSlot}
  }
  VB_Content {
   widgetClassPath=&VerticalBox
   bIsVariable=true
   parent=S_Content
   widget=VB_Content
   slot {class=CommonVisibilitySwitcherSlot, properties {padding=<ltrb 0 120 0 32>}}
  }
  WBP_Poly_MegaList {
   widgetClassPath=&/WBP_Poly_MegaList
   bIsVariable=true
   parent=VB_Content
   widget=WBP_Poly_MegaList
   properties {navigation=&self:WidgetTree.WBP_Poly_MegaList.WidgetNavigation_2}
   slot {class=VerticalBoxSlot, properties {size {sizeRule=Fill, value=1}}}
  }
  VerticalBox_Footer {
   widgetClassPath=&VerticalBox
   bIsVariable=true
   parent=VB_Content
   widget=VerticalBox_Footer
   slot {class=VerticalBoxSlot, properties {padding=<ltrb 80 72 80 0>}}
  }
  CommonRichTextBlock {
   widgetClassPath=&CommonRichTextBlock
   parent=VerticalBox_Footer
   widget=CommonRichTextBlock
   properties {
    bIsScrollingEnabled=false
    bOverrideDefaultStyle=true
    mobileTextBlockScale=1.14
    text="Manage favorites from your <img id=\"Library\"></> Library."
    decoratorClasses [/DiscoveryBrowser/RichStyles/RichTextBlockDecorator_Interests.RichTextBlockDecorator_Interests_C]
    defaultTextStyleOverride {
     overflowPolicy=Clip
     transformPolicy=None
     colorAndOpacity {colorUseRule=UseColor_Specified, specifiedColor=<rgba 1 1 1 1>}
     font {
      bForceMonospaced=false
      bMaterialIsStencil=false
      fontMaterial=null
      fontObject=&HeadingNow
      letterSpacing=10
      monospacedWidth=1
      size=15.75
      skewAmount=0
      typefaceFontName=74 Regular
      outlineSettings {
       bApplyOutlineToDropShadows=false
       bMiteredCorners=false
       bSeparateFillAlpha=false
       outlineMaterial=null
       outlineSize=0
       outlineColor=<rgba 0 0 0 1>
      }
     }
     highlightColor {colorUseRule=UseColor_Specified, specifiedColor=<rgba 0 0 0 1>}
     highlightShape {
      bIsDynamicallyLoaded=false
      drawAs=Image
      imageType=NoImage
      mirroring=NoMirror
      resourceName=None
      resourceObject=null
      tiling=NoTile
      imageSize=<32 32>
      margin=<ltrb 0 0 0 0>
      outlineSettings {
       bUseBrushTransparency=false
       roundingType=HalfHeightRadius
       width=0
       color {colorUseRule=UseColor_Specified, specifiedColor=<rgba 0 0 0 0>}
       cornerRadii=<0 0 0 0>
      }
      tintColor {colorUseRule=UseColor_Specified, specifiedColor=<rgba 1 1 1 1>}
      uVRegion {bIsValid=false, max=<0 0>, min=<0 0>}
     }
     selectedBackgroundColor {colorUseRule=UseColor_Foreground, specifiedColor=<rgba 1 0 1 1>}
     shadowColorAndOpacity=<rgba 0 0 0 1>
     shadowOffset=<0 0>
     strikeBrush {
      bIsDynamicallyLoaded=false
      drawAs=Image
      imageType=NoImage
      mirroring=NoMirror
      resourceName=None
      resourceObject=null
      tiling=NoTile
      imageSize=<32 32>
      margin=<ltrb 0 0 0 0>
      outlineSettings {
       bUseBrushTransparency=false
       roundingType=HalfHeightRadius
       width=0
       color {colorUseRule=UseColor_Specified, specifiedColor=<rgba 0 0 0 0>}
       cornerRadii=<0 0 0 0>
      }
      tintColor {colorUseRule=UseColor_Specified, specifiedColor=<rgba 1 1 1 1>}
      uVRegion {bIsValid=false, max=<0 0>, min=<0 0>}
     }
     underlineBrush {
      bIsDynamicallyLoaded=false
      drawAs=Image
      imageType=NoImage
      mirroring=NoMirror
      resourceName=None
      resourceObject=null
      tiling=NoTile
      imageSize=<32 32>
      margin=<ltrb 0 0 0 0>
      outlineSettings {
       bUseBrushTransparency=false
       roundingType=HalfHeightRadius
       width=0
       color {colorUseRule=UseColor_Specified, specifiedColor=<rgba 0 0 0 0>}
       cornerRadii=<0 0 0 0>
      }
      tintColor {colorUseRule=UseColor_Specified, specifiedColor=<rgba 1 1 1 1>}
      uVRegion {bIsValid=false, max=<0 0>, min=<0 0>}
     }
    }
   }
   slot {class=VerticalBoxSlot, properties {horizontalAlignment=HAlign_Left, padding=<ltrb 0 0 0 16>}}
  }
  HorizontalBox_219 {widgetClassPath=&HorizontalBox, parent=VerticalBox_Footer, widget=HorizontalBox_219, slot {class=VerticalBoxSlot}}
  Button_SaveSelections {
   widgetClassPath=/FortUILibrary/UIKit/Buttons/WBP_UIKit_Button_Loud
   bIsVariable=true
   parent=HorizontalBox_219
   widget=Button_SaveSelections
   properties {
    bHideInputAction=true
    minHeight=88
    minWidth=340
    navigation=&self:WidgetTree.Button_SaveSelections.WidgetNavigation_0
    text=Save Selections
   }
   slot {class=HorizontalBoxSlot, properties {horizontalAlignment=HAlign_Left, size {sizeRule=Fill, value=1}}}
  }
  Button_Skip {
   widgetClassPath=&WBP_UIKit_Button_Regular
   bIsVariable=true
   parent=HorizontalBox_219
   widget=Button_Skip
   properties {
    bHideInputAction=true
    buttonType=Quiet
    minWidth=165
    navigation=&self:WidgetTree.Button_Skip.WidgetNavigation_0
    text=Skip
   }
   slot {class=HorizontalBoxSlot}
  }
  VB_Loading {
   widgetClassPath=&VerticalBox
   bIsVariable=true
   parent=S_Content
   widget=VB_Loading
   slot {class=CommonVisibilitySwitcherSlot, properties {horizontalAlignment=HAlign_Center, verticalAlignment=VAlign_Center}}
  }
  WBP_UIKit_Throbber {
   widgetClassPath=/FortUILibrary/UIKit/Throbber/WBP_UIKit_Throbber
   parent=VB_Loading
   widget=WBP_UIKit_Throbber
   properties {displayBackground=false, throbberSize=Large}
   slot {class=VerticalBoxSlot, properties {horizontalAlignment=HAlign_Center, verticalAlignment=VAlign_Top}}
  }
  LoadingText {
   widgetClassPath=&CommonTextBlock
   parent=VB_Loading
   widget=LoadingText
   properties {
    style=&TextStyle-Base_C
    text=Curating your experience...
    font {
     bForceMonospaced=false
     bMaterialIsStencil=false
     fontMaterial=null
     fontObject=&NotoSans
     letterSpacing=0
     monospacedWidth=1
     size=24
     skewAmount=0
     typefaceFontName=NotoSans-Regular
     outlineSettings {
      bApplyOutlineToDropShadows=false
      bMiteredCorners=false
      bSeparateFillAlpha=false
      outlineMaterial=null
      outlineSize=0
      outlineColor=<rgba 0 0 0 1>
     }
    }
    shadowOffset=<0 0>
   }
   slot {class=VerticalBoxSlot}
  }
 ]
}
