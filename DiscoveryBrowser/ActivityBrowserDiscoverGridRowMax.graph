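# Discover browser grid row widget that shows a provider view model's tiles as multiple sub-rows inside WBP_Row_PanelList.
# Custom navigation steps between sub-rows and handles edge exits, while focus, hover and weight-adjusted focus offsets drive tile selection.
# The view model's row limit is clamped to available items, and visible sub-rows set the analytics impression range.
# A timer delays the animated loading-more indicator, which plays in reverse before hiding.
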WidgetBlueprint /DiscoveryBrowser/ActivityBrowserDiscoverGridRowMax
animations [OnLoadingMore {
  binding_count=1
  length_seconds=0.24
  object=&self:OnLoadingMore
  sequence {
   bindings [WBP_LoadingMorePages {
     id=#1
     possessed_class=WBP_LoadingMorePages_C
     tracks [
      {
       type=MovieSceneFloatTrack
       display_name=Render Opacity
       sections [{type=MovieSceneFloatSection, channels [None {type=MovieSceneScriptingFloatChannel, keys [[0,0],[24,1]]}]}]
      }
      {
       type=MovieScene2DTransformTrack
       sections [{type=MovieScene2DTransformSection, channels [Translation.X {type=MovieSceneScriptingFloatChannel, keys [[0,16],[24,0]]}]}]
      }
     ]
    }]
   display_rate {numerator=100}
  }
 }]
graphs [
 CustomNavigation {nodes [
   n1 = CustomNavigation!() -> (then, Navigation:EUINavigation Enum) {is=title K2Node_FunctionEntry}
   n2 = Get TileView!(self:Fort Polymorphic Tile View Object Reference=Get WBP_Row_PanelList() -> (WBP_Row_PanelList:WBP Row Panel List Object Reference)) -> (TileView:Fort Tile View Object Reference)
   > Set NumberItemsPerRow(NumberItemsPerRow:int=GetNumberItemsPerRow!(self:Fort Tile View Object Reference=@n2) -> (ReturnValue:int=0)) -> (Output_Get:int=0)
   > Set SelectedItemIndex(SelectedItemIndex:int=GetIndexForItem!(
     self:List View Object Reference=@n2
     Item:Object Reference=Get Selected Item!(self:List View Object Reference=@n2) -> (ReturnValue:Object Reference) {is=title K2Node_CallFunction}
    ) -> (ReturnValue:int=0)) -> (Output_Get:int=0)
   > n3 = Switch on EUINavigation!(execute, Selection:EUINavigation Enum=@n1) -> (Left:Exec, Right:Exec, Up:Exec, Down:Exec, Next:Exec, Previous:Exec, Invalid:Exec) {is=title K2Node_SwitchEnum}
   n4 = Get IsViewAllSelected!(self:WBP Row Panel List Object Reference=Get WBP_Row_PanelList()) -> (IsViewAllSelected:bool)
   n5 = Branch(execute=@n3.Down, Condition=@n4)
   n6 = Get WBP_Row_PanelList()
   > n7 = NavigateWithinGrid!(
    execute
    self:Fort Discover Browser Grid Row Object Reference
    Row:int=GetNumDisplayedRows!(self:WBP Row Panel List Object Reference=@n6) -> (DisplayedRows:int=0)
   ) -> (then)
   OnRowUnselected(execute=[@n7.then, @n10.then], self:WBP Row Panel List Object Reference=@n6)
   n8 = int + int(A=int / int(A=Get SelectedItemIndex() -> (SelectedItemIndex:int=0), B=Get NumberItemsPerRow() -> (NumberItemsPerRow:int=0)))
   NavigateWithinGrid(execute=@n5.else, Row=@n8)
   n9 = Get WBP_Row_PanelList()
   > n10 = Branch(Condition="integer >= integer"(A=@n8, B=GetNumDisplayedRows(self=@n9) -> (DisplayedRows=0)))
   n11 = Get TileView(self=Get WBP_Row_PanelList())
   n12 = Get NumberItemsPerRow() -> (NumberItemsPerRow=0)
   n13 = Get SelectedItemIndex() -> (SelectedItemIndex=0)
   Branch(
    execute=@n10.else
    Condition="integer >= integer"(A=int + int(A=@n13, B=@n12), B=GetNumItems!(self:List View Object Reference=@n11) -> (ReturnValue:int=0))
   )
   n14 = GetDisplayedEntryWidgets!(self:List View Base Object Reference=@n11) -> (ReturnValue:Array of User Widget Object References)
   > Return Node!(
    execute
    ReturnValue:Widget Object Reference="Get (a copy)"(
     Array:Array of User Widget Object References=@n14
     Dimension 1=Last Index!(TargetArray:Array of User Widget Object References=@n14) -> (ReturnValue=0)
    ) -> (Output:User Widget Object Reference)
   )
   n15 = Branch(execute=@n3.Up, Condition=@n4)
   > n16 = NavigateWithinGrid(Row=-1)
   OnRowUnselected(execute=[@n18.then, @n16.then], self=Get WBP_Row_PanelList())
   > Branch(Condition="Equal (Integer)"(A=Get ItemIndexRepresented() -> (ItemIndexRepresented:int=0)))
   > Call On Upper Boundary Hit!(execute:Exec, self:Activity Browser Discover Grid Row Max Object Reference) -> (then:Exec) {is=title K2Node_CallDelegate}
   n17 = int - int(A=int / int(A=Get SelectedItemIndex() -> (SelectedItemIndex=0), B=Get NumberItemsPerRow() -> (NumberItemsPerRow=0)))
   NavigateWithinGrid(execute=@n15.else, Row=@n17)
   > n18 = Branch(Condition="integer < integer"(A=@n17))
  ]}
 EventGraph {nodes [
   n1 = FinishedLoadingMoreDelay() {is=title K2Node_CustomEvent}
   > Branch(Condition=Get IsLoadingMoreQueryActive() -> (IsLoadingMoreQueryActive:bool))
   > n2 = Branch(Condition=Get IsLoadingMoreVisible() -> (IsLoadingMoreVisible:bool))
   Set IsLoadingMoreVisible(execute=@n2.else, IsLoadingMoreVisible:bool=true) -> (Output_Get:bool)
   > PlayAnimationForward(InAnimation=Get OnLoadingMore() -> (OnLoadingMore:Widget Animation Object Reference))
   > SetVisibility(self=Get WBP_LoadingMorePages() -> (WBP_LoadingMorePages:WBP Loading More Pages Object Reference), InVisibility=SelfHitTestInvisible)
   n3 = OnMaxRowsSet() {is=title K2Node_CustomEvent}
   > n4 = ValidateMaxVisibleRows!(execute, self) -> (then, ValidatedMaxVisibleRows:int=0)
   > SetMaxVisibleRows!(execute, self:Fort Discover Browser Grid Row Object Reference, MaxRowsShown:int=@n4) -> (then)
   > Set MaxRowsCached!(execute, MaxRowsCached:int=@n4, self:WBP Row Panel List Object Reference=Get WBP_Row_PanelList()) -> (then, Output_Get:int=0)
   n5 = OnTileItemGenerated!() -> (OutputDelegate, then, Widget:User Widget Object Reference) {is=title K2Node_CustomEvent}
   > SetNestedButtonGroupSelectionVM!(
    execute:Exec
    self:Object Reference=@n5.Widget
    InVM:Fort Nested Button Group Selection VM Object Reference=Get NestedButtonGroupSelectionVM() -> (NestedButtonGroupSelectionVM:Fort Nested Button Group Selection VM Object Reference)
   ) -> (then:Exec) {is=title K2Node_Message}
   n6 = OnTileFocusGained!() -> (OutputDelegate, then, ItemFocused:int) {is=title K2Node_CustomEvent}
   n7 = GetCurrentInputType(self=Get CommonInputSubsystem())
   > n8 = Branch(Condition=NOT Boolean(A="Equal (Enum)"(A=@n7, B=MouseAndKeyboard)))
   n9 = Get WBP_Row_PanelList()
   n10 = Get TileView(self=@n9)
   > n11 = NavigateWithinGrid(Row=int / int(A=@n6.ItemFocused, B=GetNumberItemsPerRow(self=@n10) -> (ReturnValue=0)))
   SelectItem!(execute=[@n11.then, @n8.else], self, IndexToSelect:int=@n6.ItemFocused) -> (then)
   n12 = OnItemHovered!() -> (OutputDelegate, then, Item:Object Reference, bIsHovered:bool=~) {is=title K2Node_CustomEvent}
   > Branch(Condition=@n12.bIsHovered)
   n13 = Get WBP_Row_PanelList()
   n14 = Get TileView(self=@n13)
   > SelectItem(IndexToSelect=GetIndexForItem(self=@n14, Item=@n12.Item) -> (ReturnValue=0))
   OnSendProviderViewModel() {is=title K2Node_CustomEvent}
   > OpenViewAllScreen!(
    execute
    self:WBP Row Panel List Object Reference=Get WBP_Row_PanelList()
    ProviderViewModel:Discover Provider View Model Object Reference=Get FortDiscoverProviderViewModel() -> (FortDiscoverProviderViewModel:Discover Provider View Model Object Reference)
   ) -> (then)
   n15 = Event OnRowIsSelectedChanged!() -> (OutputDelegate, then, bIsSelected:bool) {is=title K2Node_Event}
   > Call On Row Selection Changed!(
    execute:Exec
    self:Activity Browser Discover Grid Row Max Object Reference
    Row:Activity Browser Discover Grid Row Max Object Reference=Self-Reference()
    IsSelected:bool=@n15.bIsSelected
    IsViewAllSelected:bool=Get IsViewAllSelected(self=Get WBP_Row_PanelList())
   ) -> (then:Exec) {is=title K2Node_CallDelegate}
   Event OnOverrideItemSelectorSet!() -> (OutputDelegate, then, ItemSelector:Fort Discover Item Selector Base Object Reference) {is=title K2Node_Event}
   > UpdateItemSelectorInPanelList()
   n16 = Event SetPanelViewModel!() -> (OutputDelegate, then, ViewModel:Discover Provider View Model Object Reference) {is=title K2Node_Event}
   > SetFortDiscoverProviderViewModel!(execute, self, Viewmodel:Discover Provider View Model Object Reference=@n16.ViewModel) -> (then)
   n17 = Event SetFocusOffset!() -> (OutputDelegate, then, FocusOffset:int=0, bSelectFromBelow:bool) {is=title K2Node_Event}
   n18 = GetOwningPlayer()
   > Branch(Condition=NOT Boolean(A=IsUsingTouch(PlayerController=@n18)))
   > n19 = GetRowAndColumnIndexes!(execute, self, FocusOffset:int=@n17.FocusOffset, SelectFromBelow:bool=@n17.bSelectFromBelow) -> (then, RowIndex:int=0, ColumnIndex:int=0)
   > n20 = GetItemIndexInRowAndColumn!(execute, self, RowIndex:int=@n19, ColumnIndex:int=@n19.ColumnIndex) -> (then, ItemIndex:int=0)
   > SetSelectedIndex!(execute, self:List View Object Reference=Get TileView(self=Get WBP_Row_PanelList()), Index:int=@n20) -> (then)
   n21 = Event OpenNestedSurface!() -> (OutputDelegate, then, SurfaceTileItemVM:Fort Discover Surface Tile Item VM Object Reference) {is=title K2Node_Event}
   > Parent: OpenNestedSurface!(execute, SurfaceTileItemVM:Fort Discover Surface Tile Item VM Object Reference=@n21.SurfaceTileItemVM) -> (then) {is=title K2Node_CallParentFunction}
   n22 = Event HandleOnInputMethodChanged!() -> (OutputDelegate, then, bNewInputType:ECommonInputType Enum) {is=title K2Node_Event}
   > Parent: HandleOnInputMethodChanged!(execute, bNewInputType:ECommonInputType Enum=@n22.bNewInputType) -> (then) {is=title K2Node_CallParentFunction}
   Event On Entry Released() {is=title K2Node_Event}
   > Parent: On Entry Released() {is=title K2Node_CallParentFunction}
   > SetIsSelectedNestedVM!(execute, self:WBP Row Grid Base Object Reference, ListContainer:List View Base Object Reference=Get TileView(self=Get WBP_Row_PanelList())) -> (then)
   Event Construct()
   > Parent: Construct()
   > Bind Event to On Tile Focus Gained(self:WBP Row Panel List Object Reference=@n9, Delegate=@n6) {is=title K2Node_AddDelegate}
   > Bind Event to On View All Send Provider View Model(self:WBP Row Panel List Object Reference=@n13, Delegate=Create Event()) {is=title K2Node_AddDelegate}
   > Bind Event to On Item Is Hovered Changed(self:List View Object Reference=@n14, Delegate=@n12) {is=title K2Node_AddDelegate}
   > Bind Event to On Entry Generated(self:List View Base Object Reference=Get TileView(self=Get WBP_Row_PanelList()), Delegate=Create Event()) {is=title K2Node_AddDelegate}
   > n23 = GetSupportsLessThanMaxVisibleRows!(execute, self:Fort Discover Browser Grid Row Object Reference) -> (then, ReturnValue:bool)
   > Branch(Condition=@n23)
   > Bind Event to On Max Rows Set(self:WBP Row Panel List Object Reference=Get WBP_Row_PanelList(), Delegate=@n3) {is=title K2Node_AddDelegate}
   > UpdateItemSelectorInPanelList()
   Event Destruct()
   > Parent: Destruct()
   > Unbind Event from On Entry Generated(self:List View Base Object Reference=Get TileView(self=Get WBP_Row_PanelList()), Delegate=Create Event()) {is=title K2Node_RemoveDelegate}
   K2Node_Event_7 = OnQueryStatusChanged_1!() -> (OutputDelegate, then, bIsActive:bool=~) {is=title K2Node_CustomEvent}
   > n24 = Set IsLoadingMoreQueryActive(IsLoadingMoreQueryActive:bool=false) -> (Output_Get:bool)
   > n25 = Branch(Condition=@n24)
   > n26 = Set Timer by Event(Delegate=@n1, Time=Get LoadingMoreDisplayDelay() -> (LoadingMoreDisplayDelay:double), bLooping=false)
   > Set LoadingMoreDelayTimer(LoadingMoreDelayTimer:Timer Handle Structure=@n26) -> (Output_Get:Timer Handle Structure)
   Clear and Invalidate Timer by Handle(execute=@n25.else, Handle=Get LoadingMoreDelayTimer() -> (LoadingMoreDelayTimer:Timer Handle Structure))
   > Branch(Condition=Get IsLoadingMoreVisible())
   > Set IsLoadingMoreVisible(IsLoadingMoreVisible=false)
   n27 = Get OnLoadingMore()
   n28 = GetAnimationCurrentTime!(self:User Widget Object Reference, InAnimation:Widget Animation Object Reference=@n27) -> (ReturnValue:float)
   n29 = GetEndTime(self=@n27)
   > n30 = "Play Animation with Finished event (legacy)"!(
    execute:Exec
    Widget:User Widget Object Reference=Self-Reference()
    InAnimation:Widget Animation Object Reference=@n27
    StartAtTime:float=SelectFloat(A=float - float(A=@n29, B=@n28) {is=title K2Node_CallFunction}, B=0, bPickA=IsAnimationPlaying(InAnimation=@n27))
    NumLoopsToPlay:int=1
    PlayMode:EUMGSequencePlayMode Enum=Reverse
    PlaybackSpeed:float=1
   ) -> (then:Exec, Result:UMGSequence Player Object Reference, Finished:Exec) {is=title K2Node_PlayAnimation}
   n31 = Branch(execute=@n30.Finished, Condition=Get IsLoadingMoreVisible())
   SetVisibility(execute=@n31.else, self=Get WBP_LoadingMorePages(), InVisibility=Hidden)
   n32 = Event UpdateAnalyticsStates!() -> (OutputDelegate, then, bStopImpressions:bool) {is=title K2Node_Event}
   n33 = Get FortDiscoverProviderViewModel()
   > Is Valid(InputObject=@n33)
   > StopVisibleImpression(self:Discover Provider View Model Object Reference=@n33)
   > n34 = Branch(Condition=@n32.bStopImpressions)
   StartVisibleImpressions(execute=@n34.else, self:Discover Provider View Model Object Reference=@n33)
   n35 = GetNumberItemsPerRow(self=Get TileView(self=Get WBP_Row_PanelList())) -> (ReturnValue=0)
   n36 = GetFirstVisibleSubRow!(self:Fort Discover Browser Grid Row Object Reference) -> (ReturnValue:int=0)
   n37 = int + int(A=@n36, B=GetNumberOfVisibleSubRows!(self:Fort Discover Browser Grid Row Object Reference) -> (ReturnValue:int=0))
   > SetVisibleImpressions!(
    execute
    self:Discover Provider View Model Object Reference=Get FortDiscoverProviderViewModel()
    FirstVisibleIndex:int=int * int(A=@n36, B=@n35)
    LastVisibleIndex:int=int - int(A=int * int(A=@n37, B=@n35))
   ) -> (then)
   GetIsActive!(self:Fort Activity Browser Row Object Reference) -> (ReturnValue:bool)
   Event OnCategoryItemChanged!() -> (OutputDelegate, then, bPlayAnimation:bool, CategoryPanelName:"Name (by ref)"=None) {is=title K2Node_Event}
   Event OnRowPeekStateChanged!() -> (OutputDelegate, then, bIsInPeekState:bool) {is=title K2Node_Event}
   Event OnRowMoveUp!() -> (OutputDelegate, then, bMovingOffscreen:bool) {is=title K2Node_Event}
   Event OnRowIsActiveChanged!() -> (OutputDelegate, then, bIsActiveRow:bool) {is=title K2Node_Event}
   Event OnRowMoveDown!() -> (OutputDelegate, then, bMovingOffscreen:bool) {is=title K2Node_Event}
  ]}
 GetItemIndexInRowAndColumn {nodes [
   n1 = GetItemIndexInRowAndColumn!() -> (then, RowIndex:int, ColumnIndex:int) {is=title K2Node_FunctionEntry}
   n2 = Get TileView(self=Get WBP_Row_PanelList())
   n3 = GetNumItems(self=@n2) -> (ReturnValue=0)
   n4 = GetNumberItemsPerRow(self=@n2) -> (ReturnValue=0)
   n5 = "Min (Integer)"(A=@n1.ColumnIndex, B=int - int(A=@n4)) -> (ReturnValue=0)
   n6 = int * int(A=@n1, B=@n4)
   > Return Node!(execute, ItemIndex:int="Min (Integer)"(A=int + int(A=@n6, B=@n5), B=int - int(A=@n3)) -> (ReturnValue=0))
  ]}
 GetNumDisplayedSubRows {nodes [
   GetNumDisplayedSubRows() {is=title K2Node_FunctionEntry}
   > Return Node!(execute, ReturnValue:int=GetNumDisplayedRows(self=Get WBP_Row_PanelList()) -> (DisplayedRows=0))
  ]}
 GetNumItemsPerRow {nodes [
   GetNumItemsPerRow() {is=title K2Node_FunctionEntry}
   n1 = Get WBP_Row_PanelList()
   > Return Node(ReturnValue=GetNumberItemsPerRow(self=Get TileView(self=@n1)) -> (ReturnValue=0))
  ]}
 GetNumRowsForItems {nodes [
   n1 = GetNumRowsForItems!() -> (then, AllowNonFullRows:bool=~) {is=title K2Node_FunctionEntry}
   > n2 = Is Valid(InputObject=Get FortDiscoverProviderViewModel())
   Return Node!(execute=@n2.Is Not Valid, NumRows:int=0)
   > n3 = Branch(Condition=@n1)
   n4 = Length(TargetArray:Array of Discover Tile Item View Model Object References=Get Items!(self:Discover Provider View Model Object Reference=Get FortDiscoverProviderViewModel()) -> (Items:Array of Discover Tile Item View Model Object References))
   n5 = "To Float (Integer)"(InInt=@n4)
   n6 = GetNumberItemsPerRow(self=Get TileView(self=Get WBP_Row_PanelList())) -> (ReturnValue=0)
   > Return Node(NumRows=Ceil(A=float / float(A=@n5, B:int=@n6)) {is=title K2Node_CallFunction})
   Return Node(execute=@n3.else, NumRows=int / int(A=@n4, B=@n6))
  ]}
 GetPanelViewModel {nodes [
   GetPanelViewModel() {is=title K2Node_FunctionEntry}
   > Return Node!(execute, ReturnValue:Discover Provider View Model Object Reference=Get FortDiscoverProviderViewModel())
  ]}
 GetRowAndColumnIndexes {nodes [
   GetRowAndColumnIndexes!() -> (then, FocusOffset:int, SelectFromBelow:bool=~) {is=title K2Node_FunctionEntry}
   n1 = GetNumDisplayedRows(self=Get WBP_Row_PanelList()) -> (DisplayedRows=0)
   n2 = Get WBP_Row_PanelList()
   n3 = Get WBP_Row_PanelList()
   n4 = int - int(A=GetNumberItemsPerRow(self=Get TileView(self=@n3)) -> (ReturnValue=0))
   n5 = Get FocusOffset() -> (FocusOffset:int=0)
   n6 = Get TileFocusWeight!(self:WBP Row Panel List Object Reference=Get WBP_Row_PanelList()) -> (TileFocusWeight:double)
   n7 = Select(
    Option 0:int=Get FocusOffset() -> (FocusOffset=0)
    Option 1:int=Round(A=float / float(A:int=@n5, B:double=@n6)) {is=title K2Node_CallFunction}
    Index=ShouldUseSizeShiftedFocusOffset!() -> (ReturnValue:bool)
   ) -> (ReturnValue:int)
   > Return Node!(
    execute
    RowIndex:int=SelectInt(A=int - int(A=@n1), B=0, bPickA=Get SelectFromBelow() -> (SelectFromBelow:bool))
    ColumnIndex:int=Select(
     Option 0=@n7
     Option 1=GetNumVisibleRowItems!(self:WBP Row Panel List Object Reference=@n2) -> (NumItems:int=0)
     Index="Equal (Integer)"(A=@n7, B=@n4)
    )
   )
  ]}
 OnRowBroadcastClickToDiscover {nodes [OnRowBroadcastClickToDiscover() {is=title K2Node_FunctionEntry}]}
 OnRowSelectionChanged {nodes [OnRowSelectionChanged!() -> (then, Row:Activity Browser Discover Grid Row Max Object Reference, IsSelected:bool=~, IsViewAllSelected:bool=~) {is=title K2Node_FunctionEntry}]}
 OnUpperBoundaryHit {nodes [OnUpperBoundaryHit() {is=title K2Node_FunctionEntry}]}
 SelectItem {nodes [
   n1 = SelectItem!() -> (then, IndexToSelect:int) {is=title K2Node_FunctionEntry}
   n2 = Get WBP_Row_PanelList()
   n3 = Get TileView(self=@n2)
   > SetSelectedIndex(self=@n3, Index=@n1)
   > NavigateToIndex!(execute, self:List View Object Reference=@n3, Index:int=@n1) -> (then)
   n4 = "% (Integer)"(A=@n1, B=GetNumberItemsPerRow(self=@n3) -> (ReturnValue=0)) -> (ReturnValue=0)
   n5 = float * float(A:int=@n4, B=Get TileFocusWeight(self=@n2)) {is=title K2Node_PromotableOperator}
   > HandleActivitySelected!(
    execute
    self:Fort Discover Browser Grid Row Object Reference
    Index:int=@n1
    IndexInRow:int=Select(Option 0=@n4, Option 1=Round(A=@n5), Index=ShouldUseSizeShiftedFocusOffset())
   ) -> (then)
  ]}
 UpdateItemSelectorInPanelList {nodes [
   UpdateItemSelectorInPanelList() {is=title K2Node_FunctionEntry}
   n1 = Get OverrideItemSelector() -> (OverrideItemSelector:Fort Discover Item Selector Base Object Reference)
   > Is Valid(InputObject=@n1)
   > Set OverrideItemSelector!(
    execute
    OverrideItemSelector:Fort Discover Item Selector Base Object Reference=@n1
    self:WBP Row Panel List Object Reference=Get WBP_Row_PanelList()
   ) -> (then, Output_Get:Fort Discover Item Selector Base Object Reference)
  ]}
 ValidateMaxVisibleRows {nodes [
   ValidateMaxVisibleRows() {is=title K2Node_FunctionEntry}
   > n1 = Is Valid(InputObject=Get FortDiscoverProviderViewModel())
   n2 = Set VMMaxVisibleRows(execute=@n1.Is Not Valid, VMMaxVisibleRows:int=1) -> (Output_Get:int=0)
   n3 = GetNumberItemsPerRow(self=Get TileView(self=Get WBP_Row_PanelList())) -> (ReturnValue=0)
   n4 = Get VMMaxVisibleRows() -> (VMMaxVisibleRows:int=0)
   n5 = Branch(execute=[@n2.then, @n8.then], Condition=AND Boolean(A="integer > integer"(A=@n3), B="integer > integer"(A=@n4)))
   > n6 = GetNumRowsForItems!(execute, self, AllowNonFullRows:bool=Get bAllowNonFullRows() -> (bAllowNonFullRows:bool)) -> (then, NumRows:int=0) {is=title K2Node_CallFunction}
   n7 = Get VMMaxVisibleRows() -> (VMMaxVisibleRows=0)
   > Return Node!(execute, ValidatedMaxVisibleRows:int="Clamp (Integer)"(Value=@n6, Min=1, Max=@n7) -> (ReturnValue=0))
   Return Node(execute=@n5.else, ValidatedMaxVisibleRows=@n7)
   n8 = Set VMMaxVisibleRows(
    execute=@n1.Is Valid
    VMMaxVisibleRows=Get MaxVisibleRows!(self:Discover Provider View Model Object Reference=Get FortDiscoverProviderViewModel()) -> (MaxVisibleRows:int=0)
   ) -> (Output_Get=0)
  ]}
]
variables [
 bAllowNonFullRows {type=boolean}
 EntriesSpacing {type=number, default=32}
 FontHoverAnimateParam {type=string, default=Hover_Animate}
 FontPressedAnimateParam {type=string, default=Pressed_Animate}
 IsLoadingMoreQueryActive {type=boolean}
 IsLoadingMoreVisible {type=boolean}
 IsWaitingForInactiveAnim {type=boolean}
 LoadingMoreDelayTimer {type=TimerHandle}
 LoadingMoreDisplayDelay {type=number, default=0.2}
 OnRowBroadcastClickToDiscover {}
 OnRowSelectionChanged {}
 OnUpperBoundaryHit {}
]
widgets {
 info {
  inheritedWidgetCount=1
  parentClass=/DiscoveryBrowser/Rows/WBP_Row_Grid_Base.WBP_Row_Grid_Base_C
  rootWidgetClass=&GridPanel
  widgetCount=6
 }
 widgets [
  GridRoot_WithCustomNavEvents {
   widgetClassPath=&GridPanel
   bIsVariable=true
   widget=GridRoot_WithCustomNavEvents
   properties {
    navigation=&self:WidgetTree.GridRoot_WithCustomNavEvents.WidgetNavigation_0
    columnFill [1]
    rowFill [1]
   }
  }
  WBP_Row_PanelList {
   widgetClassPath=/DiscoveryBrowser/Rows/WBP_Row_PanelList
   bIsVariable=true
   parent=GridRoot_WithCustomNavEvents
   widget=WBP_Row_PanelList
   properties {tilesHandleTheirOwnStartVisibilityImpression=false, useFocusToTriggerOutlineVisuals=true}
   slot {class=GridSlot}
  }
  FortVisualAttachment_0 {
   widgetClassPath=&FortVisualAttachment
   parent=GridRoot_WithCustomNavEvents
   widget=FortVisualAttachment_0
   properties {contentAnchor=<1 1>}
   slot {class=GridSlot, properties {horizontalAlignment=HAlign_Right, verticalAlignment=VAlign_Top, padding=<ltrb 0 0 -32 8>}}
  }
  WBP_LoadingMorePages {
   widgetClassPath=/DiscoveryBrowser/TabScreens/WBP_LoadingMorePages
   bIsVariable=true
   parent=FortVisualAttachment_0
   widget=WBP_LoadingMorePages
   properties {visibility=Hidden}
   slot {class=SizeBoxSlot}
  }
  HighlightRow {
   widgetClassPath=&Image
   bIsVariable=true
   parent=GridRoot_WithCustomNavEvents
   widget=HighlightRow
   properties {visibility=Collapsed}
   slot {class=GridSlot}
  }
  Text_CategoryName {widgetClassPath=&CommonTextBlock, bInherited=true, widget=null}
 ]
}
