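# Discover browser row widget, built on WBP_Row_Grid_Base, that shows activity tiles in a horizontally scrolling list.
# Tile hover and focus select items and report their in-row position, while non-touch focus offsets map to list indices using scroll offset or first visible item.
# It passes selection view models and item selector overrides down and computes visibility for analytics.
# A timer delays the animated loading-more indicator, which reverses out before hiding.
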WidgetBlueprint /DiscoveryBrowser/ActivityBrowserDiscoverRow_HorizontalScrolling
animations [
 OnLoadingMore {
  binding_count=1
  length_seconds=0.24
  object=&self:OnLoadingMore
  sequence {
   bindings [WBP_LoadingMorePages {
     id=#1
     possessed_class=WBP_LoadingMorePages_C
     tracks [
      {
       type=MovieSceneFloatTrack
       display_name=Render Opacity
       sections [{type=MovieSceneFloatSection, channels [None {type=MovieSceneScriptingFloatChannel, keys [[0,0],[24,1]]}]}]
      }
      {
       type=MovieScene2DTransformTrack
       sections [{type=MovieScene2DTransformSection, channels [Translation.X {type=MovieSceneScriptingFloatChannel, keys [[0,16],[24,0]]}]}]
      }
     ]
    }]
   display_rate {numerator=100}
  }
 }
 OnRowMoveUpOutOfViewAnim {
  binding_count=1
  length_seconds=0.15
  object=&self:OnRowMoveUpOutOfViewAnim
  sequence {bindings [WBP_Row_HorizontalScrollingList {
     id=#2
     possessed_class=WBP_Row_HorizontalScrollingList_C
     tracks [{
       type=MovieSceneFloatTrack
       display_name=Render Opacity
       sections [{type=MovieSceneFloatSection, end_frame=3, start_frame=0, channels [None {type=MovieSceneScriptingFloatChannel, keys [[0,1],[3,0]]}]}]
      }]
    }]}
 }
 OnRowMoveDownIntoView {
  binding_count=1
  length_seconds=0.15
  object=&self:OnRowMoveDownIntoView
  sequence {bindings [WBP_Row_HorizontalScrollingList {
     id=#3
     possessed_class=WBP_Row_HorizontalScrollingList_C
     tracks [{
       type=MovieSceneFloatTrack
       display_name=Render Opacity
       sections [{type=MovieSceneFloatSection, end_frame=3, start_frame=0, channels [None {type=MovieSceneScriptingFloatChannel, keys [[0,0],[3,1]]}]}]
      }]
    }]}
 }
]
graphs [
 EventGraph {nodes [
   n1 = FinishedLoadingMoreDelay() {is=title K2Node_CustomEvent}
   > Branch(Condition=Get IsLoadingMoreQueryActive() -> (IsLoadingMoreQueryActive:bool))
   > n2 = Branch(Condition=Get IsLoadingMoreVisible() -> (IsLoadingMoreVisible:bool))
   Set IsLoadingMoreVisible(execute=@n2.else, IsLoadingMoreVisible:bool=true) -> (Output_Get:bool)
   > PlayAnimationForward(InAnimation=Get OnLoadingMore() -> (OnLoadingMore:Widget Animation Object Reference))
   > SetVisibility(self=Get WBP_LoadingMorePages() -> (WBP_LoadingMorePages:WBP Loading More Pages Object Reference), InVisibility=SelfHitTestInvisible)
   n3 = OnTileHovered!() -> (OutputDelegate, then, Item:Object Reference, bIsHovered:bool=~) {is=title K2Node_CustomEvent}
   > Branch(Condition=@n3.bIsHovered)
   n4 = Get WBP_Row_HorizontalScrollingList() -> (WBP_Row_HorizontalScrollingList:WBP Row Horizontal Scrolling List Object Reference)
   n5 = Get ListView!(self:Fort Polymorphic List View Object Reference=@n4) -> (ListView:Common List View Object Reference)
   > SelectItem!(execute, self, ItemIndex:int=GetIndexForItem!(self:List View Object Reference=@n5, Item:Object Reference=@n3.Item) -> (ReturnValue:int=0)) -> (then)
   n6 = OnTileFocusGained!() -> (OutputDelegate, then, ItemFocused:int) {is=title K2Node_CustomEvent}
   > SelectItem(ItemIndex=@n6.ItemFocused)
   n7 = OnTileItemGenerated!() -> (OutputDelegate, then, Widget:User Widget Object Reference) {is=title K2Node_CustomEvent}
   > SetNestedButtonGroupSelectionVM!(
    execute:Exec
    self:Object Reference=@n7.Widget
    InVM:Fort Nested Button Group Selection VM Object Reference=Get NestedButtonGroupSelectionVM() -> (NestedButtonGroupSelectionVM:Fort Nested Button Group Selection VM Object Reference)
   ) -> (then:Exec) {is=title K2Node_Message}
   Event OnOverrideItemSelectorSet!() -> (OutputDelegate, then, ItemSelector:Fort Discover Item Selector Base Object Reference) {is=title K2Node_Event}
   > UpdateItemSelectorInHoziontalScrollingList()
   n8 = Event SetPanelViewModel!() -> (OutputDelegate, then, ViewModel:Discover Provider View Model Object Reference) {is=title K2Node_Event}
   > SetFortDiscoverProviderViewModel!(execute, self, Viewmodel:Discover Provider View Model Object Reference=@n8.ViewModel) -> (then)
   n9 = Event SetFocusOffset!() -> (OutputDelegate, then, FocusOffset:int=0, bSelectFromBelow:bool) {is=title K2Node_Event}
   n10 = GetOwningPlayer()
   > Branch(Condition=NOT Boolean(A=IsUsingTouch(PlayerController=@n10)))
   n11 = Get WBP_Row_HorizontalScrollingList()
   n12 = Get ListView(self=@n11)
   n13 = GetNumItems!(self:List View Object Reference=@n12) -> (ReturnValue:int=0)
   n14 = Get TileFocusWeight!(self:WBP Row Horizontal Scrolling List Object Reference=@n11) -> (TileFocusWeight:double)
   n15 = Round(A=float + float(A=float / float(A:int=@n9.FocusOffset, B=@n14), B=GetScrollOffset!(self:List View Base Object Reference=@n12) -> (ReturnValue:float))) {is=title K2Node_CallFunction}
   n16 = ShouldUseSizeShiftedFocusOffset!() -> (ReturnValue:bool)
   n17 = int + int(
    A=@n9.FocusOffset
    B=GetFirstVisibleItemIndex!(self:WBP Row Horizontal Scrolling List Object Reference=Get WBP_Row_HorizontalScrollingList()) -> (FirstVisibleItemIndex:int=0)
   )
   > SetSelectedIndex!(
    execute
    self:List View Object Reference=@n12
    Index:int="Min (Integer)"(A=Select(Option 0:int=@n17, Option 1:int=@n15, Index=@n16) -> (ReturnValue:int), B=int - int(A=@n13)) -> (ReturnValue=0)
   ) -> (then)
   n18 = Event OpenNestedSurface!() -> (OutputDelegate, then, SurfaceTileItemVM:Fort Discover Surface Tile Item VM Object Reference) {is=title K2Node_Event}
   > Parent: OpenNestedSurface!(execute, SurfaceTileItemVM:Fort Discover Surface Tile Item VM Object Reference=@n18.SurfaceTileItemVM) -> (then) {is=title K2Node_CallParentFunction}
   n19 = Event HandleOnInputMethodChanged!() -> (OutputDelegate, then, bNewInputType:ECommonInputType Enum) {is=title K2Node_Event}
   > Parent: HandleOnInputMethodChanged!(execute, bNewInputType:ECommonInputType Enum=@n19.bNewInputType) -> (then) {is=title K2Node_CallParentFunction}
   Event On Entry Released() {is=title K2Node_Event}
   > Parent: On Entry Released() {is=title K2Node_CallParentFunction}
   > SetIsSelectedNestedVM!(
    execute
    self:WBP Row Grid Base Object Reference
    ListContainer:List View Base Object Reference=Get ListView(self=Get WBP_Row_HorizontalScrollingList())
   ) -> (then)
   Event Construct()
   > Parent: Construct()
   > Bind Event to On Tile Focus Gained(self:WBP Row Horizontal Scrolling List Object Reference=@n4, Delegate=@n6) {is=title K2Node_AddDelegate}
   > Bind Event to On Item Is Hovered Changed(self:List View Object Reference=@n5, Delegate=@n3) {is=title K2Node_AddDelegate}
   > Bind Event to On Entry Generated(self:List View Base Object Reference=Get ListView(self=Get WBP_Row_HorizontalScrollingList()), Delegate=Create Event()) {is=title K2Node_AddDelegate}
   > n20 = GetSupportsLessThanMaxVisibleRows!(execute, self:Fort Discover Browser Grid Row Object Reference) -> (then, ReturnValue:bool)
   > Branch(Condition=@n20)
   > UpdateItemSelectorInHoziontalScrollingList()
   Event Destruct()
   > Parent: Destruct()
   > Unbind Event from On Entry Generated(self:List View Base Object Reference=Get ListView(self=Get WBP_Row_HorizontalScrollingList()), Delegate=Create Event()) {is=title K2Node_RemoveDelegate}
   K2Node_Event_7 = OnQueryStatusChanged_1!() -> (OutputDelegate, then, bIsActive:bool=~) {is=title K2Node_CustomEvent}
   > n21 = Set IsLoadingMoreQueryActive(IsLoadingMoreQueryActive:bool=false) -> (Output_Get:bool)
   > n22 = Branch(Condition=@n21)
   > n23 = Set Timer by Event(Delegate=@n1, Time=Get LoadingMoreDisplayDelay() -> (LoadingMoreDisplayDelay:double), bLooping=false)
   > Set LoadingMoreDelayTimer(LoadingMoreDelayTimer:Timer Handle Structure=@n23) -> (Output_Get:Timer Handle Structure)
   Clear and Invalidate Timer by Handle(execute=@n22.else, Handle=Get LoadingMoreDelayTimer() -> (LoadingMoreDelayTimer:Timer Handle Structure))
   > Branch(Condition=Get IsLoadingMoreVisible())
   > Set IsLoadingMoreVisible(IsLoadingMoreVisible=false)
   n24 = Get OnLoadingMore()
   n25 = GetAnimationCurrentTime!(self:User Widget Object Reference, InAnimation:Widget Animation Object Reference=@n24) -> (ReturnValue:float)
   n26 = GetEndTime(self=@n24)
   > n27 = "Play Animation with Finished event (legacy)"!(
    execute:Exec
    Widget:User Widget Object Reference=Self-Reference()
    InAnimation:Widget Animation Object Reference=@n24
    StartAtTime:float=SelectFloat(A=float - float(A=@n26, B=@n25) {is=title K2Node_CallFunction}, B=0, bPickA=IsAnimationPlaying(InAnimation=@n24))
    NumLoopsToPlay:int=1
    PlayMode:EUMGSequencePlayMode Enum=Reverse
    PlaybackSpeed:float=1
   ) -> (then:Exec, Result:UMGSequence Player Object Reference, Finished:Exec) {is=title K2Node_PlayAnimation}
   n28 = Branch(execute=@n27.Finished, Condition=Get IsLoadingMoreVisible())
   SetVisibility(execute=@n28.else, self=Get WBP_LoadingMorePages(), InVisibility=Hidden)
   Event UpdateAnalyticsStates!() -> (OutputDelegate, then, bStopImpressions:bool) {is=title K2Node_Event}
   > n29 = GetWidgetAbsoluteRect!(execute, Widget:Widget Object Reference=Self-Reference()) -> (then, TopLeft:Vector 2D Structure, BottomRight:Vector 2D Structure, Size:Vector 2D Structure, DebugStringOutput:String)
   > ComputeRowVisibility!(
    execute
    ListView:List View Object Reference=Get ListView(self=Get WBP_Row_HorizontalScrollingList())
    Provider VM:Discover Provider View Model Object Reference=Get FortDiscoverProviderViewModel() -> (FortDiscoverProviderViewModel:Discover Provider View Model Object Reference)
    TopLeft:Vector 2D Structure=@n29
    BottomRight:Vector 2D Structure=@n29.BottomRight
   ) -> (then)
   Event OnCategoryItemChanged!() -> (OutputDelegate, then, bPlayAnimation:bool, CategoryPanelName:"Name (by ref)"=None) {is=title K2Node_Event}
   Event OnRowPeekStateChanged!() -> (OutputDelegate, then, bIsInPeekState:bool) {is=title K2Node_Event}
   Event OnRowIsSelectedChanged!() -> (OutputDelegate, then, bIsSelected:bool) {is=title K2Node_Event}
   Event OnRowMoveUp!() -> (OutputDelegate, then, bMovingOffscreen:bool) {is=title K2Node_Event}
   Event OnRowIsActiveChanged!() -> (OutputDelegate, then, bIsActiveRow:bool) {is=title K2Node_Event}
   Event OnRowMoveDown!() -> (OutputDelegate, then, bMovingOffscreen:bool) {is=title K2Node_Event}
   Reroute Node(InputPin:Wildcard) -> (OutputPin:Wildcard)
  ]}
 GetPanelViewModel {nodes [
   GetPanelViewModel() {is=title K2Node_FunctionEntry}
   > Return Node!(execute, ReturnValue:Discover Provider View Model Object Reference=Get FortDiscoverProviderViewModel())
  ]}
 OnRowBroadcastClickToDiscover {nodes [OnRowBroadcastClickToDiscover() {is=title K2Node_FunctionEntry}]}
 OnUpperBoundaryHit {nodes [OnUpperBoundaryHit() {is=title K2Node_FunctionEntry}]}
 SelectItem {nodes [
   n1 = SelectItem!() -> (then, ItemIndex:int) {is=title K2Node_FunctionEntry}
   n2 = Get WBP_Row_HorizontalScrollingList()
   n3 = Get ListView(self=@n2)
   > SetSelectedIndex(self=@n3, Index=@n1)
   n4 = float - float(A=float * float(A:int=@n1, B=Get TileFocusWeight(self=@n2)) {is=title K2Node_PromotableOperator}, B=GetScrollOffset(self=@n3)) {is=title K2Node_PromotableOperator}
   n5 = GetFirstVisibleItemIndex(self=@n2) -> (FirstVisibleItemIndex=0)
   > HandleActivitySelected!(
    execute
    self:Fort Discover Browser Grid Row Object Reference
    Index:int=@n1
    IndexInRow:int=Select(Option 0=int - int(A=@n1, B=@n5), Option 1=Round(A=@n4), Index=ShouldUseSizeShiftedFocusOffset())
   ) -> (then)
  ]}
 UpdateItemSelectorInHoziontalScrollingList {nodes [
   UpdateItemSelectorInHoziontalScrollingList() {is=title K2Node_FunctionEntry}
   n1 = Get OverrideItemSelector() -> (OverrideItemSelector:Fort Discover Item Selector Base Object Reference)
   > Is Valid(InputObject=@n1)
   > Set OverrideItemSelector!(
    execute
    OverrideItemSelector:Fort Discover Item Selector Base Object Reference=@n1
    self:WBP Row Horizontal Scrolling List Object Reference=Get WBP_Row_HorizontalScrollingList()
   ) -> (then, Output_Get:Fort Discover Item Selector Base Object Reference)
  ]}
]
variables [
 bAllowNonFullRows {type=boolean}
 EntriesSpacing {type=number, default=32}
 FontHoverAnimateParam {type=string, default=Hover_Animate}
 FontPressedAnimateParam {type=string, default=Pressed_Animate}
 IsLoadingMoreQueryActive {type=boolean}
 IsLoadingMoreVisible {type=boolean}
 IsWaitingForInactiveAnim {type=boolean}
 LoadingMoreDelayTimer {type=TimerHandle}
 LoadingMoreDisplayDelay {type=number, default=0.2}
 OnRowBroadcastClickToDiscover {}
 OnUpperBoundaryHit {}
]
widgets {
 info {
  inheritedWidgetCount=1
  parentClass=/DiscoveryBrowser/Rows/WBP_Row_Grid_Base.WBP_Row_Grid_Base_C
  rootWidgetClass=&GridPanel
  widgetCount=5
 }
 widgets [
  GridRoot {widgetClassPath=&GridPanel, bIsVariable=true, widget=GridRoot, properties {columnFill [1], rowFill [1]}}
  WBP_Row_HorizontalScrollingList {
   widgetClassPath=/DiscoveryBrowser/Rows/WBP_Row_HorizontalScrollingList
   bIsVariable=true
   parent=GridRoot
   widget=WBP_Row_HorizontalScrollingList
   properties {bIsFocusable=true, is Using Focus for Outline Visuals=true}
   slot {class=GridSlot}
  }
  FortVisualAttachment_0 {
   widgetClassPath=&FortVisualAttachment
   parent=GridRoot
   widget=FortVisualAttachment_0
   properties {contentAnchor=<1 1>}
   slot {class=GridSlot, properties {horizontalAlignment=HAlign_Right, verticalAlignment=VAlign_Top, padding=<ltrb 0 0 -32 8>}}
  }
  WBP_LoadingMorePages {
   widgetClassPath=/DiscoveryBrowser/TabScreens/WBP_LoadingMorePages
   bIsVariable=true
   parent=FortVisualAttachment_0
   widget=WBP_LoadingMorePages
   properties {visibility=Hidden}
   slot {class=SizeBoxSlot}
  }
  Text_CategoryName {widgetClassPath=&CommonTextBlock, bInherited=true, widget=null}
 ]
}
