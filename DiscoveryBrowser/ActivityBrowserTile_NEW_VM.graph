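# Discovery browser tile widget, derived from FortActivityBrowserTile, that uses WBP_IslandTile_Full as its root Display_TileDetails to show an island entry.
# It acts as a thin wrapper, forwarding the nested button group selection viewmodel setter and getter to that inner tile through the VM item interface.
# When the list entry is released, it casts the inner tile and runs its context menu cleanup.
# The inner tile is flagged as old Discover code and uses an empty, no-sound button style.
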WidgetBlueprint /DiscoveryBrowser/ActivityBrowserTile_NEW_VM
graphs [
 EventGraph {nodes [
   n1 = Event SetNestedButtonGroupSelectionVM!() -> (OutputDelegate, then, InVM:Fort Nested Button Group Selection VM Object Reference) {is=title K2Node_Event}
   > SetNestedButtonGroupSelectionVM!(
    execute
    self:Fort Nested Button Group VMItem Interface=Get Display_TileDetails() -> (Display_TileDetails:Fort Activity Tile Details Display Object Reference)
    InVM:Fort Nested Button Group Selection VM Object Reference=@n1.InVM
   ) -> (then)
   Event On Entry Released() {is=title K2Node_Event}
   > n2 = Cast To WBP_IslandTile_Full!(execute, Object=Get Display_TileDetails()) -> (then, CastFailed, AsWBP Island Tile Full:WBP Island Tile Full Object Reference) {is=title K2Node_DynamicCast}
   > WrapperCleanupContextMenuOnRelease(self:WBP Island Tile Minimal Object Reference=@n2)
  ]}
 GetNestedButtonGroupSelectionVM {nodes [
   GetNestedButtonGroupSelectionVM() {is=title K2Node_FunctionEntry}
   > Return Node!(
    execute
    ReturnValue:Fort Nested Button Group Selection VM Object Reference=GetNestedButtonGroupSelectionVM!(self:Fort Nested Button Group VMItem Interface=Get Display_TileDetails()) -> (ReturnValue:Fort Nested Button Group Selection VM Object Reference) {is=title K2Node_CallFunction}
   )
  ]}
]
variables []
widgets {
 info {
  inheritedWidgetCount=2
  parentClass=/Script/DiscoveryBrowserUI.FortActivityBrowserTile
  rootWidgetClass=/DiscoveryBrowser/Tiles/Islands/WBP_IslandTile_Full.WBP_IslandTile_Full_C
  widgetCount=2
 }
 widgets [
  Display_TileDetails {
   widgetClassPath=/DiscoveryBrowser/Tiles/Islands/WBP_IslandTile_Full
   bInherited=true
   bIsVariable=true
   widget=Display_TileDetails
   properties {
    holdData=&CommonUI_HoldData_Base_C
    isOldDiscoverCode=true
    style=/FortUILibrary/Foundation/Buttons/ButtonStyles/ButtonStyle_Empty_NoSound.ButtonStyle_Empty_NoSound_C
    pressedSlateSoundOverride {resourceObject=/Game/Sounds/Fort_UI_Sounds/FrontEnd/DefaultSounds/PressedButtons/MSS_UI_Button_Pressed_Default}
   }
  }
  InputActionWidget {widgetClassPath=&CommonActionWidget, bInherited=true, widget=null}
 ]
}
