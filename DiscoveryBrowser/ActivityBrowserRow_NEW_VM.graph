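# Widget Blueprint for a Discovery browser category row: a header above a horizontal-style list of activity tiles with left/right page arrows.
# Active and peek changes play or snap dim animations, toggle list hit-testing and notify the arrows.
# Generated tiles receive the row's panel name and nested button-group selection view model.
# A loading-more indicator fades in after a timer delay during active queries and reverses out afterward.
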WidgetBlueprint /DiscoveryBrowser/ActivityBrowserRow_NEW_VM
animations [
 OnRowActiveInactiveAnim {
  binding_count=2
  length_seconds=0.16
  object=&self:OnRowActiveInactiveAnim
  sequence {
   bindings [
    InactiveDarken {
     id=#1
     possessed_class=CommonBorder
     tracks [{
       type=MovieSceneColorTrack
       display_name=Content Color and Opacity
       sections [{
         type=MovieSceneColorSection
         end_frame=16
         start_frame=0
         channels [
          Color.R {type=MovieSceneScriptingFloatChannel, keys [[0,0.5],[16,1]]}
          Color.G {type=MovieSceneScriptingFloatChannel, keys [[0,0.5],[16,1]]}
          Color.B {type=MovieSceneScriptingFloatChannel, keys [[0,0.5],[16,1]]}
          Color.A {type=MovieSceneScriptingFloatChannel, const [1,2]}
         ]
        }]
      }]
    }
    InactiveOverlayDim {
     id=#2
     possessed_class=Image
     tracks [{
       type=MovieSceneFloatTrack
       display_name=Render Opacity
       sections [{type=MovieSceneFloatSection, channels [None {type=MovieSceneScriptingFloatChannel, keys [[0,0.5],[16,0]]}]}]
      }]
    }
   ]
   display_rate {numerator=100}
  }
 }
 OnPeek {
  binding_count=2
  length_seconds=0.24
  object=&self:OnPeek
  sequence {
   bindings [
    InactiveOverlayDim {
     id=#3
     possessed_class=Image
     tracks [{
       type=MovieSceneFloatTrack
       display_name=Render Opacity
       sections [{type=MovieSceneFloatSection, channels [None {type=MovieSceneScriptingFloatChannel, keys [[0,0.5],[24,0]]}]}]
      }]
    }
    InactiveDarken {
     id=#4
     possessed_class=CommonBorder
     tracks [{
       type=MovieSceneColorTrack
       display_name=Content Color and Opacity
       sections [{
         type=MovieSceneColorSection
         end_frame=24
         start_frame=0
         channels [
          Color.R {type=MovieSceneScriptingFloatChannel, keys [[0,0.5],[24,1]]}
          Color.G {type=MovieSceneScriptingFloatChannel, keys [[0,0.5],[24,1]]}
          Color.B {type=MovieSceneScriptingFloatChannel, keys [[0,0.5],[24,1]]}
          Color.A {type=MovieSceneScriptingFloatChannel, const [1,2]}
         ]
        }]
      }]
    }
   ]
   display_rate {numerator=100}
  }
 }
 OnLoadingMore {
  binding_count=1
  length_seconds=0.24
  object=&self:OnLoadingMore
  sequence {
   bindings [WBP_LoadingMorePages {
     id=#5
     possessed_class=WBP_LoadingMorePages_C
     tracks [
      {
       type=MovieSceneFloatTrack
       display_name=Render Opacity
       sections [{type=MovieSceneFloatSection, channels [None {type=MovieSceneScriptingFloatChannel, keys [[0,0],[24,1]]}]}]
      }
      {
       type=MovieScene2DTransformTrack
       sections [{type=MovieScene2DTransformSection, channels [Translation.X {type=MovieSceneScriptingFloatChannel, keys [[0,16],[24,0]]}]}]
      }
     ]
    }]
   display_rate {numerator=100}
  }
 }
]
graphs [
 EventGraph {nodes [
   n1 = "On Entry Generated (ListView_Activities)"!() -> (OutputDelegate, then, Widget:User Widget Object Reference) {is=title K2Node_ComponentBoundEvent}
   > n2 = Sequence()
   > n3 = Cast To ActivityBrowserTile_NEW_VM!(execute, Object=@n1.Widget) -> (then, CastFailed, AsActivity Browser Tile NEW VM:Activity Browser Tile NEW VM Object Reference) {is=title K2Node_DynamicCast}
   > n4 = Cast To WBP_IslandTile_Full!(
    execute
    Object=Get Display_TileDetails!(self:Fort Activity Browser Tile Object Reference=@n3) -> (Display_TileDetails:Fort Activity Tile Details Display Object Reference)
   ) -> (then, CastFailed, AsWBP Island Tile Full:WBP Island Tile Full Object Reference) {is=title K2Node_DynamicCast}
   > SetParentPanelName!(
    execute
    self:WBP Discover Tile Base Object Reference=Get WBP_Discover_Tile_Base!(self:WBP Island Tile Minimal Object Reference=@n4) -> (WBP_Discover_Tile_Base:WBP Discover Tile Base Object Reference)
    PanelName:String=Get RowPanelName() -> (RowPanelName:String)
   ) -> (then)
   SetNestedButtonGroupSelectionVM!(
    execute:Exec=@n2.then_1
    self:Object Reference=@n1.Widget
    InVM:Fort Nested Button Group Selection VM Object Reference=Get NestedButtonGroupSelectionVM() -> (NestedButtonGroupSelectionVM:Fort Nested Button Group Selection VM Object Reference)
   ) -> (then:Exec) {is=title K2Node_Message}
   HandleOnInputMethodChanged!() -> (OutputDelegate, then, bNewInputType:ECommonInputType Enum) {is=title K2Node_CustomEvent}
   > UpdateVisibility()
   n5 = FinishedLoadingMoreDelay() {is=title K2Node_CustomEvent}
   > Branch(Condition=Get IsLoadingMoreQueryActive() -> (IsLoadingMoreQueryActive:bool))
   > n6 = Branch(Condition=Get IsLoadingMoreVisible() -> (IsLoadingMoreVisible:bool))
   Set IsLoadingMoreVisible(execute=@n6.else, IsLoadingMoreVisible:bool=true) -> (Output_Get:bool)
   > PlayAnimationForward(InAnimation=Get OnLoadingMore() -> (OnLoadingMore:Widget Animation Object Reference))
   > SetVisibility(self=Get WBP_LoadingMorePages() -> (WBP_LoadingMorePages:WBP Loading More Pages Object Reference), InVisibility=SelfHitTestInvisible)
   n7 = Event OnCategoryItemChanged!() -> (OutputDelegate, then, bPlayAnimation:bool, CategoryPanelName:"Name (by ref)"=None) {is=title K2Node_Event}
   > UpdateNewRow!(
    execute
    self
    IsActive:bool=GetIsActive!(self:Fort Activity Browser Row Object Reference) -> (ReturnValue:bool)
    PlayAnimation:bool=@n7.bPlayAnimation
   ) -> (then)
   > Set RowPanelName(RowPanelName:String="To String (Name)"(InName=@n7.CategoryPanelName)) -> (Output_Get:String)
   n8 = Event OnRowPeekStateChanged!() -> (OutputDelegate, then, bIsInPeekState:bool) {is=title K2Node_Event}
   n9 = Get OnRowActiveInactiveAnim() -> (OnRowActiveInactiveAnim:Widget Animation Object Reference)
   > n10 = Branch(Condition=OR Boolean(A=IsAnimationPlaying(InAnimation=@n9), B=GetIsActive()))
   n11 = Branch(execute=@n10.else, Condition=@n8.bIsInPeekState)
   n12 = Get OnPeek() -> (OnPeek:Widget Animation Object Reference)
   > PlayAnimationForward(InAnimation=@n12)
   PlayAnimationReverse(execute=@n11.else, InAnimation=@n12)
   n13 = Event OnCategoryTextChanged!() -> (OutputDelegate, then, CategoryText:"Text (by ref)", CategorySubtitle:"Text (by ref)", bIsSponsored:bool) {is=title K2Node_Event}
   > SetHeaderText!(
    execute
    self:WBP Row Header Object Reference=Get WBP_Row_Header() -> (WBP_Row_Header:WBP Row Header Object Reference)
    HeaderText:Text=@n13.CategoryText
   ) -> (then)
   > SetSubHeaderText!(execute, self:WBP Row Header Object Reference=Get WBP_Row_Header(), SubHeaderText:Text=@n13.CategorySubtitle) -> (then)
   > SetSponsoredIconVisibility!(execute, self:WBP Row Header Object Reference=Get WBP_Row_Header(), IsVisible:bool=@n13.bIsSponsored) -> (then)
   n14 = Event OnRowIsActiveChanged!() -> (OutputDelegate, then, bIsActiveRow:bool) {is=title K2Node_Event}
   > Check and Stop Animation!(execute:Exec, Animation:Widget Animation Object Reference=Get OnPeek()) -> (then:Exec) {is=title K2Node_MacroInstance}
   > n15 = Sequence()
   > n16 = Branch(Condition=@n14.bIsActiveRow)
   n17 = Get ListView_Activities() -> (ListView_Activities:Fort Activity List View Object Reference)
   > SetVisibility(self=@n17, InVisibility=SelfHitTestInvisible)
   n18 = Get OnRowActiveInactiveAnim()
   > PlayAnimation(
    InAnimation=@n18
    StartAtTime=GetAnimationCurrentTime!(self:User Widget Object Reference, InAnimation:Widget Animation Object Reference=@n18) -> (ReturnValue:float)
   )
   SetVisibility(execute=@n16.else, self=@n17, InVisibility=HitTestInvisible)
   > n19 = "Play Animation with Finished event (legacy)"!(
    execute:Exec
    Widget:User Widget Object Reference=Self-Reference()
    InAnimation:Widget Animation Object Reference=@n18
    StartAtTime:float=GetAnimationCurrentTime(InAnimation=@n18)
    NumLoopsToPlay:int=1
    PlayMode:EUMGSequencePlayMode Enum=Reverse
    PlaybackSpeed:float=1
   ) -> (then:Exec, Result:UMGSequence Player Object Reference, Finished:Exec) {is=title K2Node_PlayAnimation}
   > Set IsWaitingForInactiveAnim(IsWaitingForInactiveAnim:bool=true) -> (Output_Get:bool)
   n20 = Branch(
    execute=@n19.Finished
    Condition=AND Boolean(
     A=GetIsInPeekState!(self:Fort Activity Browser Row Object Reference) -> (ReturnValue:bool)
     B=Get IsWaitingForInactiveAnim() -> (IsWaitingForInactiveAnim:bool)
    )
   )
   > n21 = PlayAnimationForward(InAnimation=Get OnPeek())
   Set IsWaitingForInactiveAnim(execute=[@n21.then, @n20.else], IsWaitingForInactiveAnim=false)
   n22 = For Each Loop(
    Exec=@n15.then_1
    Array:Array of Activity Browser Arrow Button Object References=Get ArrowButtons() -> (ArrowButtons:Array of Activity Browser Arrow Button Object References)
   ) -> (Array Element:Activity Browser Arrow Button Object Reference)
   > OnRowActiveChanged!(execute, self:Activity Browser Arrow Button Object Reference=@n22, IsActive:bool=@n14.bIsActiveRow) -> (then)
   Event Construct()
   > Parent: Construct()
   n23 = Create Event()
   n24 = Get CommonInputSubsystem()
   > Bind Event to On Input Method Changed(self=@n24, Delegate=@n23)
   > UpdateVisibility()
   > n25 = Sequence()
   > n26 = Cast To ActivityBrowserArrowButton!(execute, Object=Get Button_PageLeft() -> (Button_PageLeft:Common Button Base Object Reference)) -> (then, CastFailed, AsActivity Browser Arrow Button:Activity Browser Arrow Button Object Reference) {is=title K2Node_DynamicCast}
   > Add(
    TargetArray:Array of Activity Browser Arrow Button Object References=Get ArrowButtons()
    NewItem:"Activity Browser Arrow Button Object Reference (by ref)"=@n26
   ) {is=title K2Node_CallArrayFunction}
   n27 = Cast To ActivityBrowserArrowButton(execute=@n25.then_1, Object=Get Button_PageRight() -> (Button_PageRight:Common Button Base Object Reference))
   > Add(TargetArray=Get ArrowButtons(), NewItem=@n27)
   Event Destruct()
   > Parent: Destruct()
   > Unbind Event from On Input Method Changed(self=@n24, Delegate=@n23)
   > Clear(TargetArray:Array of Activity Browser Arrow Button Object References=Get ArrowButtons())
   n28 = Event OnQueryStatusChanged!() -> (OutputDelegate, then, bIsActiveStatus:bool) {is=title K2Node_Event}
   > n29 = Set IsLoadingMoreQueryActive(IsLoadingMoreQueryActive:bool=@n28.bIsActiveStatus) -> (Output_Get:bool)
   > n30 = Branch(Condition=@n29)
   > n31 = Set Timer by Event(Delegate=@n5, Time=Get LoadingMoreDisplayDelay() -> (LoadingMoreDisplayDelay:double), bLooping=false)
   > Set LoadingMoreDelayTimer(LoadingMoreDelayTimer:Timer Handle Structure=@n31) -> (Output_Get:Timer Handle Structure)
   Clear and Invalidate Timer by Handle(execute=@n30.else, Handle=Get LoadingMoreDelayTimer() -> (LoadingMoreDelayTimer:Timer Handle Structure))
   > Branch(Condition=Get IsLoadingMoreVisible())
   > Set IsLoadingMoreVisible(IsLoadingMoreVisible=false)
   n32 = Get OnLoadingMore()
   n33 = GetAnimationCurrentTime(InAnimation=@n32)
   n34 = GetEndTime(self=@n32)
   > n35 = "Play Animation with Finished event (legacy)"(
    Widget=Self-Reference()
    InAnimation=@n32
    StartAtTime=SelectFloat(A=float - float(A=@n34, B=@n33) {is=title K2Node_CallFunction}, B=0, bPickA=IsAnimationPlaying(InAnimation=@n32))
    NumLoopsToPlay=1
    PlayMode=Reverse
    PlaybackSpeed=1
   )
   n36 = Branch(execute=@n35.Finished, Condition=Get IsLoadingMoreVisible())
   SetVisibility(execute=@n36.else, self=Get WBP_LoadingMorePages(), InVisibility=Hidden)
  ]}
 ResetOnPeekAnimation {nodes [
   ResetOnPeekAnimation() {is=title K2Node_FunctionEntry}
   > SetContentColorAndOpacity!(
    execute
    self:Border Object Reference=Get InactiveDarken() -> (InactiveDarken:Common Border Object Reference)
    InContentColorAndOpacity:Linear Color Structure=(R=1,G=1,B=1,A=1)
   ) -> (then)
   > SetRenderOpacity(self=Get InactiveOverlayDim() -> (InactiveOverlayDim:Image Object Reference), InOpacity=0)
  ]}
 ResetOnRowActiveInactiveAnimation {nodes [
   n1 = ResetOnRowActiveInactiveAnimation!() -> (then, IsRowActive:bool=~) {is=title K2Node_FunctionEntry}
   > SetContentColorAndOpacity(
    self=Get InactiveDarken()
    InContentColorAndOpacity=Select(Option 0:Linear Color Structure=(R=0.5,G=0.5,B=0.5,A=1), Option 1:Linear Color Structure=(R=1,G=1,B=1,A=1), Index=@n1) -> (ReturnValue:Linear Color Structure)
   )
   > n2 = Branch(Condition=@n1)
   n3 = Get InactiveOverlayDim()
   > SetRenderOpacity(self=@n3, InOpacity=0)
   SetRenderOpacity(execute=@n2.else, self=@n3, InOpacity=0.5)
  ]}
 SnapToEndOfAnimation {nodes [
   n1 = SnapToEndOfAnimation!() -> (then, InAnimation:Widget Animation Object Reference, PlayMode:EUMGSequencePlayMode Enum=Forward) {is=title K2Node_FunctionEntry}
   > PlayAnimation(InAnimation=@n1, StartAtTime=GetEndTime(self=@n1), PlayMode=@n1.PlayMode)
  ]}
 UpdateNewRow {nodes [
   UpdateNewRow!() -> (then, IsActive:bool=~, PlayAnimation:bool=~) {is=title K2Node_FunctionEntry}
   > Set IsWaitingForInactiveAnim(IsWaitingForInactiveAnim=false)
   > n1 = Sequence()
   > n2 = Branch(Condition=Get IsActive() -> (IsActive:bool))
   > Check and Stop Animation(Animation=Get OnPeek())
   > SetVisibility(self=Get ListView_Activities(), InVisibility=SelfHitTestInvisible)
   > ResetOnRowActiveInactiveAnimation!(execute, self, IsRowActive:bool=true) -> (then) {is=title K2Node_CallFunction}
   SetVisibility(execute=@n2.else, self=Get ListView_Activities(), InVisibility=HitTestInvisible)
   > n3 = Branch(Condition=GetIsInPeekState())
   > Check and Stop Animation(Animation=Get OnRowActiveInactiveAnim())
   > ResetOnPeekAnimation() {is=title K2Node_CallFunction}
   Check and Stop Animation(execute=@n3.else, Animation=Get OnPeek())
   > ResetOnRowActiveInactiveAnimation(IsRowActive=false)
   n4 = For Each Loop(Exec=@n1.then_1, Array=Get ArrowButtons())
   > OnRowReset!(execute, self:Activity Browser Arrow Button Object Reference=@n4, IsActive:bool=Get IsActive()) -> (then)
  ]}
 UpdateVisibility {nodes [
   UpdateVisibility() {is=title K2Node_FunctionEntry}
   n1 = GetCurrentInputType(self=Get CommonInputSubsystem())
   > SetVisibility(InVisibility=Select(Option 0:ESlateVisibility Enum=SelfHitTestInvisible, Option 1:ESlateVisibility Enum=Visible, Index="Equal (Enum)"(A=@n1, B=Touch)) -> (ReturnValue:ESlateVisibility Enum))
  ]}
]
variables [
 ArrowButtons {type="array<ActivityBrowserArrowButton_C>", default []}
 FontHoverAnimateParam {type=string, default=Override Inactive}
 FontPressedAnimateParam {type=string, default=Is Inactive}
 IsLoadingMoreQueryActive {type=boolean}
 IsLoadingMoreVisible {type=boolean}
 IsWaitingForInactiveAnim {type=boolean}
 LoadingMoreDelayTimer {type=TimerHandle}
 LoadingMoreDisplayDelay {type=number, default=0.2}
 RowPanelName {type=string, default=False}
]
widgets {
 info {
  inheritedWidgetCount=4
  namedSlotCount=1
  parentClass=/Script/DiscoveryBrowserUI.FortActivityBrowserRowList
  rootWidgetClass=&GridPanel
  widgetCount=17
 }
 widgets [
  GridPanel_0 {widgetClassPath=&GridPanel, widget=GridPanel_0, properties {columnFill [1], rowFill [1]}}
  VBContent {widgetClassPath=&VerticalBox, bIsVariable=true, parent=GridPanel_0, widget=VBContent, slot {class=GridSlot}}
  WBP_Row_Header {
   widgetClassPath=/DiscoveryBrowser/Rows/WBP_Row_Header
   bIsVariable=true
   parent=VBContent
   widget=WBP_Row_Header
   slot {class=VerticalBoxSlot}
  }
  Overlay_1 {widgetClassPath=&Overlay, namedSlotHost=WBP_Row_Header, widget=Overlay_1}
  FortVisualAttachment {
   widgetClassPath=&FortVisualAttachment
   parent=Overlay_1
   widget=FortVisualAttachment
   properties {contentAnchor=<1 1>}
   slot {class=OverlaySlot, properties {horizontalAlignment=HAlign_Right, verticalAlignment=VAlign_Bottom}}
  }
  WBP_LoadingMorePages {
   widgetClassPath=/DiscoveryBrowser/TabScreens/WBP_LoadingMorePages
   bIsVariable=true
   parent=FortVisualAttachment
   widget=WBP_LoadingMorePages
   properties {renderOpacity=0, visibility=Hidden}
   slot {class=SizeBoxSlot}
  }
  Overlay_0 {widgetClassPath=&Overlay, parent=VBContent, widget=Overlay_0, slot {class=VerticalBoxSlot}}
  GridPanel_1 {
   widgetClassPath=&GridPanel
   parent=Overlay_0
   widget=GridPanel_1
   properties {columnFill [1], rowFill [1]}
   slot {class=OverlaySlot, properties {horizontalAlignment=HAlign_Fill, verticalAlignment=VAlign_Fill, padding=<ltrb 0 16 0 48>}}
  }
  InactiveDarken {
   widgetClassPath=&CommonBorder
   bIsVariable=true
   parent=GridPanel_1
   widget=InactiveDarken
   properties {padding=<ltrb 0 0 0 0>}
   slot {class=GridSlot, properties {padding=<ltrb 80 0 0 0>}}
  }
  ListView_Activities {
   widgetClassPath=/Script/DiscoveryBrowserUI.FortActivityListView
   bInherited=true
   parent=InactiveDarken
   widget=ListView_Activities
   properties {
    bAllowDragging=false
    bEnableRightClickScrolling=false
    bEnableScrollAnimation=true
    clipping=Inherit
    directionalNavigationTimeThreshold=0
    entrySpacing=24
    entryWidgetClass=&/ActivityBrowserTile_NEW_VM.ActivityBrowserTile_NEW_VM_C
    numDesignerPreviewEntries=8
    visibility=SelfHitTestInvisible
   }
   slot {class=BorderSlot, properties {padding=<ltrb 0 0 0 0>}}
  }
  Overlay_ArrowButtons {
   widgetClassPath=&Overlay
   parent=GridPanel_1
   widget=Overlay_ArrowButtons
   slot {class=GridSlot, properties {layer=1, padding=<ltrb 80 0 80 0>}}
  }
  FortVisualAttachment_1 {
   widgetClassPath=&FortVisualAttachment
   parent=Overlay_ArrowButtons
   widget=FortVisualAttachment_1
   properties {contentAnchor=<0.5 0.5>}
   slot {class=OverlaySlot, properties {verticalAlignment=VAlign_Center}}
  }
  Button_PageLeft {
   widgetClassPath=&/ActivityBrowserArrowButton
   bInherited=true
   bIsVariable=true
   parent=FortVisualAttachment_1
   widget=Button_PageLeft
   properties {
    flip=true
    onButtonBaseClicked=(&self:WidgetTree.Button_PageLeft.HandleOnButtonClicked)
    onButtonBaseDoubleClicked=(&self:WidgetTree.Button_PageLeft.HandleOnButtonDoubleClicked)
    onButtonBaseHovered=(&self:WidgetTree.Button_PageLeft.HandleOnButtonHovered)
    onButtonBaseUnhovered=(&self:WidgetTree.Button_PageLeft.HandleOnButtonUnhovered)
    onSelectedChangedBase=(&self:WidgetTree.Button_PageLeft.HandleOnSelectedChanged)
   }
   slot {class=SizeBoxSlot}
  }
  FortVisualAttachment_2 {
   widgetClassPath=&FortVisualAttachment
   parent=Overlay_ArrowButtons
   widget=FortVisualAttachment_2
   properties {contentAnchor=<0.5 0.5>}
   slot {class=OverlaySlot, properties {horizontalAlignment=HAlign_Right, verticalAlignment=VAlign_Center}}
  }
  Button_PageRight {
   widgetClassPath=&ActivityBrowserArrowButton
   bInherited=true
   bIsVariable=true
   parent=FortVisualAttachment_2
   widget=Button_PageRight
   properties {
    onButtonBaseClicked=(&self:WidgetTree.Button_PageRight.HandleOnButtonClicked)
    onButtonBaseDoubleClicked=(&self:WidgetTree.Button_PageRight.HandleOnButtonDoubleClicked)
    onButtonBaseHovered=(&self:WidgetTree.Button_PageRight.HandleOnButtonHovered)
    onButtonBaseUnhovered=(&self:WidgetTree.Button_PageRight.HandleOnButtonUnhovered)
    onSelectedChangedBase=(&self:WidgetTree.Button_PageRight.HandleOnSelectedChanged)
   }
   slot {class=SizeBoxSlot}
  }
  InactiveOverlayDim {
   widgetClassPath=&Image
   bIsVariable=true
   parent=Overlay_0
   widget=InactiveOverlayDim
   properties {
    renderOpacity=0
    visibility=Hidden
    brush {
     bIsDynamicallyLoaded=false
     drawAs=Image
     imageType=NoImage
     mirroring=NoMirror
     resourceName=None
     resourceObject=/DiscoveryBrowser/TabScreens/DiscoverHome/Materials/MI_UI_Discovery_NebulousBackground_Translucent
     tiling=NoTile
     imageSize=<32 32>
     margin=<ltrb 0 0 0 0>
     outlineSettings {
      bUseBrushTransparency=false
      roundingType=HalfHeightRadius
      width=0
      color {colorUseRule=UseColor_Specified, specifiedColor=<rgba 0 0 0 0>}
      cornerRadii=<0 0 0 0>
     }
     tintColor {colorUseRule=UseColor_Specified, specifiedColor=<rgba 1 1 1 1>}
     uVRegion {bIsValid=false, max=<0 0>, min=<0 0>}
    }
   }
   slot {class=OverlaySlot, properties {horizontalAlignment=HAlign_Fill, verticalAlignment=VAlign_Fill}}
  }
  Text_CategoryName {widgetClassPath=&CommonTextBlock, bInherited=true, widget=null}
 ]
}
