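# Discovery browser grid row widget that shows a header above a single billboard tile.
# SetPanelViewModel stores the provider view model and passes its first item to the tile as an activity view model.
# A custom navigation boundary broadcasts up/down exits via OnNavigateToNewOuterRow, while the tile's matching event calls NavigateWithinGrid one row up or down.
# UpdateAnalyticsStates stops visible impressions, then restarts them for index 0 unless told to stop.
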WidgetBlueprint /DiscoveryBrowser/Rows/Billboard/WBP_Discovery_Row_Billboard
graphs [
 DoCustomBoundaryNavigation {nodes [
   n1 = DoCustomBoundaryNavigation!() -> (then, Navigation:EUINavigation Enum) {is=title K2Node_FunctionEntry}
   > n2 = Switch on EUINavigation!(execute, Selection:EUINavigation Enum=@n1) -> (Left:Exec, Right:Exec, Up:Exec, Down:Exec, Next:Exec, Previous:Exec, Invalid:Exec) {is=title K2Node_SwitchEnum}
   Call On Navigate to New Outer Row!(
    execute:Exec=[@n2.Down, @n2.Up]
    self:WBP Discovery Row Billboard Object Reference
    IsDirectionUp:bool="Equal (Enum)"(A:EUINavigation Enum=@n1, B:EUINavigation Enum=Up)
    PreviousSubrowSelectedIndex:int
    PreviousNumItemsPerSubrow:int=1
    PreviousRowWasViewAllTileSelected:bool
   ) -> (then:Exec) {is=title K2Node_CallDelegate}
  ]}
 EventGraph {nodes [
   n1 = Navigate To New Outer Row!() -> (
    OutputDelegate
    then
    IsDirectionUp:bool=~
    PreviousSubrowSelectedIndex:int
    PreviousNumItemsPerSubrow:int
    PreviousRowWasViewAllTileSelected:bool=~
   ) {is=title K2Node_CustomEvent}
   > NavigateWithinGrid!(
    execute
    self:Fort Discover Browser Grid Row Object Reference
    Row:int=int + int(A=@n1.PreviousSubrowSelectedIndex, B=SelectInt(A=-1, B=1, bPickA=@n1.IsDirectionUp))
   ) -> (then)
   Event Construct()
   > Bind Event to On Navigate to New Outer Row(
    self:WBP Discovery Tile Billboard Object Reference=Get WBP_Discovery_Tile_Billboard() -> (WBP_Discovery_Tile_Billboard:WBP Discovery Tile Billboard Object Reference)
    Delegate=Create Event()
   ) {is=title K2Node_AddDelegate}
   Event PreConstruct()
   > SetSubHeaderText!(execute, self:WBP Row Header Object Reference=Get WBP_Row_Header() -> (WBP_Row_Header:WBP Row Header Object Reference), SubHeaderText:Text) -> (then)
   n2 = Event SetPanelViewModel!() -> (OutputDelegate, then, ViewModel:Discover Provider View Model Object Reference) {is=title K2Node_Event}
   > SetFortDiscoverProviderViewModel!(execute, self, Viewmodel:Discover Provider View Model Object Reference=@n2.ViewModel) -> (then)
   > n3 = Cast To FortActivityViewModel!(
    execute
    Object="Get (a copy)"(Array:Array of Discover Tile Item View Model Object References=Get Items!(self:Discover Provider View Model Object Reference=@n2.ViewModel) -> (Items:Array of Discover Tile Item View Model Object References)) -> (Output:Discover Tile Item View Model Object Reference)
   ) -> (then, CastFailed, AsDiscover Activity View Model:Discover Activity View Model Object Reference) {is=title K2Node_DynamicCast}
   > OnActivityViewModelSet!(
    execute
    self:WBP Discovery Tile Billboard Object Reference=Get WBP_Discovery_Tile_Billboard()
    ActivityViewModel:Discover Activity View Model Object Reference=@n3
   ) -> (then)
   n4 = Event UpdateAnalyticsStates!() -> (OutputDelegate, then, bStopImpressions:bool) {is=title K2Node_Event}
   n5 = Get FortDiscoverProviderViewModel() -> (FortDiscoverProviderViewModel:Discover Provider View Model Object Reference)
   > Is Valid(InputObject=@n5)
   > StopVisibleImpression(self:Discover Provider View Model Object Reference=@n5)
   > n6 = Branch(Condition=@n4.bStopImpressions)
   StartVisibleImpressions(execute=@n6.else, self:Discover Provider View Model Object Reference=@n5)
   > SetVisibleImpressions!(execute, self:Discover Provider View Model Object Reference=@n5, FirstVisibleIndex:int=0, LastVisibleIndex:int=0) -> (then)
  ]}
 GetPanelViewModel {nodes [
   GetPanelViewModel() {is=title K2Node_FunctionEntry}
   > Return Node!(execute, ReturnValue:Discover Provider View Model Object Reference=Get FortDiscoverProviderViewModel())
  ]}
 OnNavigateToNewOuterRow {nodes [OnNavigateToNewOuterRow!() -> (then, IsDirectionUp:bool=~, PreviousSubrowSelectedIndex:int, PreviousNumItemsPerSubrow:int, PreviousRowWasViewAllTileSelected:bool=~) {is=title K2Node_FunctionEntry}]}
]
variables [OnNavigateToNewOuterRow {}]
widgets {
 info {
  inheritedWidgetCount=1
  parentClass=/Script/DiscoveryBrowserUI.FortDiscoverBrowserGridRow
  rootWidgetClass=&GridPanel
  widgetCount=5
 }
 widgets [
  GridRoot_CustomNavBoundary {
   widgetClassPath=&GridPanel
   widget=GridRoot_CustomNavBoundary
   properties {
    navigation=&self:WidgetTree.GridRoot_CustomNavBoundary.WidgetNavigation_0
    columnFill [1]
    rowFill [1]
   }
  }
  GridPanel_0 {
   widgetClassPath=&GridPanel
   parent=GridRoot_CustomNavBoundary
   widget=GridPanel_0
   properties {columnFill [1], rowFill [0,1]}
   slot {class=GridSlot, properties {rowSpan=1, padding=<ltrb 0 0 0 48>}}
  }
  WBP_Row_Header {
   widgetClassPath=/DiscoveryBrowser/Rows/WBP_Row_Header
   bIsVariable=true
   parent=GridPanel_0
   widget=WBP_Row_Header
   slot {class=GridSlot, properties {rowSpan=1, verticalAlignment=VAlign_Top}}
  }
  WBP_Discovery_Tile_Billboard {
   widgetClassPath=/DiscoveryBrowser/Tiles/Billboard/WBP_Discovery_Tile_Billboard
   bIsVariable=true
   parent=GridPanel_0
   widget=WBP_Discovery_Tile_Billboard
   slot {class=GridSlot, properties {row=1, rowSpan=1, verticalAlignment=VAlign_Top, padding=<ltrb 80 24 80 0>}}
  }
  Text_CategoryName {widgetClassPath=&CommonTextBlock, bInherited=true, widget=null}
 ]
}
