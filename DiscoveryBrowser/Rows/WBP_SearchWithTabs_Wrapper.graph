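# Discovery browser row widget, derived from WBP_Row_Grid_Base, that wraps WBP_Row_InputFieldWithTabs, a search field with tabs.
# It stores the panel's Discover provider view model, broadcasts hover from focus-path changes and forwards blade animation completion to the inner widget.
# PreConstruct disables inner safe-zone padding and previews search bar alignment at design time through SetSearchBarAlignment, which adjusts header anchor, alignment and padding.
# GetInitialRefreshFocus requires fullscreen, finished blade animation and hover.
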WidgetBlueprint /DiscoveryBrowser/Rows/WBP_SearchWithTabs_Wrapper
graphs [
 EventGraph {nodes [
   n1 = Event SetPanelViewModel!() -> (OutputDelegate, then, ViewModel:Discover Provider View Model Object Reference) {is=title K2Node_Event}
   > SetFortDiscoverProviderViewModel!(execute, self, Viewmodel:Discover Provider View Model Object Reference=@n1.ViewModel) -> (then)
   n2 = Event PreConstruct()
   > Parent: PreConstruct(IsDesignTime=@n2.IsDesignTime)
   > n3 = Sequence()
   n4 = Get WBP_Row_InputFieldWithTabs() -> (WBP_Row_InputFieldWithTabs:WBP Row Input Field With Tabs Object Reference)
   > SetSidesToPad!(
    execute
    self:Safe Zone Object Reference=[
     Get SafeZone_List!(self:WBP Discover Surface Base Object Reference=@n4) -> (SafeZone_List:Safe Zone Object Reference)
     Get SafeZone_Header!(self:WBP Discover Surface Base Object Reference=@n4) -> (SafeZone_Header:Safe Zone Object Reference)
    ]
    InPadLeft:bool=false
    InPadRight:bool=false
    InPadTop:bool=false
    InPadBottom:bool=false
   ) -> (then)
   Branch(execute=@n3.then_1, Condition=@n2.IsDesignTime)
   > SetSearchBarAlignment!(
    execute
    self
    SearchBarAlignment:ESearchBarAlignment Enum=Get Design_Search Bar Alignment() -> (Design_Search Bar Alignment:ESearchBarAlignment Enum=None)
   ) -> (then)
   Event OnAddedToFocusPath!() -> (OutputDelegate, then, InFocusEvent:Focus Event Structure) {is=title K2Node_Event}
   > Set with Broadcast IsRowHovered!(execute, IsRowHovered:bool=true) -> (then, Output_Get:bool)
   n5 = Event SetBladeExpandContractAnimationFinished!() -> (OutputDelegate, then, bBladeExpandContractAnimationFinished:bool) {is=title K2Node_Event}
   > Set with Broadcast IsBladeExpandContractAnimationFinished!(
    execute
    IsBladeExpandContractAnimationFinished:bool=@n5.bBladeExpandContractAnimationFinished
    self:WBP Row Input Field With Tabs Object Reference=Get WBP_Row_InputFieldWithTabs()
   ) -> (then, Output_Get:bool)
   Event OnRemovedFromFocusPath!() -> (OutputDelegate, then, InFocusEvent:Focus Event Structure) {is=title K2Node_Event}
   > Set with Broadcast IsRowHovered(IsRowHovered=false)
  ]}
 GetInitialRefreshFocus {nodes [
   n1 = GetInitialRefreshFocus!() -> (then, IsDiscoverFullscreen:bool=~, bBladeExpandContractAnimationFinished:bool=~, IsRowHovered:bool=~) {is=title K2Node_FunctionEntry}
   > Return Node!(execute, ShouldRefreshFocus:bool=AND Boolean!(A=@n1, B=@n1.bBladeExpandContractAnimationFinished, C:bool=@n1.IsRowHovered) -> (ReturnValue))
  ]}
 GetPanelViewModel {nodes [
   GetPanelViewModel() {is=title K2Node_FunctionEntry}
   > Return Node!(
    execute
    ReturnValue:Discover Provider View Model Object Reference=Get FortDiscoverProviderViewModel() -> (FortDiscoverProviderViewModel:Discover Provider View Model Object Reference)
   )
  ]}
 SetSearchBarAlignment {nodes [
   SetSearchBarAlignment!() -> (then, SearchBarAlignment:ESearchBarAlignment Enum=None) {is=title K2Node_FunctionEntry}
   > Set ContentAnchor!(
    execute
    ContentAnchor:Vector 2D Structure=Select!(
     None:Vector 2D Structure
     Left:Vector 2D Structure
     Right:Vector 2D Structure=(X=1,Y=0)
     Index:ESearchBarAlignment Enum=Get SearchBarAlignment() -> (SearchBarAlignment:ESearchBarAlignment Enum=None)
    ) -> (ReturnValue:Vector 2D Structure)
    self:Common Visual Attachment Object Reference=Get CommonAttachment_CollapsibleHeader!(self:WBP Discover Surface Base Object Reference=Get WBP_Row_InputFieldWithTabs()) -> (CommonAttachment_CollapsibleHeader:Common Visual Attachment Object Reference)
   ) -> (then, Output_Get:Vector 2D Structure)
   n1 = Get WBP_Row_InputFieldWithTabs()
   > SetHorizontalAlignment(
    self:Grid Slot Object Reference=SlotAsGridSlot(Widget=Get CommonAttachment_CollapsibleHeader(self=@n1))
    InHorizontalAlignment=Select(
     None:EHorizontalAlignment Enum=HAlign_Fill
     Left:EHorizontalAlignment Enum=HAlign_Fill
     Right:EHorizontalAlignment Enum=HAlign_Right
     Index=Get SearchBarAlignment() -> (SearchBarAlignment=None)
    ) -> (ReturnValue:EHorizontalAlignment Enum)
   )
   n2 = Get WBP_Row_InputFieldWithTabs()
   > SetHorizontalAlignment(
    self:Size Box Slot Object Reference=SlotAsSizeBoxSlot!(Widget:Widget Object Reference=Get NamedSlot_SurfaceInfo!(self:WBP Discover Surface Base Object Reference=@n2) -> (NamedSlot_SurfaceInfo:Named Slot Object Reference)) -> (ReturnValue:Size Box Slot Object Reference)
    InHorizontalAlignment=Select(None=HAlign_Fill, Left=HAlign_Fill, Right=HAlign_Right, Index=Get SearchBarAlignment() -> (SearchBarAlignment=None))
   )
   n3 = Get WBP_Row_InputFieldWithTabs()
   n4 = Get SearchBarAlignment() -> (SearchBarAlignment=None)
   > SetPadding!(
    execute
    self:Grid Slot Object Reference=SlotAsGridSlot(Widget=Get SafeZone_Header(self=@n3))
    InPadding_Left:float=Select(None:float=80, Left:float=80, Right:float=0, Index=@n4) -> (ReturnValue:float)
    InPadding_Top:float=0
    InPadding_Right:float=Select(None=0, Left=0, Right=80, Index=@n4)
    InPadding_Bottom:float=0
   ) -> (then)
  ]}
]
variables [
 Design_Search Bar Alignment {type=ESearchBarAlignment, default=Left}
 IsRowHovered {type=boolean}
]
widgets {
 info {
  inheritedWidgetCount=1
  parentClass=&/WBP_Row_Grid_Base.WBP_Row_Grid_Base_C
  rootWidgetClass=&GridPanel
  widgetCount=3
 }
 widgets [
  GridPanel_24 {widgetClassPath=&GridPanel, widget=GridPanel_24, properties {columnFill [1], rowFill [1]}}
  WBP_Row_InputFieldWithTabs {
   widgetClassPath=&/WBP_Row_InputFieldWithTabs
   bIsVariable=true
   parent=GridPanel_24
   widget=WBP_Row_InputFieldWithTabs
   properties {deactivateWIdgetOnClickedBackButton=true}
   slot {class=GridSlot}
  }
  Text_CategoryName {widgetClassPath=&CommonTextBlock, bInherited=true, widget=null}
 ]
}
