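# Base activatable widget for Discovery browser surfaces, with named slots for header, content and footer around a provider mega list.
# A switcher shows the loading spinner, list or no-results view, refreshing focus once loaded.
# Swipe, mouse wheel, touch-end and list navigation collapse or restore the header and set a footer flag near the list end.
# The header animation drives AnimatorForHeaderHeight, lerping a spacer's height and top padding from a cached size.
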WidgetBlueprint /DiscoveryBrowser/WBP_Discover_Surface_Base
animations [
 ShowHeaderKeyArt {
  binding_count=1
  length_seconds=0.64
  object=&self:ShowHeaderKeyArt
  sequence {
   bindings [HeaderImage {
     id=#1
     possessed_class=Image
     tracks [{
       type=MovieSceneWidgetMaterialTrack
       sections [{
         type=MovieSceneParameterSection
         end_frame=38
         start_frame=0
         channels [
          Thumbnail Opacity {type=MovieSceneScriptingFloatChannel, keys [[0,0],[38.4,1]]}
          Thumbnail Zoom {type=MovieSceneScriptingFloatChannel, keys [[0,0.05],[38.4,0]]}
          Triangle Distance {type=MovieSceneScriptingFloatChannel, keys [[0,0.5],[38.4,0]]}
         ]
        }]
      }]
    }]
   display_rate {numerator=60}
  }
 }
 Anim_HeaderScrolling {
  binding_count=3
  length_seconds=0.2
  object=&self:Anim_HeaderScrolling
  sequence {bindings [
    NamedSlot_SurfaceInfo {
     id=#2
     possessed_class=NamedSlot
     tracks [
      {
       type=MovieSceneFloatTrack
       display_name=Render Opacity
       sections [{type=MovieSceneFloatSection, end_frame=3, start_frame=0, channels [None {type=MovieSceneScriptingFloatChannel, keys [[0,1],[3.8263,0]]}]}]
      }
      {
       type=MovieSceneByteTrack
       display_name=Visibility
       sections [{
         type=MovieSceneByteSection
         end_frame=4
         start_frame=0
         channels [None {type=MovieSceneScriptingByteChannel, keys [[0,4],[1,3],[3,3],[4,1]]}]
        }]
      }
     ]
    }
    WBP_Discover_Surface_Base_C_211 {
     id=#3
     possessed_class=WBP_Discover_Surface_Base_C
     tracks [{
       type=MovieSceneDoubleTrack
       display_name=Animator for Header Height
       sections [{type=MovieSceneDoubleSection, end_frame=4, start_frame=0, channels [None {type=MovieSceneScriptingDoubleChannel, keys [[0,0],[4,1]]}]}]
      }]
    }
    HeaderImage {
     id=#4
     possessed_class=Image
     tracks [{
       type=MovieSceneFloatTrack
       display_name=Render Opacity
       sections [{type=MovieSceneFloatSection, end_frame=3, start_frame=0, channels [None {type=MovieSceneScriptingFloatChannel, keys [[0,1],[3.826,0]]}]}]
      }]
    }
   ]}
 }
 Anim_BottomReached {
  binding_count=1
  length_seconds=0.2
  object=&self:Anim_BottomReached
  sequence {
   bindings [Image_ListSoftBottomEdge {
     id=#5
     possessed_class=Image
     tracks [{
       type=MovieSceneFloatTrack
       display_name=Render Opacity
       sections [{type=MovieSceneFloatSection, channels [None {type=MovieSceneScriptingFloatChannel, keys [[0,1],[12,0]]}]}]
      }]
    }]
   display_rate {numerator=60}
  }
 }
]
graphs [
 BP_GetDesiredFocusTarget {nodes [
   Get Desired Focus Target()
   > Branch(Condition=Get HasListLoaded() -> (HasListLoaded:bool))
   n1 = Get ListView!(self:Fort Polymorphic List View Object Reference=Get WBP_Poly_MegaList() -> (WBP_Poly_MegaList:WBP Poly Mega List Object Reference)) -> (ListView:Common List View Object Reference)
   n2 = Get Selected Item!(self:List View Object Reference=@n1) -> (ReturnValue:Object Reference) {is=title K2Node_CallFunction}
   n3 = GetIndexForItem!(self:List View Object Reference=@n1, Item:Object Reference=@n2) -> (ReturnValue:int=0)
   > SetSelectedIndex!(
    execute
    self:List View Object Reference=@n1
    Index:int=Select(Option 0:int=0, Option 1:int=@n3, Index="integer >= integer"(A=@n3)) -> (ReturnValue:int)
   ) -> (then)
   > Navigate To Item!(execute, self:List View Object Reference=@n1, Item:Object Reference=@n2) -> (then)
   > Return Node!(execute, ReturnValue:Widget Object Reference=@n1)
  ]}
 EventGraph {nodes [
   "On Swipe Up (FortSwipePanel_0)"() {is=title K2Node_ComponentBoundEvent}
   n1 = Get CommonInputSubsystem()
   n2 = AND Boolean(A=Get bDisableSwipeForTouch() -> (bDisableSwipeForTouch:bool), B="Equal (Enum)"(A=GetCurrentInputType(self=@n1), B=Touch))
   > n3 = Branch(Condition=@n2)
   n4 = Get WBP_Poly_MegaList()
   n5 = SwipeToPreviousRow_Touch(execute=@n3.else, self:WBP Poly Mega List Object Reference=@n4)
   n6 = Get ListView(self=@n4)
   TransitionToMiniHeader!(
    execute=[@n5.then, @n22.ScrollUp]
    self
    IsMiniHeaderWithFullList:bool="float > float"(A=GetScrollOffset!(self:List View Base Object Reference=@n6) -> (ReturnValue:float))
   ) -> (then)
   > Branch(Condition=Get isFooterEnabled() -> (isFooterEnabled:bool))
   > EnableFooter!(execute, self, Enable:bool=false) -> (then)
   "On Clicked (WBP_BottomBarHack_Back)"!() -> (OutputDelegate, then, Button:Common Button Base Object Reference) {is=title K2Node_ComponentBoundEvent}
   > Call On Back Button Clicked!(execute:Exec, self:WBP Discover Surface Base Object Reference) -> (then:Exec) {is=title K2Node_CallDelegate}
   > Branch(Condition=Get DeactivateWIdgetOnClickedBackButton() -> (DeactivateWIdgetOnClickedBackButton:bool))
   > DeactivateWidget()
   "On Clicked (WBP_BottomBarHack_BackToTop)"!() -> (OutputDelegate, then, Button:Common Button Base Object Reference) {is=title K2Node_ComponentBoundEvent}
   > NavigateToIndex!(execute, self:List View Object Reference=Get ListView(self=Get WBP_Poly_MegaList()), Index:int=0) -> (then)
   > Switch Back and Back To Top buttons!(execute, self, Back To Top:bool=false) -> (then)
   "On Swipe Down (FortSwipePanel_0)"() {is=title K2Node_ComponentBoundEvent}
   > n7 = Branch(Condition=@n2)
   n8 = SwipeToNextRow_Touch(execute=@n7.else, self:WBP Poly Mega List Object Reference=Get WBP_Poly_MegaList())
   TransitionToMiniHeader(execute=[@n8.then, @n22.ScrollDown], IsMiniHeaderWithFullList=true)
   n9 = Get isFooterEnabled()
   n10 = Get Distance Remaining() -> (Distance Remaining:float)
   > Branch(Condition=AND Boolean(A="float <= float"(A=@n10, B=0.2) -> (ReturnValue=~), B=NOT Boolean(A=@n9)))
   > EnableFooter(Enable=true)
   n11 = "On Active Widget Index Changed BP (Switcher_List)"!() -> (OutputDelegate, then, ActiveWidget:Widget Object Reference, ActiveIndex:int=0) {is=title K2Node_ComponentBoundEvent}
   > FireExitTabAnalytics!(execute, self, ActiveWidget:Widget Object Reference=@n11.ActiveWidget) -> (then)
   n12 = Get Switcher_List() -> (Switcher_List:Common Activatable Widget Switcher Object Reference)
   > Set isLoading(isLoading:bool="Equal (Object)"(
     A=GetActiveWidget!(self:Widget Switcher Object Reference=@n12) -> (ReturnValue:Widget Object Reference)
     B=Get LoadingResults_Spinner() -> (LoadingResults_Spinner:Overlay Object Reference)
    )) -> (Output_Get:bool)
   > Call On Switcher List Active Widget Changed!(execute:Exec, self:WBP Discover Surface Base Object Reference, ActiveWidget:Widget Object Reference=@n11.ActiveWidget) -> (then:Exec) {is=title K2Node_CallDelegate}
   > Branch(Condition=Get HasListLoaded())
   > RequestRefreshFocus(self:Common Activatable Widget Object Reference)
   ResetHeader() {is=title K2Node_CustomEvent}
   n13 = Get Anim_HeaderScrolling() -> (Anim_HeaderScrolling:Widget Animation Object Reference)
   > n14 = PlayAnimation(InAnimation=@n13, StartAtTime=GetEndTime(self=@n13), PlayMode=Reverse)
   Do Once(execute=@n21.then, Reset=[@n14.then, @n17.then])
   > UpdateSpacerHeightBeforeAnimating!(execute, self) -> (then, IsHeaderZeroSize:bool)
   n15 = Get Anim_HeaderScrolling()
   > n16 = PlayAnimationForward(InAnimation=@n15)
   Do Once(execute=@n21.else, Reset=@n16.then)
   > n17 = PlayAnimationReverse(InAnimation=@n15)
   n18 = EnableFooter!() -> (OutputDelegate, then, Enable:bool=~) {is=title K2Node_CustomEvent}
   > Set isFooterEnabled(isFooterEnabled:bool=@n18.Enable) -> (Output_Get:bool)
   n19 = TransitionToMiniHeader!() -> (OutputDelegate, then, IsMiniHeaderWithFullList:bool=~) {is=title K2Node_CustomEvent}
   > n20 = Set Is Mini Header with Full List(Is Mini Header with Full List:bool=@n19.IsMiniHeaderWithFullList) -> (Output_Get:bool)
   > n21 = Branch(Condition=@n20)
   OnListViewTouchEnd() {is=title K2Node_CustomEvent}
   > n22 = Handle List View Touch End!(execute:Exec) -> (ScrollUp:Exec, ScrollDown:Exec) {is=title K2Node_MacroInstance}
   n23 = OnListViewScrolled!() -> (OutputDelegate, then, ItemOffset:float=~, DistanceRemaining:float=~) {is=title K2Node_CustomEvent}
   > Set Distance Remaining(Distance Remaining:float=@n23.DistanceRemaining) -> (Output_Get:float)
   n24 = OnUserNavigatedMegalist!() -> (OutputDelegate, then, IsDirectionUp:bool=~, NewDesiredOuterIndex:int) {is=title K2Node_CustomEvent}
   > n25 = Branch(Condition=@n24.IsDirectionUp)
   > TransitionToMiniHeader(IsMiniHeaderWithFullList="integer > integer"(A=@n24.NewDesiredOuterIndex)) {is=title K2Node_CallFunction}
   > Branch(Condition=AND Boolean(A=Get isFooterEnabled(), B=@n24.IsDirectionUp))
   > EnableFooter(Enable=false) {is=title K2Node_CallFunction}
   TransitionToMiniHeader(execute=@n25.else, IsMiniHeaderWithFullList=true)
   n26 = Get isFooterEnabled()
   n27 = Get WBP_Poly_MegaList()
   n28 = int - int(A=GetNumItems!(self:List View Object Reference=Get ListView(self=@n27)) -> (ReturnValue:int=0))
   > Branch(Condition=AND Boolean!(A=NOT Boolean(A=@n24.IsDirectionUp), B="Equal (Integer)"(A=@n24.NewDesiredOuterIndex, B=@n28), C:bool=NOT Boolean(A=@n26)) -> (ReturnValue))
   > EnableFooter(Enable=true)
   n29 = OnInputMethodChanged_Event!() -> (OutputDelegate, then, bNewInputType:ECommonInputType Enum) {is=title K2Node_CustomEvent}
   > Branch(Condition="Equal (Enum)"(
     A=@n29.bNewInputType
     B=Literal enum ECommonInputType!(Enum:ECommonInputType Enum=MouseAndKeyboard) -> (ReturnValue:ECommonInputType Enum=~) {is=title K2Node_EnumLiteral}
    ))
   > Switch Back and Back To Top buttons(Back To Top=false)
   OnHeaderAnimationFinished() {is=title K2Node_CustomEvent}
   n30 = Get Anim_HeaderScrolling()
   > Branch(Condition="Equal (Float)"(A=GetAnimationCurrentTime!(self:User Widget Object Reference, InAnimation:Widget Animation Object Reference=@n30) -> (ReturnValue:float)))
   > UpdateSpacerHeightBeforeAnimating()
   n31 = Event PreConstruct()
   > Switch Back and Back To Top buttons(Back To Top=false)
   > SetBrushFromMaterial(
    self=Get HeaderImage() -> (HeaderImage:Image Object Reference)
    Material=Get HeaderDefaultMaterial() -> (HeaderDefaultMaterial:Material Instance Object Reference)
   )
   > ResetHeader() {is=title K2Node_CallFunction}
   n32 = Get Slot_CustomNoResults() -> (Slot_CustomNoResults:Named Slot Object Reference)
   > SetActiveWidget(
    self:Widget Switcher Object Reference=Get Switcher_NoResults() -> (Switcher_NoResults:Common Widget Switcher Legacy Object Reference)
    Widget=Select(
     Option 0:Widget Object Reference=Get WBP_Discovery_NoResultsMessage() -> (WBP_Discovery_NoResultsMessage:WBP Discovery No Results Message Object Reference)
     Option 1:Widget Object Reference=@n32
     Index=HasAnyChildren!(self:Panel Widget Object Reference=@n32) -> (ReturnValue:bool)
    ) -> (ReturnValue:Widget Object Reference)
   )
   > n33 = Branch(Condition=@n31.IsDesignTime)
   Set Padding!(
    execute=[@n34.then, @n33.then]
    Padding:Margin Structure=Get HeaderSectionPadding() -> (HeaderSectionPadding:Margin Structure)
    self:Grid Slot Object Reference=SlotAsGridSlot(Widget=Get SafeZone_Header() -> (SafeZone_Header:Safe Zone Object Reference))
   ) -> (then, Output_Get:Margin Structure)
   > SetCompactUIStyle() {is=title K2Node_CallFunction}
   n34 = SetActiveWidget(execute=@n33.else, self=Get Switcher_List(), Widget=Get LoadingResults_Spinner())
   Event Construct()
   > Bind Event to On User Navigated Mega List(self:WBP Poly Mega List Object Reference=Get WBP_Poly_MegaList(), Delegate=Create Event()) {is=title K2Node_AddDelegate}
   > BindToAnimationFinished(Animation=Get Anim_HeaderScrolling(), Delegate=Create Event())
   > Set HasListLoaded(HasListLoaded:bool=false) -> (Output_Get:bool)
   > Bind Event to On List View Scrolled(self:List View Object Reference=Get ListView(self=Get WBP_Poly_MegaList()), Delegate=Create Event()) {is=title K2Node_AddDelegate}
   > Bind Event to On List View Touch End(self:List View Object Reference=Get ListView(self=Get WBP_Poly_MegaList()), Delegate=Create Event()) {is=title K2Node_AddDelegate}
   > Bind Event to On Input Method Changed(self=Get CommonInputSubsystem(), Delegate=Create Event())
   > n35 = Branch Natural Nav Rollback Cvar!(In:Exec) -> (Use Natural Nav:Exec, Rollback:Exec) {is=title K2Node_MacroInstance}
   SetPadding!(
    execute=@n35.Rollback
    self:User Widget Object Reference=Get WBP_Poly_MegaList()
    InPadding_Left:float=0
    InPadding_Top:float=0
    InPadding_Right:float=0
    InPadding_Bottom:float=0
   ) -> (then)
   Event Destruct()
   > Unbind all Events from On User Navigated Mega List!(execute:Exec, self:WBP Poly Mega List Object Reference=Get WBP_Poly_MegaList()) -> (then:Exec) {is=title K2Node_ClearDelegate}
   > Unbind all Events from On List View Scrolled!(execute:Exec, self:List View Object Reference=Get ListView(self=Get WBP_Poly_MegaList())) -> (then:Exec) {is=title K2Node_ClearDelegate}
   > Unbind Event from On List View Touch End(self:List View Object Reference=Get ListView(self=Get WBP_Poly_MegaList()), Delegate=Create Event()) {is=title K2Node_RemoveDelegate}
   > Unbind Event from On Input Method Changed(self=Get CommonInputSubsystem(), Delegate=Create Event())
  ]}
 FillOutList {nodes [
   n1 = FillOutList!() -> (then, ProvidersArray:Array of Discover Provider View Model Object References) {is=title K2Node_FunctionEntry}
   > Set with Broadcast IsListEmpty!(execute, IsListEmpty:bool=Is Empty!(TargetArray:Array of Discover Provider View Model Object References=@n1) -> (ReturnValue:bool)) -> (then, Output_Get:bool)
   > Set List Items!(execute, self:List View Object Reference=Get ListView(self=Get WBP_Poly_MegaList()), InListItems:Array of Object References=@n1) -> (then) {is=title K2Node_CallFunction}
   > UpdateSpacerHeightBeforeAnimating()
  ]}
 FireExitTabAnalytics {nodes [FireExitTabAnalytics!() -> (then, ActiveWidget:Widget Object Reference) {is=title K2Node_FunctionEntry}]}
 HandleListViewTouchEnd {nodes [
   Inputs()
   > Branch(Condition=AND Boolean!(
     A=Get IsUsingListViewTouchEvents() -> (IsUsingListViewTouchEvents:bool)
     B=GetConfigDataBoolWithDefault!(EntryName:String=DiscoverySurfaceUsesListViewTouchEvents, bDefaultValue:bool=true) -> (ReturnValue:bool)
    ) -> (ReturnValue))
   n1 = Get CachedScrollOffset() -> (CachedScrollOffset:float)
   n2 = GetScrollOffset(self=Get ListView(self=Get WBP_Poly_MegaList()))
   > n3 = Branch(Condition="Nearly Equal (Float)"(A=@n1, B=@n2, ErrorTolerance=0.0001))
   n4 = GetOverscroll!(self:List View Base Object Reference=Get ListView(self=Get WBP_Poly_MegaList())) -> (ReturnValue:float)
   > n5 = Branch(Condition="Nearly Equal (Float)"(A=@n4, B=0, ErrorTolerance=0))
   n6 = Branch(execute=@n5.else, Condition="float > float"(B=@n4))
   Outputs!(ScrollUp:Exec=[@n6.then, @n8.then], ScrollDown:Exec=[@n6.else, @n8.else])
   n7 = Sequence(execute=@n3.else)
   > n8 = Branch(Condition="float > float"(A=@n1, B=@n2))
   Set CachedScrollOffset(execute=@n7.then_1, CachedScrollOffset:float=@n2) -> (Output_Get:float)
  ]}
 On Back Button Clicked {nodes [On Back Button Clicked() {is=title K2Node_FunctionEntry}]}
 OnMouseWheel {nodes [
   n1 = OnMouseWheel!() -> (then, MyGeometry:Geometry Structure, MouseEvent:"Pointer Event Structure (by ref)") {is=title K2Node_FunctionEntry}
   > n2 = Branch(Condition=IsActivated!(self:Common Activatable Widget Object Reference) -> (ReturnValue:bool))
   n3 = Get Wheel Delta!(Input:"Pointer Event Structure (by ref)"=@n1.MouseEvent) -> (ReturnValue:float) {is=title K2Node_CallFunction}
   > n4 = Branch(Condition="float < float"(A=@n3))
   > TransitionToMiniHeader(IsMiniHeaderWithFullList=true)
   n5 = Get isLoading() -> (isLoading:bool)
   n6 = GetScrollOffset(self=Get ListView(self=Get WBP_Poly_MegaList()))
   n7 = "float > float"(A=@n6, B=0.9)
   n8 = "float >= float"(A=@n6, B=Get Distance Remaining())
   n9 = "float < float"(A=Get Distance Remaining(), B=0.25)
   n10 = "Equal (Float)"(A=@n6, B=Get CachedScrollOffset())
   > n11 = Branch(Condition=AND Boolean!(A=OR Boolean!(A=@n8, B=@n7, C:bool=@n10, D:bool=@n9) -> (ReturnValue), B=Get isFooterEnabled(), C:bool=NOT Boolean(A=@n5)) -> (ReturnValue))
   > ScrollToBottom(self:List View Base Object Reference=Get ListView(self=Get WBP_Poly_MegaList()))
   > n12 = EnableFooter(Enable=true)
   n13 = Get WBP_Poly_MegaList()
   Set CachedScrollOffset(execute=[@n12.then, @n22.then, @n11.else], CachedScrollOffset=GetScrollOffset(self=Get ListView(self=@n13)))
   > Return Node(ReturnValue:Event Reply Structure=Handled())
   n14 = Branch(execute=@n4.else, Condition="float > float"(A=@n3))
   > n15 = Branch(Condition=Get isFooterEnabled())
   > EnableFooter(Enable=false)
   n16 = GetScrollOffset(self=Get ListView(self=Get WBP_Poly_MegaList()))
   n17 = Get Distance Remaining()
   > n18 = Branch(Condition=AND Boolean!(A="float <= float"(A=@n16, B=0.4) -> (ReturnValue=~), B="Equal (Float)"(A=@n17)) -> (ReturnValue))
   > n19 = ScrollToTop(self:List View Base Object Reference=Get ListView(self=Get WBP_Poly_MegaList()))
   n20 = Branch(execute=[@n19.then, @n15.else, @n18.else], Condition="Nearly Equal (Float)"(A=@n16, B=0, ErrorTolerance=0))
   > n21 = ScrollToTop(self=Get ListView(self=Get WBP_Poly_MegaList()))
   n22 = TransitionToMiniHeader(execute=[@n21.then, @n20.else], IsMiniHeaderWithFullList="float > float"(A=@n16))
   n23 = Parent: OnMouseWheel!(execute=[@n14.else, @n2.else], MyGeometry:Geometry Structure=@n1, MouseEvent:"Pointer Event Structure (by ref)"=@n1.MouseEvent) -> (then, ReturnValue:Event Reply Structure) {is=title K2Node_CallParentFunction}
   > Return Node(ReturnValue=@n23)
   "To String (Float)"(InDouble=@n16)
  ]}
 OnSwitcherListActiveWidgetChanged {nodes [OnSwitcherListActiveWidgetChanged!() -> (then, ActiveWidget:Widget Object Reference) {is=title K2Node_FunctionEntry}]}
 RefreshFocusWhenListHasLoaded {nodes [
   n1 = RefreshFocusWhenListHasLoaded!() -> (then, LoadingState:EDiscoverLoadingState Enum=Unset) {is=title K2Node_FunctionEntry}
   > n2 = Switch on EDiscoverLoadingState!(execute, Selection:EDiscoverLoadingState Enum=@n1) -> (Unset:Exec, Loading:Exec, Loaded:Exec, Error:Exec) {is=title K2Node_SwitchEnum}
   Set HasListLoaded(execute=[@n2.Error, @n2.Loaded], HasListLoaded=true)
   > RequestRefreshFocus()
   Set HasListLoaded(execute=[@n2.Loading, @n2.Unset], HasListLoaded=false)
  ]}
 SetAnimatorForHeaderHeight {nodes [
   n1 = SetAnimatorForHeaderHeight!() -> (then, NewValue:double=~) {is=title K2Node_FunctionEntry}
   n2 = Get CollapsibleHeaderCachedHeight() -> (CollapsibleHeaderCachedHeight:double)
   > SetSize!(
    execute
    self:Spacer Object Reference=Get AnimationSpacer() -> (AnimationSpacer:Spacer Object Reference)
    InSize:Vector 2D Structure=MakeVector2D(X=0, Y=Lerp(A=@n2, B=0, Alpha=@n1))
   ) -> (then)
   n3 = SlotAsGridSlot(Widget=Get SafeZone_Header())
   n4 = Break Margin!(Margin:"Margin Structure (by ref)"=Get Padding!(self:Grid Slot Object Reference=@n3) -> (Padding:Margin Structure)) -> (Left:float=~, Top:float=~, Right:float=~, Bottom:float=~) {is=title K2Node_BreakStruct}
   n5 = Break Margin(Margin=Get HeaderSectionPadding()) -> (Left=~, Top=~, Right=~, Bottom=~)
   n6 = Get HeaderCollapsedPaddingTop() -> (HeaderCollapsedPaddingTop:double)
   > Set Padding(Padding=Make Margin(Left=@n4, Top=Lerp(A=@n5.Top, B=@n6, Alpha=@n1), Right=@n4.Right, Bottom=@n4.Bottom), self=@n3)
  ]}
 SetBackgroundMaterialColor {nodes [
   SetBackgroundMaterialColor!() -> (then, Background Color:Linear Color Structure) {is=title K2Node_FunctionEntry}
   n1 = Get HeaderImage()
   > SetVectorParameterValue(
    self=GetDynamicMaterial!(Brush:"Slate Brush Structure (by ref)"=Get Brush!(self:Image Object Reference=@n1) -> (Brush:Slate Brush Structure)) -> (ReturnValue)
    ParameterName=Primary Background Color
    Value=Get Background Color() -> (Background Color:Linear Color Structure)
   )
   n2 = Get Image_ListSoftBottomEdge() -> (Image_ListSoftBottomEdge:Image Object Reference)
   > SetVectorParameterValue(self=GetDynamicMaterial(Brush=Get Brush(self=@n2)), ParameterName=Primary Background Color, Value=Get Background Color())
   n3 = Get Background() -> (Background:Image Object Reference)
   > SetVectorParameterValue(self=GetDynamicMaterial(Brush=Get Brush(self=@n3)), ParameterName=Primary Background Color, Value=Get Background Color())
  ]}
 SetCompactUIStyle {nodes [
   SetCompactUIStyle() {is=title K2Node_FunctionEntry}
   > Branch(Condition=IsUsingCompactUI())
   > SetWidthOverride(
    self=Get CommonAttachment_CollapsibleHeader() -> (CommonAttachment_CollapsibleHeader:Common Visual Attachment Object Reference)
    InWidthOverride=SelectFloat(
     A=1977
     B=1600
     bPickA=CompareToScreenAspectRatio!(Comparison:EAspectRatioComparison Enum=GreaterThan, AspectRatio:EAspectRatio Enum=Ratio_16_9) -> (ReturnValue:bool)
    )
   )
  ]}
 SetHeaderImageState {nodes [SetHeaderImageState!() -> (then, LoadingEnum:EActivityImageLoadingState Enum=Unset) {is=title K2Node_FunctionEntry}]}
 SetListQuerying {nodes [
   n1 = SetListQuerying!() -> (then, LoadingState:EDiscoverLoadingState Enum=Unset, IsListEmpty:bool=~) {is=title K2Node_FunctionEntry}
   > n2 = Branch(Condition="Equal (Enum)"(A:EDiscoverLoadingState Enum=@n1, B:EDiscoverLoadingState Enum=Error))
   > n3 = Set Message Text Overrides!(
    execute
    self:WBP Discovery No Results Message Object Reference=Get WBP_Discovery_NoResultsMessage()
    InTitle:Text="LOCTEXT(\"Unable To Load Content\")"
    InMessage:Text="LOCTEXT(\"Refresh the page or try again later\")"
   ) -> (then) {is=title K2Node_CallFunction}
   n4 = NOT Boolean(A=@n1.IsListEmpty)
   n5 = Get Switcher_NoResults()
   n6 = Get OverlayList() -> (OverlayList:Overlay Object Reference)
   Return Node!(
    execute=[@n3.then, @n2.else]
    SwitcherActiveWidget:Widget Object Reference=Select(Option 0=Select(Option 0=@n5, Option 1=@n6, Index=@n4), Option 1=Get LoadingResults_Spinner(), Index="Equal (Enum)"(A=@n1, B=Loading))
   )
  ]}
 SetPageStyle {nodes [
   SetPageStyle!() -> (then, LinkCategory:EFortActivityLinkCategory Enum=None) {is=title K2Node_FunctionEntry}
   > SetBrushFromMaterial(
    self=Get Background()
    Material=GetBackgroundMaterial!(LinkCategory:EFortActivityLinkCategory Enum=Get LinkCategory() -> (LinkCategory:EFortActivityLinkCategory Enum=None)) -> (BackgroundMaterial:Material Interface Object Reference)
   )
   > SetBrushFromMaterial(
    self=Get HeaderImage()
    Material=GetHeaderImageMaterial!(LinkCategory:EFortActivityLinkCategory Enum=Get LinkCategory() -> (LinkCategory=None)) -> (HeaderMaterial:Material Interface Object Reference)
   )
   > SetBrushFromMaterial(
    self=Get Image_ListSoftBottomEdge()
    Material=GetFooterMaterial!(LinkCategory:EFortActivityLinkCategory Enum=Get LinkCategory() -> (LinkCategory=None)) -> (FooterMaterial:Material Interface Object Reference)
   )
  ]}
 Switch Back and Back To Top buttons {nodes [
   n1 = Switch Back and Back To Top buttons!() -> (then, Back To Top:bool=~) {is=title K2Node_FunctionEntry}
   > SetBottomButtonVisibility!(
    execute
    self:WBP Bottom Bar Decoy Button Object Reference=Get WBP_BottomBarHack_Back() -> (WBP_BottomBarHack_Back:WBP Bottom Bar Decoy Button Object Reference)
    ShouldButtonBeVisible:bool=NOT Boolean(A=@n1)
   ) -> (then)
   > SetBottomButtonVisibility(
    self=Get WBP_BottomBarHack_BackToTop() -> (WBP_BottomBarHack_BackToTop:WBP Bottom Bar Decoy Button Object Reference)
    ShouldButtonBeVisible=@n1
   )
  ]}
 UpdateSpacerHeightBeforeAnimating {nodes [
   UpdateSpacerHeightBeforeAnimating() {is=title K2Node_FunctionEntry}
   n1 = Get NamedSlot_SurfaceInfo() -> (NamedSlot_SurfaceInfo:Named Slot Object Reference)
   n2 = BreakVector2D(InVec=GetLocalSize!(Geometry:"Geometry Structure (by ref)"=GetPaintSpaceGeometry!(self:Widget Object Reference=@n1) -> (ReturnValue:"Geometry Structure (by ref)")) -> (ReturnValue:Vector 2D Structure))
   n3 = HasAnyChildren(self=Get NamedSlot_SurfaceInfo())
   n4 = BreakVector2D(InVec=Get Size!(self:Spacer Object Reference=Get MinHeaderSize_Spacer() -> (MinHeaderSize_Spacer:Spacer Object Reference)) -> (Size:Vector 2D Structure))
   n5 = SelectFloat(A=@n2.Y, B=@n4.Y, bPickA=@n3)
   n6 = "Equal (Float)"(A=@n5)
   > n7 = Set CollapsibleHeaderCachedHeight(CollapsibleHeaderCachedHeight:double=Select(Option 0:double=Get CollapsibleHeaderCachedHeight(), Option 1:double=@n5, Index=NOT Boolean(A=@n6)) -> (ReturnValue:double)) -> (Output_Get:double)
   n8 = MakeVector2D(X=SelectFloat(A=@n2, B=@n4, bPickA=@n3), Y=@n7)
   > SetSize(self=Get AnimationSpacer(), InSize=@n8)
   > Return Node!(execute, IsHeaderZeroSize:bool=IsZero2D!(A:"Vector 2D Structure (by ref)"=@n8) -> (ReturnValue:bool))
  ]}
]
variables [
 AnimatorForHeaderHeight {type=number}
 bDisableSwipeForTouch {type=boolean}
 CachedScrollOffset {type=number}
 CanNavigateToHeaderSection {type=boolean}
 CollapsibleHeaderCachedHeight {type=number, default=300}
 DeactivateWIdgetOnClickedBackButton {type=boolean, default=true}
 Distance Remaining {type=number}
 HasListLoaded {type=boolean}
 HeaderCollapsedPaddingTop {type=number}
 HeaderDefaultMaterial {type=MaterialInstance, default=/DiscoveryBrowser/CreatorProfile/MaterialInstances/MI_CreatorPage_Header}
 HeaderDynamicMaterial {type=MaterialInstanceDynamic}
 HeaderSectionPadding {type=Margin, default=<ltrb 80 0 0 0>}
 Is Mini Header with Full List {type=boolean}
 isFooterEnabled {type=boolean}
 IsListEmpty {type=boolean}
 isLoading {type=boolean}
 IsUsingListViewTouchEvents {type=boolean}
 On Back Button Clicked {}
 OnSwitcherListActiveWidgetChanged {}
 ShouldFocusViewAll {type=boolean}
 ThumbnailOpacity {type=string, default=Thumbnail Opacity}
 ThumbnailParam {type=string, default=Thumbnail}
]
widgets {
 info {
  inheritedWidgetCount=0
  parentClass=&CommonActivatableWidget
  rootWidgetClass=/Script/FortniteUI.FortSwipePanel
  widgetCount=35
 }
 widgets [
  FortSwipePanel_0 {
   widgetClassPath=&FortSwipePanel
   bIsVariable=true
   widget=FortSwipePanel_0
   properties {bBeginSwipeOnPointerEnter=false, bConsumePointerInput=false, thresholdUnit=Inches, swipeThreshold=<0 0.25>}
  }
  GridPanel_0 {
   widgetClassPath=&GridPanel
   parent=FortSwipePanel_0
   widget=GridPanel_0
   properties {columnFill [1], rowFill [1]}
   slot {class=FortSwipePanelSlot}
  }
  Background {
   widgetClassPath=&Image
   bIsVariable=true
   parent=GridPanel_0
   widget=Background
   properties {
    visibility=HitTestInvisible
    brush {
     bIsDynamicallyLoaded=false
     drawAs=Image
     imageType=FullColor
     mirroring=NoMirror
     resourceName=None
     resourceObject=/DiscoveryBrowser/CreatorProfile/MaterialInstances/MI_CreatorPage_Bg
     tiling=NoTile
     imageSize=<32 32>
     margin=<ltrb 0 0 0 0>
     outlineSettings {
      bUseBrushTransparency=false
      roundingType=HalfHeightRadius
      width=0
      color {colorUseRule=UseColor_Specified, specifiedColor=<rgba 0 0 0 0>}
      cornerRadii=<0 0 0 0>
     }
     tintColor {colorUseRule=UseColor_Specified, specifiedColor=<rgba 1 1 1 1>}
     uVRegion {bIsValid=false, max=<0 0>, min=<0 0>}
    }
   }
   slot {class=GridSlot, properties {rowSpan=2}}
  }
  WBP_CaptureForPostBufferUpdate {
   widgetClassPath=&WBP_CaptureForPostBufferUpdate
   parent=GridPanel_0
   slot {class=GridSlot, properties {layer=1, rowSpan=2}}
  }
  HBBottomBarButtons {
   widgetClassPath=&HorizontalBox
   parent=GridPanel_0
   widget=HBBottomBarButtons
   slot {class=GridSlot, properties {horizontalAlignment=HAlign_Left, verticalAlignment=VAlign_Top}}
  }
  WBP_BottomBarHack_Back {
   widgetClassPath=/Game/UI/Frontend/BottomBar/WBP_BottomBarDecoyButton
   bIsVariable=true
   parent=HBBottomBarButtons
   widget=WBP_BottomBarHack_Back
   properties {minWidth=120, triggeringInputAction {dataTable=&InputActionDataTable, rowName=Back}}
   slot {class=HorizontalBoxSlot}
  }
  WBP_BottomBarHack_BackToTop {
   widgetClassPath=&WBP_BottomBarDecoyButton
   bIsVariable=true
   parent=HBBottomBarButtons
   widget=WBP_BottomBarHack_BackToTop
   properties {minWidth=120, triggeringInputAction {dataTable=&InputActionDataTable, rowName=BackToTop}}
   slot {class=HorizontalBoxSlot}
  }
  WBP_BottomBarHack_FavoriteCollection {
   widgetClassPath=&WBP_BottomBarDecoyButton
   bIsVariable=true
   parent=HBBottomBarButtons
   widget=WBP_BottomBarHack_FavoriteCollection
   properties {minWidth=120}
   slot {class=HorizontalBoxSlot}
  }
  WBP_BottomBarHack_SupportCreator {
   widgetClassPath=&WBP_BottomBarDecoyButton
   bIsVariable=true
   parent=HBBottomBarButtons
   widget=WBP_BottomBarHack_SupportCreator
   properties {minWidth=120}
   slot {class=HorizontalBoxSlot}
  }
  O_HeaderImage {
   widgetClassPath=&Overlay
   parent=GridPanel_0
   widget=O_HeaderImage
   slot {class=GridSlot, properties {columnSpan=1, rowSpan=2, verticalAlignment=VAlign_Top}}
  }
  HeaderImage {
   widgetClassPath=&Image
   bIsVariable=true
   parent=O_HeaderImage
   widget=HeaderImage
   properties {
    visibility=SelfHitTestInvisible
    brush {
     bIsDynamicallyLoaded=false
     drawAs=Image
     imageType=FullColor
     mirroring=NoMirror
     resourceName=None
     resourceObject=&MI_CreatorPage_Header
     tiling=NoTile
     imageSize=<32 32>
     margin=<ltrb 0 0 0 0>
     outlineSettings {
      bUseBrushTransparency=false
      roundingType=HalfHeightRadius
      width=0
      color {colorUseRule=UseColor_Specified, specifiedColor=<rgba 0 0 0 0>}
      cornerRadii=<0 0 0 0>
     }
     tintColor {colorUseRule=UseColor_Specified, specifiedColor=<rgba 1 1 1 1>}
     uVRegion {bIsValid=false, max=<0 0>, min=<0 0>}
    }
   }
   slot {class=OverlaySlot, properties {horizontalAlignment=HAlign_Fill, verticalAlignment=VAlign_Fill, padding=<ltrb 0 0 0 -120>}}
  }
  HeaderLoadingThrobber {
   widgetClassPath=/FortUILibrary/UIKit/Throbber/WBP_UIKit_Throbber
   bIsVariable=true
   parent=O_HeaderImage
   widget=HeaderLoadingThrobber
   properties {throbberSize=Medium}
   slot {class=OverlaySlot, properties {horizontalAlignment=HAlign_Center, verticalAlignment=VAlign_Center}}
  }
  MinHeaderSize_Spacer {
   widgetClassPath=&Spacer
   bIsVariable=true
   parent=O_HeaderImage
   widget=MinHeaderSize_Spacer
   properties {visibility=HitTestInvisible, size=<1 408>}
   slot {class=OverlaySlot, properties {horizontalAlignment=HAlign_Fill}}
  }
  GP_Content {
   widgetClassPath=&GridPanel
   parent=GridPanel_0
   widget=GP_Content
   properties {columnFill [1], rowFill [0,1]}
   slot {class=GridSlot, properties {layer=10, rowSpan=2}}
  }
  SafeZone_Header {
   widgetClassPath=&SafeZone
   bIsVariable=true
   parent=GP_Content
   widget=SafeZone_Header
   properties {padBottom=false, padTop=false}
   slot {class=GridSlot, properties {layer=2, verticalAlignment=VAlign_Top, padding=<ltrb 80 0 0 0>}}
  }
  Grid_HeaderContent {
   widgetClassPath=&GridPanel
   bIsVariable=true
   parent=SafeZone_Header
   widget=Grid_HeaderContent
   properties {columnFill [1], rowFill [0,0]}
   slot {class=SafeZoneSlot}
  }
  O_TopHeader_NoCollapse {
   widgetClassPath=&Overlay
   parent=Grid_HeaderContent
   widget=O_TopHeader_NoCollapse
   slot {class=GridSlot, properties {columnSpan=1, rowSpan=1}}
  }
  NamedSlot_NonCollapsibleSurfaceInfo {
   widgetClassPath=&NamedSlot
   bIsVariable=true
   parent=O_TopHeader_NoCollapse
   widget=NamedSlot_NonCollapsibleSurfaceInfo
   slot {class=OverlaySlot, properties {horizontalAlignment=HAlign_Fill, verticalAlignment=VAlign_Fill}}
  }
  CommonAttachment_CollapsibleHeader {
   widgetClassPath=&CommonVisualAttachment
   bIsVariable=true
   parent=Grid_HeaderContent
   widget=CommonAttachment_CollapsibleHeader
   properties {bOverride_WidthOverride=true, widthOverride=1760}
   slot {class=GridSlot, properties {row=1, rowSpan=1}}
  }
  NamedSlot_SurfaceInfo {
   widgetClassPath=&NamedSlot
   bIsVariable=true
   parent=CommonAttachment_CollapsibleHeader
   widget=NamedSlot_SurfaceInfo
   properties {visibility=Visible}
   slot {class=SizeBoxSlot}
  }
  AnimationSpacer {
   widgetClassPath=&Spacer
   bIsVariable=true
   parent=Grid_HeaderContent
   widget=AnimationSpacer
   properties {size=<1920 408>}
   slot {class=GridSlot, properties {row=1, rowSpan=1, verticalAlignment=VAlign_Top}}
  }
  SafeZone_List {
   widgetClassPath=&SafeZone
   bIsVariable=true
   parent=GP_Content
   widget=SafeZone_List
   properties {padBottom=false, padTop=false}
   slot {class=GridSlot, properties {layer=1, row=1, rowSpan=1}}
  }
  SB_MegaListContentHeightLock {widgetClassPath=&SizeBox, bIsVariable=true, parent=SafeZone_List, widget=SB_MegaListContentHeightLock, slot {class=SafeZoneSlot}}
  Switcher_List {
   widgetClassPath=&CommonActivatableWidgetSwitcher
   bIsVariable=true
   parent=SB_MegaListContentHeightLock
   widget=Switcher_List
   properties {transitionDuration=0}
   slot {class=SizeBoxSlot}
  }
  OverlayList {
   widgetClassPath=&Overlay
   bIsVariable=true
   parent=Switcher_List
   widget=OverlayList
   properties {clipping=ClipToBounds}
   slot {class=WidgetSwitcherSlot}
  }
  WBP_Poly_MegaList {
   widgetClassPath=&/WBP_Poly_MegaList
   bIsVariable=true
   parent=OverlayList
   widget=WBP_Poly_MegaList
   slot {class=OverlaySlot, properties {horizontalAlignment=HAlign_Fill, verticalAlignment=VAlign_Fill, padding=<ltrb 0 10 0 85>}}
  }
  Switcher_NoResults {
   widgetClassPath=&CommonWidgetSwitcherLegacy
   bIsVariable=true
   parent=Switcher_List
   widget=Switcher_NoResults
   properties {transitionDuration=0}
   slot {class=WidgetSwitcherSlot}
  }
  WBP_Discovery_NoResultsMessage {
   widgetClassPath=&/WBP_Discovery_MessageNoResults.WBP_Discovery_NoResultsMessage
   bIsVariable=true
   parent=Switcher_NoResults
   widget=WBP_Discovery_NoResultsMessage
   properties {flavorTitleBase=Nothing Here Yet}
   slot {class=WidgetSwitcherSlot, properties {horizontalAlignment=HAlign_Center, verticalAlignment=VAlign_Center}}
  }
  Slot_CustomNoResults {
   widgetClassPath=&NamedSlot
   bIsVariable=true
   parent=Switcher_NoResults
   widget=Slot_CustomNoResults
   slot {class=WidgetSwitcherSlot}
  }
  LoadingResults_Spinner {
   widgetClassPath=&Overlay
   bIsVariable=true
   parent=Switcher_List
   widget=LoadingResults_Spinner
   slot {class=WidgetSwitcherSlot, properties {padding=<ltrb 20 0 0 0>}}
  }
  WBP_UIKit_Throbber {
   widgetClassPath=&WBP_UIKit_Throbber
   bIsVariable=true
   parent=LoadingResults_Spinner
   widget=WBP_UIKit_Throbber
   properties {displayBackground=false, throbberSize=Large}
   slot {class=OverlaySlot, properties {horizontalAlignment=HAlign_Center, verticalAlignment=VAlign_Center, padding=<ltrb 0 0 0 40>}}
  }
  NamedSlot_SwitcherListContent {
   widgetClassPath=&NamedSlot
   bIsVariable=true
   parent=Switcher_List
   widget=NamedSlot_SwitcherListContent
   slot {class=WidgetSwitcherSlot}
  }
  NamedSlot_OverTopContent {
   widgetClassPath=&NamedSlot
   bIsVariable=true
   parent=GP_Content
   widget=NamedSlot_OverTopContent
   slot {class=GridSlot, properties {layer=25, row=1}}
  }
  NamedSlot_Footer {
   widgetClassPath=&NamedSlot
   bIsVariable=true
   parent=GP_Content
   widget=NamedSlot_Footer
   slot {class=GridSlot, properties {layer=25, row=2, verticalAlignment=VAlign_Bottom, padding=<ltrb 80 0 0 0>}}
  }
  Image_ListSoftBottomEdge {
   widgetClassPath=&Image
   bIsVariable=true
   parent=GP_Content
   widget=Image_ListSoftBottomEdge
   properties {
    visibility=HitTestInvisible
    brush {
     bIsDynamicallyLoaded=false
     drawAs=Image
     imageType=FullColor
     mirroring=NoMirror
     resourceName=None
     resourceObject=/DiscoveryBrowser/TabScreens/DiscoverHome/Materials/MI_UI_Discovery_NebulousBackgroundWithBottomMask
     tiling=NoTile
     imageSize=<32 90>
     margin=<ltrb 0 0 0 0>
     outlineSettings {
      bUseBrushTransparency=false
      roundingType=HalfHeightRadius
      width=0
      color {colorUseRule=UseColor_Specified, specifiedColor=<rgba 0 0 0 0>}
      cornerRadii=<0 0 0 0>
     }
     tintColor {colorUseRule=UseColor_Specified, specifiedColor=<rgba 1 1 1 1>}
     uVRegion {bIsValid=false, max=<0 0>, min=<0 0>}
    }
   }
   slot {class=GridSlot, properties {layer=10, row=1, verticalAlignment=VAlign_Bottom}}
  }
 ]
}
