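# Nested Discover surface screen for a collection, extending WBP_Discover_Surface_Base with a banner, logo or fallback title, and favorite toggle.
# The tile item view model supplies name and images, whose loading states drive the header material, throbber and logo.
# Activation starts impression analytics and can show a focus-taking favorite coach mark.
# Input method repositions that coach mark and, with a user setting, limits header parallax to mouse and keyboard.
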WidgetBlueprint /DiscoveryBrowser/WBP_Discover_NestedSurface
graphs [
 BP_GetDesiredFocusTarget {nodes [
   Get Desired Focus Target()
   n1 = Get WBP_Nested_FavoriteCollection_CoachMark() -> (WBP_Nested_FavoriteCollection_CoachMark:WBP Nested Favorite Collection Coach Mark Object Reference)
   > n2 = Branch(Condition=NOT Boolean(A=IsActivated!(self:Common Activatable Widget Object Reference=@n1) -> (ReturnValue:bool)))
   > n3 = Parent: Get Desired Focus Target!(execute) -> (then, ReturnValue:Widget Object Reference) {is=title K2Node_CallParentFunction}
   > Return Node!(execute, ReturnValue:Widget Object Reference=@n3)
   Return Node(execute=@n2.else, ReturnValue=Get WBP_Nested_FavoriteCollection_CoachMark())
  ]}
 EventGraph {nodes [
   n1 = OnInputMethodChanged!() -> (OutputDelegate, then, InputType:ECommonInputType Enum) {is=title K2Node_CustomEvent}
   > n2 = Branch(Condition=IsActivated(self=Get WBP_Nested_FavoriteCollection_CoachMark()))
   > n3 = UpdateNestedTooltipPosition()
   "Set IsUsingM&&K"(
    execute=[@n3.then, @n2.else]
    "IsUsingM&&K":bool=Select!(MouseAndKeyboard:bool=true, Gamepad:bool=false, Touch:bool=false, Count:bool=false, Index:ECommonInputType Enum=@n1.InputType) -> (ReturnValue:bool)
   ) -> (Output_Get:bool)
   > SetParallax()
   On Disover Surface Tile Item VM Set() {is=title K2Node_CustomEvent}
   n4 = Get FortDiscoverSurfaceTileItemVM() -> (FortDiscoverSurfaceTileItemVM:Fort Discover Surface Tile Item VM Object Reference)
   > n5 = Branch(Condition=HasFetchedImages!(self:Fort Discover Link Entry Tile Item VM Object Reference=@n4) -> (ReturnValue:bool))
   n6 = Get FortDiscoverSurfaceTileItemVM()
   "SetText (Text)"(
    execute=[@n7.then, @n5.then]
    self=Get Text_Title() -> (Text_Title:Common Text Object Reference)
    InText=GetDisplayName!(self:Fort Discover Link Entry Tile Item VM Object Reference=@n6) -> (ReturnValue:Text)
   )
   > SetHeaderImageState!(
    execute
    self:WBP Discover Surface Base Object Reference
    LoadingEnum:EActivityImageLoadingState Enum=GetSurfaceBannerLoadingState!(self:Fort Discover Surface Tile Item VM Object Reference=@n6) -> (ReturnValue:EActivityImageLoadingState Enum=Unset)
   ) -> (then)
   > SetHeaderLogo()
   n7 = FetchImages(execute=@n5.else, self:Fort Discover Link Entry Tile Item VM Object Reference=@n4)
   OnButtonFavoriteCollectionClicked!() -> (OutputDelegate, then, Button:Common Button Base Object Reference) {is=title K2Node_CustomEvent}
   n8 = Get FortDiscoverSurfaceTileItemVM()
   > ToggleFavorite(self:Fort Discover Link Entry Tile Item VM Object Reference=@n8)
   > n9 = Branch(Condition=Get bIsFavorite!(self:Fort Discover Link Entry Tile Item VM Object Reference=@n8) -> (bIsFavorite:bool))
   > PlaySound2D(Sound=Get Favorited Sound() -> (Favorited Sound:Sound Base Object Reference))
   PlaySound2D(execute=@n9.else, Sound=Get Unfavorited Sound() -> (Unfavorited Sound:Sound Base Object Reference))
   n10 = OnTooltipButtonClicked() {is=title K2Node_CustomEvent}
   n11 = Get WBP_Nested_FavoriteCollection_CoachMark()
   > DeactivateWidget(self=@n11)
   > RequestRefreshFocus(self:Common Activatable Widget Object Reference)
   Event Construct()
   > Parent: Construct()
   n12 = Create Event()
   n13 = Get CommonInputSubsystem()
   > Bind Event to On Input Method Changed(self=@n13, Delegate=@n12)
   n14 = GetCurrentInputType(self=@n13)
   > OnInputMethodChanged!(execute, self, InputType:ECommonInputType Enum=@n14) -> (then) {is=title K2Node_CallFunction}
   > Bind Event to On Clicked(
    self:Common Button Base Object Reference=Get WBP_BottomBarHack_FavoriteCollection() -> (WBP_BottomBarHack_FavoriteCollection:WBP Bottom Bar Decoy Button Object Reference)
    Delegate=Create Event()
   ) {is=title K2Node_AddDelegate}
   > "Set IsUsingM&&K"("IsUsingM&&K"=Select(MouseAndKeyboard=true, Gamepad=false, Touch=false, Count=false, Index=@n14))
   Event On Activated()
   > Parent: On Activated() {is=title K2Node_CallParentFunction}
   > Fire Event Main View Opened!(execute, View:Widget Object Reference=Self-Reference(), AdditionalInfo:Flow Analytics Additional Info Structure, bIsOverlay:bool=false) -> (then)
   n15 = Get FortDiscoverSurfaceViewModel() -> (FortDiscoverSurfaceViewModel:Discover Surface View Model Object Reference)
   > StartAnalyticsImpressions(self:Discover Surface View Model Object Reference=@n15)
   > Branch(Condition=ShouldShowOnboarding!(self:Fort Sidebar Onboard Tooltip Widget Object Reference=@n11) -> (ReturnValue:bool))
   > ActivateWidget(self=@n11)
   > UpdateNestedTooltipPosition()
   > SetFocus(self=Get OtherAction_Button!(self:Fort Sidebar Onboard Tooltip Widget Object Reference=@n11) -> (OtherAction_Button:Common Button Legacy Object Reference))
   > Bind Event to On Button Other Action Clicked(self:WBP Nested Favorite Collection Coach Mark Object Reference=@n11, Delegate=@n10) {is=title K2Node_AddDelegate}
   Event Destruct()
   > Unbind Event from On Input Method Changed(self=@n13, Delegate=@n12)
   Event On Deactivated()
   > Parent: On Deactivated() {is=title K2Node_CallParentFunction}
   > FireExitedImpressionsFromSurface(self:Discover Surface View Model Object Reference=@n15)
  ]}
 SetHeaderImageState {nodes [
   n1 = SetHeaderImageState!() -> (then, LoadingEnum:EActivityImageLoadingState Enum=Unset) {is=title K2Node_FunctionEntry}
   > Parent: SetHeaderImageState!(execute, LoadingEnum:EActivityImageLoadingState Enum=@n1) -> (then) {is=title K2Node_CallParentFunction}
   > SetVisibility(
    self=Get HeaderLoadingThrobber() -> (HeaderLoadingThrobber:WBP UIKit Throbber Object Reference)
    InVisibility=Select!(
     Option 0:ESlateVisibility Enum=Collapsed
     Option 1:ESlateVisibility Enum=SelfHitTestInvisible
     Index:bool="Equal (Enum)"(A:EActivityImageLoadingState Enum=@n1, B:EActivityImageLoadingState Enum=Loading)
    ) -> (ReturnValue:ESlateVisibility Enum)
   )
   > n2 = GetDynamicMaterial(self=Get HeaderImage() -> (HeaderImage:Image Object Reference))
   > Set HeaderDynamicMaterial(HeaderDynamicMaterial:Material Instance Dynamic Object Reference=@n2) -> (Output_Get:Material Instance Dynamic Object Reference)
   > n3 = Switch on EActivityImageLoadingState!(execute, Selection:EActivityImageLoadingState Enum=@n1) -> (Unset:Exec, Loading:Exec, Loaded:Exec, Error:Exec) {is=title K2Node_SwitchEnum}
   SetScalarParameterValue(
    execute=@n3.Error
    self=Get HeaderDynamicMaterial() -> (HeaderDynamicMaterial:Material Instance Dynamic Object Reference)
    ParameterName=Get ThumbnailOpacity() -> (ThumbnailOpacity:Name=None)
    Value=1
   )
   > SetScalarParameterValue(self=Get HeaderDynamicMaterial(), ParameterName=IsError, Value=1)
   SetHeaderImageTexture(execute=@n3.Loaded)
   > SetScalarParameterValue(self=Get HeaderDynamicMaterial(), ParameterName=Get ThumbnailOpacity() -> (ThumbnailOpacity=None), Value=0)
   > PlayAnimationForward(InAnimation=Get ShowHeaderKeyArt() -> (ShowHeaderKeyArt:Widget Animation Object Reference))
   > SetParallax()
   SetTextureParameterValue(
    execute=@n3.Unset
    self=Get HeaderDynamicMaterial()
    ParameterName=Get ThumbnailParam() -> (ThumbnailParam:Name=None)
    Value=/Game/UI/DiscoveryShared/Art/T_UI_Discovery_DefaultEmpty
   )
   > SetScalarParameterValue(self=Get HeaderDynamicMaterial(), ParameterName=IsError, Value=0)
  ]}
 SetHeaderImageTexture {nodes [
   SetHeaderImageTexture() {is=title K2Node_FunctionEntry}
   > SetTextureParameterValue(
    self=Get HeaderDynamicMaterial()
    ParameterName=Get ThumbnailParam() -> (ThumbnailParam=None)
    Value=GetSurfaceBannerTexture!(self:Fort Discover Surface Tile Item VM Object Reference=Get FortDiscoverSurfaceTileItemVM()) -> (ReturnValue:Texture Object Reference)
   )
  ]}
 SetHeaderLogo {nodes [
   SetHeaderLogo() {is=title K2Node_FunctionEntry}
   n1 = GetSurfacePageLogoLoadingState!(self:Fort Discover Surface Tile Item VM Object Reference=Get FortDiscoverSurfaceTileItemVM()) -> (ReturnValue:EActivityImageLoadingState Enum=Unset)
   > Set Visibility(
    Visibility=Select!(
     Unset:ESlateVisibility Enum=SelfHitTestInvisible
     Loading:ESlateVisibility Enum=SelfHitTestInvisible
     Loaded:ESlateVisibility Enum=Collapsed
     Error:ESlateVisibility Enum=SelfHitTestInvisible
     Index:EActivityImageLoadingState Enum=@n1
    ) -> (ReturnValue)
    self=Get Text_Title()
   )
   > Set Visibility(
    Visibility=Select(Unset=Collapsed, Loading=Collapsed, Loaded=SelfHitTestInvisible, Error=Collapsed, Index=@n1)
    self=Get Logo() -> (Logo:Image Object Reference)
   )
   > n2 = Switch on EActivityImageLoadingState(Selection=@n1)
   n3 = GetDynamicMaterial(execute=@n2.Loaded, self=Get Logo())
   > SetTextureParameterValue(
    self=@n3
    ParameterName=Logo
    Value=GetSurfacePageLogoTexture!(self:Fort Discover Surface Tile Item VM Object Reference=Get FortDiscoverSurfaceTileItemVM()) -> (ReturnValue:Texture Object Reference)
   )
  ]}
 SetHeaderLogoState {nodes [
   n1 = SetHeaderLogoState!() -> (then, LoadingEnum:EActivityImageLoadingState Enum=Unset) {is=title K2Node_FunctionEntry}
   > n2 = Switch on EActivityImageLoadingState(Selection=@n1)
   SetHeaderLogo(execute=@n2.Loaded) {is=title K2Node_CallFunction}
  ]}
 SetIsFavoriteCollection {nodes [
   n1 = SetIsFavoriteCollection!() -> (then, IsFavorite:bool=~) {is=title K2Node_FunctionEntry}
   > n2 = Branch(Condition=Get bEnableFortniteInterests!(self:Fort Runtime Options Object Reference=GetRuntimeOptions()) -> (bEnableFortniteInterests:bool))
   > SetTriggeringInputAction!(
    execute
    self:Common Button Base Object Reference=Get WBP_BottomBarHack_FavoriteCollection()
    InputActionRow:"Data Table Row Handle Structure (by ref)"=Select!(
     Option 0:"Data Table Row Handle Structure (by ref)"=Get FavoriteCollectionInput() -> (FavoriteCollectionInput:Data Table Row Handle Structure)
     Option 1:"Data Table Row Handle Structure (by ref)"=Get UnfavoriteCollectionInput() -> (UnfavoriteCollectionInput:Data Table Row Handle Structure)
     Index:bool=@n1
    ) -> (ReturnValue:"Data Table Row Handle Structure (by ref)")
   ) -> (then)
   SetVisibility(execute=@n2.else, self=Get WBP_BottomBarHack_FavoriteCollection(), InVisibility=Collapsed)
  ]}
 SetParallax {nodes [
   SetParallax() {is=title K2Node_FunctionEntry}
   n1 = "Get IsUsingM&&K"() -> ("IsUsingM&&K":bool)
   n2 = Get bAllowUIParallax!(self:Fort Game User Settings Object Reference=GetFortGameUserSettings!() -> (ReturnValue:Fort Game User Settings Object Reference)) -> (bAllowUIParallax:bool)
   > SetScalarParameterValue(self=Get HeaderDynamicMaterial(), ParameterName=UseParallax, Value="To Float (Boolean)"(InBool=AND Boolean(A=@n1, B=@n2)))
  ]}
 UpdateNestedTooltipPosition {nodes [
   UpdateNestedTooltipPosition() {is=title K2Node_FunctionEntry}
   n1 = Get CommonInputSubsystem()
   > SetPadding!(
    execute
    self:Grid Slot Object Reference=SlotAsGridSlot(Widget=Get SafeZone_Coachmark() -> (SafeZone_Coachmark:Safe Zone Object Reference))
    InPadding_Left:float=0
    InPadding_Top:float=0
    InPadding_Right:float=Select!(MouseAndKeyboard:float=370, Gamepad:float=320, Touch:float=370, Count:float=0, Index:ECommonInputType Enum=GetCurrentInputType(self=@n1)) -> (ReturnValue:float)
    InPadding_Bottom:float=SelectFloat(A=150, B=100, bPickA=IsMobileGame())
   ) -> (then)
  ]}
]
variables [
 Favorited Sound {type=SoundBase, default=/Game/Sounds/Fort_UI_Sounds/FrontEnd/MultiProduct/Common/Buttons/MSS_UI_MP_Button_Click_Favorite}
 "IsUsingM&&K" {type=boolean}
 Unfavorited Sound {type=SoundBase, default=/Game/Sounds/Fort_UI_Sounds/FrontEnd/MultiProduct/Common/Buttons/MSS_UI_MP_Button_Click_Unfavorite}
]
widgets {
 info {
  inheritedWidgetCount=0
  namedSlotCount=2
  parentClass=&/WBP_Discover_Surface_Base.WBP_Discover_Surface_Base_C
  rootWidgetClass=null
  widgetCount=12
 }
 widgets [
  O_HeaderContent {widgetClassPath=&Overlay, widget=O_HeaderContent}
  SafeZone_Title {
   widgetClassPath=&SafeZone
   bIsVariable=true
   parent=O_HeaderContent
   widget=SafeZone_Title
   properties {padBottom=false, padTop=false}
   slot {class=OverlaySlot, properties {horizontalAlignment=HAlign_Fill, padding=<ltrb 0 336 0 0>}}
  }
  FortVisualAttachment_0 {
   widgetClassPath=&FortVisualAttachment
   parent=SafeZone_Title
   widget=FortVisualAttachment_0
   properties {contentAnchor=<0 1>}
   slot {class=SafeZoneSlot}
  }
  Text_Title {
   widgetClassPath=&CommonTextBlock
   bIsVariable=true
   parent=FortVisualAttachment_0
   widget=Text_Title
   properties {
    text="{Surface Title}"
    wrapTextAt=928
    font {
     bForceMonospaced=false
     bMaterialIsStencil=false
     fontMaterial=null
     fontObject=&HeadingNow
     letterSpacing=-40
     monospacedWidth=1
     size=41.25
     skewAmount=0
     typefaceFontName=86 Bold Italic
     outlineSettings {
      bApplyOutlineToDropShadows=false
      bMiteredCorners=false
      bSeparateFillAlpha=false
      outlineMaterial=/FortUILibrary/MaterialInstances/Text/MI_UI_Block_Text_Outline
      outlineSize=4
      outlineColor=<rgba 0 0 0 1>
     }
    }
    shadowOffset=<0 0>
   }
   slot {class=SizeBoxSlot, properties {verticalAlignment=VAlign_Bottom}}
  }
  SafeZone_Logo {
   widgetClassPath=&SafeZone
   bIsVariable=true
   parent=O_HeaderContent
   widget=SafeZone_Logo
   properties {padBottom=false, padTop=false}
   slot {class=OverlaySlot, properties {verticalAlignment=VAlign_Center}}
  }
  FortVisualAttachment_1 {
   widgetClassPath=&FortVisualAttachment
   parent=SafeZone_Logo
   widget=FortVisualAttachment_1
   properties {contentAnchor=<0 0.5>}
   slot {class=SafeZoneSlot}
  }
  Logo {
   widgetClassPath=&Image
   bIsVariable=true
   parent=FortVisualAttachment_1
   widget=Logo
   properties {
    visibility=SelfHitTestInvisible
    brush {
     bIsDynamicallyLoaded=false
     drawAs=Image
     imageType=FullColor
     mirroring=NoMirror
     resourceName=None
     resourceObject=/DiscoveryBrowser/Art/MaterialInstances/MI_UI_Discovery_Logo
     tiling=NoTile
     imageSize=<512 284>
     margin=<ltrb 0 0 0 0>
     outlineSettings {
      bUseBrushTransparency=false
      roundingType=HalfHeightRadius
      width=0
      color {colorUseRule=UseColor_Specified, specifiedColor=<rgba 0 0 0 0>}
      cornerRadii=<0 0 0 0>
     }
     tintColor {colorUseRule=UseColor_Specified, specifiedColor=<rgba 1 1 1 1>}
     uVRegion {bIsValid=false, max=<0 0>, min=<0 0>}
    }
   }
   slot {class=SizeBoxSlot, properties {horizontalAlignment=HAlign_Left, verticalAlignment=VAlign_Bottom}}
  }
  Spacer_HeaderHeight {
   widgetClassPath=&Spacer
   parent=O_HeaderContent
   widget=Spacer_HeaderHeight
   properties {size=<1 420>}
   slot {class=OverlaySlot, properties {horizontalAlignment=HAlign_Fill}}
  }
  GridPanel_26 {widgetClassPath=&GridPanel, widget=GridPanel_26, properties {columnFill [1], rowFill [1]}}
  SafeZone_Coachmark {
   widgetClassPath=&SafeZone
   bIsVariable=true
   parent=GridPanel_26
   widget=SafeZone_Coachmark
   properties {padLeft=false, padTop=false}
   slot {class=GridSlot, properties {horizontalAlignment=HAlign_Right, layer=25, row=1, verticalAlignment=VAlign_Bottom, padding=<ltrb 0 0 350 152>}}
  }
  TooltipAttachment {
   widgetClassPath=&CommonVisualAttachment
   bIsVariable=true
   parent=SafeZone_Coachmark
   widget=TooltipAttachment
   properties {contentAnchor=<1 1>}
   slot {class=SafeZoneSlot, properties {hAlign=HAlign_Right, vAlign=VAlign_Bottom}}
  }
  WBP_Nested_FavoriteCollection_CoachMark {
   widgetClassPath=&/WBP_Nested_FavoriteCollection_CoachMark
   bIsVariable=true
   parent=TooltipAttachment
   widget=WBP_Nested_FavoriteCollection_CoachMark
   properties {visibility=Collapsed}
   slot {class=SizeBoxSlot}
  }
 ]
}
