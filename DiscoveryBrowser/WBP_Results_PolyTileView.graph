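# Polymorphic tile view for Discovery browser results that picks an entry widget class for each tile item view model.
# GetDesiredEntryClassForItem uses a valid class from the optional OverrideItemSelector.
# Otherwise it switches on item type to island, creator, playing-now, IP, community or category tiles, with mini genre tiles for flagged surface tiles.
# PreConstruct turns off proximate entry navigation on the natural-nav cvar's rollback path and enables return-focus-to-selection.
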WidgetBlueprint /DiscoveryBrowser/WBP_Results_PolyTileView
graphs [
 EventGraph {nodes [
   Event PreConstruct()
   > n1 = Branch Natural Nav Rollback Cvar!(In:Exec) -> (Use Natural Nav:Exec, Rollback:Exec) {is=title K2Node_MacroInstance}
   SetEnableProximateEntryNavigation!(execute=@n1.Rollback, self:List View Object Reference=Get TileView() -> (TileView:Fort Tile View Object Reference), bNewValue:bool=false) -> (then)
   > SetReturnFocusToSelection!(execute, self:List View Object Reference=Get TileView(), bEnabled:bool=true) -> (then)
  ]}
 GetDesiredEntryClassForItem {nodes [
   n1 = GetDesiredEntryClassForItem!() -> (then, Item:Object Reference) {is=title K2Node_FunctionEntry}
   > n2 = Cast To FortDiscoverTileItemVM!(execute, Object=@n1) -> (then, CastFailed, AsDiscover Tile Item View Model:Discover Tile Item View Model Object Reference) {is=title K2Node_DynamicCast}
   n3 = Get OverrideItemSelector() -> (OverrideItemSelector:Fort Discover Item Selector Base Object Reference)
   > n4 = Is Valid(InputObject=@n3)
   n5 = GetTileItemClass!(self:Fort Discover Item Selector Base Object Reference=@n3, InTileItemViewModel:Discover Tile Item View Model Object Reference=@n2) -> (ReturnValue:User Widget Class Reference)
   > n6 = Branch(Condition=IsValidClass!(Class:Object Class Reference=@n5) -> (ReturnValue:bool))
   > Return Node!(execute, ReturnValue:User Widget Class Reference=@n5)
   n7 = Switch on EActivityItemType!(
    execute=[@n6.else, @n4.Is Not Valid]
    Selection:EActivityItemType Enum=Byte to Enum EActivityItemType!(Byte:Byte=Get ItemType!(self:Discover Tile Item View Model Object Reference=@n2) -> (ItemType:Byte=0)) -> (ReturnValue:EActivityItemType Enum) {is=title K2Node_CastByteToEnum}
   ) -> (Island:Exec, Creator:Exec, PlayingNow:Exec, SurfaceTile:Exec, IPTile:Exec, PanelTile:Exec, CommunityHighlight:Exec) {is=title K2Node_SwitchEnum}
   Return Node(execute=@n7.CommunityHighlight, ReturnValue=/DiscoveryBrowser/Tiles/Community/WBP_CommunityPost_Tile.WBP_CommunityPost_Tile_C)
   Return Node(execute=@n7.Creator, ReturnValue=/DiscoveryBrowser/Tiles/Creator/WBP_CreatorTile.WBP_CreatorTile_C)
   Return Node(execute=@n7.IPTile, ReturnValue=/DiscoveryBrowser/Tiles/Ip/WBP_Tile_IP_NEW.WBP_Tile_IP_NEW_C)
   > Return Node(ReturnValue=/DiscoveryBrowser/Tiles/Islands/WBP_IslandTile_Full.WBP_IslandTile_Full_C)
   Return Node(execute=@n7.PlayingNow, ReturnValue=/DiscoveryBrowser/Tiles/PlayingNow/WBP_PlayingNow_Tile.WBP_PlayingNow_Tile_C)
   n8 = Branch(
    execute=@n7.SurfaceTile
    Condition=Get bDisplayAsMiniGenreTile!(self:Discover Tile Item View Model Object Reference=@n2) -> (bDisplayAsMiniGenreTile:bool)
   )
   > Return Node(ReturnValue=/DiscoveryBrowser/Tiles/Genre/WBP_Tile_Genre.WBP_Tile_Genre_C)
   Return Node(execute=@n8.else, ReturnValue=/DiscoveryBrowser/Tiles/Categories/WBP_Tile_Category.WBP_Tile_Category_C)
  ]}
]
variables [OverrideItemSelector {type=FortDiscoverItemSelectorBase, default=null}]
widgets {
 info {
  inheritedWidgetCount=1
  parentClass=/Script/FortniteUI.FortPolymorphicTileView
  rootWidgetClass=/Script/FortniteUI.FortTileView
  widgetCount=1
 }
 widgets [TileView {
   widgetClassPath=&FortTileView
   bInherited=true
   bIsVariable=true
   widget=TileView
   properties {
    bAllowDragging=false
    bEnableFixedLineOffset=true
    bEnableProximateEntryNavigation=true
    bEntrySizeIncludesEntrySpacing=false
    clipping=Inherit
    entryHeight=248
    entryWidgetClass=&WBP_IslandTile_Full_C
    entryWidth=248
    horizontalEntrySpacing=24
    numDesignerPreviewEntries=20
    tileAlignment=LeftAligned
    verticalEntrySpacing=24
    scrollBarStyle {
     thickness=16
     draggedThumbImage {
      bIsDynamicallyLoaded=false
      drawAs=NoDrawType
      imageType=FullColor
      mirroring=NoMirror
      resourceName=../../../Engine/Content/Slate/Common/Scrollbar_Thumb.png
      resourceObject=null
      tiling=NoTile
      imageSize=<0 0>
      margin=<ltrb 0.25 0.25 0.25 0.25>
      outlineSettings {
       bUseBrushTransparency=false
       roundingType=HalfHeightRadius
       width=0
       color {colorUseRule=UseColor_Specified, specifiedColor=<rgba 0 0 0 0>}
       cornerRadii=<0 0 0 0>
      }
      tintColor {colorUseRule=UseColor_Specified, specifiedColor=<rgba 1 1 1 1>}
      uVRegion {bIsValid=false, max=<0 0>, min=<0 0>}
     }
     horizontalBackgroundImage {
      bIsDynamicallyLoaded=false
      drawAs=NoDrawType
      imageType=NoImage
      mirroring=NoMirror
      resourceName=None
      resourceObject=null
      tiling=NoTile
      imageSize=<0 0>
      margin=<ltrb 0 0 0 0>
      outlineSettings {
       bUseBrushTransparency=false
       roundingType=HalfHeightRadius
       width=0
       color {colorUseRule=UseColor_Specified, specifiedColor=<rgba 0 0 0 0>}
       cornerRadii=<0 0 0 0>
      }
      tintColor {colorUseRule=UseColor_Specified, specifiedColor=<rgba 1 1 1 1>}
      uVRegion {bIsValid=false, max=<0 0>, min=<0 0>}
     }
     horizontalBottomSlotImage {
      bIsDynamicallyLoaded=false
      drawAs=NoDrawType
      imageType=FullColor
      mirroring=NoMirror
      resourceName=../../../Engine/Content/Slate/Common/Scrollbar_Background_Horizontal.png
      resourceObject=null
      tiling=NoTile
      imageSize=<8 8>
      margin=<ltrb 0 0 0 0>
      outlineSettings {
       bUseBrushTransparency=false
       roundingType=HalfHeightRadius
       width=0
       color {colorUseRule=UseColor_Specified, specifiedColor=<rgba 0 0 0 0>}
       cornerRadii=<0 0 0 0>
      }
      tintColor {colorUseRule=UseColor_Specified, specifiedColor=<rgba 1 1 1 1>}
      uVRegion {bIsValid=false, max=<0 0>, min=<0 0>}
     }
     horizontalTopSlotImage {
      bIsDynamicallyLoaded=false
      drawAs=NoDrawType
      imageType=FullColor
      mirroring=NoMirror
      resourceName=../../../Engine/Content/Slate/Common/Scrollbar_Background_Horizontal.png
      resourceObject=null
      tiling=NoTile
      imageSize=<8 8>
      margin=<ltrb 0 0 0 0>
      outlineSettings {
       bUseBrushTransparency=false
       roundingType=HalfHeightRadius
       width=0
       color {colorUseRule=UseColor_Specified, specifiedColor=<rgba 0 0 0 0>}
       cornerRadii=<0 0 0 0>
      }
      tintColor {colorUseRule=UseColor_Specified, specifiedColor=<rgba 1 1 1 1>}
      uVRegion {bIsValid=false, max=<0 0>, min=<0 0>}
     }
     hoveredThumbImage {
      bIsDynamicallyLoaded=false
      drawAs=NoDrawType
      imageType=FullColor
      mirroring=NoMirror
      resourceName=../../../Engine/Content/Slate/Common/Scrollbar_Thumb.png
      resourceObject=null
      tiling=NoTile
      imageSize=<0 0>
      margin=<ltrb 0.25 0.25 0.25 0.25>
      outlineSettings {
       bUseBrushTransparency=false
       roundingType=HalfHeightRadius
       width=0
       color {colorUseRule=UseColor_Specified, specifiedColor=<rgba 0 0 0 0>}
       cornerRadii=<0 0 0 0>
      }
      tintColor {colorUseRule=UseColor_Specified, specifiedColor=<rgba 1 1 1 1>}
      uVRegion {bIsValid=false, max=<0 0>, min=<0 0>}
     }
     normalThumbImage {
      bIsDynamicallyLoaded=false
      drawAs=NoDrawType
      imageType=FullColor
      mirroring=NoMirror
      resourceName=../../../Engine/Content/Slate/Common/Scrollbar_Thumb.png
      resourceObject=null
      tiling=NoTile
      imageSize=<0 0>
      margin=<ltrb 0.25 0.25 0.25 0.25>
      outlineSettings {
       bUseBrushTransparency=false
       roundingType=HalfHeightRadius
       width=0
       color {colorUseRule=UseColor_Specified, specifiedColor=<rgba 0 0 0 0>}
       cornerRadii=<0 0 0 0>
      }
      tintColor {colorUseRule=UseColor_Specified, specifiedColor=<rgba 1 1 1 1>}
      uVRegion {bIsValid=false, max=<0 0>, min=<0 0>}
     }
     verticalBackgroundImage {
      bIsDynamicallyLoaded=false
      drawAs=NoDrawType
      imageType=NoImage
      mirroring=NoMirror
      resourceName=None
      resourceObject=null
      tiling=NoTile
      imageSize=<0 0>
      margin=<ltrb 0 0 0 0>
      outlineSettings {
       bUseBrushTransparency=false
       roundingType=HalfHeightRadius
       width=0
       color {colorUseRule=UseColor_Specified, specifiedColor=<rgba 0 0 0 0>}
       cornerRadii=<0 0 0 0>
      }
      tintColor {colorUseRule=UseColor_Specified, specifiedColor=<rgba 1 1 1 1>}
      uVRegion {bIsValid=false, max=<0 0>, min=<0 0>}
     }
     verticalBottomSlotImage {
      bIsDynamicallyLoaded=false
      drawAs=NoDrawType
      imageType=FullColor
      mirroring=NoMirror
      resourceName=../../../Engine/Content/Slate/Common/Scrollbar_Background_Vertical.png
      resourceObject=null
      tiling=NoTile
      imageSize=<8 8>
      margin=<ltrb 0 0 0 0>
      outlineSettings {
       bUseBrushTransparency=false
       roundingType=HalfHeightRadius
       width=0
       color {colorUseRule=UseColor_Specified, specifiedColor=<rgba 0 0 0 0>}
       cornerRadii=<0 0 0 0>
      }
      tintColor {colorUseRule=UseColor_Specified, specifiedColor=<rgba 1 1 1 1>}
      uVRegion {bIsValid=false, max=<0 0>, min=<0 0>}
     }
     verticalTopSlotImage {
      bIsDynamicallyLoaded=false
      drawAs=NoDrawType
      imageType=FullColor
      mirroring=NoMirror
      resourceName=../../../Engine/Content/Slate/Common/Scrollbar_Background_Vertical.png
      resourceObject=null
      tiling=NoTile
      imageSize=<8 8>
      margin=<ltrb 0 0 0 0>
      outlineSettings {
       bUseBrushTransparency=false
       roundingType=HalfHeightRadius
       width=0
       color {colorUseRule=UseColor_Specified, specifiedColor=<rgba 0 0 0 0>}
       cornerRadii=<0 0 0 0>
      }
      tintColor {colorUseRule=UseColor_Specified, specifiedColor=<rgba 1 1 1 1>}
      uVRegion {bIsValid=false, max=<0 0>, min=<0 0>}
     }
    }
   }
  }]
}
