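# Activatable "View All" page for Discovery, showing one surface panel's items in a tile grid with a no-results fallback.
# It picks a provider from the surface view model by panel name, fills the tile view, sizes entries from tile widgets and requests further pages as items initialize.
# Scroll offset fades the header edge and swaps Back and Back to Top buttons.
# Background materials follow the parent tile's link category.
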WidgetBlueprint /DiscoveryBrowser/WBP_ViewAllPage
animations [Anim_FadeInOut {
  binding_count=1
  length_seconds=0.3
  object=&self:Anim_FadeInOut
  sequence {bindings [Image_HeaderSoftEdge {
     id=#1
     possessed_class=Image
     tracks [
      {
       type=MovieSceneFloatTrack
       display_name=Render Opacity
       sections [{type=MovieSceneFloatSection, end_frame=6, start_frame=0, channels [None {type=MovieSceneScriptingFloatChannel, keys [[0,0],[6,1]]}]}]
      }
      {
       type=MovieSceneByteTrack
       display_name=Visibility
       sections [{type=MovieSceneByteSection, end_frame=6, start_frame=0, channels [None {type=MovieSceneScriptingByteChannel, keys [[0,1],[1,4],[6,3]]}]}]
      }
     ]
    }]}
 }]
graphs [
 BP_GetDesiredFocusTarget {nodes [
   Get Desired Focus Target()
   > Return Node!(
    execute
    ReturnValue:Widget Object Reference=Get TileView!(self:Fort Polymorphic Tile View Object Reference=Get WBP_Results_PolyTileView() -> (WBP_Results_PolyTileView:WBP Results Poly Tile View Object Reference)) -> (TileView:Fort Tile View Object Reference)
   )
  ]}
 EventGraph {nodes [
   "On Clicked (WBP_BottomBarHack_BackToTop)"!() -> (OutputDelegate, then, Button:Common Button Base Object Reference) {is=title K2Node_ComponentBoundEvent}
   n1 = Get TileView(self=Get WBP_Results_PolyTileView())
   > SetSelectedIndex!(execute, self:List View Object Reference=@n1, Index:int=0) -> (then)
   > Navigate To Item!(
    execute
    self:List View Object Reference=@n1
    Item:Object Reference=Get Selected Item!(self:List View Object Reference=@n1) -> (ReturnValue:Object Reference) {is=title K2Node_CallFunction}
   ) -> (then)
   "On Clicked (WBP_BottomBarHack_Back)"!() -> (OutputDelegate, then, Button:Common Button Base Object Reference) {is=title K2Node_ComponentBoundEvent}
   > DeactivateWidget()
   HandleInputMethodChanged!() -> (OutputDelegate, then, bNewInputType:ECommonInputType Enum) {is=title K2Node_CustomEvent}
   n2 = Get WBP_BottomBarHack_Back() -> (WBP_BottomBarHack_Back:WBP Bottom Bar Decoy Button Object Reference)
   > SetBottomButtonVisibility!(
    execute
    self:WBP Bottom Bar Decoy Button Object Reference=Get WBP_BottomBarHack_BackToTop() -> (WBP_BottomBarHack_BackToTop:WBP Bottom Bar Decoy Button Object Reference)
    ShouldButtonBeVisible:bool=Select(Option 0:bool=true, Option 1:bool=false, Index=IsVisible(self=@n2)) -> (ReturnValue:bool)
   ) -> (then)
   n3 = OnTileViewEntryGenerated!() -> (OutputDelegate, then, Widget:User Widget Object Reference) {is=title K2Node_CustomEvent}
   > n4 = Sequence()
   > n5 = Cast To WBP_IslandTile_Minimal!(execute, Object=@n3.Widget) -> (then, CastFailed, AsWBP Island Tile Minimal:WBP Island Tile Minimal Object Reference) {is=title K2Node_DynamicCast}
   > Update Tile View Layout!(
    execute
    self
    TileBaseWidget:WBP Discover Tile Base Object Reference=Get WBP_Discover_Tile_Base!(self:WBP Island Tile Minimal Object Reference=@n5) -> (WBP_Discover_Tile_Base:WBP Discover Tile Base Object Reference)
   ) -> (then)
   n6 = Cast To WBP_Discover_Tile_Button!(execute=@n5.CastFailed, Object=@n3.Widget) -> (then, CastFailed, AsWBP Discover Tile Button:WBP Discover Tile Button Object Reference) {is=title K2Node_DynamicCast}
   > Update Tile View Layout(TileBaseWidget=Get WBP_Discover_Tile_Base(self:WBP Discover Tile Button Object Reference=@n6))
   SetNestedButtonGroupSelectionVM!(
    execute:Exec=@n4.then_1
    self:Object Reference=@n3.Widget
    InVM:Fort Nested Button Group Selection VM Object Reference=Get ButtonGroupSelectionVM() -> (ButtonGroupSelectionVM:Fort Nested Button Group Selection VM Object Reference)
   ) -> (then:Exec) {is=title K2Node_Message}
   n7 = OnItemInitialized!() -> (OutputDelegate, then, Item:Object Reference, Widget:User Widget Object Reference) {is=title K2Node_CustomEvent}
   n8 = Get WBP_Results_PolyTileView()
   > CheckAndFetchNextPage!(
    execute
    self:Discover Provider View Model Object Reference=Get FortDiscoverProviderViewModel() -> (FortDiscoverProviderViewModel:Discover Provider View Model Object Reference)
    SelectionIndex:int=GetIndexForItem!(self:List View Object Reference=Get TileView(self=@n8), Item:Object Reference=@n7.Item) -> (ReturnValue:int=0)
    bForceFetchNextPage:bool=false
   ) -> (then)
   n9 = OnTileViewEntryInitialized!() -> (OutputDelegate, then, Item:Object Reference, Widget:User Widget Object Reference) {is=title K2Node_CustomEvent}
   > n10 = Cast To WBP_IslandTile_Minimal(Object=@n9.Widget)
   n11 = Get WBP_Discover_Tile_Base(self:WBP Island Tile Minimal Object Reference=@n10)
   n12 = GetIndexForItem(self=Get TileView(self=Get WBP_Results_PolyTileView()), Item=@n9.Item) -> (ReturnValue=0)
   > Update Base Tile Properties!(execute, self, ItemIndex:int=@n12, TileBaseWidget:WBP Discover Tile Base Object Reference=@n11) -> (then)
   > SetIsPagination!(execute, self:WBP Discover Tile Base Object Reference=@n11, bInIsPagination:bool=true) -> (then)
   n13 = Cast To WBP_CreatorTile!(execute=@n10.CastFailed, Object=@n9.Widget) -> (then, CastFailed, AsWBP Creator Tile:WBP Creator Tile Object Reference) {is=title K2Node_DynamicCast}
   > Update Base Tile Properties(ItemIndex=@n12, TileBaseWidget=Get WBP_Discover_Tile_Base(self:WBP Discover Tile Button Object Reference=@n13))
   n14 = OnTileViewScrolled!() -> (OutputDelegate, then, ItemOffset:float=~, DistanceRemaining:float=~) {is=title K2Node_CustomEvent}
   > UpdateHeaderSoftEdgeVisibility!(execute, self, ScrollOffset:double=@n14.ItemOffset) -> (then)
   n15 = "float > float"(A=@n14.ItemOffset)
   > SetBottomButtonVisibility(self=Get WBP_BottomBarHack_BackToTop(), ShouldButtonBeVisible=@n15)
   > SetBottomButtonVisibility(self=Get WBP_BottomBarHack_Back(), ShouldButtonBeVisible=NOT Boolean(A=@n15))
   Event Construct()
   n16 = Create Event()
   n17 = Get CommonInputSubsystem()
   > Bind Event to On Input Method Changed(self=@n17, Delegate=@n16)
   > HandleInputMethodChanged!(execute, self, bNewInputType:ECommonInputType Enum=GetCurrentInputType(self=@n17)) -> (then) {is=title K2Node_CallFunction}
   > SetBottomButtonVisibility(self=Get WBP_BottomBarHack_BackToTop(), ShouldButtonBeVisible=false)
   n18 = Get TileView(self=Get WBP_Results_PolyTileView())
   > Bind Event to On Entry Generated(self:List View Base Object Reference=@n18, Delegate=Create Event()) {is=title K2Node_AddDelegate}
   > Bind Event to On Entry Initialized(self:List View Object Reference=@n18, Delegate=Create Event()) {is=title K2Node_AddDelegate}
   > Bind Event to On List View Scrolled(self:List View Object Reference=@n18, Delegate=Create Event()) {is=title K2Node_AddDelegate}
   > n19 = Branch Natural Nav Rollback Cvar!(In:Exec) -> (Use Natural Nav:Exec, Rollback:Exec) {is=title K2Node_MacroInstance}
   SetDesiredFocusWidget!(execute=@n19.Rollback, self:User Widget Object Reference, Widget:Widget Object Reference=Get WBP_Results_PolyTileView()) -> (then, ReturnValue:bool)
   Event OnInitialized()
   > n20 = Construct Object from Class(Class=/Script/FortniteUI.FortNestedButtonGroupSelectionVM) -> (ReturnValue:Fort Nested Button Group Selection VM Object Reference)
   > Set ButtonGroupSelectionVM(ButtonGroupSelectionVM:Fort Nested Button Group Selection VM Object Reference=@n20) -> (Output_Get:Fort Nested Button Group Selection VM Object Reference)
   Event PreConstruct()
   > SetScrollbarVisibility!(
    execute
    self:List View Base Object Reference=Get TileView(self=Get WBP_Results_PolyTileView())
    InVisibility:ESlateVisibility Enum=Collapsed
   ) -> (then)
   Event Destruct()
   > Unbind Event from On Input Method Changed(self=@n17, Delegate=@n16)
   > Unbind all Events from On Entry Generated!(execute:Exec, self:List View Base Object Reference=@n18) -> (then:Exec) {is=title K2Node_ClearDelegate}
   > Unbind all Events from On Entry Initialized!(execute:Exec, self:List View Object Reference=@n18) -> (then:Exec) {is=title K2Node_ClearDelegate}
   > Unbind all Events from On List View Scrolled!(execute:Exec, self:List View Object Reference=@n18) -> (then:Exec) {is=title K2Node_ClearDelegate}
   Event On Activated()
   n21 = Get FortDiscoverProviderViewModel()
   > StartVisibleImpressions(self:Discover Provider View Model Object Reference=@n21)
   > Set bIsViewAllScreen!(execute, bIsViewAllScreen:bool=true, self:Discover Provider View Model Object Reference=@n21) -> (then, Output_Get:bool)
   Event On Deactivated()
   > StopVisibleImpression(self:Discover Provider View Model Object Reference=@n21)
   > Set bIsViewAllScreen(bIsViewAllScreen=false, self=@n21)
   n22 = Get TileView(self=Get WBP_Results_PolyTileView())
   n23 = Branch(Condition="integer > integer"(A=GetNumItems!(self:List View Object Reference=@n22) -> (ReturnValue:int=0)))
   n24 = Get WidgetSwitcher_Results() -> (WidgetSwitcher_Results:Common Activatable Widget Switcher Object Reference)
   > SetActiveWidgetIndex(self=@n24, Index=1)
   SetActiveWidgetIndex(execute=@n23.else, self=@n24, Index=0)
   K2Node_Event_3 = BP_OnTileViewUpdated_1() {is=title K2Node_CustomEvent}
  ]}
 Fill Out Tile View Data {nodes [
   Fill Out Tile View Data() {is=title K2Node_FunctionEntry}
   n1 = Get TileView(self=Get WBP_Results_PolyTileView())
   > Set LastSelectedIndex(LastSelectedIndex:int=GetIndexForItem(self=@n1, Item=Get Selected Item(self=@n1)) -> (ReturnValue=0)) -> (Output_Get:int=0)
   > ClearListItems(self=Get TileView(self=Get WBP_Results_PolyTileView()))
   n2 = Get FortDiscoverProviderViewModel()
   > Is Valid(InputObject=@n2)
   n3 = Get TileView(self=Get WBP_Results_PolyTileView())
   > Set List Items!(
    execute
    self:List View Object Reference=@n3
    InListItems:Array of Object References=Get Items!(self:Discover Provider View Model Object Reference=@n2) -> (Items:Array of Discover Tile Item View Model Object References)
   ) -> (then) {is=title K2Node_CallFunction}
   n4 = Get LastSelectedIndex() -> (LastSelectedIndex:int=0)
   > SetSelectedIndex(self=@n3, Index=SelectInt(A=@n4, B=0, bPickA="integer >= integer"(A=@n4)))
   Get FortDiscoverSurfaceViewModel() -> (FortDiscoverSurfaceViewModel:Discover Surface View Model Object Reference)
  ]}
 On Querying More Pages {nodes [
   n1 = On Querying More Pages!() -> (then, LoadingState:EDiscoverLoadingState Enum=Unset) {is=title K2Node_FunctionEntry}
   > SetVisibility(
    self=Get WBP_LoadingMorePages() -> (WBP_LoadingMorePages:WBP Loading More Pages Object Reference)
    InVisibility=Select(
     Option 0:ESlateVisibility Enum=Collapsed
     Option 1:ESlateVisibility Enum=HitTestInvisible
     Index="Equal (Enum)"(A:EDiscoverLoadingState Enum=@n1, B:EDiscoverLoadingState Enum=Loading)
    ) -> (ReturnValue:ESlateVisibility Enum)
   )
  ]}
 On Update Providers {nodes [
   n1 = On Update Providers!() -> (then, Tiles:Array of Discover Tile Item View Model Object References) {is=title K2Node_FunctionEntry}
   > Fill Out Tile View Data() {is=title K2Node_CallFunction}
   Reroute Node(InputPin:Wildcard) -> (OutputPin:Wildcard)
   Reroute Node(InputPin:Array of Discover Tile Item View Model Object References=@n1) -> (OutputPin:Array of Discover Tile Item View Model Object References)
  ]}
 SetInitialSurfacePanelIfUninitialized {nodes [
   n1 = SetInitialSurfacePanelIfUninitialized!() -> (then, Providers:Array of Discover Provider View Model Object References) {is=title K2Node_FunctionEntry}
   > Is Valid(InputObject=Get FortDiscoverProviderViewModel())
   n2 = Get PanelName() -> (PanelName:Name=None)
   n3 = SelectInt(
    A=0
    B=GetProviderIndexFromPanelName!(self:Discover Surface View Model Object Reference=Get FortDiscoverSurfaceViewModel(), PanelName:Name=@n2) -> (ReturnValue:int=0)
    bPickA="Equal (Name)"(A=@n2)
   )
   > Branch(Condition=Is Valid Index!(TargetArray:Array of Discover Provider View Model Object References=@n1, IndexToTest:int=@n3) -> (ReturnValue:bool))
   > Set FortDiscoverProviderViewModel(FortDiscoverProviderViewModel:Discover Provider View Model Object Reference="Get (a copy)"(Array:Array of Discover Provider View Model Object References=@n1, Dimension 1=@n3) -> (Output:Discover Provider View Model Object Reference)) -> (Output_Get:Discover Provider View Model Object Reference)
   > On Querying More Pages!(execute, self, LoadingState:EDiscoverLoadingState Enum=Loaded) -> (then) {is=title K2Node_CallFunction}
  ]}
 Setup with Panel Name {nodes [
   n1 = Setup with Panel Name!() -> (then, Display Name:Text) {is=title K2Node_FunctionEntry}
   > Branch(Condition=IsValid(Object=Get FortDiscoverProviderViewModel()))
   > SetHeaderText!(execute, self:WBP Row Header Object Reference=Get WBP_Row_Header() -> (WBP_Row_Header:WBP Row Header Object Reference), HeaderText:Text=@n1) -> (then)
   > Bind Event to On Entry Initialized(self=Get TileView(self=Get WBP_Results_PolyTileView()), Delegate=Create Event())
   > StartVisibleImpressions(self=Get FortDiscoverProviderViewModel())
  ]}
 SetupBackground {nodes [
   n1 = SetupBackground!() -> (then, ParentSurfaceTileItemVM:Fort Discover Surface Tile Item VM Object Reference) {is=title K2Node_FunctionEntry}
   > Is Valid(InputObject=@n1)
   n2 = GetLinkCategory!(self:Fort Discover Link Entry Tile Item VM Object Reference=@n1) -> (ReturnValue:EFortActivityLinkCategory Enum=None)
   > SetBrushFromMaterial(
    self=Get Image_Background() -> (Image_Background:Image Object Reference)
    Material=GetBackgroundMaterial!(LinkCategory:EFortActivityLinkCategory Enum=@n2) -> (BackgroundMaterial:Material Interface Object Reference)
   )
   > SetBrushFromMaterial(
    self=Get Image_HeaderSoftEdge() -> (Image_HeaderSoftEdge:Image Object Reference)
    Material=GetTopEdgeMaterial!(LinkCategory:EFortActivityLinkCategory Enum=@n2) -> (TopEdgeMaterial:Material Interface Object Reference)
   )
   > SetBrushFromMaterial(
    self=Get BgFooter() -> (BgFooter:Image Object Reference)
    Material=GetFooterMaterial!(LinkCategory:EFortActivityLinkCategory Enum=@n2) -> (FooterMaterial:Material Interface Object Reference)
   )
  ]}
 SetupSurface {nodes [
   n1 = SetupSurface!() -> (then, Surface Name:Name, Panel Name:Name) {is=title K2Node_FunctionEntry}
   > Set PanelName(PanelName:Name=@n1.Panel Name) -> (Output_Get:Name=None)
   > ClearListItems(self=Get TileView(self=Get WBP_Results_PolyTileView()))
   n2 = Get WBP_Row_Header()
   > SetHeaderText(self=@n2)
   > SetSubHeaderText!(execute, self:WBP Row Header Object Reference=@n2, SubHeaderText:Text) -> (then)
   > SetSurfaceName!(execute, self:Discover Surface View Model Object Reference=Get FortDiscoverSurfaceViewModel(), InSurfaceName:Name=@n1) -> (then)
   n3 = Get FortDiscoverSurfaceViewModel()
   n4 = Get PanelName() -> (PanelName=None)
   n5 = SelectInt(A=0, B=GetProviderIndexFromPanelName(self=@n3, PanelName=@n4) -> (ReturnValue=0), bPickA="Equal (Name)"(A=@n4))
   > n6 = Branch(Condition=Is Valid Index(
     TargetArray=Get Providers!(self:Discover Surface View Model Object Reference=@n3) -> (Providers:Array of Discover Provider View Model Object References)
     IndexToTest=@n5
    ))
   n7 = "Get (a copy)"(Array=Get Providers(self=Get FortDiscoverSurfaceViewModel()), Dimension 1=@n5)
   > Is Valid(InputObject=@n7)
   > Set FortDiscoverProviderViewModel(FortDiscoverProviderViewModel=@n7)
   On Querying More Pages(execute=@n6.else, LoadingState=Loading)
  ]}
 Update Base Tile Properties {nodes [
   n1 = Update Base Tile Properties!() -> (then, ItemIndex:int, TileBaseWidget:WBP Discover Tile Base Object Reference) {is=title K2Node_FunctionEntry}
   > Set TileItemIndex!(execute, TileItemIndex:int=@n1, self:WBP Discover Tile Base Object Reference=@n1.TileBaseWidget) -> (then, Output_Get:int=0)
  ]}
 Update Tile View Layout {nodes [
   n1 = Update Tile View Layout!() -> (then, TileBaseWidget:WBP Discover Tile Base Object Reference) {is=title K2Node_FunctionEntry}
   n2 = BreakVector2D(InVec=Get Dynamic Tile Size!(self:WBP Discover Tile Base Object Reference=@n1) -> (DynamicTileSize:Vector 2D Structure) {is=title K2Node_CallFunction})
   n3 = Get TileView(self=Get WBP_Results_PolyTileView())
   > SetEntryWidth!(execute, self:Tile View Object Reference=@n3, NewWidth:float=@n2) -> (then)
   > SetEntryHeight!(execute, self:Tile View Object Reference=@n3, NewHeight:float=@n2.Y) -> (then)
  ]}
 UpdateHeaderSoftEdgeVisibility {nodes [
   n1 = UpdateHeaderSoftEdgeVisibility!() -> (then, ScrollOffset:double=~) {is=title K2Node_FunctionEntry}
   n2 = Get Image_HeaderSoftEdge()
   n3 = "float > float"(A=@n1)
   > Branch(Condition=XOR Boolean!(A:bool=IsVisible(self=@n2), B:bool=@n3) -> (ReturnValue:bool))
   > n4 = Branch(Condition=@n3)
   n5 = Get Anim_FadeInOut() -> (Anim_FadeInOut:Widget Animation Object Reference)
   > PlayAnimationForward(InAnimation=@n5)
   PlayAnimationReverse(execute=@n4.else, InAnimation=@n5)
  ]}
]
variables [
 ButtonGroupSelectionVM {type=FortNestedButtonGroupSelectionVM, default=null}
 LastSelectedIndex {type=integer}
 OverrideItemSelector {type=FortDiscoverItemSelectorBase, default=null}
 PanelName {type=string, default=None}
 SurfaceName {type=string, default=None}
]
widgets {
 info {inheritedWidgetCount=0, parentClass=&CommonActivatableWidget, rootWidgetClass=&Overlay, widgetCount=20}
 widgets [
  Overlay_0 {widgetClassPath=&Overlay, widget=Overlay_0}
  Image_Background {
   widgetClassPath=&Image
   bIsVariable=true
   parent=Overlay_0
   widget=Image_Background
   properties {
    visibility=SelfHitTestInvisible
    brush {
     bIsDynamicallyLoaded=false
     drawAs=Image
     imageType=FullColor
     mirroring=NoMirror
     resourceName=None
     resourceObject=/FortUILibrary/MaterialInstances/Background/Nebulous/MI_UI_Background_Nebulous
     tiling=NoTile
     imageSize=<32 32>
     margin=<ltrb 0 0 0 0>
     outlineSettings {
      bUseBrushTransparency=false
      roundingType=HalfHeightRadius
      width=0
      color {colorUseRule=UseColor_Specified, specifiedColor=<rgba 0 0 0 0>}
      cornerRadii=<0 0 0 0>
     }
     tintColor {colorUseRule=UseColor_Specified, specifiedColor=<rgba 1 1 1 1>}
     uVRegion {bIsValid=false, max=<0 0>, min=<0 0>}
    }
   }
   slot {class=OverlaySlot, properties {horizontalAlignment=HAlign_Fill, verticalAlignment=VAlign_Fill}}
  }
  WBP_CaptureForPostBufferUpdate {
   widgetClassPath=&WBP_CaptureForPostBufferUpdate
   bIsVariable=true
   parent=Overlay_0
   slot {class=OverlaySlot, properties {horizontalAlignment=HAlign_Fill, verticalAlignment=VAlign_Fill}}
  }
  SafeZone_PageContent {
   widgetClassPath=&SafeZone
   bIsVariable=true
   parent=Overlay_0
   widget=SafeZone_PageContent
   slot {class=OverlaySlot, properties {horizontalAlignment=HAlign_Fill, verticalAlignment=VAlign_Fill}}
  }
  VerticalBox_1 {
   widgetClassPath=&VerticalBox
   parent=SafeZone_PageContent
   widget=VerticalBox_1
   slot {class=SafeZoneSlot, properties {padding=<ltrb 0 128 0 0>}}
  }
  WBP_Row_Header {
   widgetClassPath=/DiscoveryBrowser/Rows/WBP_Row_Header
   bIsVariable=true
   parent=VerticalBox_1
   widget=WBP_Row_Header
   slot {class=VerticalBoxSlot}
  }
  WidgetSwitcher_Results {
   widgetClassPath=&CommonActivatableWidgetSwitcher
   bIsVariable=true
   parent=VerticalBox_1
   widget=WidgetSwitcher_Results
   properties {activeWidgetIndex=1, clipping=ClipToBounds}
   slot {class=VerticalBoxSlot, properties {size {sizeRule=Fill, value=1}}}
  }
  WBP_Discovery_NoResultsMessage {
   widgetClassPath=&/WBP_Discovery_MessageNoResults.WBP_Discovery_NoResultsMessage
   bIsVariable=true
   parent=WidgetSwitcher_Results
   widget=WBP_Discovery_NoResultsMessage
   properties {flavorMessageBase=Check back later for new content., flavorTitleBase=Currently Unavailable}
   slot {class=WidgetSwitcherSlot, properties {verticalAlignment=VAlign_Center, padding=<ltrb 0 0 0 80>}}
  }
  Overlay_1 {
   widgetClassPath=&Overlay
   parent=WidgetSwitcher_Results
   widget=Overlay_1
   slot {class=WidgetSwitcherSlot, properties {padding=<ltrb 70 0 70 100>}}
  }
  SizeBox {
   widgetClassPath=&SizeBox
   parent=Overlay_1
   widget=SizeBox
   properties {bOverride_MinDesiredWidth=true, minDesiredWidth=1760}
   slot {class=OverlaySlot, properties {horizontalAlignment=HAlign_Fill, verticalAlignment=VAlign_Fill}}
  }
  WBP_Results_PolyTileView {
   widgetClassPath=&/WBP_Results_PolyTileView
   bIsVariable=true
   parent=SizeBox
   widget=WBP_Results_PolyTileView
   slot {class=SizeBoxSlot, properties {padding=<ltrb 0 24 0 0>}}
  }
  Image_HeaderSoftEdge {
   widgetClassPath=&Image
   bIsVariable=true
   parent=Overlay_1
   widget=Image_HeaderSoftEdge
   properties {
    visibility=Collapsed
    brush {
     bIsDynamicallyLoaded=false
     drawAs=Image
     imageType=FullColor
     mirroring=NoMirror
     resourceName=None
     resourceObject=/DiscoveryBrowser/TabScreens/DiscoverHome/Materials/MI_UI_Discovery_NebulousBackgroundWithTopMask
     tiling=NoTile
     imageSize=<32 50>
     margin=<ltrb 0 0 0 0>
     outlineSettings {
      bUseBrushTransparency=false
      roundingType=HalfHeightRadius
      width=0
      color {colorUseRule=UseColor_Specified, specifiedColor=<rgba 0 0 0 0>}
      cornerRadii=<0 0 0 0>
     }
     tintColor {colorUseRule=UseColor_Specified, specifiedColor=<rgba 1 1 1 1>}
     uVRegion {bIsValid=false, max=<0 0>, min=<0 0>}
    }
   }
   slot {class=OverlaySlot, properties {horizontalAlignment=HAlign_Fill}}
  }
  Overlay_LoadingAndGradientBacker {
   widgetClassPath=&Overlay
   parent=Overlay_0
   widget=Overlay_LoadingAndGradientBacker
   slot {class=OverlaySlot, properties {horizontalAlignment=HAlign_Fill, verticalAlignment=VAlign_Bottom}}
  }
  BgFooter {
   widgetClassPath=&Image
   bIsVariable=true
   parent=Overlay_LoadingAndGradientBacker
   widget=BgFooter
   properties {
    visibility=SelfHitTestInvisible
    brush {
     bIsDynamicallyLoaded=false
     drawAs=Image
     imageType=FullColor
     mirroring=NoMirror
     resourceName=None
     resourceObject=/DiscoveryBrowser/TabScreens/DiscoverHome/Materials/MI_UI_Discovery_NebulousBackgroundWithBottomMask
     tiling=NoTile
     imageSize=<32 32>
     margin=<ltrb 0 0 0 0>
     outlineSettings {
      bUseBrushTransparency=false
      roundingType=HalfHeightRadius
      width=0
      color {colorUseRule=UseColor_Specified, specifiedColor=<rgba 0 0 0 0>}
      cornerRadii=<0 0 0 0>
     }
     tintColor {colorUseRule=UseColor_Specified, specifiedColor=<rgba 1 1 1 1>}
     uVRegion {bIsValid=false, max=<0 0>, min=<0 0>}
    }
   }
   slot {class=OverlaySlot, properties {horizontalAlignment=HAlign_Fill, verticalAlignment=VAlign_Fill}}
  }
  SZBottomButtons {
   widgetClassPath=&SafeZone
   bIsVariable=true
   parent=Overlay_LoadingAndGradientBacker
   widget=SZBottomButtons
   properties {padTop=false}
   slot {class=OverlaySlot, properties {horizontalAlignment=HAlign_Fill, verticalAlignment=VAlign_Fill}}
  }
  Overlay_2 {widgetClassPath=&Overlay, parent=SZBottomButtons, widget=Overlay_2, slot {class=SafeZoneSlot}}
  WBP_LoadingMorePages {
   widgetClassPath=/DiscoveryBrowser/TabScreens/WBP_LoadingMorePages
   bIsVariable=true
   parent=Overlay_2
   widget=WBP_LoadingMorePages
   slot {class=OverlaySlot, properties {horizontalAlignment=HAlign_Center, padding=<ltrb 8 8 8 8>}}
  }
  HorizontalBox_0 {
   widgetClassPath=&HorizontalBox
   parent=Overlay_2
   widget=HorizontalBox_0
   slot {class=OverlaySlot, properties {horizontalAlignment=HAlign_Right, verticalAlignment=VAlign_Bottom}}
  }
  WBP_BottomBarHack_BackToTop {
   widgetClassPath=/Game/UI/Frontend/BottomBar/WBP_BottomBarDecoyButton
   bIsVariable=true
   parent=HorizontalBox_0
   widget=WBP_BottomBarHack_BackToTop
   properties {minWidth=120, triggeringInputAction {dataTable=&InputActionDataTable, rowName=BackToTop}}
   slot {class=HorizontalBoxSlot}
  }
  WBP_BottomBarHack_Back {
   widgetClassPath=&WBP_BottomBarDecoyButton
   bIsVariable=true
   parent=HorizontalBox_0
   widget=WBP_BottomBarHack_Back
   properties {minWidth=120, triggeringInputAction {dataTable=&InputActionDataTable, rowName=Back}}
   slot {class=HorizontalBoxSlot}
  }
 ]
}
